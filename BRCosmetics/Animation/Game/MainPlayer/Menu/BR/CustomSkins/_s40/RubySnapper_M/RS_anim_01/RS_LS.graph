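# Level Sequence for animating the RubySnapper_M Battle Royale skin's menu animation on a spawned FortPlayerMannequin at the origin, about 324 frames long.
# A skeletal animation track sits alongside Control Rig tracks for Fortnite_Biped_CtrlRig and Fortnite_Legacy_Face_CtrlRig.
# The body rig carries dense, baked-looking keys, with switches held constant for legs, arms, spine and neck.
# The visible face rig channels hold constant values.
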LevelSequence /BRCosmetics/Animation/Game/MainPlayer/Menu/BR/CustomSkins/_s40/RubySnapper_M/RS_anim_01/RS_LS
bindings [
 FortPlayerMannequin {
  id=#1
  possessed_class=FortPlayerMannequin
  tracks [
   {type=MovieSceneSpawnTrack, sections [{type=MovieSceneSpawnSection, channels [None {type=MovieSceneScriptingBoolChannel, default=true}]}]}
   {
    type=MovieScene3DTransformTrack
    sections [{
      type=MovieScene3DTransformSection
      channels [
       Location.X {type=MovieSceneScriptingDoubleChannel, default=0}
       Location.Y {type=MovieSceneScriptingDoubleChannel, default=0}
       Location.Z {type=MovieSceneScriptingDoubleChannel, default=0}
       Rotation.X {type=MovieSceneScriptingDoubleChannel, default=0}
       Rotation.Y {type=MovieSceneScriptingDoubleChannel, default=0}
       Rotation.Z {type=MovieSceneScriptingDoubleChannel, default=0}
       Scale.X {type=MovieSceneScriptingDoubleChannel, default=1}
       Scale.Y {type=MovieSceneScriptingDoubleChannel, default=1}
       Scale.Z {type=MovieSceneScriptingDoubleChannel, default=1}
      ]
     }]
   }
   {type=MovieSceneSkeletalAnimationTrack, sections [{type=MovieSceneSkeletalAnimationSection, end_frame=324, start_frame=0}]}
   {
    type=MovieSceneControlRigParameterTrack
    display_name=Fortnite_Legacy_Face_CtrlRig
    sections [{
      type=MovieSceneControlRigParameterSection
      channels [
       Flipbook {type=MovieSceneScriptingBoolChannel, default=false, const [false,10]}
       Face_Ctrl_Master.Location.X {type=MovieSceneScriptingFloatChannel, default=-21.3302, const [-21.3302,10]}
       Face_Ctrl_Master.Location.Y {type=MovieSceneScriptingFloatChannel, default=0, const [0,10]}
       Face_Ctrl_Master.Location.Z {type=MovieSceneScriptingFloatChannel, default=0, const [0,10]}
       Face_Ctrl_Master.Rotation.X {type=MovieSceneScriptingFloatChannel, default=0, const [0,10]}
       Face_Ctrl_Master.Rotation.Y {type=MovieSceneScriptingFloatChannel, default=0, const [0,10]}
       Face_Ctrl_Master.Rotation.Z {type=MovieSceneScriptingFloatChannel, default=0, const [0,10]}
       Face_Ctrl_Master.Scale.X {type=MovieSceneScriptingFloatChannel, default=1, const [1,10]}
       Face_Ctrl_Master.Scale.Y {type=MovieSceneScriptingFloatChannel, default=1, const [1,10]}
       Face_Ctrl_Master.Scale.Z {type=MovieSceneScriptingFloatChannel, default=1, const [1,10]}
       gui_follow_head {type=MovieSceneScriptingFloatChannel, default=1, const [1,10]}
      ]
     }]
   }
   {
    type=MovieSceneControlRigParameterTrack
    display_name=Fortnite_Biped_CtrlRig
    sections [{
      type=MovieSceneControlRigParameterSection
      channels [
       leg_l_stretch_switch {type=MovieSceneScriptingBoolChannel, default=false, const [false,325]}
       leg_r_stretch_switch {type=MovieSceneScriptingBoolChannel, default=false, const [false,325]}
       arm_l_fk_local {type=MovieSceneScriptingBoolChannel, default=true, const [true,325]}
       Auto Interpolate_3 {type=MovieSceneScriptingBoolChannel, default=false, const [false,325]}
       Auto Interpolate {type=MovieSceneScriptingBoolChannel, default=false, const [false,325]}
       arm_l_stretch_switch {type=MovieSceneScriptingBoolChannel, default=false, const [false,325]}
       Keep Aligned {type=MovieSceneScriptingBoolChannel, default=false, const [false,325]}
       arm_r_fk_local {type=MovieSceneScriptingBoolChannel, default=true, const [true,325]}
       Auto Interpolate_4 {type=MovieSceneScriptingBoolChannel, default=false, const [false,325]}
       Auto Interpolate_2 {type=MovieSceneScriptingBoolChannel, default=false, const [false,325]}
       arm_r_stretch_switch {type=MovieSceneScriptingBoolChannel, default=false, const [false,325]}
       Keep Aligned_2 {type=MovieSceneScriptingBoolChannel, default=false, const [false,325]}
       head_local {type=MovieSceneScriptingBoolChannel, default=true, const [true,325]}
       Stretch {type=MovieSceneScriptingBoolChannel, default=false, const [false,325]}
       head_stretch {type=MovieSceneScriptingBoolChannel, default=false, const [false,325]}
       spine_stretch_switch {type=MovieSceneScriptingBoolChannel, default=false, const [false,325]}
       arm_l_fk_ik_switch {type=MovieSceneScriptingBoolChannel, default=false, const [false,325]}
       leg_l_fk_ik_switch {type=MovieSceneScriptingBoolChannel, default=true, const [true,325]}
       arm_r_fk_ik_switch {type=MovieSceneScriptingBoolChannel, default=false, const [false,325]}
       leg_r_fk_ik_switch {type=MovieSceneScriptingBoolChannel, default=true, const [true,325]}
       ShowBodyControls {type=MovieSceneScriptingBoolChannel, default=true, const [true,325]}
       spine_fk_ik_switch {type=MovieSceneScriptingBoolChannel, default=false, const [false,325]}
       neck_fk_ik_switch {type=MovieSceneScriptingBoolChannel, default=false, const [false,325]}
       space {type=MovieSceneScriptingByteChannel, default=1, const [1,325]}
       global_ctrl.Location.X {type=MovieSceneScriptingFloatChannel, default=0, const [0,325]}
       global_ctrl.Location.Y {type=MovieSceneScriptingFloatChannel, default=0, const [0,325]}
       global_ctrl.Location.Z {type=MovieSceneScriptingFloatChannel, default=0, const [0,325]}
       global_ctrl.Rotation.X {type=MovieSceneScriptingFloatChannel, default=0, const [0,325]}
       global_ctrl.Rotation.Y {type=MovieSceneScriptingFloatChannel, default=0, const [0,325]}
       global_ctrl.Rotation.Z {type=MovieSceneScriptingFloatChannel, default=0, const [0,325]}
       root_ctrl.Location.X {type=MovieSceneScriptingFloatChannel, default=0, const [0,325]}
       root_ctrl.Location.Y {type=MovieSceneScriptingFloatChannel, default=0, const [0,325]}
       root_ctrl.Location.Z {type=MovieSceneScriptingFloatChannel, default=0, const [0,325]}
       root_ctrl.Rotation.X {type=MovieSceneScriptingFloatChannel, default=0, const [0,325]}
       root_ctrl.Rotation.Y {type=MovieSceneScriptingFloatChannel, default=0, const [0,325]}
       root_ctrl.Rotation.Z {type=MovieSceneScriptingFloatChannel, default=0, const [0,325]}
       root_ctrl.Scale.X {type=MovieSceneScriptingFloatChannel, default=1, const [1,325]}
       root_ctrl.Scale.Y {type=MovieSceneScriptingFloatChannel, default=1, const [1,325]}
       root_ctrl.Scale.Z {type=MovieSceneScriptingFloatChannel, default=1, const [1,325]}
       body_offset_ctrl.Location.X {type=MovieSceneScriptingFloatChannel, default=0, const [0,325]}
       body_offset_ctrl.Location.Y {type=MovieSceneScriptingFloatChannel, default=0, const [0,325]}
       body_offset_ctrl.Location.Z {type=MovieSceneScriptingFloatChannel, default=0, const [0,325]}
       body_offset_ctrl.Rotation.X {type=MovieSceneScriptingFloatChannel, default=0, const [0,325]}
       body_offset_ctrl.Rotation.Y {type=MovieSceneScriptingFloatChannel, default=0, const [0,325]}
       body_offset_ctrl.Rotation.Z {type=MovieSceneScriptingFloatChannel, default=0, const [0,325]}
       body_offset_ctrl.Scale.X {type=MovieSceneScriptingFloatChannel, default=1, const [1,325]}
       body_offset_ctrl.Scale.Y {type=MovieSceneScriptingFloatChannel, default=1, const [1,325]}
       body_offset_ctrl.Scale.Z {type=MovieSceneScriptingFloatChannel, default=1, const [1,325]}
       thigh_l_fk_ctrl.Location.X {type=MovieSceneScriptingFloatChannel, default=0, thinned_from=325, keys [[0,0],[324,0]]}
       thigh_l_fk_ctrl.Location.Y {type=MovieSceneScriptingFloatChannel, default=0, thinned_from=325, keys [[0,0],[324,0]]}
       thigh_l_fk_ctrl.Location.Z {type=MovieSceneScriptingFloatChannel, default=0, thinned_from=325, keys [[0,0],[324,0]]}
       thigh_l_fk_ctrl.Rotation.X {
        type=MovieSceneScriptingFloatChannel
        default=1.216
        thinned_from=325
        keys [[0,42.9532],[29,43.0312],[31,42.8934],[32,42.5964],[33,42.1148],[34,41.4317],[35,40.8477],[40,40.4782],[86,40.3917],[88,40.6267],[89,40.8723],[90,41.351],[91,41.9802],[92,42.4394],[94,42.8297],[184,43.0544],[186,43.1817],[187,43.3283],[188,43.6025],[192,45.3906],[193,45.692],[200,46.1059],[247,46.0198],[249,45.8394],[250,45.5927],[251,45.145],[252,44.2575],[253,43.5332],[255,43.1709],[324,42.954]]
       }
       thigh_l_fk_ctrl.Rotation.Y {
        type=MovieSceneScriptingFloatChannel
        default=-1.2128
        thinned_from=325
        keys [[0,-2.8599],[30,-2.8798],[31,-2.8886],[32,-2.9217],[34,-3.0227],[37,-3.1358],[38,-3.1584],[40,-3.1758],[85,-3.1543],[87,-3.1301],[88,-3.1051],[89,-3.0658],[90,-3.0156],[91,-2.9589],[93,-2.8265],[94,-2.7652],[95,-2.7174],[96,-2.6833],[97,-2.658],[98,-2.641],[99,-2.6309],[178,-2.6057],[183,-2.6254],[184,-2.6176],[185,-2.5897],[186,-2.5331],[187,-2.3631],[188,-2.1672],[189,-2.0812],[190,-2.0521],[191,-2.0472],[192,-2.0678],[193,-2.1074],[194,-2.1292],[233,-2.1009],[247,-2.083],[248,-2.0855],[249,-2.0995],[250,-2.1258],[251,-2.1841],[252,-2.2731],[253,-2.4983],[254,-2.7015],[255,-2.7449],[256,-2.7737],[257,-2.7907],[258,-2.7988],[282,-2.8479],[283,-2.8544],[324,-2.8607]]
       }
       thigh_l_fk_ctrl.Rotation.Z {
        type=MovieSceneScriptingFloatChannel
        default=-0.1664
        thinned_from=325
        keys [[0,3.9491],[1,3.9491],[3,3.9121],[5,3.8531],[8,3.73],[14,3.4418],[16,3.3582],[22,3.1481],[24,3.0931],[25,3.0736],[26,3.061],[28,3.061],[31,3.0479],[32,3.022],[33,2.9764],[34,2.9048],[35,2.845],[40,2.9087],[43,2.9649],[45,3.0181],[53,3.3237],[55,3.3827],[67,3.5722],[68,3.5722],[69,3.5953],[70,3.5953],[71,3.6162],[72,3.6162],[73,3.6342],[74,3.6342],[75,3.6494],[77,3.6494],[78,3.6666],[81,3.6666],[82,3.6787],[85,3.6787],[89,3.6182],[90,3.6304],[91,3.6591],[92,3.6576],[93,3.6187],[98,3.3574],[100,3.2685],[106,3.0602],[108,3.0057],[109,2.9863],[110,2.9863],[111,2.9694],[115,2.9694],[116,2.9859],[117,2.9859],[118,3.0017],[119,3.0017],[120,3.0228],[121,3.0228],[122,3.0491],[127,3.1374],[129,3.1907],[137,3.4961],[139,3.5552],[150,3.7325],[152,3.7325],[153,3.7683],[154,3.7683],[155,3.7892],[156,3.7892],[157,3.8073],[158,3.8073],[159,3.8226],[161,3.8226],[162,3.8397],[165,3.8397],[166,3.8521],[169,3.8521],[171,3.8114],[173,3.7451],[176,3.6092],[181,3.3583],[183,3.2765],[184,3.247],[186,3.2063],[187,3.1998],[188,3.2157],[189,3.2663],[192,3.4784],[193,3.5258],[200,3.6128],[201,3.6128],[202,3.6295],[203,3.6295],[204,3.6501],[205,3.6501],[206,3.6761],[211,3.7658],[213,3.8231],[220,4.1171],[222,4.1819],[232,4.3473],[233,4.3473],[234,4.3728],[235,4.3728],[236,4.3959],[237,4.3959],[238,4.4164],[239,4.4164],[240,4.4342],[241,4.4342],[242,4.4493],[244,4.4493],[245,4.4664],[248,4.4676],[249,4.4494],[250,4.4101],[251,4.3345],[252,4.1798],[253,4.0453],[256,3.8895],[267,3.3838],[274,3.1373],[276,3.083],[277,3.0638],[278,3.0638],[279,3.0475],[282,3.0475],[283,3.0596],[284,3.0596],[285,3.0735],[286,3.0735],[287,3.0928],[288,3.0928],[289,3.1173],[294,3.2013],[296,3.246],[299,3.3418],[306,3.6076],[315,3.8369],[319,3.9105],[320,3.9237],[321,3.9237],[322,3.9425],[324,3.9425]]
       }
       thigh_l_fk_ctrl.Scale.X {type=MovieSceneScriptingFloatChannel, default=1, const [1,325]}
       thigh_l_fk_ctrl.Scale.Y {type=MovieSceneScriptingFloatChannel, default=1, const [1,325]}
       thigh_l_fk_ctrl.Scale.Z {type=MovieSceneScriptingFloatChannel, default=1, const [1,325]}
       calf_l_fk_ctrl.Location.X {
        type=MovieSceneScriptingFloatChannel
        default=0
        thinned_from=325
        keys [[0,-0.6326],[2,-0.6339],[3,-0.6363],[4,-0.6404],[5,-0.6468],[6,-0.656],[7,-0.6683],[8,-0.6844],[9,-0.7056],[10,-0.732],[11,-0.7622],[14,-0.8605],[15,-0.8909],[16,-0.9175],[17,-0.939],[19,-0.9723],[21,-0.9986],[23,-1.0177],[25,-1.0293],[27,-1.0333],[31,-1.0318],[38,-1.0052],[41,-0.9874],[43,-0.9712],[44,-0.9597],[45,-0.9439],[46,-0.9247],[47,-0.9027],[52,-0.7768],[53,-0.7536],[54,-0.7328],[55,-0.7151],[56,-0.7012],[58,-0.6799],[60,-0.6625],[62,-0.6487],[64,-0.6381],[66,-0.6303],[69,-0.6232],[75,-0.6203],[78,-0.6228],[80,-0.6277],[82,-0.6362],[83,-0.6423],[84,-0.6497],[85,-0.6589],[86,-0.6701],[87,-0.6834],[88,-0.6986],[89,-0.7156],[91,-0.7545],[96,-0.8593],[98,-0.8973],[100,-0.9313],[106,-1.0199],[108,-1.0426],[109,-1.0507],[110,-1.0558],[111,-1.0577],[113,-1.0554],[116,-1.0442],[120,-1.0179],[127,-0.9533],[129,-0.9298],[132,-0.8854],[137,-0.806],[139,-0.7791],[143,-0.7381],[147,-0.7028],[150,-0.6802],[154,-0.6558],[157,-0.6422],[160,-0.6336],[164,-0.6289],[168,-0.6283],[169,-0.6294],[170,-0.6326],[171,-0.6378],[172,-0.6447],[173,-0.6531],[175,-0.6738],[178,-0.7115],[183,-0.7784],[186,-0.8106],[188,-0.8186],[190,-0.8335],[191,-0.8425],[194,-0.8746],[195,-0.8826],[198,-0.8989],[202,-0.912],[211,-0.9186],[213,-0.9147],[215,-0.9044],[217,-0.8889],[220,-0.8585],[224,-0.8102],[226,-0.7809],[234,-0.6445],[236,-0.6132],[237,-0.5991],[238,-0.5862],[239,-0.5748],[240,-0.5649],[241,-0.5567],[242,-0.5499],[243,-0.5443],[245,-0.5363],[247,-0.5319],[248,-0.5313],[249,-0.532],[250,-0.5343],[251,-0.5405],[252,-0.5516],[253,-0.5713],[254,-0.5876],[256,-0.5991],[260,-0.6168],[262,-0.6286],[264,-0.6441],[266,-0.6636],[268,-0.6877],[269,-0.7016],[270,-0.7206],[271,-0.7475],[272,-0.7801],[275,-0.8907],[276,-0.9249],[277,-0.9543],[278,-0.9766],[279,-0.9898],[282,-1.0096],[286,-1.0246],[295,-1.031],[298,-1.0266],[301,-1.0137],[303,-1],[305,-0.982],[307,-0.9594],[308,-0.9463],[309,-0.9307],[310,-0.9117],[311,-0.8901],[313,-0.841],[317,-0.7365],[318,-0.7125],[319,-0.6906],[320,-0.6713],[321,-0.6553],[322,-0.6431],[323,-0.6353],[324,-0.6326]]
       }
       calf_l_fk_ctrl.Location.Y {
        type=MovieSceneScriptingFloatChannel
        default=0
        thinned_from=325
        keys [[0,0.0509],[3,0.0512],[5,0.0521],[6,0.0528],[7,0.0538],[8,0.0551],[9,0.0568],[10,0.0589],[12,0.064],[14,0.0693],[15,0.0717],[16,0.0739],[17,0.0756],[19,0.0783],[21,0.0804],[23,0.0819],[25,0.0829],[27,0.0832],[31,0.0831],[39,0.0805],[42,0.0789],[43,0.0782],[44,0.0773],[45,0.076],[46,0.0745],[48,0.0707],[52,0.0625],[54,0.059],[55,0.0576],[56,0.0565],[59,0.054],[62,0.0522],[66,0.0507],[72,0.0499],[78,0.0501],[81,0.0508],[83,0.0517],[85,0.053],[87,0.055],[89,0.0576],[97,0.0708],[100,0.075],[106,0.0821],[108,0.084],[109,0.0846],[110,0.085],[112,0.0851],[115,0.0845],[119,0.0826],[127,0.0768],[129,0.0749],[137,0.0649],[139,0.0627],[146,0.0573],[150,0.0548],[151,0.0482],[152,0.0477],[153,0.0532],[158,0.0514],[168,0.0506],[170,0.0509],[172,0.0519],[175,0.0542],[183,0.0627],[186,0.0653],[188,0.0659],[191,0.0678],[194,0.0704],[197,0.072],[202,0.0734],[211,0.074],[213,0.0736],[215,0.0728],[218,0.0708],[223,0.0663],[226,0.0629],[235,0.0506],[238,0.0472],[240,0.0455],[242,0.0443],[245,0.0432],[248,0.0428],[250,0.043],[251,0.0435],[252,0.0444],[253,0.046],[254,0.0473],[262,0.0506],[265,0.0526],[268,0.0554],[269,0.0565],[270,0.058],[271,0.0602],[272,0.0628],[275,0.0717],[276,0.0745],[277,0.0768],[278,0.0786],[279,0.0797],[282,0.0813],[286,0.0825],[296,0.083],[299,0.0824],[302,0.0811],[305,0.0791],[308,0.0762],[309,0.0749],[310,0.0734],[312,0.0698],[317,0.0593],[319,0.0556],[320,0.054],[321,0.0528],[322,0.0518],[323,0.0511],[324,0.0509]]
       }
       calf_l_fk_ctrl.Location.Z {
        type=MovieSceneScriptingFloatChannel
        default=0
        thinned_from=325
        keys [[0,0.0004],[150,0.0004],[151,-0.0059],[152,-0.0059],[153,0.0004],[324,0.0004]]
       }
       calf_l_fk_ctrl.Rotation.X {
        type=MovieSceneScriptingFloatChannel
        default=0.0543
        thinned_from=325
        keys [[0,0.0033],[150,0.0033],[151,0.0025],[152,0.0025],[153,0.0033],[324,0.0033]]
       }
       calf_l_fk_ctrl.Rotation.Y {
        type=MovieSceneScriptingFloatChannel
        default=1.2238
        thinned_from=325
        keys [[0,0.0047],[150,0.0047],[151,0.0137],[152,0.0137],[153,0.0047],[324,0.0047]]
       }
       calf_l_fk_ctrl.Rotation.Z {type=MovieSceneScriptingFloatChannel, default=0.3732, thinned_from=325, keys [[0,-4.9606],[150,-4.9607],[152,-4.9518],[324,-4.9606]]}
       calf_l_fk_ctrl.Scale.X {type=MovieSceneScriptingFloatChannel, default=1, const [1,325]}
       calf_l_fk_ctrl.Scale.Y {type=MovieSceneScriptingFloatChannel, default=1, const [1,325]}
       calf_l_fk_ctrl.Scale.Z {type=MovieSceneScriptingFloatChannel, default=1, const [1,325]}
       foot_l_fk_ctrl.Location.X {
        type=MovieSceneScriptingFloatChannel
        default=0
        thinned_from=325
        keys [[0,-0.6707],[2,-0.6722],[3,-0.6747],[4,-0.6793],[5,-0.6863],[6,-0.6963],[7,-0.7098],[8,-0.7273],[9,-0.7505],[10,-0.7794],[11,-0.8125],[14,-0.92],[15,-0.9531],[16,-0.9823],[17,-1.0057],[19,-1.0422],[21,-1.0709],[23,-1.0918],[25,-1.1045],[27,-1.1088],[31,-1.1073],[38,-1.0781],[41,-1.0587],[43,-1.041],[44,-1.0284],[45,-1.0111],[46,-0.9901],[47,-0.966],[52,-0.8284],[53,-0.803],[54,-0.7803],[55,-0.7609],[56,-0.7457],[58,-0.7224],[60,-0.7034],[62,-0.6883],[64,-0.6767],[66,-0.6682],[69,-0.6604],[73,-0.6571],[77,-0.6586],[79,-0.6622],[81,-0.6694],[83,-0.6813],[84,-0.6894],[85,-0.6994],[86,-0.7118],[87,-0.7262],[88,-0.7429],[89,-0.7615],[91,-0.804],[96,-0.9186],[98,-0.9602],[100,-0.9973],[105,-1.0791],[107,-1.1077],[108,-1.1191],[109,-1.1279],[110,-1.1335],[111,-1.1355],[113,-1.133],[116,-1.1208],[119,-1.1003],[127,-1.0213],[129,-0.9957],[132,-0.9472],[136,-0.8768],[138,-0.8449],[139,-0.8309],[143,-0.786],[147,-0.7475],[151,-0.7154],[154,-0.6961],[157,-0.6812],[160,-0.6718],[164,-0.6667],[168,-0.666],[169,-0.6672],[170,-0.6707],[171,-0.6763],[172,-0.6839],[173,-0.6931],[175,-0.7157],[178,-0.7569],[183,-0.8302],[186,-0.8654],[188,-0.8741],[190,-0.8904],[191,-0.9003],[194,-0.9353],[195,-0.9441],[198,-0.9619],[202,-0.9762],[211,-0.9834],[213,-0.9792],[215,-0.9679],[217,-0.951],[220,-0.9177],[224,-0.8649],[226,-0.8328],[234,-0.6837],[236,-0.6495],[237,-0.6341],[238,-0.62],[239,-0.6074],[240,-0.5967],[241,-0.5877],[242,-0.5802],[243,-0.5741],[245,-0.5654],[247,-0.5606],[248,-0.5599],[249,-0.5606],[250,-0.5632],[251,-0.5699],[252,-0.5822],[253,-0.6037],[254,-0.6215],[256,-0.6341],[260,-0.6535],[262,-0.6663],[264,-0.6832],[266,-0.7046],[268,-0.731],[269,-0.7461],[270,-0.767],[271,-0.7964],[272,-0.832],[275,-0.953],[276,-0.9904],[277,-1.0224],[278,-1.0469],[279,-1.0613],[282,-1.0829],[285,-1.0963],[291,-1.1062],[296,-1.1057],[299,-1.0979],[301,-1.0874],[303,-1.0725],[305,-1.0528],[307,-1.028],[308,-1.0137],[309,-0.9966],[310,-0.976],[311,-0.9523],[313,-0.8986],[317,-0.7843],[318,-0.7581],[319,-0.7341],[320,-0.713],[321,-0.6955],[322,-0.6822],[323,-0.6737],[324,-0.6707]]
       }
       foot_l_fk_ctrl.Location.Y {
        type=MovieSceneScriptingFloatChannel
        default=0.0398
        thinned_from=325
        keys [[0,-0.0042],[9,-0.0047],[24,-0.0068],[46,-0.0061],[84,-0.0043],[111,-0.007],[172,-0.0042],[218,-0.0058],[250,-0.0035],[270,-0.0048],[278,-0.0065],[309,-0.0062],[323,-0.0042],[324,-0.0042]]
       }
       foot_l_fk_ctrl.Location.Z {type=MovieSceneScriptingFloatChannel, default=-0.8885, thinned_from=325, keys [[0,0.0005],[324,0.0005]]}
       foot_l_fk_ctrl.Rotation.X {
        type=MovieSceneScriptingFloatChannel
        default=-1.1647
        thinned_from=325
        keys [[0,0.3492],[30,0.3577],[31,0.3585],[32,0.3614],[33,0.3661],[34,0.3719],[35,0.3766],[37,0.3805],[40,0.3825],[86,0.3756],[87,0.3748],[88,0.3732],[89,0.3705],[90,0.3661],[92,0.3557],[94,0.3492],[95,0.3471],[97,0.3452],[171,0.3391],[186,0.346],[189,0.345],[191,0.3466],[194,0.3482],[211,0.3472],[247,0.3387],[249,0.3392],[251,0.3418],[253,0.3468],[294,0.3558],[324,0.3492]]
       }
       foot_l_fk_ctrl.Rotation.Y {
        type=MovieSceneScriptingFloatChannel
        default=0.0792
        thinned_from=325
        keys [[0,2.7789],[25,2.8823],[29,2.8848],[30,2.8893],[31,2.8997],[32,2.937],[33,2.9952],[34,3.0692],[35,3.129],[37,3.1763],[38,3.1907],[41,3.2033],[86,3.112],[87,3.102],[88,3.0817],[89,3.0483],[90,2.993],[91,2.9235],[92,2.8629],[93,2.8182],[94,2.7822],[95,2.7572],[96,2.7423],[98,2.7311],[111,2.7684],[169,2.6514],[186,2.7448],[187,2.7422],[188,2.7319],[189,2.7289],[191,2.744],[192,2.7444],[194,2.76],[211,2.7457],[247,2.6355],[249,2.6436],[250,2.6565],[251,2.6775],[252,2.7158],[253,2.7461],[278,2.867],[294,2.8651],[324,2.7789]]
       }
       foot_l_fk_ctrl.Rotation.Z {
        type=MovieSceneScriptingFloatChannel
        default=-0.2074
        thinned_from=325
        keys [[0,6.635],[30,6.7309],[32,6.7072],[34,6.664],[35,6.6536],[37,6.5895],[85,6.4824],[87,6.5015],[91,6.6109],[92,6.6571],[94,6.7744],[96,6.858],[98,6.9092],[184,6.9217],[187,6.8854],[188,6.8588],[189,6.7924],[191,6.5961],[192,6.5134],[193,6.4458],[194,6.4113],[247,6.2881],[249,6.3063],[250,6.3305],[251,6.3798],[253,6.5752],[255,6.6477],[258,6.7018],[294,6.7253],[324,6.635]]
       }
       foot_l_fk_ctrl.Scale.X {type=MovieSceneScriptingFloatChannel, default=1, const [1,325]}
       foot_l_fk_ctrl.Scale.Y {type=MovieSceneScriptingFloatChannel, default=1, const [1,325]}
       foot_l_fk_ctrl.Scale.Z {type=MovieSceneScriptingFloatChannel, default=1, const [1,325]}
       ball_l_fk_ctrl.Location.X {type=MovieSceneScriptingFloatChannel, default=0, const [0,325]}
       ball_l_fk_ctrl.Location.Y {type=MovieSceneScriptingFloatChannel, default=0, const [0,325]}
       ball_l_fk_ctrl.Location.Z {type=MovieSceneScriptingFloatChannel, default=0, const [0,325]}
       ball_l_fk_ctrl.Rotation.X {type=MovieSceneScriptingFloatChannel, default=0, const [0,325]}
       ball_l_fk_ctrl.Rotation.Y {type=MovieSceneScriptingFloatChannel, default=0, const [0,325]}
       ball_l_fk_ctrl.Rotation.Z {type=MovieSceneScriptingFloatChannel, default=0, const [0,325]}
       ball_l_fk_ctrl.Scale.X {type=MovieSceneScriptingFloatChannel, default=1, const [1,325]}
       ball_l_fk_ctrl.Scale.Y {type=MovieSceneScriptingFloatChannel, default=1, const [1,325]}
       ball_l_fk_ctrl.Scale.Z {type=MovieSceneScriptingFloatChannel, default=1, const [1,325]}
       foot_l_ik_ctrl.Location.X {
        type=MovieSceneScriptingFloatChannel
        default=0
        thinned_from=325
        keys [[0,-2.2125],[16,-2.2136],[17,-2.2451],[22,-2.182],[23,-2.2114],[28,-2.2154],[38,-2.2111],[39,-2.1763],[44,-2.2165],[46,-2.2429],[47,-2.2128],[67,-2.2123],[68,-2.2064],[69,-2.2124],[70,-2.2067],[71,-2.2128],[72,-2.2075],[73,-2.2124],[74,-2.2083],[75,-2.2127],[77,-2.2058],[78,-2.2124],[81,-2.2063],[82,-2.2129],[84,-2.2131],[85,-2.2215],[86,-2.2127],[100,-2.2127],[101,-2.2451],[106,-2.182],[107,-2.2135],[109,-2.2125],[110,-2.2182],[111,-2.2127],[115,-2.2082],[116,-2.2131],[117,-2.2095],[118,-2.2118],[119,-2.2086],[120,-2.2118],[121,-2.2071],[122,-2.2124],[123,-2.181],[128,-2.216],[130,-2.2388],[131,-2.2122],[151,-2.2125],[152,-2.2076],[153,-2.2126],[154,-2.2078],[155,-2.2125],[156,-2.2085],[157,-2.2127],[158,-2.2088],[159,-2.2126],[161,-2.2073],[162,-2.2125],[165,-2.2073],[166,-2.2126],[183,-2.213],[184,-2.2159],[185,-2.1487],[186,-2.2132],[201,-2.2104],[202,-2.213],[203,-2.2077],[204,-2.2127],[205,-2.2074],[206,-2.2126],[232,-2.2128],[233,-2.2097],[234,-2.2126],[235,-2.2102],[236,-2.2126],[244,-2.2103],[268,-2.2136],[269,-2.2449],[274,-2.1821],[275,-2.2114],[278,-2.2146],[282,-2.2021],[283,-2.2135],[284,-2.2073],[285,-2.2131],[286,-2.2061],[287,-2.2125],[288,-2.2042],[289,-2.2112],[290,-2.2113],[291,-2.181],[296,-2.2158],[298,-2.2375],[299,-2.2125],[320,-2.2126],[321,-2.2087],[322,-2.2127],[324,-2.2102]]
       }
       foot_l_ik_ctrl.Location.Y {
        type=MovieSceneScriptingFloatChannel
        default=0
        thinned_from=325
        keys [[0,-2.3767],[1,-2.3829],[2,-2.3771],[14,-2.3764],[16,-2.3797],[17,-2.4715],[22,-2.2886],[23,-2.3735],[26,-2.3769],[27,-2.3831],[28,-2.3835],[31,-2.3754],[37,-2.3759],[38,-2.3731],[39,-2.2867],[43,-2.3618],[44,-2.3867],[46,-2.4525],[47,-2.3776],[67,-2.3763],[68,-2.3611],[69,-2.3765],[70,-2.3625],[71,-2.3774],[72,-2.3641],[73,-2.3764],[74,-2.3662],[75,-2.3771],[77,-2.3604],[78,-2.3765],[81,-2.3634],[82,-2.3776],[84,-2.3753],[85,-2.382],[95,-2.3752],[100,-2.3773],[101,-2.471],[106,-2.2891],[107,-2.3795],[109,-2.3765],[110,-2.3946],[111,-2.3771],[114,-2.3674],[115,-2.3606],[116,-2.3783],[117,-2.3659],[118,-2.3748],[119,-2.3631],[120,-2.3747],[121,-2.3583],[122,-2.3764],[123,-2.2864],[127,-2.3617],[128,-2.3867],[130,-2.4527],[131,-2.3758],[151,-2.3766],[152,-2.3603],[153,-2.3768],[154,-2.3615],[155,-2.3768],[156,-2.3636],[157,-2.3773],[158,-2.3652],[159,-2.3771],[161,-2.36],[162,-2.3766],[165,-2.3622],[166,-2.3768],[168,-2.3759],[169,-2.3818],[180,-2.3784],[181,-2.3754],[183,-2.3782],[184,-2.4793],[185,-2.4612],[186,-2.3766],[200,-2.3769],[201,-2.3652],[202,-2.3769],[203,-2.3636],[204,-2.3769],[205,-2.3598],[206,-2.3769],[232,-2.3768],[233,-2.3582],[234,-2.3769],[235,-2.3596],[236,-2.3769],[237,-2.3615],[238,-2.377],[239,-2.3637],[240,-2.377],[241,-2.3656],[242,-2.3769],[244,-2.3593],[245,-2.3771],[247,-2.3672],[248,-2.3769],[267,-2.375],[268,-2.3797],[269,-2.4708],[274,-2.2891],[275,-2.3735],[277,-2.3767],[278,-2.3949],[279,-2.3769],[281,-2.3752],[282,-2.369],[283,-2.3794],[284,-2.3685],[285,-2.3783],[286,-2.3654],[287,-2.3767],[288,-2.3595],[289,-2.3729],[290,-2.3733],[291,-2.2858],[295,-2.3616],[296,-2.3862],[298,-2.4493],[299,-2.3766],[300,-2.3814],[320,-2.3768],[321,-2.3617],[322,-2.3772],[324,-2.3673]]
       }
       foot_l_ik_ctrl.Location.Z {
        type=MovieSceneScriptingFloatChannel
        default=0
        thinned_from=325
        keys [[0,0],[16,0],[17,0.0008],[22,-0.0006],[23,0],[38,0],[46,0.0003],[47,0],[84,-0.0001],[85,-0.0003],[86,0],[100,0],[101,0.0006],[106,-0.0004],[107,0],[110,0.0001],[115,-0.0002],[116,0],[117,-0.0001],[118,0],[121,-0.0002],[122,0],[123,-0.0004],[129,0.0002],[130,0.0005],[131,0],[151,0],[152,-0.0002],[153,0],[154,-0.0001],[155,0],[156,-0.0001],[157,0],[161,-0.0002],[166,0],[168,-0.0002],[169,0.0001],[183,0],[184,0.005],[185,0.0131],[186,-0.0001],[200,0],[201,-0.0002],[202,0],[203,-0.0001],[204,0],[205,-0.0003],[206,0],[211,-0.0002],[232,0],[233,-0.0004],[234,0],[235,-0.0004],[236,0],[237,-0.0004],[238,0],[239,-0.0003],[240,0],[241,-0.0003],[242,0],[244,-0.0004],[245,0],[247,-0.0003],[248,0],[268,0],[269,0.0008],[274,-0.0006],[275,0],[277,0],[278,0.0003],[279,0],[282,0.0003],[283,0],[290,0],[291,-0.0006],[293,-0.0003],[294,-0.0003],[298,0.0006],[299,0],[320,0],[321,-0.0002],[322,0],[324,-0.0001]]
       }
       foot_l_ik_ctrl.Rotation.X {
        type=MovieSceneScriptingFloatChannel
        default=0
        thinned_from=325
        keys [[0,-0.0001],[1,0.0033],[2,0.0001],[3,0.0004],[4,-0.0002],[6,-0.0003],[8,0.0003],[9,0.0004],[10,0],[12,0.0004],[13,0.0004],[14,-0.0003],[15,0.0012],[16,0.0014],[17,0.0557],[18,0.0347],[19,0.0127],[20,-0.0096],[21,-0.0314],[22,-0.052],[23,-0.0017],[24,-0.0014],[25,-0.0001],[26,0],[27,0.0035],[28,0.0035],[29,0.0006],[30,-0.0004],[31,-0.0008],[32,-0.0004],[33,-0.0004],[34,0.0001],[35,0],[36,0.0003],[37,-0.0006],[38,-0.0018],[39,-0.0558],[40,-0.0455],[41,-0.0344],[42,-0.0223],[43,-0.0094],[44,0.006],[45,0.025],[46,0.0468],[47,0.0007],[48,-0.0005],[49,-0.0005],[50,-0.0001],[52,-0.0001],[53,0.0003],[54,0.0003],[55,-0.0001],[56,0.0004],[57,0.0004],[58,-0.0002],[59,0.0001],[60,0.0001],[61,-0.0006],[62,0.0001],[63,0.0001],[64,-0.0006],[65,0.0002],[66,-0.0002],[67,-0.0003],[68,-0.0088],[69,-0.0003],[70,-0.0082],[71,0.0002],[72,-0.0072],[73,-0.0003],[74,-0.0053],[75,0.0001],[76,-0.005],[77,-0.0094],[78,-0.0002],[79,-0.0031],[80,-0.0057],[81,-0.0082],[82,0.0003],[83,-0.0011],[84,-0.0002],[85,0.007],[86,0.0002],[91,-0.0001],[92,0],[93,-0.0006],[94,-0.0002],[95,-0.0009],[96,0.0003],[97,0.0001],[99,-0.0009],[100,0.0001],[101,0.0557],[102,0.0347],[103,0.0127],[104,-0.0096],[105,-0.0314],[106,-0.052],[107,0.0012],[108,-0.0002],[109,-0.0001],[110,0.0088],[111,0.0001],[112,0],[114,-0.0044],[115,-0.0077],[116,0.0007],[117,-0.0053],[118,-0.001],[119,-0.0067],[120,-0.001],[121,-0.0091],[122,-0.0001],[123,-0.0536],[124,-0.0437],[125,-0.033],[126,-0.0214],[127,-0.009],[128,0.0058],[129,0.024],[130,0.0449],[131,-0.0008],[132,0.0006],[133,-0.0004],[134,0.0006],[135,-0.0001],[136,0.0004],[139,0.0003],[140,0],[141,0.0004],[142,-0.0002],[147,-0.0002],[148,0.0001],[150,-0.0003],[151,-0.0002],[152,-0.0083],[153,-0.0001],[154,-0.0078],[155,-0.0001],[156,-0.0067],[157,0.0002],[158,-0.006],[159,0.0001],[161,-0.0086],[162,-0.0002],[163,-0.0033],[164,-0.0061],[165,-0.0079],[166,-0.0001],[167,-0.0005],[168,-0.0005],[169,0.0024],[170,0.0004],[171,-0.0002],[173,-0.0001],[175,0.0004],[176,-0.0003],[177,-0.0002],[178,0.0004],[179,0.0001],[180,0.0006],[181,-0.0008],[182,0.0004],[183,0.0004],[184,0.0418],[185,-0.0053],[186,0.0002],[188,-0.0001],[189,0.0001],[190,0],[191,0.0003],[192,0],[193,0],[194,-0.0002],[196,0.0001],[197,-0.0001],[198,0.0003],[200,0],[201,-0.0048],[202,0.0002],[203,-0.0071],[204,0.0001],[205,-0.0086],[206,0],[207,-0.0001],[209,0],[211,0.0042],[212,0.0001],[214,0.0001],[215,-0.0001],[218,0],[221,-0.0001],[225,0.0001],[231,0],[232,0.0001],[233,-0.0077],[234,0],[235,-0.007],[236,0],[237,-0.0062],[238,0],[239,-0.0054],[240,0],[241,-0.0047],[242,0],[243,-0.0037],[244,-0.007],[245,0],[246,-0.0022],[247,-0.0036],[248,0],[255,0],[256,-0.0003],[258,-0.0002],[259,0.0002],[260,0.0003],[261,0.0002],[262,-0.0002],[264,0.0003],[265,0.0002],[267,-0.0009],[268,0.0014],[269,0.0557],[270,0.0347],[271,0.0127],[272,-0.0096],[273,-0.0314],[274,-0.052],[275,-0.0017],[276,-0.0014],[277,-0.0002],[278,0.0069],[279,0.0001],[280,-0.0014],[281,-0.0041],[282,-0.0089],[283,0.0013],[284,-0.006],[285,0.0006],[286,-0.0075],[287,0],[288,-0.0104],[289,-0.0021],[290,-0.0018],[291,-0.0536],[292,-0.0437],[293,-0.0296],[294,-0.0148],[295,-0.009],[296,0.0055],[297,0.0228],[298,0.0425],[299,0],[300,0.0023],[301,0],[302,0],[303,-0.0005],[305,0.0006],[306,0.0001],[307,0.0002],[308,0],[309,0.0002],[310,-0.0001],[312,-0.0002],[313,0.0003],[314,0.0003],[315,0.0001],[317,0.0001],[318,-0.0002],[319,0.0001],[320,-0.0001],[321,-0.0072],[322,0.0001],[323,-0.0031],[324,-0.0045]]
       }
       foot_l_ik_ctrl.Rotation.Y {
        type=MovieSceneScriptingFloatChannel
        default=0
        thinned_from=325
        keys [[0,-0.0001],[1,0.003],[2,0.0001],[3,0.0004],[4,-0.0002],[6,-0.0003],[9,0.0004],[10,0],[12,0.0003],[13,0.0003],[14,-0.0003],[15,0.0012],[16,0.0014],[17,0.0529],[18,0.033],[19,0.0121],[20,-0.0091],[21,-0.0298],[22,-0.0494],[23,-0.0016],[24,-0.0013],[25,-0.0001],[26,0],[27,0.0031],[28,0.0031],[29,0.0006],[30,-0.0004],[31,-0.0007],[32,-0.0004],[33,-0.0004],[34,0.0001],[35,0],[36,0.0003],[37,-0.0005],[38,-0.0016],[39,-0.0485],[40,-0.0395],[41,-0.0298],[42,-0.0194],[43,-0.0081],[44,0.0052],[45,0.0217],[46,0.0406],[47,0.0006],[48,-0.0004],[49,-0.0004],[50,0],[52,-0.0001],[53,0.0002],[54,0.0002],[55,-0.0001],[56,0.0004],[57,0.0004],[58,-0.0002],[59,0.0001],[60,0.0001],[61,-0.0005],[62,0.0001],[63,0.0001],[64,-0.0005],[65,0.0001],[67,-0.0003],[68,-0.0078],[69,-0.0002],[70,-0.0071],[71,0.0002],[72,-0.0063],[73,-0.0003],[74,-0.0046],[75,0.0001],[76,-0.0044],[77,-0.008],[78,-0.0002],[80,-0.0047],[81,-0.0063],[82,0.0003],[83,-0.0006],[84,-0.0012],[85,0.0003],[91,0],[92,0],[93,-0.0005],[94,-0.0002],[95,-0.0009],[96,0.0002],[97,0.0001],[99,-0.0008],[100,0.0001],[101,0.0529],[102,0.033],[103,0.0121],[104,-0.0091],[105,-0.0298],[106,-0.0493],[107,0.0012],[108,-0.0002],[109,-0.0001],[110,0.0089],[111,0.0001],[112,-0.0001],[113,-0.0023],[114,-0.0048],[115,-0.0085],[116,0.0006],[117,-0.0057],[118,-0.001],[119,-0.0072],[120,-0.0009],[121,-0.0096],[122,-0.0001],[123,-0.0509],[124,-0.0415],[125,-0.0313],[126,-0.0204],[127,-0.0086],[128,0.0055],[129,0.0228],[130,0.0427],[131,-0.0008],[132,0.0006],[133,-0.0004],[134,0.0006],[135,-0.0001],[136,0.0004],[139,0.0003],[140,0],[141,0.0004],[142,-0.0002],[147,-0.0002],[148,0.0001],[150,-0.0003],[151,-0.0002],[152,-0.0086],[153,-0.0001],[154,-0.008],[155,-0.0001],[156,-0.0069],[157,0.0002],[158,-0.006],[159,0.0001],[161,-0.0088],[162,-0.0002],[164,-0.0057],[165,-0.0074],[166,-0.0001],[167,-0.0005],[168,-0.0005],[169,0.0025],[170,0.0004],[171,-0.0002],[173,-0.0001],[175,0.0004],[176,-0.0003],[177,-0.0002],[178,0.0004],[179,0],[180,0.0006],[181,-0.0008],[182,0.0004],[183,0.0005],[184,0.0675],[185,0.0759],[186,-0.0004],[188,0.0001],[189,-0.0001],[190,0],[191,-0.0002],[192,0],[194,0.0001],[196,-0.0001],[197,0.0001],[198,-0.0002],[200,0],[201,-0.0064],[202,-0.0001],[203,-0.0066],[204,0],[205,-0.0089],[206,0],[207,0.0001],[209,0],[210,0.0022],[211,0.0048],[212,-0.0001],[214,-0.0001],[215,0.0002],[218,0],[221,0.0002],[222,-0.0001],[225,-0.0003],[227,0.0001],[230,-0.0001],[231,0.0001],[232,-0.0001],[233,-0.0106],[234,0],[235,-0.0099],[236,0],[237,-0.0089],[238,0.0001],[239,-0.0076],[240,0.0001],[241,-0.0065],[242,-0.0001],[243,-0.0053],[244,-0.0101],[245,0.0001],[246,-0.0036],[247,-0.0061],[248,0],[250,0],[251,0.0001],[252,-0.0001],[253,0],[254,-0.0001],[255,0],[256,-0.0003],[258,-0.0002],[259,0.0002],[260,0.0003],[262,-0.0002],[264,0.0002],[265,0.0001],[267,-0.0008],[268,0.0013],[269,0.0529],[270,0.033],[271,0.0121],[272,-0.0091],[273,-0.0298],[274,-0.0493],[275,-0.0016],[276,-0.0013],[277,-0.0001],[278,0.0105],[279,0.0001],[280,0.001],[281,0.0007],[282,-0.0016],[283,0.0012],[284,-0.0036],[285,0.0006],[286,-0.005],[287,0.0001],[288,-0.0079],[289,-0.0019],[290,-0.0017],[291,-0.0509],[292,-0.0415],[293,-0.03],[294,-0.0176],[295,-0.0086],[296,0.0052],[297,0.0216],[298,0.0404],[299,0],[300,0.0022],[301,0.0001],[302,0.0001],[303,-0.0004],[305,0.0006],[306,0],[307,0.0002],[308,0],[309,0.0002],[310,-0.0001],[312,-0.0002],[313,0.0003],[317,0.0001],[318,-0.0002],[319,0.0001],[320,0],[321,-0.0082],[322,0.0001],[323,-0.0037],[324,-0.0051]]
       }
       foot_l_ik_ctrl.Rotation.Z {
        type=MovieSceneScriptingFloatChannel
        default=0
        thinned_from=325
        keys [[0,-62.9662],[16,-62.966],[22,-63.0339],[38,-62.9642],[100,-62.9674],[106,-63.0339],[122,-62.9647],[183,-62.9679],[185,-62.9012],[268,-62.9665],[274,-63.0338],[290,-62.9654],[324,-62.9661]]
       }
       foot_l_ik_ctrl.Scale.X {type=MovieSceneScriptingFloatChannel, default=1, const [1,325]}
       foot_l_ik_ctrl.Scale.Y {type=MovieSceneScriptingFloatChannel, default=1, const [1,325]}
       foot_l_ik_ctrl.Scale.Z {type=MovieSceneScriptingFloatChannel, default=1, const [1,325]}
       ball_l_ik_ctrl.Location.X {
        type=MovieSceneScriptingFloatChannel
        default=0
        thinned_from=325
        keys [[0,0],[12,0],[13,0.0001],[50,0],[51,0.0002],[53,0],[54,0.0002],[88,0],[89,0.0001],[146,0],[147,-0.0002],[219,0],[220,-0.0001],[222,0],[223,-0.0001],[301,0],[302,-0.0004],[303,0],[316,0],[317,-0.0001],[324,0]]
       }
       ball_l_ik_ctrl.Location.Y {type=MovieSceneScriptingFloatChannel, default=0, thinned_from=325, keys [[0,0],[324,0]]}
       ball_l_ik_ctrl.Location.Z {type=MovieSceneScriptingFloatChannel, default=0, thinned_from=325, keys [[0,0],[324,0]]}
       ball_l_ik_ctrl.Rotation.X {
        type=MovieSceneScriptingFloatChannel
        default=0
        thinned_from=325
        keys [[0,0],[12,0],[13,-0.0001],[301,0],[302,0.0002],[316,0],[317,0.0001],[324,0]]
       }
       ball_l_ik_ctrl.Rotation.Y {type=MovieSceneScriptingFloatChannel, default=0, thinned_from=325, keys [[0,0],[12,0],[13,-0.0001],[301,0],[302,0.0002],[324,0]]}
       ball_l_ik_ctrl.Rotation.Z {
        type=MovieSceneScriptingFloatChannel
        default=0
        thinned_from=325
        keys [[0,0],[12,0],[13,-0.0009],[14,0],[50,0],[51,-0.0006],[52,0],[53,0],[54,-0.0006],[55,0],[88,0],[89,-0.0005],[90,0],[146,0],[147,0.0007],[148,0],[301,0],[302,0.0018],[303,0],[316,0],[317,0.0007],[318,0],[324,0]]
       }
       ball_l_ik_ctrl.Scale.X {type=MovieSceneScriptingFloatChannel, default=1, const [1,325]}
       ball_l_ik_ctrl.Scale.Y {type=MovieSceneScriptingFloatChannel, default=1, const [1,325]}
       ball_l_ik_ctrl.Scale.Z {type=MovieSceneScriptingFloatChannel, default=1, const [1,325]}
       foot_bk1_l_ctrl.Location.X {type=MovieSceneScriptingFloatChannel, default=0, const [0,325]}
       foot_bk1_l_ctrl.Location.Y {type=MovieSceneScriptingFloatChannel, default=0, const [0,325]}
       foot_bk1_l_ctrl.Location.Z {type=MovieSceneScriptingFloatChannel, default=0, const [0,325]}
       foot_bk1_l_ctrl.Rotation.X {type=MovieSceneScriptingFloatChannel, default=0, const [0,325]}
       foot_bk1_l_ctrl.Rotation.Y {type=MovieSceneScriptingFloatChannel, default=0, const [0,325]}
       foot_bk1_l_ctrl.Rotation.Z {type=MovieSceneScriptingFloatChannel, default=0, const [0,325]}
       foot_bk1_l_ctrl.Scale.X {type=MovieSceneScriptingFloatChannel, default=1, const [1,325]}
       foot_bk1_l_ctrl.Scale.Y {type=MovieSceneScriptingFloatChannel, default=1, const [1,325]}
       foot_bk1_l_ctrl.Scale.Z {type=MovieSceneScriptingFloatChannel, default=1, const [1,325]}
       foot_roll_l_ctrl.Location.X {type=MovieSceneScriptingFloatChannel, default=0, const [0,325]}
       foot_roll_l_ctrl.Location.Y {type=MovieSceneScriptingFloatChannel, default=0, const [0,325]}
       foot_roll_l_ctrl.Location.Z {type=MovieSceneScriptingFloatChannel, default=0, const [0,325]}
       foot_roll_l_ctrl.Rotation.X {type=MovieSceneScriptingFloatChannel, default=0, const [0,325]}
       foot_roll_l_ctrl.Rotation.Y {type=MovieSceneScriptingFloatChannel, default=0, const [0,325]}
       foot_roll_l_ctrl.Rotation.Z {type=MovieSceneScriptingFloatChannel, default=0, const [0,325]}
       foot_roll_l_ctrl.Scale.X {type=MovieSceneScriptingFloatChannel, default=1, const [1,325]}
       foot_roll_l_ctrl.Scale.Y {type=MovieSceneScriptingFloatChannel, default=1, const [1,325]}
       foot_roll_l_ctrl.Scale.Z {type=MovieSceneScriptingFloatChannel, default=1, const [1,325]}
       roll_blend_l_ctrl {type=MovieSceneScriptingFloatChannel, default=1, const [1,325]}
       tip_l_ctrl.Location.X {type=MovieSceneScriptingFloatChannel, default=0, const [0,325]}
       tip_l_ctrl.Location.Y {type=MovieSceneScriptingFloatChannel, default=0, const [0,325]}
       tip_l_ctrl.Location.Z {type=MovieSceneScriptingFloatChannel, default=0, const [0,325]}
       tip_l_ctrl.Rotation.X {type=MovieSceneScriptingFloatChannel, default=0, const [0,325]}
       tip_l_ctrl.Rotation.Y {type=MovieSceneScriptingFloatChannel, default=0, const [0,325]}
       tip_l_ctrl.Rotation.Z {type=MovieSceneScriptingFloatChannel, default=0, const [0,325]}
       tip_l_ctrl.Scale.X {type=MovieSceneScriptingFloatChannel, default=1, const [1,325]}
       tip_l_ctrl.Scale.Y {type=MovieSceneScriptingFloatChannel, default=1, const [1,325]}
       tip_l_ctrl.Scale.Z {type=MovieSceneScriptingFloatChannel, default=1, const [1,325]}
       heel_l_ctrl.Location.X {type=MovieSceneScriptingFloatChannel, default=0, const [0,325]}
       heel_l_ctrl.Location.Y {type=MovieSceneScriptingFloatChannel, default=0, const [0,325]}
       heel_l_ctrl.Location.Z {type=MovieSceneScriptingFloatChannel, default=0, const [0,325]}
       heel_l_ctrl.Rotation.X {type=MovieSceneScriptingFloatChannel, default=0, const [0,325]}
       heel_l_ctrl.Rotation.Y {type=MovieSceneScriptingFloatChannel, default=0, const [0,325]}
       heel_l_ctrl.Rotation.Z {type=MovieSceneScriptingFloatChannel, default=0, const [0,325]}
       heel_l_ctrl.Scale.X {type=MovieSceneScriptingFloatChannel, default=1, const [1,325]}
       heel_l_ctrl.Scale.Y {type=MovieSceneScriptingFloatChannel, default=1, const [1,325]}
       heel_l_ctrl.Scale.Z {type=MovieSceneScriptingFloatChannel, default=1, const [1,325]}
       leg_l_softness {type=MovieSceneScriptingFloatChannel, default=0, const [0,325]}
       leg_l_pv_ik_ctrl.Location.X {
        type=MovieSceneScriptingFloatChannel
        default=0
        thinned_from=325
        keys [[0,-0.1492],[12,-0.1495],[13,-0.1489],[14,-0.1494],[29,-0.1496],[31,-0.1491],[32,-0.1477],[34,-0.1435],[36,-0.1405],[37,-0.1393],[38,-0.1384],[41,-0.1376],[50,-0.1376],[51,-0.1372],[52,-0.1377],[53,-0.1377],[54,-0.1373],[55,-0.1377],[84,-0.1376],[86,-0.1381],[88,-0.1391],[89,-0.1406],[91,-0.1453],[94,-0.1533],[95,-0.1553],[96,-0.1567],[97,-0.1578],[99,-0.159],[143,-0.1596],[146,-0.1597],[147,-0.1603],[148,-0.1598],[172,-0.1599],[173,-0.1601],[182,-0.1594],[188,-0.157],[189,-0.1551],[191,-0.1488],[192,-0.1463],[193,-0.1443],[194,-0.143],[198,-0.1414],[209,-0.1408],[212,-0.1409],[248,-0.1417],[250,-0.1425],[251,-0.1436],[253,-0.1482],[255,-0.1502],[257,-0.1513],[294,-0.1499],[301,-0.1493],[302,-0.1505],[303,-0.1492],[313,-0.149],[314,-0.1488],[317,-0.1496],[318,-0.1491],[324,-0.1492]]
       }
       leg_l_pv_ik_ctrl.Location.Y {type=MovieSceneScriptingFloatChannel, default=0, thinned_from=325, keys [[0,-1.9438],[324,-1.9438]]}
       leg_l_pv_ik_ctrl.Location.Z {type=MovieSceneScriptingFloatChannel, default=0, thinned_from=325, keys [[0,-1.8045],[324,-1.8046]]}
       leg_l_pv_ik_ctrl.Rotation.X {type=MovieSceneScriptingFloatChannel, default=0, thinned_from=325, keys [[0,2.6168],[324,2.6168]]}
       leg_l_pv_ik_ctrl.Rotation.Y {
        type=MovieSceneScriptingFloatChannel
        default=0
        thinned_from=325
        keys [[0,-0.0193],[32,-0.0192],[89,-0.0188],[188,-0.0198],[251,-0.019],[324,-0.0193]]
       }
       leg_l_pv_ik_ctrl.Rotation.Z {
        type=MovieSceneScriptingFloatChannel
        default=0
        thinned_from=325
        keys [[0,0.2001],[12,0.2004],[13,0.1996],[14,0.2004],[29,0.2007],[31,0.1999],[32,0.1979],[34,0.192],[36,0.1877],[37,0.186],[38,0.1848],[41,0.1837],[50,0.1837],[51,0.1831],[52,0.1838],[53,0.1838],[54,0.1832],[55,0.1838],[84,0.1837],[86,0.1843],[88,0.1857],[89,0.1879],[91,0.1946],[94,0.2059],[95,0.2087],[96,0.2108],[97,0.2123],[99,0.214],[143,0.2148],[146,0.215],[147,0.2157],[148,0.215],[172,0.2152],[173,0.2155],[180,0.2151],[187,0.2119],[188,0.2112],[189,0.2084],[191,0.1996],[192,0.196],[193,0.193],[194,0.1912],[198,0.1891],[209,0.1881],[230,0.1882],[248,0.1894],[250,0.1905],[251,0.1922],[252,0.1956],[253,0.1986],[255,0.2015],[257,0.203],[259,0.2036],[301,0.2003],[302,0.2019],[303,0.2],[313,0.1998],[314,0.1995],[317,0.2006],[318,0.1999],[324,0.2001]]
       }
       leg_l_pv_ik_ctrl.Scale.X {type=MovieSceneScriptingFloatChannel, default=1, const [1,325]}
       leg_l_pv_ik_ctrl.Scale.Y {type=MovieSceneScriptingFloatChannel, default=1, const [1,325]}
       leg_l_pv_ik_ctrl.Scale.Z {type=MovieSceneScriptingFloatChannel, default=1, const [1,325]}
       thigh_r_fk_ctrl.Location.X {type=MovieSceneScriptingFloatChannel, default=0, thinned_from=325, keys [[0,-0.0002],[324,-0.0002]]}
       thigh_r_fk_ctrl.Location.Y {type=MovieSceneScriptingFloatChannel, default=0, thinned_from=325, keys [[0,0],[324,0]]}
       thigh_r_fk_ctrl.Location.Z {type=MovieSceneScriptingFloatChannel, default=0, thinned_from=325, keys [[0,0],[324,0]]}
       thigh_r_fk_ctrl.Rotation.X {
        type=MovieSceneScriptingFloatChannel
        default=1.2139
        thinned_from=325
        keys [[0,-7.5892],[28,-7.4932],[29,-7.4734],[30,-7.4275],[31,-7.3378],[32,-7.0438],[33,-6.5662],[34,-5.8875],[35,-5.3075],[36,-5.1993],[37,-5.1117],[38,-5.0432],[39,-4.9926],[40,-4.9587],[41,-4.9401],[85,-5.0152],[86,-5.037],[87,-5.1126],[88,-5.2601],[89,-5.4967],[90,-5.965],[91,-6.5824],[92,-7.0288],[93,-7.2448],[94,-7.392],[95,-7.4817],[96,-7.5252],[183,-7.5254],[184,-7.5428],[185,-7.5834],[186,-7.6542],[187,-7.7898],[188,-8.0522],[189,-8.4534],[192,-9.8134],[193,-10.112],[195,-10.3076],[197,-10.4371],[199,-10.5088],[247,-10.618],[248,-10.5714],[249,-10.4488],[250,-10.2046],[251,-9.761],[252,-8.8789],[253,-8.1604],[254,-7.9489],[255,-7.7958],[256,-7.691],[257,-7.625],[258,-7.5882],[324,-7.5885]]
       }
       thigh_r_fk_ctrl.Rotation.Y {
        type=MovieSceneScriptingFloatChannel
        default=-1.213
        thinned_from=325
        keys [[0,-3.2368],[29,-3.2809],[30,-3.2769],[31,-3.2658],[32,-3.2263],[34,-3.104],[35,-3.0563],[37,-2.9793],[38,-2.9554],[40,-2.9355],[85,-2.9007],[87,-2.9331],[88,-2.962],[89,-3.0066],[90,-3.0659],[91,-3.1356],[93,-3.2879],[94,-3.3558],[95,-3.4091],[96,-3.4476],[97,-3.4761],[98,-3.4961],[100,-3.5159],[183,-3.4979],[184,-3.5081],[185,-3.5389],[186,-3.5992],[187,-3.7745],[188,-3.979],[189,-4.0779],[190,-4.1224],[191,-4.1437],[192,-4.1404],[193,-4.1134],[194,-4.0971],[247,-4.1029],[248,-4.0966],[249,-4.0778],[250,-4.0423],[251,-3.9683],[252,-3.8514],[253,-3.6071],[254,-3.3995],[255,-3.3529],[256,-3.3226],[257,-3.305],[259,-3.2953],[283,-3.3055],[324,-3.2375]]
       }
       thigh_r_fk_ctrl.Rotation.Z {
        type=MovieSceneScriptingFloatChannel
        default=-0.1592
        thinned_from=325
        keys [[0,2.634],[1,2.634],[3,2.597],[4,2.5703],[6,2.5009],[8,2.4151],[14,2.1285],[16,2.0454],[22,1.8359],[23,1.806],[24,1.7811],[25,1.7616],[26,1.749],[29,1.7503],[30,1.7621],[31,1.7844],[32,1.8498],[33,1.9527],[34,2.0923],[35,2.2132],[37,2.2992],[38,2.3337],[40,2.3873],[44,2.476],[46,2.5392],[52,2.7765],[54,2.8445],[55,2.8718],[66,3.0494],[68,3.0738],[69,3.0738],[70,3.0958],[71,3.0958],[72,3.1152],[73,3.1152],[74,3.1318],[75,3.1318],[76,3.1455],[78,3.1455],[79,3.1602],[85,3.1602],[86,3.143],[87,3.1043],[88,3.0422],[89,2.9538],[90,2.8186],[91,2.6533],[92,2.5111],[93,2.4037],[94,2.3085],[95,2.2271],[96,2.1584],[97,2.0994],[98,2.0472],[99,2.0002],[100,1.9586],[105,1.7834],[107,1.7212],[108,1.6963],[109,1.6769],[110,1.6641],[115,1.6641],[116,1.6763],[117,1.6763],[118,1.6921],[119,1.6921],[120,1.7131],[123,1.7543],[126,1.807],[128,1.8513],[129,1.8805],[131,1.9507],[136,2.1498],[138,2.2172],[139,2.2442],[148,2.3942],[152,2.4454],[153,2.4454],[154,2.4674],[155,2.4674],[156,2.4867],[157,2.4867],[158,2.5034],[159,2.5034],[160,2.5174],[162,2.5174],[163,2.5322],[170,2.5322],[172,2.4696],[173,2.4333],[175,2.3462],[180,2.094],[182,2.0029],[185,1.8909],[186,1.8508],[187,1.8027],[188,1.7375],[190,1.5916],[192,1.4521],[193,1.4066],[194,1.3893],[195,1.3758],[196,1.368],[198,1.3616],[200,1.367],[201,1.367],[202,1.3822],[203,1.3822],[204,1.4031],[206,1.4292],[209,1.4789],[211,1.5191],[212,1.5446],[213,1.5766],[214,1.6139],[219,1.8302],[220,1.8698],[221,1.9049],[222,1.9342],[223,1.9563],[231,2.0821],[234,2.12],[235,2.12],[236,2.1422],[237,2.1422],[238,2.162],[239,2.162],[240,2.1792],[241,2.1792],[242,2.1937],[243,2.1937],[244,2.2057],[246,2.2057],[248,2.2287],[249,2.2485],[250,2.2854],[251,2.3478],[252,2.4681],[253,2.5555],[254,2.5631],[255,2.5576],[256,2.5408],[257,2.5143],[258,2.4798],[260,2.3959],[266,2.112],[268,2.03],[273,1.8566],[275,1.795],[276,1.7704],[277,1.7514],[278,1.7394],[283,1.7394],[284,1.7539],[285,1.7539],[286,1.7707],[287,1.7707],[288,1.7926],[291,1.8353],[294,1.8896],[296,1.9343],[298,1.995],[305,2.261],[307,2.3267],[315,2.5232],[318,2.5804],[320,2.609],[321,2.609],[322,2.6274],[324,2.6274]]
       }
       thigh_r_fk_ctrl.Scale.X {type=MovieSceneScriptingFloatChannel, default=1, const [1,325]}
       thigh_r_fk_ctrl.Scale.Y {type=MovieSceneScriptingFloatChannel, default=1, const [1,325]}
       thigh_r_fk_ctrl.Scale.Z {type=MovieSceneScriptingFloatChannel, default=1, const [1,325]}
       calf_r_fk_ctrl.Location.X {
        type=MovieSceneScriptingFloatChannel
        default=0
        thinned_from=325
        keys [[0,-0.6018],[2,-0.6031],[3,-0.6054],[4,-0.6096],[5,-0.616],[6,-0.6251],[7,-0.6375],[8,-0.6535],[9,-0.6747],[10,-0.7011],[11,-0.7313],[14,-0.8296],[15,-0.8599],[16,-0.8866],[17,-0.908],[19,-0.9413],[21,-0.9675],[23,-0.9866],[25,-0.9982],[27,-1.0021],[35,-1.0023],[38,-0.9955],[40,-0.9858],[42,-0.972],[43,-0.9635],[44,-0.952],[45,-0.9362],[46,-0.917],[47,-0.895],[52,-0.7692],[53,-0.7461],[54,-0.7253],[55,-0.7076],[56,-0.6937],[58,-0.6724],[60,-0.6551],[62,-0.6413],[64,-0.6307],[66,-0.6229],[69,-0.6158],[75,-0.6129],[78,-0.6155],[80,-0.6203],[82,-0.6289],[83,-0.6349],[84,-0.6422],[85,-0.6512],[86,-0.6619],[88,-0.6873],[91,-0.7312],[94,-0.7814],[99,-0.8697],[105,-0.9606],[107,-0.9868],[108,-0.9973],[109,-1.0053],[110,-1.0105],[111,-1.0123],[113,-1.01],[116,-0.9988],[119,-0.98],[125,-0.9275],[128,-0.8967],[130,-0.8703],[137,-0.7601],[139,-0.7331],[143,-0.692],[147,-0.6566],[151,-0.6272],[154,-0.6095],[157,-0.5959],[160,-0.5873],[164,-0.5826],[168,-0.582],[169,-0.583],[170,-0.5863],[171,-0.5914],[172,-0.5984],[173,-0.6068],[175,-0.6276],[178,-0.6655],[184,-0.7479],[185,-0.7639],[186,-0.7838],[187,-0.8125],[188,-0.8441],[190,-0.8992],[191,-0.9238],[192,-0.9445],[193,-0.9614],[194,-0.9749],[195,-0.9851],[197,-1.0002],[200,-1.0155],[205,-1.0263],[211,-1.0277],[213,-1.0238],[215,-1.0133],[217,-0.9976],[220,-0.9668],[224,-0.9182],[226,-0.8887],[234,-0.7518],[236,-0.7204],[238,-0.6933],[239,-0.6818],[240,-0.6719],[241,-0.6637],[243,-0.6511],[245,-0.6431],[249,-0.634],[250,-0.6306],[251,-0.6237],[252,-0.6118],[253,-0.5922],[254,-0.5774],[255,-0.5737],[256,-0.5722],[257,-0.5726],[258,-0.5747],[260,-0.5827],[262,-0.5946],[264,-0.6101],[266,-0.6298],[268,-0.6541],[269,-0.668],[270,-0.6871],[271,-0.7141],[272,-0.7467],[275,-0.8577],[276,-0.892],[277,-0.9214],[278,-0.9438],[279,-0.9571],[282,-0.9771],[285,-0.9896],[290,-0.9985],[296,-0.9989],[299,-0.9919],[301,-0.9824],[303,-0.9689],[305,-0.9509],[307,-0.9283],[308,-0.9153],[309,-0.8997],[310,-0.8808],[311,-0.8592],[313,-0.8102],[317,-0.7057],[318,-0.6817],[319,-0.6598],[320,-0.6405],[321,-0.6245],[322,-0.6123],[323,-0.6045],[324,-0.6018]]
       }
       calf_r_fk_ctrl.Location.Y {
        type=MovieSceneScriptingFloatChannel
        default=0
        thinned_from=325
        keys [[0,0.0495],[3,0.0498],[5,0.0506],[6,0.0514],[7,0.0524],[8,0.0537],[9,0.0555],[10,0.0576],[12,0.0628],[14,0.0682],[15,0.0707],[16,0.0729],[17,0.0747],[19,0.0774],[21,0.0796],[23,0.0811],[25,0.0821],[27,0.0824],[35,0.0824],[38,0.0819],[41,0.0805],[43,0.0792],[44,0.0783],[45,0.077],[46,0.0754],[48,0.0716],[52,0.0633],[54,0.0596],[55,0.0582],[56,0.057],[59,0.0545],[62,0.0527],[66,0.0512],[72,0.0504],[78,0.0506],[81,0.0513],[83,0.0522],[85,0.0535],[88,0.0565],[92,0.0614],[99,0.0715],[106,0.0801],[108,0.082],[109,0.0827],[110,0.0831],[112,0.0832],[115,0.0825],[119,0.0806],[127,0.0746],[129,0.0727],[137,0.0625],[139,0.0603],[146,0.0547],[152,0.051],[157,0.049],[163,0.0479],[169,0.0479],[171,0.0486],[173,0.0499],[176,0.0526],[184,0.0615],[185,0.0628],[186,0.0645],[187,0.0668],[188,0.0694],[190,0.0739],[191,0.076],[192,0.0777],[193,0.0791],[194,0.0802],[196,0.0817],[199,0.0832],[203,0.0842],[211,0.0845],[213,0.0842],[215,0.0833],[218,0.0813],[223,0.0766],[226,0.0731],[235,0.0605],[238,0.057],[240,0.0552],[242,0.054],[245,0.0529],[250,0.0518],[251,0.0513],[252,0.0503],[253,0.0487],[254,0.0475],[256,0.047],[258,0.0472],[261,0.0484],[264,0.0502],[267,0.0527],[269,0.0549],[270,0.0565],[271,0.0587],[272,0.0614],[275,0.0705],[276,0.0734],[277,0.0758],[278,0.0776],[279,0.0787],[282,0.0804],[286,0.0816],[295,0.0822],[298,0.0818],[301,0.0808],[304,0.079],[307,0.0763],[308,0.0753],[309,0.074],[310,0.0724],[312,0.0687],[317,0.058],[319,0.0542],[320,0.0527],[321,0.0513],[322,0.0503],[323,0.0497],[324,0.0495]]
       }
       calf_r_fk_ctrl.Location.Z {type=MovieSceneScriptingFloatChannel, default=0, thinned_from=325, keys [[0,0.0006],[324,0.0006]]}
       calf_r_fk_ctrl.Rotation.X {type=MovieSceneScriptingFloatChannel, default=0.0561, thinned_from=325, keys [[0,0.0055],[324,0.0055]]}
       calf_r_fk_ctrl.Rotation.Y {type=MovieSceneScriptingFloatChannel, default=1.1982, thinned_from=325, keys [[0,0.0049],[324,0.0049]]}
       calf_r_fk_ctrl.Rotation.Z {type=MovieSceneScriptingFloatChannel, default=0.2616, thinned_from=325, keys [[0,-4.9607],[324,-4.9607]]}
       calf_r_fk_ctrl.Scale.X {type=MovieSceneScriptingFloatChannel, default=1, const [1,325]}
       calf_r_fk_ctrl.Scale.Y {type=MovieSceneScriptingFloatChannel, default=1, const [1,325]}
       calf_r_fk_ctrl.Scale.Z {type=MovieSceneScriptingFloatChannel, default=1, const [1,325]}
       foot_r_fk_ctrl.Location.X {
        type=MovieSceneScriptingFloatChannel
        default=0
        thinned_from=325
        keys [[0,-0.6372],[2,-0.6386],[3,-0.6412],[4,-0.6457],[5,-0.6527],[6,-0.6627],[7,-0.6762],[8,-0.6938],[9,-0.7169],[10,-0.7458],[11,-0.7789],[14,-0.8864],[15,-0.9195],[16,-0.9487],[17,-0.9721],[19,-1.0085],[21,-1.0372],[23,-1.0581],[25,-1.0708],[27,-1.075],[35,-1.0752],[38,-1.0679],[40,-1.0572],[42,-1.0421],[43,-1.0328],[44,-1.0202],[45,-1.003],[46,-0.982],[47,-0.9579],[52,-0.8203],[53,-0.795],[54,-0.7723],[55,-0.7529],[56,-0.7377],[58,-0.7145],[60,-0.6955],[62,-0.6804],[64,-0.6688],[66,-0.6603],[69,-0.6525],[73,-0.6493],[77,-0.6507],[79,-0.6544],[81,-0.6616],[83,-0.6734],[84,-0.6814],[85,-0.6912],[86,-0.703],[88,-0.7308],[91,-0.7788],[94,-0.8337],[99,-0.9302],[105,-1.0297],[107,-1.0583],[108,-1.0697],[109,-1.0785],[110,-1.0842],[111,-1.0862],[113,-1.0837],[116,-1.0714],[119,-1.0509],[125,-0.9935],[128,-0.9598],[130,-0.9309],[137,-0.8103],[139,-0.7808],[143,-0.7358],[147,-0.6972],[151,-0.665],[154,-0.6456],[157,-0.6307],[159,-0.6238],[162,-0.6179],[168,-0.6155],[169,-0.6167],[170,-0.6202],[171,-0.6259],[172,-0.6335],[173,-0.6427],[175,-0.6654],[177,-0.6923],[184,-0.797],[185,-0.8145],[186,-0.8363],[187,-0.8676],[188,-0.9022],[190,-0.9625],[191,-0.9894],[192,-1.012],[193,-1.0306],[194,-1.0452],[195,-1.0564],[197,-1.0729],[200,-1.0896],[204,-1.1003],[211,-1.103],[213,-1.0987],[215,-1.0873],[217,-1.0701],[220,-1.0365],[224,-0.9832],[226,-0.951],[234,-0.8013],[236,-0.767],[238,-0.7373],[239,-0.7247],[240,-0.7139],[241,-0.7049],[243,-0.6912],[245,-0.6823],[249,-0.6725],[250,-0.6687],[251,-0.6611],[252,-0.6481],[253,-0.6268],[254,-0.6105],[255,-0.6065],[256,-0.6048],[257,-0.6053],[258,-0.6075],[259,-0.6113],[261,-0.6223],[263,-0.6373],[265,-0.6565],[267,-0.6805],[269,-0.7096],[270,-0.7305],[271,-0.76],[272,-0.7957],[275,-0.9171],[276,-0.9546],[277,-0.9867],[278,-1.0113],[279,-1.0259],[282,-1.0476],[285,-1.0613],[290,-1.0711],[296,-1.0715],[299,-1.0639],[301,-1.0535],[303,-1.0387],[305,-1.019],[307,-0.9944],[308,-0.9801],[309,-0.963],[310,-0.9424],[311,-0.9187],[313,-0.8651],[317,-0.7508],[318,-0.7246],[319,-0.7006],[320,-0.6795],[321,-0.662],[322,-0.6487],[323,-0.6402],[324,-0.6372]]
       }
       foot_r_fk_ctrl.Location.Y {
        type=MovieSceneScriptingFloatChannel
        default=-0.0329
        thinned_from=325
        keys [[0,-0.0029],[44,-0.0046],[92,-0.0036],[186,-0.0037],[270,-0.0033],[308,-0.0044],[324,-0.0029]]
       }
       foot_r_fk_ctrl.Location.Z {type=MovieSceneScriptingFloatChannel, default=-0.9075, thinned_from=325, keys [[0,0.0007],[324,0.0007]]}
       foot_r_fk_ctrl.Rotation.X {
        type=MovieSceneScriptingFloatChannel
        default=-1.1648
        thinned_from=325
        keys [[0,0.3482],[30,0.3496],[31,0.3489],[32,0.3462],[34,0.3375],[37,0.3282],[38,0.3264],[42,0.3243],[84,0.3236],[86,0.3246],[87,0.3258],[88,0.328],[89,0.3314],[91,0.3408],[93,0.3518],[94,0.3567],[95,0.3605],[97,0.3653],[99,0.3676],[182,0.3669],[188,0.3626],[189,0.3589],[191,0.3468],[193,0.3376],[194,0.3353],[199,0.3322],[248,0.3337],[250,0.3352],[251,0.3373],[253,0.3459],[255,0.3496],[257,0.3516],[324,0.3482]]
       }
       foot_r_fk_ctrl.Rotation.Y {
        type=MovieSceneScriptingFloatChannel
        default=0.1067
        thinned_from=325
        keys [[0,3.2822],[29,3.3043],[31,3.2917],[32,3.2571],[34,3.1475],[35,3.1028],[37,3.0293],[38,3.0064],[40,2.9877],[84,2.9704],[86,2.9828],[87,2.9981],[88,3.0259],[89,3.0685],[90,3.1247],[93,3.3268],[94,3.3886],[95,3.4368],[96,3.4715],[97,3.4972],[99,3.5262],[181,3.521],[187,3.4767],[188,3.4641],[189,3.4185],[191,3.2685],[192,3.2077],[193,3.1545],[194,3.1253],[197,3.0962],[247,3.0997],[249,3.1094],[250,3.1213],[251,3.1478],[252,3.2045],[253,3.254],[255,3.3],[257,3.3245],[324,3.2822]]
       }
       foot_r_fk_ctrl.Rotation.Z {
        type=MovieSceneScriptingFloatChannel
        default=-0.1036
        thinned_from=325
        keys [[0,4.5761],[27,4.7186],[30,4.7089],[31,4.6941],[32,4.6458],[33,4.5702],[34,4.472],[35,4.39],[37,4.3286],[39,4.2943],[84,4.1622],[86,4.1732],[87,4.1919],[88,4.2249],[89,4.2751],[90,4.356],[91,4.4566],[92,4.5434],[93,4.6081],[94,4.662],[95,4.7037],[97,4.7558],[111,4.8348],[170,4.7061],[186,4.8004],[187,4.8119],[189,4.8535],[192,4.8906],[210,4.8809],[248,4.7755],[250,4.7477],[251,4.7169],[252,4.6584],[253,4.6154],[279,4.7308],[295,4.6991],[324,4.5761]]
       }
       foot_r_fk_ctrl.Scale.X {type=MovieSceneScriptingFloatChannel, default=1, const [1,325]}
       foot_r_fk_ctrl.Scale.Y {type=MovieSceneScriptingFloatChannel, default=1, const [1,325]}
       foot_r_fk_ctrl.Scale.Z {type=MovieSceneScriptingFloatChannel, default=1, const [1,325]}
       ball_r_fk_ctrl.Location.X {type=MovieSceneScriptingFloatChannel, default=0, const [0,325]}
       ball_r_fk_ctrl.Location.Y {type=MovieSceneScriptingFloatChannel, default=0, const [0,325]}
       ball_r_fk_ctrl.Location.Z {type=MovieSceneScriptingFloatChannel, default=0, const [0,325]}
       ball_r_fk_ctrl.Rotation.X {type=MovieSceneScriptingFloatChannel, default=0, const [0,325]}
       ball_r_fk_ctrl.Rotation.Y {type=MovieSceneScriptingFloatChannel, default=0, const [0,325]}
       ball_r_fk_ctrl.Rotation.Z {type=MovieSceneScriptingFloatChannel, default=0, const [0,325]}
       ball_r_fk_ctrl.Scale.X {type=MovieSceneScriptingFloatChannel, default=1, const [1,325]}
       ball_r_fk_ctrl.Scale.Y {type=MovieSceneScriptingFloatChannel, default=1, const [1,325]}
       ball_r_fk_ctrl.Scale.Z {type=MovieSceneScriptingFloatChannel, default=1, const [1,325]}
       foot_r_ik_ctrl.Location.X {
        type=MovieSceneScriptingFloatChannel
        default=0
        thinned_from=325
        keys [[0,-7.0484],[16,-7.0474],[17,-7.0057],[22,-7.0886],[23,-7.0495],[38,-7.0496],[39,-7.0944],[46,-7.01],[47,-7.0476],[84,-7.0486],[85,-7.039],[100,-7.0484],[101,-7.0057],[106,-7.0886],[107,-7.0476],[122,-7.0483],[123,-7.0898],[130,-7.0139],[131,-7.0491],[169,-7.0505],[183,-7.0481],[184,-7.0339],[185,-7.1031],[186,-7.0477],[268,-7.0474],[269,-7.0059],[274,-7.0884],[275,-7.0495],[283,-7.0571],[286,-7.0482],[290,-7.0495],[291,-7.0898],[298,-7.0157],[299,-7.0483],[324,-7.0526]]
       }
       foot_r_ik_ctrl.Location.Y {
        type=MovieSceneScriptingFloatChannel
        default=0
        thinned_from=325
        keys [[0,5.7088],[1,5.7024],[16,5.7057],[17,5.5871],[22,5.822],[23,5.7118],[38,5.7116],[39,5.8242],[43,5.728],[45,5.6569],[46,5.6117],[47,5.7065],[68,5.7084],[69,5.7241],[70,5.7085],[71,5.7215],[72,5.7088],[73,5.7207],[74,5.7086],[75,5.718],[76,5.7088],[78,5.7233],[79,5.7083],[84,5.7215],[100,5.7086],[101,5.5876],[106,5.8215],[107,5.7063],[115,5.7186],[116,5.7074],[117,5.7193],[118,5.7104],[119,5.7217],[122,5.7085],[123,5.825],[127,5.7281],[129,5.6565],[130,5.6111],[131,5.7105],[152,5.7085],[153,5.7241],[154,5.7093],[155,5.7223],[156,5.7089],[157,5.7202],[158,5.7091],[159,5.7184],[160,5.7079],[162,5.7234],[163,5.7086],[168,5.7204],[169,5.7146],[170,5.6945],[171,5.7089],[183,5.7078],[184,5.5768],[185,5.6001],[186,5.7089],[200,5.7086],[201,5.7187],[202,5.7086],[203,5.7222],[204,5.7086],[234,5.7085],[235,5.7244],[236,5.7085],[237,5.7227],[238,5.7085],[239,5.7208],[240,5.7085],[241,5.7192],[242,5.7086],[243,5.7174],[244,5.7086],[246,5.7215],[247,5.7088],[268,5.7059],[269,5.5878],[274,5.8215],[275,5.7118],[281,5.7056],[283,5.7156],[284,5.7095],[285,5.7175],[286,5.7082],[287,5.7218],[288,5.7103],[290,5.7119],[291,5.8255],[295,5.7282],[298,5.6157],[299,5.7083],[320,5.7087],[321,5.7227],[322,5.7083],[324,5.7174]]
       }
       foot_r_ik_ctrl.Location.Z {
        type=MovieSceneScriptingFloatChannel
        default=0
        thinned_from=325
        keys [[0,0.0002],[22,0.0003],[29,0.0001],[38,0.0001],[39,-0.0006],[44,0.0003],[46,0.001],[47,0.0002],[68,0.0002],[69,0],[70,0.0002],[71,0],[72,0.0001],[73,0],[78,0],[84,0.0002],[85,0.0006],[86,0.0002],[104,0.0002],[106,0.0004],[107,0.0002],[122,0.0002],[123,0.0004],[130,0.0001],[166,0.0002],[168,0],[183,0.0002],[184,-0.0036],[185,-0.0105],[186,0.0002],[200,0.0002],[208,0.0002],[209,0.0003],[234,0.0001],[235,0.0003],[274,0.0003],[283,0],[284,0.0002],[294,0.0002],[295,0],[324,0.0001]]
       }
       foot_r_ik_ctrl.Rotation.X {
        type=MovieSceneScriptingFloatChannel
        default=0
        thinned_from=325
        keys [[0,-0.0001],[1,0.0044],[2,0.0001],[3,0.0006],[4,-0.0003],[6,-0.0004],[7,-0.0002],[8,0.0003],[9,0.0005],[10,0],[12,0.0005],[13,0.0003],[14,-0.0004],[15,0.0017],[16,0.002],[17,0.0762],[18,0.0475],[19,0.0174],[20,-0.0131],[21,-0.043],[22,-0.0711],[23,-0.0024],[24,-0.0019],[25,-0.0002],[26,0],[27,0.0046],[28,0.0046],[29,0.0008],[30,-0.0006],[31,-0.001],[32,-0.0005],[33,-0.0005],[34,0.0002],[35,0.0002],[36,0.0005],[37,-0.0008],[38,-0.0024],[39,-0.0737],[40,-0.0601],[41,-0.0454],[42,-0.0295],[43,-0.0124],[44,0.0079],[45,0.0329],[46,0.0618],[47,0.001],[48,-0.0006],[49,-0.0006],[50,0],[52,-0.0002],[53,0.0004],[54,0.0003],[55,-0.0001],[56,0.0005],[57,0.0005],[58,-0.0003],[59,0.0001],[60,0.0001],[61,-0.0007],[62,0.0002],[63,0.0002],[64,-0.0008],[65,0.0002],[67,-0.0005],[68,0.0001],[69,-0.0116],[70,0],[71,-0.0097],[72,-0.0002],[73,-0.009],[74,0.001],[75,-0.007],[76,-0.0002],[77,-0.0058],[78,-0.0109],[79,0.0002],[80,-0.0033],[81,-0.0061],[82,-0.0073],[83,-0.0088],[84,-0.0082],[85,-0.0015],[86,0.0002],[88,0],[92,0],[93,-0.0008],[94,-0.0003],[95,-0.0013],[96,0.0003],[97,0.0002],[99,-0.0012],[100,0.0001],[101,0.0762],[102,0.0475],[103,0.0174],[104,-0.0131],[105,-0.0429],[106,-0.0711],[107,0.0017],[108,-0.0003],[110,-0.0002],[111,0.0042],[112,0.0042],[114,-0.0024],[115,-0.0074],[116,0.0009],[117,-0.0078],[118,-0.0014],[119,-0.0097],[120,-0.0014],[122,-0.0001],[123,-0.0733],[124,-0.0598],[125,-0.0452],[126,-0.0293],[127,-0.0123],[128,0.0079],[129,0.0328],[130,0.0614],[131,-0.0011],[132,0.0009],[133,-0.0006],[134,0.0009],[135,-0.0001],[136,0.0006],[137,0.0004],[139,0.0004],[140,0],[141,0.0005],[142,-0.0002],[145,-0.0003],[146,-0.0001],[147,-0.0001],[148,0.0002],[149,0],[150,-0.0005],[152,0],[153,-0.0114],[154,-0.0005],[155,-0.0101],[156,-0.0003],[157,-0.0086],[158,-0.0004],[159,-0.0073],[160,0.0005],[161,-0.0056],[162,-0.0109],[163,-0.0001],[164,-0.0039],[165,-0.0063],[166,-0.008],[167,-0.0086],[168,-0.0086],[169,-0.0044],[170,0.0102],[171,-0.0003],[173,-0.0001],[175,0.0006],[176,-0.0004],[177,-0.0004],[178,0.0005],[179,0],[180,0.0009],[181,-0.0011],[182,0.0006],[183,0.0006],[184,0.0737],[185,0.0407],[186,0],[194,-0.0001],[196,0.0001],[197,0],[200,0],[201,-0.0067],[202,0.0001],[203,-0.0097],[204,0],[208,0],[210,0.0058],[211,0],[234,0],[235,-0.0113],[236,0],[237,-0.0101],[238,0],[239,-0.0087],[240,0],[241,-0.0075],[242,0],[243,-0.0062],[244,0],[245,-0.0048],[246,-0.0087],[247,-0.0001],[251,0],[253,0],[255,-0.0001],[256,-0.0004],[257,-0.0003],[258,-0.0004],[259,0.0003],[260,0.0005],[261,0.0002],[262,-0.0002],[264,0.0003],[265,0.0003],[267,-0.0012],[268,0.0019],[269,0.0762],[270,0.0475],[271,0.0174],[272,-0.0131],[273,-0.0429],[274,-0.0711],[275,-0.0023],[276,-0.0019],[277,-0.0002],[278,0],[279,0.004],[280,0.0034],[281,0.0011],[282,-0.0041],[283,-0.007],[284,-0.0005],[285,-0.0069],[286,0.0002],[287,-0.0103],[288,-0.0014],[289,-0.0028],[290,-0.0024],[291,-0.0733],[292,-0.0598],[293,-0.0451],[294,-0.0258],[295,-0.005],[296,0.0075],[297,0.0312],[298,0.0582],[299,0],[300,0.0032],[301,0.0001],[302,0.0003],[303,-0.0006],[305,0.0008],[306,0.0001],[307,0.0003],[308,0],[309,0.0003],[310,-0.0001],[312,-0.0002],[313,0.0004],[315,0.0002],[317,0.0003],[318,-0.0004],[319,0.0001],[320,-0.0001],[321,-0.0107],[322,0.0002],[323,-0.0049],[324,-0.0067]]
       }
       foot_r_ik_ctrl.Rotation.Y {
        type=MovieSceneScriptingFloatChannel
        default=0
        thinned_from=325
        keys [[0,0],[1,-0.0005],[2,0],[14,0],[15,-0.0002],[16,-0.0002],[17,-0.0096],[21,0.0054],[22,0.009],[23,0.0003],[26,0],[27,-0.0003],[31,0.0001],[36,0],[38,0.0002],[39,0.006],[42,0.0024],[43,0.001],[44,-0.0006],[45,-0.0026],[46,-0.0049],[47,-0.0001],[68,0],[69,0.0008],[70,0],[71,0.0007],[72,0],[73,0.0007],[74,-0.0001],[75,0.0006],[76,0],[78,0.001],[79,0],[82,0.0013],[83,0.002],[84,0.0039],[85,0.008],[86,0],[94,0],[95,0.0002],[96,0],[99,0.0002],[100,0],[101,-0.0096],[105,0.0054],[106,0.009],[107,-0.0002],[108,0],[110,0],[111,-0.0005],[112,-0.0005],[114,0.0001],[115,0.0006],[116,-0.0001],[117,0.0007],[118,0.0002],[119,0.0009],[120,0.0002],[122,0],[123,0.0093],[125,0.0057],[127,0.0016],[128,-0.001],[129,-0.0041],[130,-0.0077],[131,0.0001],[132,-0.0001],[133,0.0001],[134,-0.0001],[135,0],[141,-0.0001],[152,0],[153,0.0012],[154,0.0001],[155,0.0011],[156,0],[157,0.0009],[158,0.0001],[159,0.0008],[160,-0.0001],[162,0.0012],[163,0],[165,0.0008],[168,0.001],[169,0.0006],[170,-0.0012],[171,0],[175,-0.0001],[177,0.0001],[180,-0.0001],[181,0.0001],[182,-0.0001],[183,-0.0001],[184,-0.0297],[185,-0.0643],[186,0.0005],[188,-0.0001],[189,0.0002],[190,0],[191,0.0003],[192,0],[193,0],[194,-0.0002],[196,0.0002],[197,-0.0001],[198,0.0004],[200,0],[201,0.0025],[202,0.0002],[203,0.0014],[204,0.0001],[207,-0.0001],[208,0],[210,-0.0012],[211,-0.0001],[212,0.0001],[214,0.0001],[215,-0.0002],[216,0.0001],[218,0.0001],[219,-0.0001],[220,0],[221,-0.0002],[222,0.0001],[225,0.0003],[227,-0.0001],[230,0.0001],[231,-0.0001],[232,0.0001],[234,0],[235,0.0018],[236,0],[237,0.0017],[238,-0.0001],[239,0.0014],[240,-0.0001],[241,0.0014],[242,0.0001],[243,0.0013],[244,0],[245,0.0011],[246,0.0025],[247,0.0003],[248,0],[251,-0.0001],[252,0.0001],[253,0],[258,0.0001],[267,0.0002],[268,-0.0002],[269,-0.0096],[273,0.0054],[274,0.009],[275,0.0003],[278,0],[279,-0.001],[281,-0.0023],[282,-0.0025],[283,-0.0032],[284,0.0001],[285,-0.0003],[289,0.0003],[290,0.0003],[291,0.0093],[293,0.0057],[295,0.0034],[296,-0.0009],[297,-0.0039],[298,-0.0073],[299,0],[300,-0.0004],[301,0],[305,-0.0001],[320,0],[321,0.0003],[322,0],[324,0.0001]]
       }
       foot_r_ik_ctrl.Rotation.Z {
        type=MovieSceneScriptingFloatChannel
        default=0
        thinned_from=325
        keys [[0,26.6285],[16,26.6283],[17,26.5562],[22,26.6962],[23,26.628],[38,26.6265],[39,26.6972],[46,26.5712],[100,26.6297],[101,26.5562],[106,26.6962],[107,26.6293],[122,26.627],[123,26.6983],[130,26.5702],[131,26.6309],[183,26.6294],[184,26.5495],[268,26.6289],[269,26.5562],[274,26.6962],[275,26.6282],[290,26.6277],[291,26.6983],[298,26.5733],[324,26.6285]]
       }
       foot_r_ik_ctrl.Scale.X {type=MovieSceneScriptingFloatChannel, default=1, const [1,325]}
       foot_r_ik_ctrl.Scale.Y {type=MovieSceneScriptingFloatChannel, default=1, const [1,325]}
       foot_r_ik_ctrl.Scale.Z {type=MovieSceneScriptingFloatChannel, default=1, const [1,325]}
       ball_r_ik_ctrl.Location.X {
        type=MovieSceneScriptingFloatChannel
        default=0
        thinned_from=325
        keys [[0,0],[90,0],[91,0.0002],[109,0],[110,-0.0004],[111,0],[176,0],[177,-0.0001],[182,0],[183,-0.0002],[184,0],[193,0.0001],[254,0],[255,-0.0001],[264,0],[265,0.0003],[266,0],[310,0],[311,-0.0002],[324,0]]
       }
       ball_r_ik_ctrl.Location.Y {type=MovieSceneScriptingFloatChannel, default=0, thinned_from=325, keys [[0,0],[144,0],[145,-0.0001],[182,0],[183,0.0001],[324,0]]}
       ball_r_ik_ctrl.Location.Z {type=MovieSceneScriptingFloatChannel, default=0, thinned_from=325, keys [[0,0],[324,0]]}
       ball_r_ik_ctrl.Rotation.X {
        type=MovieSceneScriptingFloatChannel
        default=0
        thinned_from=325
        keys [[0,0],[90,0],[91,-0.0001],[109,0],[110,0.0002],[264,0],[265,-0.0001],[324,0]]
       }
       ball_r_ik_ctrl.Rotation.Y {type=MovieSceneScriptingFloatChannel, default=0, thinned_from=325, keys [[0,0],[324,0]]}
       ball_r_ik_ctrl.Rotation.Z {
        type=MovieSceneScriptingFloatChannel
        default=0
        thinned_from=325
        keys [[0,0],[90,0],[91,-0.0009],[92,0],[109,0],[110,0.0015],[111,0],[144,0],[145,0.0003],[146,0],[176,0],[177,0.0004],[178,0],[182,0],[183,0.0009],[184,0],[254,0],[255,0.0004],[256,0],[264,0],[265,-0.0012],[266,0],[310,0],[311,0.0007],[312,0],[324,0]]
       }
       ball_r_ik_ctrl.Scale.X {type=MovieSceneScriptingFloatChannel, default=1, const [1,325]}
       ball_r_ik_ctrl.Scale.Y {type=MovieSceneScriptingFloatChannel, default=1, const [1,325]}
       ball_r_ik_ctrl.Scale.Z {type=MovieSceneScriptingFloatChannel, default=1, const [1,325]}
       foot_bk1_r_ctrl.Location.X {type=MovieSceneScriptingFloatChannel, default=0, const [0,325]}
       foot_bk1_r_ctrl.Location.Y {type=MovieSceneScriptingFloatChannel, default=0, const [0,325]}
       foot_bk1_r_ctrl.Location.Z {type=MovieSceneScriptingFloatChannel, default=0, const [0,325]}
       foot_bk1_r_ctrl.Rotation.X {type=MovieSceneScriptingFloatChannel, default=0, const [0,325]}
       foot_bk1_r_ctrl.Rotation.Y {type=MovieSceneScriptingFloatChannel, default=0, const [0,325]}
       foot_bk1_r_ctrl.Rotation.Z {type=MovieSceneScriptingFloatChannel, default=0, const [0,325]}
       foot_bk1_r_ctrl.Scale.X {type=MovieSceneScriptingFloatChannel, default=1, const [1,325]}
       foot_bk1_r_ctrl.Scale.Y {type=MovieSceneScriptingFloatChannel, default=1, const [1,325]}
       foot_bk1_r_ctrl.Scale.Z {type=MovieSceneScriptingFloatChannel, default=1, const [1,325]}
       foot_roll_r_ctrl.Location.X {type=MovieSceneScriptingFloatChannel, default=0, const [0,325]}
       foot_roll_r_ctrl.Location.Y {type=MovieSceneScriptingFloatChannel, default=0, const [0,325]}
       foot_roll_r_ctrl.Location.Z {type=MovieSceneScriptingFloatChannel, default=0, const [0,325]}
       foot_roll_r_ctrl.Rotation.X {type=MovieSceneScriptingFloatChannel, default=0, const [0,325]}
       foot_roll_r_ctrl.Rotation.Y {type=MovieSceneScriptingFloatChannel, default=0, const [0,325]}
       foot_roll_r_ctrl.Rotation.Z {type=MovieSceneScriptingFloatChannel, default=0, const [0,325]}
       foot_roll_r_ctrl.Scale.X {type=MovieSceneScriptingFloatChannel, default=1, const [1,325]}
       foot_roll_r_ctrl.Scale.Y {type=MovieSceneScriptingFloatChannel, default=1, const [1,325]}
       foot_roll_r_ctrl.Scale.Z {type=MovieSceneScriptingFloatChannel, default=1, const [1,325]}
       roll_blend_r_ctrl {type=MovieSceneScriptingFloatChannel, default=1, const [1,325]}
       tip_r_ctrl.Location.X {type=MovieSceneScriptingFloatChannel, default=0, const [0,325]}
       tip_r_ctrl.Location.Y {type=MovieSceneScriptingFloatChannel, default=0, const [0,325]}
       tip_r_ctrl.Location.Z {type=MovieSceneScriptingFloatChannel, default=0, const [0,325]}
       tip_r_ctrl.Rotation.X {type=MovieSceneScriptingFloatChannel, default=0, const [0,325]}
       tip_r_ctrl.Rotation.Y {type=MovieSceneScriptingFloatChannel, default=0, const [0,325]}
       tip_r_ctrl.Rotation.Z {type=MovieSceneScriptingFloatChannel, default=0, const [0,325]}
       tip_r_ctrl.Scale.X {type=MovieSceneScriptingFloatChannel, default=1, const [1,325]}
       tip_r_ctrl.Scale.Y {type=MovieSceneScriptingFloatChannel, default=1, const [1,325]}
       tip_r_ctrl.Scale.Z {type=MovieSceneScriptingFloatChannel, default=1, const [1,325]}
       heel_r_ctrl.Location.X {type=MovieSceneScriptingFloatChannel, default=0, const [0,325]}
       heel_r_ctrl.Location.Y {type=MovieSceneScriptingFloatChannel, default=0, const [0,325]}
       heel_r_ctrl.Location.Z {type=MovieSceneScriptingFloatChannel, default=0, const [0,325]}
       heel_r_ctrl.Rotation.X {type=MovieSceneScriptingFloatChannel, default=0, const [0,325]}
       heel_r_ctrl.Rotation.Y {type=MovieSceneScriptingFloatChannel, default=0, const [0,325]}
       heel_r_ctrl.Rotation.Z {type=MovieSceneScriptingFloatChannel, default=0, const [0,325]}
       heel_r_ctrl.Scale.X {type=MovieSceneScriptingFloatChannel, default=1, const [1,325]}
       heel_r_ctrl.Scale.Y {type=MovieSceneScriptingFloatChannel, default=1, const [1,325]}
       heel_r_ctrl.Scale.Z {type=MovieSceneScriptingFloatChannel, default=1, const [1,325]}
       leg_r_softness {type=MovieSceneScriptingFloatChannel, default=0, const [0,325]}
       leg_r_pv_ik_ctrl.Location.X {
        type=MovieSceneScriptingFloatChannel
        default=0
        thinned_from=325
        keys [[0,-0.056],[24,-0.0594],[29,-0.0595],[31,-0.059],[32,-0.0577],[33,-0.0558],[34,-0.0529],[35,-0.0502],[37,-0.0484],[38,-0.0477],[40,-0.0469],[48,-0.0457],[49,-0.0457],[56,-0.0444],[62,-0.044],[63,-0.0436],[64,-0.0438],[84,-0.0433],[85,-0.0434],[86,-0.0437],[87,-0.0444],[88,-0.0455],[89,-0.0472],[90,-0.0495],[91,-0.0521],[92,-0.0552],[93,-0.0569],[94,-0.0584],[95,-0.0594],[96,-0.0602],[98,-0.061],[105,-0.062],[109,-0.0623],[110,-0.0634],[111,-0.0624],[117,-0.0624],[144,-0.0601],[145,-0.0603],[146,-0.0601],[171,-0.0596],[173,-0.0598],[176,-0.0604],[177,-0.0609],[178,-0.0607],[182,-0.0612],[183,-0.0619],[184,-0.0614],[185,-0.0614],[187,-0.0619],[189,-0.0628],[191,-0.0634],[192,-0.0638],[202,-0.0641],[210,-0.0637],[212,-0.0633],[213,-0.0633],[218,-0.0624],[219,-0.0624],[222,-0.0619],[248,-0.0608],[249,-0.0606],[250,-0.0601],[251,-0.0593],[252,-0.0579],[253,-0.0569],[254,-0.0569],[255,-0.0572],[256,-0.0571],[264,-0.0584],[265,-0.0577],[266,-0.0587],[274,-0.0597],[284,-0.0598],[295,-0.0591],[310,-0.057],[311,-0.0574],[312,-0.0568],[324,-0.0561]]
       }
       leg_r_pv_ik_ctrl.Location.Y {type=MovieSceneScriptingFloatChannel, default=0, thinned_from=325, keys [[0,-1.9484],[324,-1.9484]]}
       leg_r_pv_ik_ctrl.Location.Z {type=MovieSceneScriptingFloatChannel, default=0, thinned_from=325, keys [[0,-1.7956],[324,-1.7956]]}
       leg_r_pv_ik_ctrl.Rotation.X {type=MovieSceneScriptingFloatChannel, default=0, thinned_from=325, keys [[0,2.6046],[324,2.6046]]}
       leg_r_pv_ik_ctrl.Rotation.Y {type=MovieSceneScriptingFloatChannel, default=0, thinned_from=325, keys [[0,-0.0134],[32,-0.0135],[88,-0.0128],[110,-0.0139],[324,-0.0134]]}
       leg_r_pv_ik_ctrl.Rotation.Z {
        type=MovieSceneScriptingFloatChannel
        default=0
        thinned_from=325
        keys [[0,0.0681],[2,0.0681],[23,0.0728],[29,0.0731],[30,0.0728],[31,0.0722],[32,0.0705],[33,0.0677],[34,0.0637],[35,0.0599],[36,0.0587],[37,0.0573],[38,0.0563],[40,0.0552],[48,0.0534],[49,0.0534],[50,0.0529],[56,0.0516],[57,0.0516],[58,0.0514],[62,0.051],[63,0.0504],[64,0.0507],[70,0.0504],[84,0.0501],[85,0.0502],[86,0.0507],[87,0.0516],[88,0.0532],[89,0.0556],[90,0.0588],[91,0.0625],[92,0.0669],[93,0.0694],[94,0.0714],[95,0.0729],[96,0.0739],[98,0.0752],[101,0.076],[105,0.0766],[106,0.0766],[107,0.0769],[109,0.077],[110,0.0786],[111,0.0771],[117,0.0771],[126,0.0764],[127,0.0762],[128,0.0762],[138,0.0745],[139,0.0745],[144,0.0739],[145,0.0742],[146,0.0738],[148,0.0738],[171,0.0732],[172,0.0734],[173,0.0734],[176,0.0743],[177,0.0749],[178,0.0747],[182,0.0755],[183,0.0765],[184,0.0757],[185,0.0757],[187,0.0764],[189,0.0777],[191,0.0785],[192,0.0791],[195,0.0793],[196,0.0795],[202,0.0795],[210,0.079],[211,0.0785],[213,0.0784],[218,0.0771],[219,0.0771],[222,0.0764],[223,0.0764],[224,0.0762],[225,0.0762],[226,0.076],[230,0.0758],[232,0.0755],[238,0.0753],[239,0.0751],[247,0.075],[249,0.0746],[250,0.0739],[251,0.0728],[252,0.0707],[253,0.0694],[254,0.0694],[255,0.0698],[256,0.0696],[264,0.0714],[265,0.0705],[266,0.0719],[274,0.0732],[275,0.0732],[276,0.0735],[284,0.0735],[290,0.0729],[291,0.0727],[292,0.0727],[293,0.0725],[295,0.0725],[296,0.0721],[310,0.0695],[311,0.0701],[312,0.0692],[314,0.0691],[319,0.0683],[321,0.0683],[324,0.0682]]
       }
       leg_r_pv_ik_ctrl.Scale.X {type=MovieSceneScriptingFloatChannel, default=1, const [1,325]}
       leg_r_pv_ik_ctrl.Scale.Y {type=MovieSceneScriptingFloatChannel, default=1, const [1,325]}
       leg_r_pv_ik_ctrl.Scale.Z {type=MovieSceneScriptingFloatChannel, default=1, const [1,325]}
       upperarm_l_fk_ctrl.Location.X {type=MovieSceneScriptingFloatChannel, default=0, thinned_from=325, keys [[0,0],[324,0]]}
       upperarm_l_fk_ctrl.Location.Y {type=MovieSceneScriptingFloatChannel, default=0, thinned_from=325, keys [[0,0],[324,0]]}
       upperarm_l_fk_ctrl.Location.Z {type=MovieSceneScriptingFloatChannel, default=0, thinned_from=325, keys [[0,0],[324,0]]}
       upperarm_l_fk_ctrl.Rotation.X {
        type=MovieSceneScriptingFloatChannel
        default=0
        thinned_from=325
        keys [[0,-10.8373],[4,-10.8926],[14,-11.2743],[25,-11.48],[28,-11.4614],[29,-11.4334],[30,-11.3807],[31,-11.29],[32,-11.0125],[33,-10.5719],[34,-9.9567],[35,-9.4362],[37,-9.2857],[40,-9.1596],[44,-9.0771],[47,-8.9467],[52,-8.6736],[54,-8.5987],[56,-8.5769],[80,-8.4047],[82,-8.365],[83,-8.3295],[84,-8.2725],[86,-8.1321],[88,-7.9378],[89,-7.9569],[90,-8.2159],[91,-8.6689],[92,-9.0001],[93,-9.1649],[96,-9.5428],[99,-9.7818],[108,-10.3799],[110,-10.4558],[124,-10.3829],[128,-10.2999],[137,-9.9301],[139,-9.8954],[144,-9.9605],[149,-10.0881],[169,-10.0237],[172,-10.0958],[180,-10.4561],[182,-10.5013],[185,-10.4653],[186,-10.4756],[187,-10.5646],[188,-10.7781],[189,-11.1087],[190,-11.5019],[192,-12.2247],[193,-12.4414],[197,-12.5746],[214,-12.6299],[219,-12.5611],[222,-12.598],[234,-12.9267],[247,-12.9584],[248,-12.9242],[249,-12.8183],[250,-12.6051],[251,-12.2076],[252,-11.3852],[253,-10.6828],[254,-10.4996],[255,-10.3812],[256,-10.2991],[257,-10.2494],[258,-10.2314],[270,-10.4185],[276,-10.6425],[278,-10.6769],[304,-10.555],[307,-10.6098],[311,-10.7322],[324,-10.8363]]
       }
       upperarm_l_fk_ctrl.Rotation.Y {
        type=MovieSceneScriptingFloatChannel
        default=0
        thinned_from=325
        keys [[0,-11.6401],[4,-11.6906],[9,-11.888],[11,-12.0162],[22,-13.1222],[25,-13.318],[28,-13.4239],[30,-13.4413],[31,-13.4165],[32,-13.3153],[33,-13.1439],[34,-12.8886],[35,-12.6657],[39,-12.3969],[45,-12.0633],[54,-11.3682],[58,-11.0165],[60,-10.8931],[67,-10.6502],[72,-10.5804],[75,-10.5941],[76,-10.5696],[77,-10.5897],[79,-10.5778],[83,-10.6451],[85,-10.752],[86,-10.8362],[87,-11.0036],[88,-11.309],[89,-11.7134],[90,-12.3311],[92,-13.4808],[93,-13.9635],[94,-14.397],[95,-14.7439],[96,-14.9942],[97,-15.2073],[98,-15.3873],[99,-15.5359],[101,-15.7251],[105,-15.8836],[124,-15.8376],[129,-15.7064],[132,-15.5306],[142,-14.4605],[149,-13.5926],[150,-13.5098],[155,-13.3175],[160,-13.2269],[162,-13.2293],[163,-13.2081],[169,-13.2126],[174,-13.3277],[177,-13.4925],[180,-13.7662],[181,-13.8832],[182,-14.0291],[186,-14.8378],[188,-15.0857],[190,-15.5344],[191,-15.7902],[192,-16.0854],[193,-16.3371],[194,-16.5019],[196,-16.7281],[198,-16.8719],[202,-16.9736],[210,-16.8881],[212,-16.7785],[217,-16.185],[223,-15.2708],[227,-14.762],[233,-14.19],[236,-13.9982],[242,-13.8265],[245,-13.8002],[249,-13.6326],[251,-13.4412],[252,-13.3067],[253,-13.3634],[254,-13.4813],[255,-13.4945],[262,-13.941],[267,-14.3554],[272,-14.6286],[284,-14.875],[290,-14.8401],[294,-14.7038],[296,-14.5857],[298,-14.4017],[304,-13.5572],[309,-12.7495],[311,-12.4751],[316,-12.0028],[319,-11.7936],[322,-11.6666],[324,-11.6469]]
       }
       upperarm_l_fk_ctrl.Rotation.Z {
        type=MovieSceneScriptingFloatChannel
        default=0
        thinned_from=325
        keys [[0,-9.5453],[1,-9.5453],[3,-9.4459],[5,-9.2814],[7,-9.0518],[9,-8.7662],[12,-8.2874],[15,-7.8717],[17,-7.6368],[22,-7.0525],[23,-6.9451],[24,-6.8532],[25,-6.7835],[26,-6.7432],[27,-6.7388],[31,-6.7791],[32,-6.7536],[33,-6.6802],[35,-6.3938],[37,-6.2336],[38,-6.174],[39,-6.144],[40,-6.1576],[41,-6.2148],[42,-6.3025],[43,-6.415],[44,-6.5626],[45,-6.753],[46,-6.9775],[47,-7.2277],[51,-8.2884],[52,-8.5334],[53,-8.7587],[54,-8.9553],[55,-9.1153],[56,-9.2343],[59,-9.4858],[62,-9.6745],[78,-10.2959],[80,-10.4175],[81,-10.5069],[82,-10.6216],[83,-10.8106],[84,-11.1175],[85,-11.5273],[86,-12.0296],[87,-12.9448],[88,-14.0948],[89,-14.6999],[90,-14.7675],[91,-14.6345],[92,-14.1913],[93,-13.5555],[96,-11.4818],[97,-10.8551],[98,-10.3011],[99,-9.8367],[100,-9.4738],[103,-8.6268],[105,-8.1016],[106,-7.8625],[107,-7.6498],[108,-7.4691],[109,-7.3257],[110,-7.2254],[111,-7.1734],[112,-7.1509],[115,-7.1389],[119,-7.2306],[122,-7.3444],[125,-7.5063],[127,-7.6463],[128,-7.7422],[129,-7.8713],[130,-8.0262],[133,-8.5734],[137,-9.3777],[138,-9.5535],[139,-9.7029],[140,-9.8192],[143,-10.074],[156,-10.7496],[163,-10.9735],[169,-11.0342],[170,-10.9763],[171,-10.8962],[172,-10.787],[174,-10.5005],[177,-9.968],[180,-9.398],[181,-9.2272],[182,-9.0866],[183,-8.9899],[184,-8.9435],[185,-8.9444],[186,-8.9793],[187,-9.0395],[188,-9.1247],[191,-9.4273],[192,-9.5136],[193,-9.5753],[196,-9.697],[198,-9.7181],[200,-9.6697],[205,-9.391],[208,-9.3027],[211,-9.2962],[212,-9.3197],[213,-9.377],[214,-9.4618],[216,-9.6886],[218,-9.9431],[220,-10.1416],[221,-10.2068],[222,-10.2439],[223,-10.2444],[232,-10.0974],[247,-10.2145],[249,-10.1806],[250,-10.1295],[251,-10.0382],[253,-9.7416],[255,-9.592],[262,-9.2956],[263,-9.2676],[266,-9.2502],[267,-9.2289],[269,-9.123],[270,-9.0504],[271,-8.952],[275,-8.4645],[276,-8.3648],[277,-8.2902],[278,-8.2463],[279,-8.2395],[286,-8.3699],[295,-8.6258],[296,-8.6683],[304,-9.2296],[306,-9.3305],[307,-9.3522],[311,-9.3438],[321,-9.5322],[324,-9.5438]]
       }
       upperarm_l_fk_ctrl.Scale.X {type=MovieSceneScriptingFloatChannel, default=1, const [1,325]}
       upperarm_l_fk_ctrl.Scale.Y {type=MovieSceneScriptingFloatChannel, default=1, const [1,325]}
       upperarm_l_fk_ctrl.Scale.Z {type=MovieSceneScriptingFloatChannel, default=1, const [1,325]}
       lowerarm_l_fk_ctrl.Location.X {
        type=MovieSceneScriptingFloatChannel
        default=0
        thinned_from=325
        keys [[0,0],[1,0.0024],[2,0],[112,0],[113,0.0021],[114,0],[115,-0.0013],[116,0],[165,0],[166,-0.002],[167,0],[168,0],[169,0.0023],[170,0],[244,0],[245,-0.0015],[246,0],[323,0],[324,-0.0014]]
       }
       lowerarm_l_fk_ctrl.Location.Y {
        type=MovieSceneScriptingFloatChannel
        default=0
        thinned_from=325
        keys [[0,0],[1,0.0043],[2,0],[112,0],[113,0.0039],[114,0],[115,-0.0023],[116,0],[165,0],[166,-0.0037],[167,0],[168,0],[169,0.004],[170,0],[244,0],[245,-0.0026],[246,0],[323,0],[324,-0.0025]]
       }
       lowerarm_l_fk_ctrl.Location.Z {
        type=MovieSceneScriptingFloatChannel
        default=0
        thinned_from=325
        keys [[0,0],[1,-0.0024],[2,0],[112,0],[113,0.0002],[114,0],[115,0.0012],[116,0],[165,0],[166,0.0011],[167,0],[168,0],[169,-0.0048],[170,0],[244,0],[245,0.0056],[246,0],[323,0],[324,0.0007]]
       }
       lowerarm_l_fk_ctrl.Rotation.X {
        type=MovieSceneScriptingFloatChannel
        default=0
        thinned_from=325
        keys [[0,0.0229],[1,0.0231],[2,0.0302],[3,0.0337],[4,0.0368],[5,0.0396],[6,0.0421],[7,0.0443],[8,0.0459],[9,0.0469],[10,0.0475],[11,0.0477],[12,0.0477],[13,0.0471],[14,0.0454],[16,0.0407],[17,0.0381],[18,0.0353],[19,0.032],[20,0.0284],[21,0.0242],[22,0.0197],[23,0.0147],[24,0.0096],[25,0.0042],[26,-0.0014],[27,-0.0074],[28,-0.0136],[29,-0.0201],[31,-0.0341],[32,-0.0408],[33,-0.0471],[34,-0.0531],[35,-0.0587],[36,-0.0637],[37,-0.0679],[38,-0.0714],[39,-0.0744],[41,-0.0799],[42,-0.0822],[43,-0.0842],[44,-0.086],[45,-0.0874],[46,-0.0885],[47,-0.0892],[48,-0.0892],[49,-0.0886],[50,-0.0859],[51,-0.0816],[52,-0.0758],[53,-0.0684],[54,-0.0594],[55,-0.0495],[56,-0.0392],[57,-0.0282],[59,-0.0048],[60,0.0063],[62,0.028],[63,0.038],[64,0.0466],[65,0.0539],[66,0.0602],[67,0.0657],[68,0.0705],[69,0.0746],[70,0.0779],[71,0.0808],[72,0.0833],[73,0.0853],[74,0.087],[75,0.0883],[76,0.0894],[77,0.0901],[78,0.0906],[79,0.0909],[81,0.0909],[82,0.09],[83,0.0872],[84,0.0814],[85,0.071],[86,0.055],[87,0.0278],[88,-0.017],[89,-0.0536],[90,-0.0709],[91,-0.0827],[92,-0.0932],[93,-0.103],[94,-0.1118],[95,-0.119],[97,-0.1312],[98,-0.137],[99,-0.1422],[100,-0.1468],[105,-0.1675],[108,-0.1789],[110,-0.1856],[112,-0.1914],[113,-0.1928],[114,-0.1961],[115,-0.1959],[116,-0.1999],[118,-0.2027],[120,-0.2045],[121,-0.2051],[122,-0.2051],[123,-0.2055],[125,-0.2055],[127,-0.2044],[128,-0.2033],[129,-0.2017],[130,-0.1994],[131,-0.1964],[132,-0.1925],[133,-0.1869],[134,-0.179],[135,-0.1693],[136,-0.1585],[139,-0.1237],[140,-0.1117],[141,-0.0985],[143,-0.0713],[144,-0.0584],[146,-0.033],[147,-0.0209],[148,-0.0098],[149,0.0006],[150,0.0103],[151,0.0192],[152,0.0271],[153,0.0339],[154,0.0399],[155,0.0453],[156,0.0502],[157,0.0548],[158,0.0589],[159,0.0626],[160,0.0659],[161,0.0686],[162,0.0709],[165,0.0767],[166,0.0802],[167,0.0798],[168,0.0811],[169,0.0783],[170,0.0836],[171,0.0845],[172,0.0848],[173,0.0846],[174,0.0839],[175,0.0828],[176,0.0813],[177,0.0794],[178,0.0772],[179,0.0745],[180,0.071],[181,0.0666],[182,0.0614],[183,0.0555],[184,0.0491],[185,0.0421],[186,0.0343],[188,0.0181],[192,-0.0124],[194,-0.028],[195,-0.0357],[196,-0.0429],[197,-0.0495],[198,-0.0557],[199,-0.0614],[200,-0.0665],[201,-0.0709],[202,-0.0745],[203,-0.0777],[205,-0.0836],[206,-0.0861],[207,-0.0883],[208,-0.0901],[209,-0.0915],[210,-0.0926],[211,-0.0932],[213,-0.0932],[214,-0.0928],[215,-0.0918],[216,-0.0901],[217,-0.0877],[218,-0.0847],[219,-0.081],[220,-0.0767],[221,-0.0718],[222,-0.0662],[223,-0.0603],[224,-0.054],[225,-0.0476],[226,-0.0409],[228,-0.0269],[229,-0.0195],[231,-0.0043],[232,0.0028],[233,0.0093],[234,0.0156],[235,0.0214],[236,0.0267],[237,0.0312],[238,0.0353],[239,0.0389],[240,0.0422],[241,0.045],[242,0.0475],[244,0.0517],[245,0.0565],[246,0.0552],[248,0.0581],[250,0.0605],[252,0.0623],[254,0.0635],[261,0.0655],[262,0.0655],[263,0.0658],[266,0.0658],[273,0.0634],[278,0.0607],[285,0.0554],[290,0.0506],[306,0.0313],[310,0.0273],[312,0.0257],[315,0.024],[317,0.0234],[318,0.0234],[319,0.0231],[323,0.0231],[324,0.0246]]
       }
       lowerarm_l_fk_ctrl.Rotation.Y {
        type=MovieSceneScriptingFloatChannel
        default=0
        thinned_from=325
        keys [[0,0.0923],[1,0.1039],[2,0.1241],[3,0.1396],[4,0.1535],[5,0.1662],[6,0.1778],[7,0.1879],[8,0.1956],[9,0.2005],[10,0.2031],[11,0.2041],[12,0.2041],[13,0.2012],[14,0.1932],[15,0.1826],[16,0.1712],[17,0.1595],[18,0.1464],[19,0.1319],[20,0.1159],[21,0.0981],[22,0.0788],[23,0.0584],[24,0.0375],[25,0.0162],[26,-0.0059],[27,-0.0285],[28,-0.0515],[29,-0.0756],[30,-0.1005],[31,-0.125],[32,-0.1478],[33,-0.1688],[34,-0.1888],[35,-0.2071],[36,-0.2231],[37,-0.2363],[38,-0.2472],[39,-0.2566],[40,-0.2653],[41,-0.2734],[42,-0.2805],[43,-0.2867],[44,-0.2918],[45,-0.2961],[46,-0.2994],[47,-0.3016],[48,-0.3016],[49,-0.2996],[50,-0.2916],[51,-0.2785],[52,-0.2608],[53,-0.2377],[54,-0.2093],[55,-0.1771],[56,-0.1423],[57,-0.1042],[58,-0.0624],[60,0.0245],[61,0.0687],[62,0.1144],[63,0.1587],[64,0.1988],[65,0.2342],[66,0.2662],[67,0.295],[68,0.321],[69,0.3438],[70,0.3633],[71,0.3803],[72,0.3952],[73,0.4079],[74,0.4183],[75,0.4267],[76,0.4334],[77,0.4384],[78,0.4415],[79,0.443],[81,0.443],[82,0.4378],[83,0.4198],[84,0.384],[85,0.324],[86,0.2397],[87,0.1135],[88,-0.064],[89,-0.1905],[90,-0.2455],[91,-0.2819],[92,-0.3135],[93,-0.3422],[94,-0.3673],[95,-0.3877],[97,-0.4215],[98,-0.437],[99,-0.4511],[100,-0.4633],[105,-0.5171],[108,-0.5457],[110,-0.5623],[113,-0.5828],[114,-0.588],[115,-0.5913],[116,-0.597],[119,-0.6063],[121,-0.6096],[125,-0.6106],[127,-0.6079],[128,-0.6053],[129,-0.6014],[130,-0.5959],[131,-0.5886],[132,-0.5792],[133,-0.5655],[134,-0.5459],[135,-0.5216],[136,-0.4938],[137,-0.4636],[139,-0.4007],[140,-0.367],[141,-0.3291],[142,-0.2886],[145,-0.1642],[147,-0.0784],[148,-0.0375],[149,0.0019],[150,0.0403],[151,0.0767],[152,0.1102],[153,0.1403],[154,0.1675],[155,0.1926],[156,0.2163],[157,0.2388],[158,0.2597],[159,0.2789],[160,0.296],[161,0.3108],[162,0.3234],[164,0.3454],[165,0.3559],[166,0.3672],[168,0.3822],[169,0.3805],[170,0.3971],[171,0.4024],[172,0.4045],[173,0.4031],[174,0.3988],[175,0.392],[176,0.383],[177,0.3721],[178,0.3591],[179,0.3433],[180,0.3238],[181,0.3],[182,0.2723],[183,0.2422],[184,0.2109],[185,0.1776],[187,0.1063],[188,0.0725],[189,0.0409],[190,0.0106],[191,-0.0187],[192,-0.0474],[193,-0.0758],[194,-0.1037],[195,-0.1303],[196,-0.1547],[197,-0.1769],[198,-0.1972],[199,-0.2156],[200,-0.2318],[201,-0.2456],[202,-0.257],[203,-0.2668],[205,-0.2846],[206,-0.2923],[207,-0.299],[208,-0.3044],[209,-0.3085],[210,-0.3116],[211,-0.3134],[213,-0.3134],[214,-0.3123],[215,-0.3094],[216,-0.3044],[217,-0.2971],[218,-0.2879],[219,-0.2768],[220,-0.2637],[221,-0.2484],[222,-0.231],[223,-0.212],[224,-0.1918],[225,-0.1707],[226,-0.1482],[227,-0.1245],[228,-0.0997],[229,-0.0732],[231,-0.0167],[232,0.0105],[233,0.0365],[234,0.062],[235,0.0863],[236,0.1085],[237,0.1285],[238,0.1466],[239,0.1631],[240,0.1781],[241,0.1915],[242,0.2031],[243,0.2135],[244,0.2233],[245,0.2444],[246,0.2407],[247,0.2483],[249,0.2618],[250,0.2677],[251,0.2729],[252,0.2772],[253,0.2805],[255,0.2854],[260,0.293],[261,0.294],[262,0.294],[263,0.2954],[266,0.2954],[269,0.2912],[273,0.2831],[276,0.2752],[280,0.2619],[284,0.2462],[287,0.2329],[290,0.2179],[302,0.1496],[304,0.1388],[306,0.1287],[308,0.1194],[309,0.1151],[310,0.1112],[311,0.1075],[312,0.1042],[313,0.1015],[314,0.0992],[315,0.0973],[316,0.0958],[317,0.0947],[318,0.0947],[319,0.0932],[323,0.0932],[324,0.0942]]
       }
       lowerarm_l_fk_ctrl.Rotation.Z {
        type=MovieSceneScriptingFloatChannel
        default=0
        thinned_from=325
        keys [[0,-1.2143],[1,-1.4109],[2,-1.6324],[3,-1.8356],[4,-2.0189],[5,-2.1848],[6,-2.3377],[7,-2.469],[8,-2.5701],[9,-2.6345],[10,-2.6683],[11,-2.6814],[12,-2.6814],[13,-2.643],[14,-2.5395],[15,-2.3998],[16,-2.2507],[17,-2.0969],[18,-1.925],[19,-1.7344],[20,-1.5245],[21,-1.2913],[22,-1.0366],[23,-0.7677],[24,-0.4922],[25,-0.2103],[26,0.0814],[27,0.3809],[28,0.6862],[29,1.0056],[30,1.3374],[31,1.6639],[32,1.9676],[33,2.2493],[34,2.5167],[35,2.7618],[36,2.9766],[37,3.1545],[38,3.3011],[39,3.4274],[40,3.5442],[41,3.6534],[42,3.7495],[43,3.8324],[44,3.9018],[45,3.9594],[46,4.0046],[47,4.0342],[48,4.0342],[49,4.0075],[50,3.8985],[51,3.7222],[52,3.4833],[53,3.1729],[54,2.7922],[55,2.3597],[56,1.8944],[57,1.3866],[58,0.8311],[59,0.2536],[60,-0.3202],[61,-0.904],[62,-1.5049],[63,-2.0863],[64,-2.6121],[65,-3.0747],[66,-3.4926],[67,-3.8691],[68,-4.2072],[69,-4.5036],[70,-4.7581],[71,-4.9782],[72,-5.1716],[73,-5.3371],[74,-5.4717],[75,-5.5802],[76,-5.6676],[77,-5.732],[78,-5.772],[79,-5.7924],[81,-5.7924],[82,-5.7242],[83,-5.4905],[84,-5.0263],[85,-4.2459],[86,-3.1472],[87,-1.4938],[88,0.8523],[89,2.5397],[90,3.278],[91,3.768],[92,4.1953],[93,4.5842],[94,4.9236],[95,5.2017],[97,5.6617],[98,5.8741],[99,6.0669],[100,6.234],[105,6.9718],[108,7.3663],[110,7.596],[113,7.8854],[116,8.0776],[119,8.2056],[121,8.2513],[125,8.2664],[127,8.2281],[128,8.1917],[129,8.1373],[130,8.0614],[131,7.9604],[132,7.8307],[133,7.64],[134,7.3694],[135,7.0347],[136,6.6523],[137,6.2381],[139,5.3788],[140,4.9204],[141,4.4068],[142,3.8589],[143,3.298],[145,2.1876],[146,1.6136],[147,1.044],[148,0.4999],[149,-0.0215],[150,-0.5286],[151,-1.0091],[152,-1.4504],[153,-1.8457],[154,-2.2027],[155,-2.5314],[156,-2.8414],[157,-3.1355],[158,-3.4085],[159,-3.6581],[160,-3.8821],[161,-4.0747],[162,-4.2387],[164,-4.5252],[166,-4.7956],[168,-5.0037],[169,-5.0931],[170,-5.1963],[171,-5.2657],[172,-5.2929],[173,-5.2738],[174,-5.2186],[175,-5.1307],[176,-5.0132],[177,-4.8719],[178,-4.7033],[179,-4.4974],[180,-4.2442],[181,-3.9335],[182,-3.5731],[183,-3.18],[184,-2.7709],[185,-2.3352],[187,-1.3992],[188,-0.9531],[189,-0.537],[190,-0.1368],[191,0.2513],[192,0.631],[193,1.0087],[194,1.3804],[195,1.735],[196,2.061],[197,2.3575],[198,2.6299],[199,2.8761],[200,3.0943],[201,3.2787],[202,3.4322],[203,3.5652],[205,3.8052],[206,3.9091],[207,3.9985],[208,4.0715],[209,4.128],[210,4.1689],[211,4.1937],[213,4.1937],[214,4.1792],[215,4.1396],[216,4.0715],[217,3.9736],[218,3.8496],[219,3.6995],[220,3.5232],[221,3.3173],[222,3.0829],[223,2.8271],[224,2.5573],[225,2.2737],[226,1.9731],[227,1.6571],[228,1.327],[229,0.9739],[230,0.6],[231,0.2242],[232,-0.1349],[233,-0.4789],[234,-0.8157],[235,-1.1353],[236,-1.428],[237,-1.6906],[238,-1.9287],[239,-2.1448],[240,-2.3416],[241,-2.516],[242,-2.6679],[243,-2.8043],[245,-3.0583],[247,-3.2588],[249,-3.435],[250,-3.5127],[251,-3.5806],[252,-3.6362],[253,-3.68],[255,-3.7437],[260,-3.8423],[261,-3.8561],[262,-3.8561],[263,-3.8735],[266,-3.8735],[270,-3.7956],[274,-3.6817],[277,-3.5714],[281,-3.3883],[285,-3.1763],[288,-2.9949],[291,-2.7916],[302,-1.9675],[304,-1.8258],[306,-1.6928],[308,-1.5709],[309,-1.5149],[310,-1.4626],[311,-1.4143],[312,-1.3716],[313,-1.3355],[314,-1.3054],[315,-1.2807],[316,-1.2609],[317,-1.2455],[318,-1.2455],[319,-1.2257],[323,-1.2257],[324,-1.232]]
       }
       lowerarm_l_fk_ctrl.Scale.X {type=MovieSceneScriptingFloatChannel, default=1, const [1,325]}
       lowerarm_l_fk_ctrl.Scale.Y {type=MovieSceneScriptingFloatChannel, default=1, const [1,325]}
       lowerarm_l_fk_ctrl.Scale.Z {type=MovieSceneScriptingFloatChannel, default=1, const [1,325]}
       hand_l_fk_ctrl.Location.X {type=MovieSceneScriptingFloatChannel, default=0, const [0,325]}
       hand_l_fk_ctrl.Location.Y {type=MovieSceneScriptingFloatChannel, default=0, const [0,325]}
       hand_l_fk_ctrl.Location.Z {type=MovieSceneScriptingFloatChannel, default=0, const [0,325]}
       hand_l_fk_ctrl.Rotation.X {type=MovieSceneScriptingFloatChannel, default=0, const [-0.8734,325]}
       hand_l_fk_ctrl.Rotation.Y {type=MovieSceneScriptingFloatChannel, default=0, const [10.019,325]}
       hand_l_fk_ctrl.Rotation.Z {type=MovieSceneScriptingFloatChannel, default=0, const [-11.4504,325]}
       hand_l_fk_ctrl.Scale.X {type=MovieSceneScriptingFloatChannel, default=1, const [1,325]}
       hand_l_fk_ctrl.Scale.Y {type=MovieSceneScriptingFloatChannel, default=1, const [1,325]}
       hand_l_fk_ctrl.Scale.Z {type=MovieSceneScriptingFloatChannel, default=1, const [1,325]}
       thumb_01_l_ctrl.Location.X {type=MovieSceneScriptingFloatChannel, default=0, const [0.0686,325]}
       thumb_01_l_ctrl.Location.Y {type=MovieSceneScriptingFloatChannel, default=0, const [-0.3985,325]}
       thumb_01_l_ctrl.Location.Z {type=MovieSceneScriptingFloatChannel, default=0, const [-0.4179,325]}
       thumb_01_l_ctrl.Rotation.X {type=MovieSceneScriptingFloatChannel, default=0, const [9.0456,325]}
       thumb_01_l_ctrl.Rotation.Y {type=MovieSceneScriptingFloatChannel, default=0, const [-15.3775,325]}
       thumb_01_l_ctrl.Rotation.Z {type=MovieSceneScriptingFloatChannel, default=0, const [-6.5547,325]}
       thumb_01_l_ctrl.Scale.X {type=MovieSceneScriptingFloatChannel, default=1, const [1,325]}
       thumb_01_l_ctrl.Scale.Y {type=MovieSceneScriptingFloatChannel, default=1, const [1,325]}
       thumb_01_l_ctrl.Scale.Z {type=MovieSceneScriptingFloatChannel, default=1, const [1,325]}
       thumb_02_l_ctrl.Location.X {type=MovieSceneScriptingFloatChannel, default=0, const [0,325]}
       thumb_02_l_ctrl.Location.Y {type=MovieSceneScriptingFloatChannel, default=0, const [0,325]}
       thumb_02_l_ctrl.Location.Z {type=MovieSceneScriptingFloatChannel, default=0, const [0,325]}
       thumb_02_l_ctrl.Rotation.X {type=MovieSceneScriptingFloatChannel, default=0, const [-28.0109,325]}
       thumb_02_l_ctrl.Rotation.Y {type=MovieSceneScriptingFloatChannel, default=0, const [-14.6405,325]}
       thumb_02_l_ctrl.Rotation.Z {type=MovieSceneScriptingFloatChannel, default=0, const [-20.6578,325]}
       thumb_02_l_ctrl.Scale.X {type=MovieSceneScriptingFloatChannel, default=1, const [1,325]}
       thumb_02_l_ctrl.Scale.Y {type=MovieSceneScriptingFloatChannel, default=1, const [1,325]}
       thumb_02_l_ctrl.Scale.Z {type=MovieSceneScriptingFloatChannel, default=1, const [1,325]}
       thumb_03_l_ctrl.Location.X {type=MovieSceneScriptingFloatChannel, default=0, const [0,325]}
       thumb_03_l_ctrl.Location.Y {type=MovieSceneScriptingFloatChannel, default=0, const [0,325]}
       thumb_03_l_ctrl.Location.Z {type=MovieSceneScriptingFloatChannel, default=0, const [0,325]}
       thumb_03_l_ctrl.Rotation.X {type=MovieSceneScriptingFloatChannel, default=0, const [-4.6289,325]}
       thumb_03_l_ctrl.Rotation.Y {type=MovieSceneScriptingFloatChannel, default=0, const [-5.8424,325]}
       thumb_03_l_ctrl.Rotation.Z {type=MovieSceneScriptingFloatChannel, default=0, const [-69.3636,325]}
       thumb_03_l_ctrl.Scale.X {type=MovieSceneScriptingFloatChannel, default=1, const [1,325]}
       thumb_03_l_ctrl.Scale.Y {type=MovieSceneScriptingFloatChannel, default=1, const [1,325]}
       thumb_03_l_ctrl.Scale.Z {type=MovieSceneScriptingFloatChannel, default=1, const [1,325]}
       index_metacarpal_l_ctrl.Location.X {type=MovieSceneScriptingFloatChannel, default=0, const [0,325]}
       index_metacarpal_l_ctrl.Location.Y {type=MovieSceneScriptingFloatChannel, default=0, const [0,325]}
       index_metacarpal_l_ctrl.Location.Z {type=MovieSceneScriptingFloatChannel, default=0, const [0,325]}
       index_metacarpal_l_ctrl.Rotation.X {type=MovieSceneScriptingFloatChannel, default=0, const [-2.9693,325]}
       index_metacarpal_l_ctrl.Rotation.Y {type=MovieSceneScriptingFloatChannel, default=0, const [3.5534,325]}
       index_metacarpal_l_ctrl.Rotation.Z {type=MovieSceneScriptingFloatChannel, default=0, const [-13.3204,325]}
       index_metacarpal_l_ctrl.Scale.X {type=MovieSceneScriptingFloatChannel, default=1, const [1,325]}
       index_metacarpal_l_ctrl.Scale.Y {type=MovieSceneScriptingFloatChannel, default=1, const [1,325]}
       index_metacarpal_l_ctrl.Scale.Z {type=MovieSceneScriptingFloatChannel, default=1, const [1,325]}
       index_01_l_ctrl.Location.X {type=MovieSceneScriptingFloatChannel, default=0, const [-0.4363,325]}
       index_01_l_ctrl.Location.Y {type=MovieSceneScriptingFloatChannel, default=0, const [-1.0252,325]}
       index_01_l_ctrl.Location.Z {type=MovieSceneScriptingFloatChannel, default=0, const [-0.034,325]}
       index_01_l_ctrl.Rotation.X {type=MovieSceneScriptingFloatChannel, default=0, const [8.4656,325]}
       index_01_l_ctrl.Rotation.Y {type=MovieSceneScriptingFloatChannel, default=0, const [-1.3715,325]}
       index_01_l_ctrl.Rotation.Z {type=MovieSceneScriptingFloatChannel, default=0, const [-66.4322,325]}
       index_01_l_ctrl.Scale.X {type=MovieSceneScriptingFloatChannel, default=1, const [1,325]}
       index_01_l_ctrl.Scale.Y {type=MovieSceneScriptingFloatChannel, default=1, const [1,325]}
       index_01_l_ctrl.Scale.Z {type=MovieSceneScriptingFloatChannel, default=1, const [1,325]}
       index_02_l_ctrl.Location.X {type=MovieSceneScriptingFloatChannel, default=0, const [0,325]}
       index_02_l_ctrl.Location.Y {type=MovieSceneScriptingFloatChannel, default=0, const [0,325]}
       index_02_l_ctrl.Location.Z {type=MovieSceneScriptingFloatChannel, default=0, const [0,325]}
       index_02_l_ctrl.Rotation.X {type=MovieSceneScriptingFloatChannel, default=0, const [2.7848,325]}
       index_02_l_ctrl.Rotation.Y {type=MovieSceneScriptingFloatChannel, default=0, const [-2.6518,325]}
       index_02_l_ctrl.Rotation.Z {type=MovieSceneScriptingFloatChannel, default=0, const [-67.2511,325]}
       index_02_l_ctrl.Scale.X {type=MovieSceneScriptingFloatChannel, default=1, const [1,325]}
       index_02_l_ctrl.Scale.Y {type=MovieSceneScriptingFloatChannel, default=1, const [1,325]}
       index_02_l_ctrl.Scale.Z {type=MovieSceneScriptingFloatChannel, default=1, const [1,325]}
       index_03_l_ctrl.Location.X {type=MovieSceneScriptingFloatChannel, default=0, const [-0.0001,325]}
       index_03_l_ctrl.Location.Y {type=MovieSceneScriptingFloatChannel, default=0, const [0.0003,325]}
       index_03_l_ctrl.Location.Z {type=MovieSceneScriptingFloatChannel, default=0, const [0,325]}
       index_03_l_ctrl.Rotation.X {type=MovieSceneScriptingFloatChannel, default=0, const [0.046,325]}
       index_03_l_ctrl.Rotation.Y {type=MovieSceneScriptingFloatChannel, default=0, const [-0.1016,325]}
       index_03_l_ctrl.Rotation.Z {type=MovieSceneScriptingFloatChannel, default=0, const [-90.9605,325]}
       index_03_l_ctrl.Scale.X {type=MovieSceneScriptingFloatChannel, default=1, const [1,325]}
       index_03_l_ctrl.Scale.Y {type=MovieSceneScriptingFloatChannel, default=1, const [1,325]}
       index_03_l_ctrl.Scale.Z {type=MovieSceneScriptingFloatChannel, default=1, const [1,325]}
       middle_metacarpal_l_ctrl.Location.X {type=MovieSceneScriptingFloatChannel, default=0, const [0,325]}
       middle_metacarpal_l_ctrl.Location.Y {type=MovieSceneScriptingFloatChannel, default=0, const [0,325]}
       middle_metacarpal_l_ctrl.Location.Z {type=MovieSceneScriptingFloatChannel, default=0, const [0,325]}
       middle_metacarpal_l_ctrl.Rotation.X {type=MovieSceneScriptingFloatChannel, default=0, const [0.2131,325]}
       middle_metacarpal_l_ctrl.Rotation.Y {type=MovieSceneScriptingFloatChannel, default=0, const [-1.1482,325]}
       middle_metacarpal_l_ctrl.Rotation.Z {type=MovieSceneScriptingFloatChannel, default=0, const [-10.5148,325]}
       middle_metacarpal_l_ctrl.Scale.X {type=MovieSceneScriptingFloatChannel, default=1, const [1,325]}
       middle_metacarpal_l_ctrl.Scale.Y {type=MovieSceneScriptingFloatChannel, default=1, const [1,325]}
       middle_metacarpal_l_ctrl.Scale.Z {type=MovieSceneScriptingFloatChannel, default=1, const [1,325]}
       middle_01_l_ctrl.Location.X {type=MovieSceneScriptingFloatChannel, default=0, const [-0.5532,325]}
       middle_01_l_ctrl.Location.Y {type=MovieSceneScriptingFloatChannel, default=0, const [-0.9634,325]}
       middle_01_l_ctrl.Location.Z {type=MovieSceneScriptingFloatChannel, default=0, const [-0.0915,325]}
       middle_01_l_ctrl.Rotation.X {type=MovieSceneScriptingFloatChannel, default=0, const [0,325]}
       middle_01_l_ctrl.Rotation.Y {type=MovieSceneScriptingFloatChannel, default=0, const [0,325]}
       middle_01_l_ctrl.Rotation.Z {type=MovieSceneScriptingFloatChannel, default=0, const [-80.8858,325]}
       middle_01_l_ctrl.Scale.X {type=MovieSceneScriptingFloatChannel, default=1, const [1,325]}
       middle_01_l_ctrl.Scale.Y {type=MovieSceneScriptingFloatChannel, default=1, const [1,325]}
       middle_01_l_ctrl.Scale.Z {type=MovieSceneScriptingFloatChannel, default=1, const [1,325]}
       middle_02_l_ctrl.Location.X {type=MovieSceneScriptingFloatChannel, default=0, const [-0.0452,325]}
       middle_02_l_ctrl.Location.Y {type=MovieSceneScriptingFloatChannel, default=0, const [0.0507,325]}
       middle_02_l_ctrl.Location.Z {type=MovieSceneScriptingFloatChannel, default=0, const [0.1146,325]}
       middle_02_l_ctrl.Rotation.X {type=MovieSceneScriptingFloatChannel, default=0, const [-6.9244,325]}
       middle_02_l_ctrl.Rotation.Y {type=MovieSceneScriptingFloatChannel, default=0, const [-1.6646,325]}
       middle_02_l_ctrl.Rotation.Z {type=MovieSceneScriptingFloatChannel, default=0, const [-62.576,325]}
       middle_02_l_ctrl.Scale.X {type=MovieSceneScriptingFloatChannel, default=1, const [1,325]}
       middle_02_l_ctrl.Scale.Y {type=MovieSceneScriptingFloatChannel, default=1, const [1,325]}
       middle_02_l_ctrl.Scale.Z {type=MovieSceneScriptingFloatChannel, default=1, const [1,325]}
       middle_03_l_ctrl.Location.X {type=MovieSceneScriptingFloatChannel, default=0, const [0.0206,325]}
       middle_03_l_ctrl.Location.Y {type=MovieSceneScriptingFloatChannel, default=0, const [-0.0583,325]}
       middle_03_l_ctrl.Location.Z {type=MovieSceneScriptingFloatChannel, default=0, const [0.1324,325]}
       middle_03_l_ctrl.Rotation.X {type=MovieSceneScriptingFloatChannel, default=0, const [-2.0926,325]}
       middle_03_l_ctrl.Rotation.Y {type=MovieSceneScriptingFloatChannel, default=0, const [1.0729,325]}
       middle_03_l_ctrl.Rotation.Z {type=MovieSceneScriptingFloatChannel, default=0, const [-76.641,325]}
       middle_03_l_ctrl.Scale.X {type=MovieSceneScriptingFloatChannel, default=1, const [1,325]}
       middle_03_l_ctrl.Scale.Y {type=MovieSceneScriptingFloatChannel, default=1, const [1,325]}
       middle_03_l_ctrl.Scale.Z {type=MovieSceneScriptingFloatChannel, default=1, const [1,325]}
       ring_metacarpal_l_ctrl.Location.X {type=MovieSceneScriptingFloatChannel, default=0, const [0,325]}
       ring_metacarpal_l_ctrl.Location.Y {type=MovieSceneScriptingFloatChannel, default=0, const [0,325]}
       ring_metacarpal_l_ctrl.Location.Z {type=MovieSceneScriptingFloatChannel, default=0, const [0,325]}
       ring_metacarpal_l_ctrl.Rotation.X {type=MovieSceneScriptingFloatChannel, default=0, const [5.3943,325]}
       ring_metacarpal_l_ctrl.Rotation.Y {type=MovieSceneScriptingFloatChannel, default=0, const [-2.9431,325]}
       ring_metacarpal_l_ctrl.Rotation.Z {type=MovieSceneScriptingFloatChannel, default=0, const [-10.4667,325]}
       ring_metacarpal_l_ctrl.Scale.X {type=MovieSceneScriptingFloatChannel, default=1, const [1,325]}
       ring_metacarpal_l_ctrl.Scale.Y {type=MovieSceneScriptingFloatChannel, default=1, const [1,325]}
       ring_metacarpal_l_ctrl.Scale.Z {type=MovieSceneScriptingFloatChannel, default=1, const [1,325]}
       ring_01_l_ctrl.Location.X {type=MovieSceneScriptingFloatChannel, default=0, const [-0.3709,325]}
       ring_01_l_ctrl.Location.Y {type=MovieSceneScriptingFloatChannel, default=0, const [-1.0509,325]}
       ring_01_l_ctrl.Location.Z {type=MovieSceneScriptingFloatChannel, default=0, const [-0.0244,325]}
       ring_01_l_ctrl.Rotation.X {type=MovieSceneScriptingFloatChannel, default=0, const [-0.8239,325]}
       ring_01_l_ctrl.Rotation.Y {type=MovieSceneScriptingFloatChannel, default=0, const [-2.6856,325]}
       ring_01_l_ctrl.Rotation.Z {type=MovieSceneScriptingFloatChannel, default=0, const [-84.1083,325]}
       ring_01_l_ctrl.Scale.X {type=MovieSceneScriptingFloatChannel, default=1, const [1,325]}
       ring_01_l_ctrl.Scale.Y {type=MovieSceneScriptingFloatChannel, default=1, const [1,325]}
       ring_01_l_ctrl.Scale.Z {type=MovieSceneScriptingFloatChannel, default=1, const [1,325]}
       ring_02_l_ctrl.Location.X {type=MovieSceneScriptingFloatChannel, default=0, const [0,325]}
       ring_02_l_ctrl.Location.Y {type=MovieSceneScriptingFloatChannel, default=0, const [0,325]}
       ring_02_l_ctrl.Location.Z {type=MovieSceneScriptingFloatChannel, default=0, const [0,325]}
       ring_02_l_ctrl.Rotation.X {type=MovieSceneScriptingFloatChannel, default=0, const [-12.8636,325]}
       ring_02_l_ctrl.Rotation.Y {type=MovieSceneScriptingFloatChannel, default=0, const [-0.6911,325]}
       ring_02_l_ctrl.Rotation.Z {type=MovieSceneScriptingFloatChannel, default=0, const [-59.6606,325]}
       ring_02_l_ctrl.Scale.X {type=MovieSceneScriptingFloatChannel, default=1, const [1,325]}
       ring_02_l_ctrl.Scale.Y {type=MovieSceneScriptingFloatChannel, default=1, const [1,325]}
       ring_02_l_ctrl.Scale.Z {type=MovieSceneScriptingFloatChannel, default=1, const [1,325]}
       ring_03_l_ctrl.Location.X {type=MovieSceneScriptingFloatChannel, default=0, const [0,325]}
       ring_03_l_ctrl.Location.Y {type=MovieSceneScriptingFloatChannel, default=0, const [0,325]}
       ring_03_l_ctrl.Location.Z {type=MovieSceneScriptingFloatChannel, default=0, const [0,325]}
       ring_03_l_ctrl.Rotation.X {type=MovieSceneScriptingFloatChannel, default=0, const [0,325]}
       ring_03_l_ctrl.Rotation.Y {type=MovieSceneScriptingFloatChannel, default=0, const [0,325]}
       ring_03_l_ctrl.Rotation.Z {type=MovieSceneScriptingFloatChannel, default=0, const [-88.4683,325]}
       ring_03_l_ctrl.Scale.X {type=MovieSceneScriptingFloatChannel, default=1, const [1,325]}
       ring_03_l_ctrl.Scale.Y {type=MovieSceneScriptingFloatChannel, default=1, const [1,325]}
       ring_03_l_ctrl.Scale.Z {type=MovieSceneScriptingFloatChannel, default=1, const [1,325]}
       pinky_metacarpal_l_ctrl.Location.X {type=MovieSceneScriptingFloatChannel, default=0, const [0,325]}
       pinky_metacarpal_l_ctrl.Location.Y {type=MovieSceneScriptingFloatChannel, default=0, const [0,325]}
       pinky_metacarpal_l_ctrl.Location.Z {type=MovieSceneScriptingFloatChannel, default=0, const [0,325]}
       pinky_metacarpal_l_ctrl.Rotation.X {type=MovieSceneScriptingFloatChannel, default=0, const [9.8795,325]}
       pinky_metacarpal_l_ctrl.Rotation.Y {type=MovieSceneScriptingFloatChannel, default=0, const [-6.9881,325]}
       pinky_metacarpal_l_ctrl.Rotation.Z {type=MovieSceneScriptingFloatChannel, default=0, const [-8.607,325]}
       pinky_metacarpal_l_ctrl.Scale.X {type=MovieSceneScriptingFloatChannel, default=1, const [1,325]}
       pinky_metacarpal_l_ctrl.Scale.Y {type=MovieSceneScriptingFloatChannel, default=1, const [1,325]}
       pinky_metacarpal_l_ctrl.Scale.Z {type=MovieSceneScriptingFloatChannel, default=1, const [1,325]}
       pinky_01_l_ctrl.Location.X {type=MovieSceneScriptingFloatChannel, default=0, const [-0.3836,325]}
       pinky_01_l_ctrl.Location.Y {type=MovieSceneScriptingFloatChannel, default=0, const [-1.0453,325]}
       pinky_01_l_ctrl.Location.Z {type=MovieSceneScriptingFloatChannel, default=0, const [-0.0533,325]}
       pinky_01_l_ctrl.Rotation.X {type=MovieSceneScriptingFloatChannel, default=0, const [-7.5972,325]}
       pinky_01_l_ctrl.Rotation.Y {type=MovieSceneScriptingFloatChannel, default=0, const [-7.928,325]}
       pinky_01_l_ctrl.Rotation.Z {type=MovieSceneScriptingFloatChannel, default=0, const [-83.5712,325]}
       pinky_01_l_ctrl.Scale.X {type=MovieSceneScriptingFloatChannel, default=1, const [1,325]}
       pinky_01_l_ctrl.Scale.Y {type=MovieSceneScriptingFloatChannel, default=1, const [1,325]}
       pinky_01_l_ctrl.Scale.Z {type=MovieSceneScriptingFloatChannel, default=1, const [1,325]}
       pinky_02_l_ctrl.Location.X {type=MovieSceneScriptingFloatChannel, default=0, const [0.3529,325]}
       pinky_02_l_ctrl.Location.Y {type=MovieSceneScriptingFloatChannel, default=0, const [0.0017,325]}
       pinky_02_l_ctrl.Location.Z {type=MovieSceneScriptingFloatChannel, default=0, const [-0.0756,325]}
       pinky_02_l_ctrl.Rotation.X {type=MovieSceneScriptingFloatChannel, default=0, const [1.9939,325]}
       pinky_02_l_ctrl.Rotation.Y {type=MovieSceneScriptingFloatChannel, default=0, const [10.0966,325]}
       pinky_02_l_ctrl.Rotation.Z {type=MovieSceneScriptingFloatChannel, default=0, const [-57.9295,325]}
       pinky_02_l_ctrl.Scale.X {type=MovieSceneScriptingFloatChannel, default=1, const [1,325]}
       pinky_02_l_ctrl.Scale.Y {type=MovieSceneScriptingFloatChannel, default=1, const [1,325]}
       pinky_02_l_ctrl.Scale.Z {type=MovieSceneScriptingFloatChannel, default=1, const [1,325]}
       pinky_03_l_ctrl.Location.X {type=MovieSceneScriptingFloatChannel, default=0, const [0,325]}
       pinky_03_l_ctrl.Location.Y {type=MovieSceneScriptingFloatChannel, default=0, const [0,325]}
       pinky_03_l_ctrl.Location.Z {type=MovieSceneScriptingFloatChannel, default=0, const [0,325]}
       pinky_03_l_ctrl.Rotation.X {type=MovieSceneScriptingFloatChannel, default=0, const [3.1496,325]}
       pinky_03_l_ctrl.Rotation.Y {type=MovieSceneScriptingFloatChannel, default=0, const [-4.3172,325]}
       pinky_03_l_ctrl.Rotation.Z {type=MovieSceneScriptingFloatChannel, default=0, const [-111.6788,325]}
       pinky_03_l_ctrl.Scale.X {type=MovieSceneScriptingFloatChannel, default=1, const [1,325]}
       pinky_03_l_ctrl.Scale.Y {type=MovieSceneScriptingFloatChannel, default=1, const [1,325]}
       pinky_03_l_ctrl.Scale.Z {type=MovieSceneScriptingFloatChannel, default=1, const [1,325]}
       hand_l_ik_ctrl.Location.X {
        type=MovieSceneScriptingFloatChannel
        default=0
        thinned_from=325
        keys [[0,-6.2394],[1,-6.2657],[2,-6.3103],[4,-6.4368],[6,-6.6093],[8,-6.8194],[10,-7.06],[11,-7.1982],[14,-7.6705],[15,-7.813],[16,-7.9358],[19,-8.2126],[21,-8.3496],[22,-8.3969],[24,-8.4257],[27,-8.4066],[30,-8.3596],[31,-8.357],[32,-8.4106],[33,-8.5036],[35,-8.7896],[37,-8.9166],[39,-8.9417],[43,-8.8494],[45,-8.7532],[49,-8.5092],[53,-8.2113],[55,-8.0963],[57,-8.0231],[59,-7.9934],[62,-8.0206],[77,-8.3301],[80,-8.4398],[82,-8.5616],[83,-8.6415],[84,-8.7481],[85,-8.8739],[86,-9.0232],[87,-9.2909],[88,-9.5776],[89,-9.6476],[90,-9.3629],[91,-9.0998],[92,-8.8129],[93,-8.573],[94,-8.3703],[95,-8.2192],[96,-8.1095],[97,-8.0383],[98,-7.9993],[99,-7.9891],[101,-8.018],[106,-8.1692],[116,-8.0975],[123,-7.9329],[127,-7.7584],[129,-7.6391],[131,-7.5046],[135,-7.1558],[138,-6.9407],[140,-6.8251],[142,-6.7486],[144,-6.7069],[147,-6.7002],[150,-6.7354],[159,-6.9834],[169,-7.3298],[178,-7.9173],[183,-8.3107],[185,-8.4499],[186,-8.4925],[187,-8.4942],[188,-8.4585],[189,-8.3764],[190,-8.2165],[191,-8.0218],[192,-7.8038],[193,-7.4711],[194,-7.1795],[195,-7.0349],[196,-6.923],[197,-6.8323],[210,-6.0976],[211,-6.0338],[212,-5.9575],[214,-5.7714],[217,-5.4638],[222,-4.8635],[225,-4.5383],[226,-4.4407],[227,-4.3522],[228,-4.275],[230,-4.1485],[232,-4.0468],[234,-3.9688],[236,-3.9145],[238,-3.879],[240,-3.864],[243,-3.871],[247,-3.9206],[248,-3.9502],[249,-4.0061],[250,-4.1053],[251,-4.2863],[252,-4.6093],[253,-5.0653],[254,-5.5467],[255,-5.9791],[256,-6.3106],[257,-6.5405],[258,-6.742],[259,-6.9212],[261,-7.237],[263,-7.5367],[266,-8.0606],[267,-8.2216],[268,-8.3614],[273,-8.9661],[274,-9.0751],[276,-9.2372],[278,-9.3458],[282,-9.4082],[287,-9.3787],[291,-9.2806],[294,-9.1349],[296,-8.9983],[299,-8.7182],[301,-8.4731],[308,-7.4727],[310,-7.2092],[312,-6.9806],[315,-6.6951],[317,-6.5306],[319,-6.396],[321,-6.2996],[322,-6.2675],[323,-6.2486],[324,-6.2438]]
       }
       hand_l_ik_ctrl.Location.Y {
        type=MovieSceneScriptingFloatChannel
        default=0
        thinned_from=325
        keys [[0,7.2485],[6,7.4556],[9,7.506],[11,7.5183],[15,7.5879],[21,7.6309],[22,7.6275],[27,7.4973],[29,7.4174],[31,7.2965],[32,7.2028],[34,6.9582],[35,6.8591],[36,6.7742],[39,6.5937],[43,6.4308],[45,6.3777],[46,6.3651],[49,6.3986],[50,6.4339],[52,6.5544],[57,6.9464],[59,7.1301],[63,7.5599],[65,7.7365],[68,7.9378],[73,8.1733],[75,8.2485],[76,8.2495],[77,8.2812],[79,8.2993],[85,8.4543],[86,8.4641],[87,8.5529],[89,8.7885],[90,8.9279],[94,9.3308],[95,9.3806],[97,9.342],[100,9.2274],[102,9.1058],[107,8.718],[109,8.5961],[112,8.481],[121,8.3768],[129,8.3645],[133,8.4156],[135,8.4969],[143,8.9752],[149,9.2073],[150,9.2636],[156,9.6596],[159,9.81],[162,9.9157],[163,9.9269],[167,10.0152],[170,10.0375],[181,9.8369],[182,9.8452],[184,9.9219],[185,9.9761],[186,10.0534],[187,10.0456],[188,10.0065],[189,9.9271],[191,9.6214],[192,9.513],[193,9.3679],[194,9.2557],[209,8.6188],[211,8.5447],[216,8.2325],[221,7.9923],[225,7.8925],[227,7.8765],[233,7.9758],[241,8.2258],[242,8.2359],[243,8.2684],[245,8.2899],[250,8.2522],[251,8.2716],[252,8.3955],[253,8.6362],[254,8.8371],[256,9.1441],[257,9.2553],[259,9.4252],[266,9.9232],[268,10.0115],[271,10.0354],[274,9.9682],[277,9.8768],[288,9.7498],[291,9.6649],[294,9.5054],[297,9.2779],[299,9.0797],[309,7.8913],[310,7.7846],[311,7.6943],[314,7.5141],[317,7.3822],[320,7.2944],[323,7.2545],[324,7.2564]]
       }
       hand_l_ik_ctrl.Location.Z {
        type=MovieSceneScriptingFloatChannel
        default=0
        thinned_from=325
        keys [[0,-13.0602],[4,-13.2906],[7,-13.3608],[9,-13.3345],[12,-13.1873],[14,-13.0218],[16,-12.811],[22,-12.1095],[23,-11.9591],[31,-11.0695],[32,-11.0127],[34,-10.965],[36,-10.9748],[37,-11.0161],[42,-11.0691],[44,-11.1403],[46,-11.2846],[49,-11.6631],[50,-11.8321],[52,-12.25],[62,-14.7627],[64,-15.1963],[66,-15.5383],[69,-15.9277],[73,-16.2637],[76,-16.4049],[83,-16.5793],[84,-16.5852],[85,-16.5487],[86,-16.4659],[87,-16.4276],[88,-16.2429],[89,-15.8817],[90,-15.4354],[91,-14.769],[93,-13.2238],[94,-12.4978],[95,-11.8624],[96,-11.3124],[97,-10.8291],[98,-10.4156],[99,-10.0771],[100,-9.8214],[101,-9.5986],[103,-9.2406],[105,-8.9273],[107,-8.6448],[109,-8.4346],[111,-8.2918],[114,-8.1688],[118,-8.0919],[121,-8.0788],[127,-8.1496],[129,-8.2167],[130,-8.2745],[132,-8.4763],[133,-8.6077],[134,-8.7773],[135,-8.985],[140,-10.226],[147,-12.0962],[150,-12.8018],[152,-13.1988],[154,-13.5218],[157,-13.9072],[160,-14.2006],[163,-14.3831],[168,-14.5519],[170,-14.5717],[172,-14.5261],[175,-14.3164],[179,-13.8998],[181,-13.6239],[183,-13.2919],[185,-12.9175],[188,-12.4793],[189,-12.2348],[190,-11.91],[191,-11.5179],[192,-11.0566],[193,-10.426],[194,-9.8505],[195,-9.5014],[196,-9.2145],[197,-8.9821],[198,-8.7943],[200,-8.4867],[202,-8.2286],[204,-8.0064],[206,-7.817],[208,-7.66],[210,-7.5359],[211,-7.4863],[213,-7.4377],[214,-7.4376],[215,-7.4559],[216,-7.4918],[217,-7.5434],[218,-7.6131],[220,-7.8045],[225,-8.3743],[228,-8.7692],[234,-9.6462],[236,-9.8997],[238,-10.1041],[241,-10.3315],[246,-10.5598],[248,-10.6314],[249,-10.695],[250,-10.8123],[251,-10.9964],[252,-11.3111],[253,-11.6707],[254,-12.0824],[255,-12.5392],[256,-12.8436],[257,-13.0065],[258,-13.1343],[260,-13.3112],[266,-13.6569],[268,-13.7109],[271,-13.6384],[274,-13.4814],[277,-13.2889],[298,-12.6698],[304,-12.6433],[312,-12.7526],[322,-13.0486],[324,-13.0646]]
       }
       hand_l_ik_ctrl.Rotation.X {
        type=MovieSceneScriptingFloatChannel
        default=0
        thinned_from=325
        keys [[0,-29.6831],[11,-29.6156],[16,-29.8462],[17,-29.8193],[22,-30.2937],[23,-30.2815],[38,-30.9587],[39,-31.0443],[46,-30.7768],[47,-30.7974],[54,-30.2871],[64,-29.2762],[74,-28.9176],[83,-28.8978],[86,-29.0822],[87,-29.2131],[91,-30.2049],[94,-31.1635],[97,-31.8748],[99,-32.2045],[100,-32.3154],[101,-32.3196],[106,-32.7619],[107,-32.7252],[122,-32.9407],[123,-33.0174],[130,-32.7673],[131,-32.7933],[141,-31.6979],[149,-30.5839],[160,-29.9499],[175,-29.881],[181,-30.1884],[183,-30.3955],[184,-30.4254],[185,-30.5648],[186,-30.8174],[194,-31.95],[209,-32.6949],[218,-32.461],[237,-31.0093],[250,-30.6376],[255,-30.3329],[268,-30.3999],[269,-30.3328],[274,-30.5534],[275,-30.4887],[290,-30.6107],[291,-30.6892],[298,-30.4687],[299,-30.4961],[323,-29.6803],[324,-29.6837]]
       }
       hand_l_ik_ctrl.Rotation.Y {
        type=MovieSceneScriptingFloatChannel
        default=0
        thinned_from=325
        keys [[0,-12.7116],[3,-13.2631],[6,-13.7242],[8,-13.9501],[10,-14.0621],[12,-14.1167],[13,-14.1115],[14,-14.0482],[17,-13.7507],[19,-13.4432],[21,-13.0602],[24,-12.4048],[28,-11.4073],[31,-10.5676],[33,-10.0441],[35,-9.4818],[36,-9.2162],[37,-8.9805],[38,-8.7797],[39,-8.6007],[40,-8.4827],[41,-8.4016],[42,-8.3522],[43,-8.3316],[44,-8.3374],[45,-8.3655],[47,-8.452],[48,-8.5379],[49,-8.6441],[50,-8.8176],[51,-9.047],[52,-9.3272],[53,-9.6651],[54,-10.0594],[55,-10.492],[56,-10.9479],[57,-11.4344],[63,-14.5879],[64,-15.0508],[65,-15.4584],[66,-15.8241],[68,-16.4507],[69,-16.7116],[71,-17.139],[73,-17.478],[76,-17.8093],[79,-17.9967],[80,-18.0339],[81,-18.108],[82,-18.1346],[83,-18.0879],[84,-17.9395],[85,-17.6037],[86,-17.0941],[87,-16.4578],[88,-15.3942],[89,-14.4102],[91,-13.1055],[92,-12.2756],[93,-11.3648],[94,-10.4821],[95,-9.6976],[96,-9.0065],[97,-8.385],[98,-7.8467],[99,-7.4065],[100,-7.0777],[101,-6.8388],[102,-6.5816],[103,-6.3395],[104,-6.1131],[105,-5.9022],[106,-5.7069],[108,-5.4025],[109,-5.2605],[110,-5.1349],[111,-5.0223],[112,-4.9246],[113,-4.8426],[115,-4.7057],[116,-4.6652],[117,-4.612],[118,-4.5747],[120,-4.5219],[121,-4.5064],[122,-4.5058],[123,-4.4711],[125,-4.4778],[127,-4.518],[128,-4.555],[129,-4.611],[131,-4.7615],[132,-4.8792],[133,-5.0477],[134,-5.2932],[135,-5.6046],[136,-5.9717],[137,-6.369],[139,-7.2005],[140,-7.6417],[141,-8.1316],[147,-11.2585],[149,-12.2398],[151,-13.1508],[152,-13.5591],[153,-13.9257],[155,-14.5624],[157,-15.1208],[159,-15.6062],[160,-15.8168],[163,-16.2929],[168,-16.8509],[170,-17.0101],[171,-17.0597],[172,-17.0685],[174,-16.9678],[176,-16.7563],[178,-16.476],[179,-16.3],[180,-16.0882],[181,-15.8281],[184,-14.9273],[186,-14.2468],[187,-13.9351],[192,-12.6174],[196,-11.4588],[198,-10.9358],[200,-10.4452],[202,-10.0135],[204,-9.6315],[206,-9.2945],[208,-9.019],[210,-8.8094],[211,-8.7279],[213,-8.6102],[214,-8.5652],[215,-8.5426],[216,-8.5449],[217,-8.5735],[218,-8.6244],[219,-8.697],[220,-8.7918],[221,-8.9127],[222,-9.0592],[224,-9.4056],[226,-9.8101],[228,-10.2786],[229,-10.5433],[234,-11.945],[236,-12.4426],[238,-12.8518],[240,-13.1897],[242,-13.462],[246,-13.8818],[250,-14.1727],[255,-14.4098],[258,-14.6418],[262,-15.081],[266,-15.6929],[269,-15.9799],[272,-16.0165],[274,-15.987],[282,-15.7945],[289,-15.4262],[290,-15.3559],[298,-14.6486],[308,-13.3975],[311,-13.1055],[315,-12.875],[319,-12.7548],[324,-12.7241]]
       }
       hand_l_ik_ctrl.Rotation.Z {
        type=MovieSceneScriptingFloatChannel
        default=0
        thinned_from=325
        keys [[0,-14.4749],[8,-13.677],[11,-13.4477],[13,-13.2509],[15,-13.1171],[16,-13.0743],[17,-12.9948],[20,-12.9794],[22,-13.0395],[23,-13.0368],[25,-13.1674],[28,-13.4603],[33,-14.128],[38,-14.94],[39,-15.1238],[43,-15.45],[46,-15.5972],[47,-15.6892],[49,-15.7567],[51,-15.7221],[52,-15.6653],[54,-15.4638],[57,-15.0532],[59,-14.6808],[63,-13.7975],[65,-13.4259],[67,-13.1295],[69,-12.8875],[73,-12.5285],[75,-12.3897],[76,-12.3683],[77,-12.3155],[82,-12.1673],[83,-12.1776],[84,-12.2299],[85,-12.3654],[86,-12.5829],[87,-12.8363],[88,-13.2006],[89,-13.4198],[90,-13.3782],[92,-13.1501],[94,-13.0624],[95,-13.0676],[100,-13.3748],[101,-13.3734],[106,-13.8963],[107,-13.9303],[112,-14.2769],[121,-14.5389],[122,-14.5482],[123,-14.6245],[129,-14.6087],[130,-14.5872],[131,-14.6337],[133,-14.6136],[134,-14.5652],[136,-14.3858],[140,-13.9197],[150,-12.4884],[152,-12.1436],[156,-11.5965],[159,-11.2502],[162,-10.9892],[163,-10.9354],[167,-10.6806],[171,-10.4978],[173,-10.469],[180,-10.5168],[182,-10.5437],[183,-10.5172],[184,-10.4321],[185,-10.3966],[192,-9.5132],[193,-9.4278],[194,-9.4038],[195,-9.4297],[197,-9.5287],[207,-10.3689],[211,-10.6353],[212,-10.7206],[216,-11.1671],[221,-11.6109],[224,-11.7244],[226,-11.7433],[228,-11.6982],[241,-10.7946],[243,-10.6916],[245,-10.614],[248,-10.5887],[249,-10.6082],[250,-10.6611],[251,-10.7808],[254,-11.1841],[255,-11.2588],[256,-11.2603],[258,-11.1358],[260,-10.9458],[262,-10.7152],[266,-10.1446],[269,-9.8208],[271,-9.7365],[274,-9.6979],[275,-9.6446],[282,-9.6285],[287,-9.7365],[290,-9.8707],[291,-9.9738],[294,-10.1957],[296,-10.3933],[297,-10.5139],[298,-10.6558],[305,-12.1473],[309,-13.0442],[311,-13.4207],[314,-13.8203],[317,-14.1351],[320,-14.3553],[322,-14.4398],[324,-14.4616]]
       }
       hand_l_ik_ctrl.Scale.X {type=MovieSceneScriptingFloatChannel, default=1, const [1,325]}
       hand_l_ik_ctrl.Scale.Y {type=MovieSceneScriptingFloatChannel, default=1, const [1,325]}
       hand_l_ik_ctrl.Scale.Z {type=MovieSceneScriptingFloatChannel, default=1, const [1,325]}
       Softness {type=MovieSceneScriptingFloatChannel, default=0, const [0,325]}
       hand_l_ikorient_ctrl.Rotation.Roll {type=MovieSceneScriptingFloatChannel, default=0, const [0,325]}
       hand_l_ikorient_ctrl.Rotation.Pitch {type=MovieSceneScriptingFloatChannel, default=0, const [0,325]}
       hand_l_ikorient_ctrl.Rotation.Yaw {type=MovieSceneScriptingFloatChannel, default=0, const [0,325]}
       arm_l_pv_ik_ctrl.Location.X {
        type=MovieSceneScriptingFloatChannel
        default=0
        thinned_from=325
        keys [[0,-21.4724],[11,-21.6759],[16,-21.9748],[17,-21.9846],[22,-22.3254],[31,-22.4549],[38,-22.3134],[46,-22.1305],[74,-21.8817],[81,-21.9862],[83,-22.1017],[84,-22.2244],[86,-22.62],[87,-23.0104],[88,-23.5325],[89,-23.8811],[90,-24.001],[95,-24.339],[100,-24.371],[101,-24.322],[106,-24.2934],[122,-24.0862],[123,-24.1191],[130,-23.9271],[150,-23.122],[176,-23.1223],[181,-23.3277],[184,-23.5556],[186,-23.7912],[188,-23.7588],[189,-23.6807],[191,-23.369],[194,-23.011],[211,-22.7013],[232,-21.2724],[242,-21.0381],[249,-20.9819],[250,-20.9997],[251,-21.0711],[252,-21.2692],[254,-21.9084],[256,-22.3286],[259,-22.6765],[268,-23.3961],[269,-23.3963],[274,-23.5949],[291,-23.5742],[297,-23.2523],[311,-21.9461],[321,-21.5039],[324,-21.4767]]
       }
       arm_l_pv_ik_ctrl.Location.Y {
        type=MovieSceneScriptingFloatChannel
        default=0
        thinned_from=325
        keys [[0,-7.4463],[12,-7.5347],[16,-7.586],[17,-7.5706],[22,-7.5684],[23,-7.5368],[28,-7.4976],[30,-7.5101],[31,-7.5301],[32,-7.6032],[33,-7.7138],[35,-8.0297],[36,-8.1557],[37,-8.3014],[38,-8.3915],[39,-8.4509],[46,-8.66],[54,-9.114],[60,-9.2914],[79,-9.5177],[81,-9.578],[82,-9.6261],[83,-9.7075],[84,-9.8409],[85,-10.0247],[86,-10.253],[87,-10.6486],[88,-11.0985],[89,-11.2319],[90,-10.982],[91,-10.5258],[92,-10.0171],[93,-9.5664],[94,-9.1488],[95,-8.7925],[96,-8.4914],[97,-8.2409],[98,-8.0377],[99,-7.8806],[101,-7.6746],[105,-7.4288],[106,-7.3783],[107,-7.3104],[109,-7.2267],[112,-7.1629],[122,-7.0904],[123,-7.0995],[130,-7.0253],[134,-7.0835],[151,-7.6214],[160,-7.77],[171,-7.7945],[179,-7.7444],[183,-7.7902],[184,-7.7827],[187,-7.8458],[188,-7.8222],[189,-7.6847],[190,-7.4359],[191,-7.1122],[192,-6.721],[193,-6.1315],[194,-5.6099],[195,-5.3533],[196,-5.1595],[197,-5.0183],[198,-4.9216],[200,-4.7835],[203,-4.6019],[212,-4.1643],[214,-4.0886],[216,-4.0347],[219,-3.9838],[222,-3.9407],[226,-3.8476],[229,-3.8047],[233,-3.7807],[247,-3.7949],[248,-3.8109],[249,-3.864],[250,-3.9835],[251,-4.203],[252,-4.6067],[253,-5.1171],[254,-5.6764],[255,-6.2571],[256,-6.6561],[257,-6.8811],[258,-7.0591],[259,-7.1988],[260,-7.3092],[266,-7.8126],[267,-7.8861],[268,-7.9386],[271,-8.0085],[274,-8.0267],[278,-7.9748],[289,-7.9232],[291,-7.9095],[298,-7.7032],[299,-7.6925],[310,-7.4434],[324,-7.4479]]
       }
       arm_l_pv_ik_ctrl.Location.Z {
        type=MovieSceneScriptingFloatChannel
        default=0
        thinned_from=325
        keys [[0,0.7869],[1,0.7911],[2,0.8215],[3,0.8641],[4,0.9219],[5,0.9954],[6,1.0836],[7,1.1861],[8,1.3022],[9,1.4328],[10,1.5765],[11,1.7404],[12,1.9262],[14,2.3206],[15,2.5096],[16,2.6771],[17,2.8617],[18,2.9848],[19,3.0991],[20,3.203],[21,3.2953],[22,3.3744],[23,3.486],[24,3.5457],[25,3.5908],[26,3.6233],[27,3.6437],[29,3.6698],[31,3.7073],[32,3.7702],[33,3.8533],[34,3.9601],[35,4.0275],[36,4.0193],[37,3.9931],[38,3.9493],[39,3.8649],[40,3.8319],[43,3.77],[44,3.743],[45,3.7045],[46,3.6563],[47,3.5536],[49,3.3858],[50,3.2801],[54,2.7959],[55,2.6826],[56,2.5808],[58,2.3986],[59,2.3175],[60,2.2456],[61,2.1802],[63,2.0619],[64,2.0082],[65,1.9602],[66,1.9163],[67,1.8779],[68,1.8456],[69,1.8188],[70,1.7987],[71,1.7854],[72,1.7788],[73,1.7787],[75,1.791],[76,1.805],[77,1.8134],[78,1.8431],[80,1.9228],[81,1.9893],[82,2.0664],[83,2.1881],[84,2.3702],[85,2.6122],[86,2.929],[87,3.4513],[88,4.0966],[89,4.4475],[90,4.3591],[91,4.3279],[92,4.2045],[94,3.8839],[95,3.7511],[96,3.646],[97,3.5743],[98,3.5355],[99,3.5297],[100,3.5564],[101,3.6386],[105,3.7643],[106,3.7893],[107,3.8578],[109,3.8914],[111,3.8972],[114,3.8817],[115,3.8683],[116,3.8635],[119,3.806],[122,3.7217],[123,3.6502],[126,3.5468],[127,3.5074],[128,3.4607],[129,3.4025],[130,3.3336],[131,3.2078],[132,3.1038],[133,2.9819],[135,2.7189],[137,2.4728],[138,2.354],[139,2.2422],[140,2.1402],[142,1.9551],[143,1.8699],[144,1.792],[145,1.72],[146,1.6517],[147,1.5875],[148,1.528],[149,1.4731],[150,1.4271],[151,1.3908],[152,1.3591],[153,1.3314],[154,1.3087],[155,1.2907],[156,1.2769],[157,1.2669],[158,1.2627],[159,1.2627],[160,1.2701],[161,1.2742],[162,1.2871],[163,1.3085],[164,1.3168],[165,1.34],[166,1.3589],[167,1.3849],[169,1.4478],[170,1.5006],[171,1.5616],[172,1.6332],[173,1.7137],[174,1.8028],[175,1.8991],[176,2.0012],[177,2.1092],[178,2.2218],[179,2.3422],[180,2.4745],[181,2.6134],[183,2.9158],[184,3.1121],[185,3.2668],[186,3.3932],[187,3.5372],[188,3.6731],[189,3.8217],[190,3.9564],[191,4.0776],[192,4.1833],[193,4.2554],[194,4.3063],[195,4.3479],[196,4.3779],[198,4.4037],[199,4.4034],[201,4.3529],[205,4.1716],[210,3.8988],[211,3.8322],[212,3.748],[213,3.645],[214,3.5269],[215,3.3976],[217,3.1204],[218,2.9616],[221,2.4272],[222,2.2534],[223,2.0879],[225,1.7696],[226,1.6175],[227,1.4736],[228,1.3405],[229,1.2163],[230,1.0979],[231,0.9861],[232,0.8814],[233,0.785],[234,0.6972],[235,0.6182],[236,0.5479],[237,0.4866],[238,0.4344],[239,0.3913],[240,0.3564],[241,0.3312],[242,0.314],[243,0.2979],[244,0.2908],[245,0.2851],[246,0.29],[247,0.2892],[248,0.3009],[249,0.3134],[250,0.3354],[251,0.37],[252,0.4301],[253,0.5496],[254,0.6915],[255,0.8074],[256,0.9306],[257,1.0585],[258,1.1931],[259,1.335],[260,1.4831],[261,1.6371],[262,1.7958],[263,1.97],[264,2.1637],[266,2.5663],[267,2.7536],[268,2.9198],[269,3.1049],[273,3.6534],[274,3.7801],[275,3.9429],[276,4.0524],[277,4.1429],[278,4.2122],[279,4.2565],[282,4.3307],[284,4.3564],[287,4.3585],[289,4.3378],[290,4.3191],[291,4.2579],[293,4.2047],[295,4.1287],[296,4.0757],[297,4.0086],[298,3.9276],[299,3.7873],[300,3.662],[301,3.514],[307,2.4936],[308,2.3295],[310,2.0164],[311,1.8707],[312,1.738],[314,1.4946],[315,1.3791],[316,1.2699],[317,1.1683],[318,1.0753],[319,0.9932],[320,0.9226],[321,0.8653],[322,0.8228],[323,0.8013],[324,0.7909]]
       }
       arm_l_pv_ik_ctrl.Rotation.X {
        type=MovieSceneScriptingFloatChannel
        default=0
        thinned_from=325
        keys [[0,5.5349],[1,5.5281],[5,5.5929],[8,5.681],[10,5.7591],[11,5.8143],[13,5.973],[15,6.1505],[16,6.2274],[17,6.3709],[22,6.5612],[23,6.6818],[26,6.7959],[30,6.8779],[32,6.8841],[33,6.8631],[34,6.8023],[35,6.7107],[38,6.3819],[39,6.2273],[40,6.1907],[41,6.1891],[42,6.2083],[43,6.2423],[44,6.291],[46,6.4287],[47,6.4202],[50,6.5622],[53,6.6638],[55,6.6948],[62,6.646],[69,6.5429],[72,6.5497],[75,6.6005],[77,6.6201],[78,6.6488],[80,6.7454],[81,6.8302],[82,6.9302],[83,7.1139],[84,7.4149],[85,7.8214],[86,8.3501],[87,9.2933],[88,10.4946],[89,11.2022],[90,11.3973],[91,11.4165],[92,11.2056],[93,10.8605],[96,9.7033],[97,9.3638],[98,9.0806],[99,8.8638],[100,8.7201],[101,8.6925],[103,8.4339],[106,8.1059],[107,8.1142],[109,8.0123],[112,7.9251],[115,7.8935],[116,7.9039],[122,7.8735],[123,7.7928],[130,7.8356],[131,7.7513],[135,7.6305],[141,7.569],[144,7.4902],[149,7.2979],[150,7.2781],[165,7.3659],[168,7.3788],[175,7.3574],[177,7.3923],[179,7.4837],[181,7.625],[182,7.7302],[183,7.8673],[184,8.103],[185,8.2693],[186,8.4837],[189,9.313],[191,9.8125],[192,10.032],[193,10.2146],[194,10.3511],[195,10.444],[197,10.5488],[198,10.5656],[199,10.5607],[201,10.4638],[211,9.6571],[213,9.4436],[218,8.8098],[222,8.2267],[226,7.7107],[228,7.4898],[231,7.2096],[233,7.0569],[235,6.9406],[237,6.8577],[240,6.7818],[242,6.7482],[246,6.7052],[248,6.6527],[250,6.5648],[252,6.3787],[253,6.2956],[254,6.2323],[255,6.2111],[256,6.243],[257,6.3175],[258,6.4126],[259,6.5277],[260,6.6597],[261,6.8074],[262,6.9691],[263,7.1553],[266,7.7861],[267,7.9654],[268,8.1089],[269,8.2899],[270,8.3518],[272,8.4252],[274,8.4487],[275,8.5469],[282,8.6915],[284,8.7078],[288,8.6951],[290,8.659],[291,8.5592],[294,8.4647],[297,8.3065],[298,8.2341],[299,8.058],[300,7.9219],[308,6.6494],[310,6.3582],[311,6.2362],[312,6.1365],[315,5.8964],[317,5.763],[319,5.656],[321,5.5801],[322,5.5557],[324,5.5415]]
       }
       arm_l_pv_ik_ctrl.Rotation.Y {
        type=MovieSceneScriptingFloatChannel
        default=0
        thinned_from=325
        keys [[0,-9.5734],[5,-9.6224],[10,-9.7721],[12,-9.8906],[16,-10.2184],[17,-10.2758],[22,-10.7176],[26,-10.8936],[29,-10.9652],[34,-10.9989],[39,-10.9155],[42,-10.7734],[54,-9.7921],[58,-9.43],[60,-9.3069],[66,-9.1078],[72,-9.0242],[74,-9.0208],[75,-9.0331],[76,-9.008],[77,-9.0272],[79,-9.001],[80,-9.0125],[85,-8.9501],[86,-8.9182],[87,-8.8552],[88,-8.8478],[89,-9.0011],[90,-9.3156],[91,-9.7225],[94,-11.2938],[95,-11.7334],[96,-12.0786],[97,-12.3797],[98,-12.6331],[99,-12.8358],[100,-12.9779],[103,-13.2096],[105,-13.3135],[106,-13.3445],[123,-13.369],[129,-13.2275],[131,-13.1546],[134,-12.9408],[141,-12.1998],[148,-11.338],[149,-11.2291],[150,-11.1476],[154,-10.9777],[160,-10.8545],[162,-10.8551],[163,-10.8299],[167,-10.8433],[172,-10.8747],[175,-10.9602],[178,-11.1148],[180,-11.2656],[181,-11.3584],[182,-11.4762],[184,-11.7711],[185,-11.9436],[186,-12.2278],[189,-12.8766],[193,-13.684],[194,-13.836],[197,-14.1243],[203,-14.4109],[210,-14.5307],[212,-14.488],[217,-14.1252],[226,-13.1928],[233,-12.6691],[236,-12.5191],[242,-12.3926],[245,-12.3674],[249,-12.2296],[250,-12.1561],[254,-11.6503],[255,-11.561],[256,-11.5251],[268,-11.7861],[269,-11.7912],[274,-11.9397],[288,-12.0437],[291,-12.0325],[295,-11.9025],[297,-11.786],[299,-11.6282],[304,-11.0998],[310,-10.3556],[312,-10.1653],[317,-9.8093],[320,-9.6572],[323,-9.5761],[324,-9.5787]]
       }
       arm_l_pv_ik_ctrl.Rotation.Z {
        type=MovieSceneScriptingFloatChannel
        default=0
        thinned_from=325
        keys [[0,-35.0456],[16,-35.1428],[17,-35.0913],[22,-35.2346],[38,-35.0394],[46,-35.0896],[82,-35.6764],[84,-35.8528],[86,-36.196],[87,-36.5403],[88,-36.958],[89,-37.1629],[90,-37.1533],[91,-37.0633],[97,-35.786],[99,-35.5244],[101,-35.3362],[106,-35.2435],[122,-35.0782],[130,-35.0302],[183,-35.4177],[184,-35.372],[195,-35.866],[255,-34.8475],[268,-35.4684],[269,-35.4431],[274,-35.6375],[290,-35.6111],[324,-35.0463]]
       }
       arm_l_pv_ik_ctrl.Scale.X {type=MovieSceneScriptingFloatChannel, default=1, const [1,325]}
       arm_l_pv_ik_ctrl.Scale.Y {type=MovieSceneScriptingFloatChannel, default=1, const [1,325]}
       arm_l_pv_ik_ctrl.Scale.Z {type=MovieSceneScriptingFloatChannel, default=1, const [1,325]}
       upperarm_r_fk_ctrl.Location.X {type=MovieSceneScriptingFloatChannel, default=0, thinned_from=325, keys [[0,0],[324,0]]}
       upperarm_r_fk_ctrl.Location.Y {type=MovieSceneScriptingFloatChannel, default=0, thinned_from=325, keys [[0,0],[324,0]]}
       upperarm_r_fk_ctrl.Location.Z {type=MovieSceneScriptingFloatChannel, default=0, thinned_from=325, keys [[0,0],[324,0]]}
       upperarm_r_fk_ctrl.Rotation.X {
        type=MovieSceneScriptingFloatChannel
        default=0
        thinned_from=325
        keys [[0,14.4837],[4,14.4003],[16,13.6723],[24,13.3425],[29,13.3264],[30,13.298],[31,13.2323],[32,12.9884],[33,12.5849],[34,12.0091],[35,11.5186],[37,11.3791],[39,11.3052],[41,11.2949],[45,11.3954],[53,11.8013],[56,11.8791],[81,12.2855],[86,12.2842],[87,12.3204],[88,12.4177],[89,12.5981],[90,12.9541],[91,13.4793],[92,13.8486],[93,14.0035],[94,14.0928],[95,14.1268],[96,14.1168],[108,13.5629],[111,13.5183],[127,13.7238],[137,14.2307],[169,14.3596],[173,14.2377],[177,13.9976],[183,13.5062],[184,13.4643],[185,13.454],[186,13.4763],[187,13.5795],[188,13.8081],[189,14.1569],[192,15.3622],[193,15.6288],[196,15.8951],[199,16.0601],[211,16.4232],[221,17.2227],[240,17.7651],[247,17.8446],[248,17.8074],[249,17.6962],[250,17.4673],[251,17.0415],[252,16.1952],[253,15.4679],[254,15.1865],[255,14.9942],[256,14.8487],[258,14.6573],[264,14.2021],[273,13.4358],[276,13.2587],[278,13.192],[285,13.177],[295,13.3634],[300,13.6255],[307,14.0934],[320,14.4584],[324,14.4821]]
       }
       upperarm_r_fk_ctrl.Rotation.Y {
        type=MovieSceneScriptingFloatChannel
        default=0
        thinned_from=325
        keys [[0,-4.6136],[1,-4.5974],[10,-4.6921],[12,-4.7415],[15,-4.8358],[17,-4.8643],[23,-4.8671],[29,-4.8064],[30,-4.8147],[31,-4.8481],[32,-4.9661],[33,-5.1592],[34,-5.4396],[35,-5.6751],[36,-5.7297],[37,-5.7663],[40,-5.8178],[48,-5.8873],[59,-5.8588],[64,-5.7705],[73,-5.4792],[74,-5.4363],[75,-5.4218],[77,-5.3656],[78,-5.3653],[81,-5.3315],[84,-5.3234],[85,-5.3497],[86,-5.3601],[87,-5.3594],[88,-5.3222],[89,-5.2414],[90,-4.9832],[91,-4.8123],[92,-4.7009],[93,-4.6563],[94,-4.6343],[95,-4.6295],[97,-4.6553],[102,-4.7816],[109,-5.0082],[112,-5.0687],[113,-5.0974],[114,-5.0952],[129,-5.2105],[137,-5.2229],[147,-5.1321],[149,-5.1395],[152,-5.1763],[168,-5.2266],[169,-5.2502],[171,-5.2714],[175,-5.3696],[178,-5.4769],[180,-5.5794],[184,-5.8438],[185,-5.93],[186,-6.0421],[188,-6.4765],[189,-6.6037],[190,-6.7001],[191,-6.769],[193,-6.8297],[194,-6.8768],[197,-6.9556],[201,-6.9887],[207,-6.9344],[216,-6.6976],[236,-5.8935],[240,-5.7698],[241,-5.73],[242,-5.7152],[244,-5.6499],[245,-5.6395],[248,-5.5667],[251,-5.5289],[252,-5.5475],[253,-5.4133],[254,-5.1779],[255,-5.0983],[256,-5.0507],[257,-5.0335],[258,-5.0321],[259,-5.0445],[261,-5.1024],[262,-5.1459],[263,-5.2032],[265,-5.3545],[272,-5.9274],[273,-5.9913],[274,-6.0363],[278,-6.1444],[283,-6.205],[285,-6.205],[291,-6.1999],[294,-6.1585],[296,-6.1065],[297,-6.0661],[299,-5.9462],[304,-5.5277],[308,-5.1602],[310,-5.003],[311,-4.9384],[313,-4.839],[316,-4.7287],[319,-4.6569],[322,-4.6203],[324,-4.6152]]
       }
       upperarm_r_fk_ctrl.Rotation.Z {
        type=MovieSceneScriptingFloatChannel
        default=0
        thinned_from=325
        keys [[0,-12.1],[1,-12.0877],[2,-12.0311],[3,-11.9478],[5,-11.7042],[7,-11.3854],[14,-10.0651],[15,-9.9113],[16,-9.7925],[17,-9.7157],[21,-9.5257],[23,-9.4687],[24,-9.4639],[25,-9.4907],[26,-9.5504],[31,-10.0618],[32,-10.2105],[33,-10.4021],[34,-10.6419],[35,-10.8468],[40,-11.1432],[42,-11.2131],[45,-11.2516],[47,-11.3234],[51,-11.5827],[53,-11.6661],[55,-11.6744],[59,-11.526],[60,-11.5286],[62,-11.6288],[64,-11.7874],[66,-12.0232],[74,-13.211],[76,-13.4314],[78,-13.5819],[81,-13.7054],[82,-13.7054],[83,-13.7361],[84,-13.7361],[85,-13.6805],[86,-13.5332],[88,-13.1718],[89,-13.0604],[90,-12.9147],[91,-12.7086],[92,-12.5372],[95,-12.1861],[102,-11.5067],[107,-10.8493],[108,-10.7397],[109,-10.6541],[110,-10.5982],[112,-10.5777],[118,-10.6435],[127,-10.8286],[128,-10.8664],[130,-11.0165],[133,-11.3217],[135,-11.4759],[137,-11.5567],[138,-11.5638],[139,-11.5434],[145,-11.1345],[150,-10.6446],[151,-10.568],[155,-10.4329],[157,-10.3946],[158,-10.3946],[159,-10.3709],[164,-10.3325],[169,-10.2408],[170,-10.1864],[171,-10.1049],[172,-9.9973],[173,-9.8645],[174,-9.7075],[175,-9.5272],[176,-9.3244],[177,-9.096],[178,-8.8423],[179,-8.5529],[182,-7.548],[183,-7.2344],[184,-6.9621],[186,-6.455],[187,-6.188],[188,-5.9008],[190,-5.3023],[191,-5.0457],[192,-4.8246],[193,-4.6769],[194,-4.6179],[195,-4.6047],[196,-4.6228],[197,-4.6561],[198,-4.7038],[199,-4.7648],[200,-4.8381],[202,-5.0155],[204,-5.2269],[206,-5.4729],[208,-5.7551],[210,-6.0751],[211,-6.2493],[212,-6.4501],[213,-6.6892],[214,-6.9589],[215,-7.2512],[219,-8.4911],[220,-8.7807],[221,-9.0468],[222,-9.2815],[223,-9.4771],[228,-10.2893],[234,-11.1319],[239,-11.7164],[244,-12.1862],[248,-12.5058],[249,-12.6033],[250,-12.7281],[251,-12.9016],[252,-13.1915],[253,-13.4009],[254,-13.4318],[255,-13.4163],[256,-13.3594],[257,-13.2629],[258,-13.128],[259,-12.9552],[260,-12.745],[261,-12.4967],[262,-12.2095],[263,-11.8673],[264,-11.4678],[267,-10.1425],[269,-9.3309],[271,-8.478],[272,-8.0655],[273,-7.7023],[274,-7.4179],[275,-7.1956],[276,-7.003],[277,-6.8453],[278,-6.7279],[279,-6.6561],[281,-6.5819],[283,-6.5464],[285,-6.5464],[287,-6.5834],[289,-6.6512],[291,-6.7511],[293,-6.8922],[294,-6.9821],[295,-7.087],[296,-7.2195],[297,-7.3906],[298,-7.5976],[299,-7.8362],[300,-8.1043],[302,-8.7066],[305,-9.6524],[306,-9.95],[307,-10.2253],[308,-10.4701],[310,-10.8896],[312,-11.2253],[314,-11.4981],[316,-11.7184],[318,-11.8873],[320,-12.0063],[323,-12.0942],[324,-12.0942]]
       }
       upperarm_r_fk_ctrl.Scale.X {type=MovieSceneScriptingFloatChannel, default=1, const [1,325]}
       upperarm_r_fk_ctrl.Scale.Y {type=MovieSceneScriptingFloatChannel, default=1, const [1,325]}
       upperarm_r_fk_ctrl.Scale.Z {type=MovieSceneScriptingFloatChannel, default=1, const [1,325]}
       lowerarm_r_fk_ctrl.Location.X {
        type=MovieSceneScriptingFloatChannel
        default=0
        thinned_from=325
        keys [[0,-0.0001],[81,-0.0001],[82,-0.002],[83,-0.0001],[84,-0.0007],[85,-0.0001],[157,-0.0001],[158,0.0014],[159,-0.0001],[160,0.0015],[161,-0.0001],[162,0.0017],[163,-0.0001],[164,0.0022],[165,-0.0001],[283,-0.0001],[284,0.0011],[285,0.0002],[286,-0.0001],[323,-0.0001],[324,-0.0014]]
       }
       lowerarm_r_fk_ctrl.Location.Y {
        type=MovieSceneScriptingFloatChannel
        default=0
        thinned_from=325
        keys [[0,0],[81,0],[82,-0.0036],[83,0],[84,-0.0011],[85,0],[157,0],[158,0.003],[159,0],[160,0.0032],[161,0],[162,0.0033],[163,0],[164,0.0043],[165,0],[283,0],[284,0.0021],[285,0.0004],[286,0],[323,0],[324,-0.0025]]
       }
       lowerarm_r_fk_ctrl.Location.Z {
        type=MovieSceneScriptingFloatChannel
        default=0
        thinned_from=325
        keys [[0,0],[81,0],[82,-0.0006],[83,0],[84,-0.0003],[85,0],[157,0],[158,0.0056],[159,0],[160,0.0035],[161,0],[162,0.0018],[163,0],[164,0.0006],[165,0],[283,0],[285,-0.0042],[286,0],[323,0],[324,-0.0005]]
       }
       lowerarm_r_fk_ctrl.Rotation.X {
        type=MovieSceneScriptingFloatChannel
        default=0
        thinned_from=325
        keys [[0,-0.0556],[6,-0.0837],[8,-0.0935],[10,-0.1042],[13,-0.1211],[16,-0.1391],[24,-0.1891],[30,-0.2234],[31,-0.2286],[32,-0.2333],[33,-0.2372],[34,-0.2403],[35,-0.2423],[36,-0.243],[37,-0.2413],[38,-0.237],[39,-0.2311],[41,-0.2174],[44,-0.1951],[48,-0.1639],[49,-0.1557],[50,-0.147],[51,-0.138],[52,-0.1286],[53,-0.119],[54,-0.1091],[56,-0.0886],[59,-0.0567],[62,-0.0261],[63,-0.0165],[64,-0.0081],[65,-0.0006],[66,0.006],[67,0.0119],[68,0.0171],[69,0.0216],[70,0.0254],[71,0.0287],[72,0.0317],[73,0.0344],[74,0.0368],[75,0.0388],[76,0.0406],[77,0.042],[81,0.0455],[82,0.048],[83,0.0469],[84,0.0476],[85,0.0472],[86,0.0464],[87,0.0449],[88,0.0423],[89,0.038],[90,0.0322],[93,0.0122],[94,0.0051],[97,-0.0173],[102,-0.0556],[103,-0.0631],[104,-0.0699],[107,-0.0886],[109,-0.1006],[111,-0.112],[113,-0.1227],[115,-0.1326],[117,-0.1417],[119,-0.1502],[121,-0.1578],[123,-0.1643],[125,-0.1697],[129,-0.1785],[131,-0.1823],[133,-0.1849],[134,-0.1857],[137,-0.1857],[138,-0.1846],[139,-0.1828],[140,-0.1798],[141,-0.1754],[142,-0.1693],[143,-0.1608],[144,-0.1497],[145,-0.1371],[147,-0.1101],[148,-0.0973],[149,-0.0833],[150,-0.0682],[151,-0.0546],[152,-0.045],[153,-0.0388],[154,-0.0335],[155,-0.0289],[156,-0.0251],[157,-0.0219],[158,-0.0128],[159,-0.0168],[160,-0.0103],[161,-0.0132],[162,-0.0085],[163,-0.0107],[164,-0.0075],[165,-0.0091],[167,-0.0082],[168,-0.0079],[169,-0.0079],[170,-0.0077],[172,-0.0077],[173,-0.008],[174,-0.0089],[182,-0.0197],[184,-0.0229],[186,-0.0266],[189,-0.033],[193,-0.0426],[195,-0.0478],[197,-0.0535],[199,-0.0598],[201,-0.0666],[204,-0.0777],[210,-0.1009],[211,-0.1045],[212,-0.1077],[213,-0.1104],[214,-0.1126],[215,-0.1144],[217,-0.1174],[218,-0.1184],[219,-0.1192],[221,-0.1192],[222,-0.1184],[223,-0.1171],[224,-0.1154],[225,-0.113],[226,-0.1101],[227,-0.1069],[229,-0.0997],[230,-0.0958],[232,-0.0872],[233,-0.0825],[235,-0.0724],[236,-0.0679],[237,-0.0639],[238,-0.0602],[239,-0.0568],[240,-0.054],[241,-0.0516],[242,-0.0498],[243,-0.0483],[244,-0.0471],[245,-0.0462],[246,-0.0456],[247,-0.0456],[248,-0.0453],[254,-0.0453],[255,-0.0459],[256,-0.0467],[257,-0.0479],[258,-0.0496],[259,-0.0518],[260,-0.0547],[261,-0.0585],[262,-0.0631],[263,-0.0697],[264,-0.0779],[265,-0.0859],[267,-0.1011],[268,-0.1082],[269,-0.1141],[270,-0.1189],[271,-0.1234],[272,-0.1273],[273,-0.1307],[274,-0.1335],[275,-0.1354],[278,-0.1397],[281,-0.143],[283,-0.1446],[284,-0.1443],[285,-0.1432],[286,-0.1463],[287,-0.1467],[288,-0.1467],[289,-0.1471],[291,-0.1471],[293,-0.1457],[294,-0.1445],[295,-0.1428],[296,-0.1407],[297,-0.1383],[298,-0.1354],[299,-0.132],[300,-0.128],[301,-0.1232],[302,-0.118],[303,-0.1122],[304,-0.1059],[305,-0.0992],[306,-0.0919],[307,-0.0839],[309,-0.0667],[310,-0.0568],[311,-0.0464],[312,-0.0385],[313,-0.0334],[314,-0.0296],[315,-0.0271],[316,-0.0263],[317,-0.0273],[318,-0.03],[320,-0.0374],[321,-0.0414],[322,-0.046],[323,-0.0508],[324,-0.0547]]
       }
       lowerarm_r_fk_ctrl.Rotation.Y {
        type=MovieSceneScriptingFloatChannel
        default=0
        thinned_from=325
        keys [[0,-0.1985],[8,-0.3159],[15,-0.4277],[22,-0.5418],[25,-0.5869],[30,-0.654],[32,-0.6766],[33,-0.6856],[34,-0.6925],[35,-0.697],[36,-0.6986],[37,-0.6949],[38,-0.6852],[39,-0.6716],[41,-0.64],[43,-0.6053],[46,-0.5487],[48,-0.509],[49,-0.488],[50,-0.4653],[51,-0.4411],[52,-0.4157],[53,-0.389],[54,-0.361],[55,-0.3318],[56,-0.3014],[57,-0.2695],[58,-0.2362],[60,-0.1684],[61,-0.1338],[62,-0.0986],[63,-0.0643],[64,-0.033],[65,-0.0048],[66,0.0211],[67,0.0447],[68,0.066],[69,0.0847],[70,0.1008],[71,0.1151],[72,0.1282],[73,0.1403],[74,0.151],[75,0.1603],[76,0.1683],[77,0.1747],[78,0.1795],[81,0.1912],[82,0.1929],[83,0.198],[84,0.1982],[85,0.1992],[86,0.1955],[87,0.1885],[88,0.176],[89,0.1563],[90,0.1305],[91,0.102],[92,0.0738],[93,0.0461],[95,-0.0113],[96,-0.0394],[97,-0.0671],[98,-0.0944],[99,-0.1213],[101,-0.1731],[102,-0.1986],[103,-0.2226],[104,-0.2441],[106,-0.2828],[108,-0.3193],[110,-0.3534],[112,-0.3848],[114,-0.4133],[116,-0.4392],[118,-0.4626],[120,-0.4838],[122,-0.5021],[124,-0.5173],[127,-0.5356],[130,-0.551],[132,-0.5591],[134,-0.5638],[137,-0.5638],[138,-0.5612],[139,-0.5567],[140,-0.5493],[141,-0.5383],[142,-0.523],[143,-0.5011],[144,-0.4724],[145,-0.4387],[146,-0.4019],[147,-0.364],[148,-0.327],[149,-0.2855],[150,-0.2387],[151,-0.1952],[152,-0.1638],[153,-0.1428],[154,-0.1245],[155,-0.1087],[156,-0.0952],[157,-0.0836],[158,-0.0634],[159,-0.0654],[160,-0.0521],[161,-0.0521],[162,-0.0445],[163,-0.0428],[164,-0.0393],[165,-0.0368],[166,-0.0348],[167,-0.0334],[168,-0.0324],[169,-0.0324],[170,-0.0315],[172,-0.0315],[173,-0.0326],[174,-0.036],[175,-0.0407],[177,-0.0512],[180,-0.0656],[181,-0.0706],[182,-0.0759],[183,-0.0815],[184,-0.0874],[185,-0.0938],[186,-0.1006],[187,-0.1078],[188,-0.1152],[190,-0.1308],[192,-0.1471],[194,-0.1641],[195,-0.173],[196,-0.1822],[198,-0.2019],[200,-0.2229],[202,-0.2452],[210,-0.3374],[211,-0.348],[212,-0.3572],[213,-0.3647],[214,-0.371],[215,-0.3763],[217,-0.3846],[218,-0.3876],[219,-0.3895],[221,-0.3895],[222,-0.3874],[223,-0.3839],[224,-0.3789],[225,-0.3723],[226,-0.3641],[228,-0.3447],[229,-0.3341],[230,-0.3225],[231,-0.3102],[232,-0.2972],[233,-0.2828],[235,-0.252],[236,-0.2379],[237,-0.2252],[238,-0.2134],[239,-0.2026],[240,-0.1933],[241,-0.1857],[242,-0.1795],[243,-0.1745],[244,-0.1705],[245,-0.1676],[246,-0.1658],[247,-0.1658],[248,-0.1647],[254,-0.1647],[255,-0.1666],[256,-0.1693],[257,-0.1733],[258,-0.1789],[259,-0.1863],[260,-0.1958],[261,-0.2078],[262,-0.2227],[263,-0.2434],[264,-0.2689],[265,-0.2932],[267,-0.3382],[268,-0.3587],[269,-0.3754],[270,-0.389],[271,-0.4013],[272,-0.4122],[273,-0.4215],[274,-0.4289],[275,-0.4343],[278,-0.4458],[282,-0.4569],[283,-0.4589],[285,-0.4751],[286,-0.4634],[287,-0.4643],[288,-0.4643],[289,-0.4654],[291,-0.4654],[293,-0.4618],[294,-0.4585],[295,-0.454],[296,-0.4485],[297,-0.442],[298,-0.4343],[299,-0.425],[300,-0.4139],[301,-0.4009],[302,-0.3862],[303,-0.3699],[304,-0.352],[305,-0.3327],[306,-0.3112],[307,-0.2871],[308,-0.2612],[309,-0.2342],[310,-0.2024],[311,-0.1683],[312,-0.1419],[313,-0.1244],[314,-0.111],[315,-0.1024],[316,-0.0994],[317,-0.1031],[318,-0.1124],[319,-0.1249],[320,-0.1379],[321,-0.1515],[322,-0.1669],[323,-0.1829],[324,-0.2003]]
       }
       lowerarm_r_fk_ctrl.Rotation.Z {
        type=MovieSceneScriptingFloatChannel
        default=0
        thinned_from=325
        keys [[0,2.6708],[7,4.049],[9,4.4598],[15,5.7697],[22,7.3355],[25,7.9598],[30,8.8924],[32,9.2087],[33,9.3349],[34,9.432],[35,9.4945],[36,9.5166],[37,9.4649],[38,9.3291],[39,9.1386],[41,8.697],[43,8.2146],[46,7.4312],[48,6.8841],[49,6.5953],[50,6.2838],[51,5.9529],[52,5.6062],[53,5.2432],[54,4.8625],[55,4.466],[56,4.0551],[57,3.6251],[58,3.1768],[60,2.2669],[61,1.8056],[62,1.3354],[63,0.8802],[64,0.4638],[65,0.0905],[66,-0.2519],[67,-0.5635],[68,-0.8444],[69,-1.0902],[70,-1.3023],[71,-1.4899],[72,-1.6624],[73,-1.8211],[74,-1.9615],[75,-2.084],[76,-2.1891],[77,-2.2726],[78,-2.3355],[81,-2.4887],[82,-2.5486],[83,-2.578],[84,-2.5959],[85,-2.5932],[86,-2.5445],[87,-2.4531],[88,-2.29],[89,-2.031],[90,-1.6931],[91,-1.3179],[92,-0.9469],[93,-0.5821],[94,-0.2044],[95,0.1763],[96,0.5498],[97,0.9163],[98,1.2802],[99,1.6379],[101,2.3305],[102,2.6716],[103,2.9939],[104,3.2825],[106,3.8035],[108,4.2971],[110,4.7592],[112,5.1854],[114,5.5739],[116,5.9273],[118,6.2478],[120,6.5377],[122,6.7893],[124,6.9985],[127,7.2504],[130,7.4632],[132,7.5748],[134,7.6397],[137,7.6397],[138,7.6036],[139,7.5419],[140,7.44],[141,7.2881],[142,7.0761],[143,6.7752],[144,6.3809],[145,5.9198],[146,5.4184],[147,4.9033],[148,4.4009],[149,3.8402],[150,3.2097],[151,2.6259],[152,2.2053],[153,1.9247],[154,1.6809],[155,1.4705],[156,1.29],[157,1.1361],[158,1.0122],[159,0.8939],[161,0.7178],[162,0.6581],[163,0.5945],[164,0.5602],[165,0.5149],[166,0.4885],[167,0.4695],[168,0.4566],[169,0.4566],[170,0.4445],[172,0.4445],[173,0.4597],[174,0.5045],[175,0.567],[176,0.6372],[177,0.7054],[180,0.8973],[181,0.9641],[182,1.0339],[183,1.1079],[184,1.187],[185,1.2721],[186,1.3629],[187,1.4578],[188,1.5566],[190,1.7644],[192,1.9832],[194,2.2102],[195,2.3284],[196,2.4524],[198,2.7164],[200,2.9981],[202,3.2975],[210,4.5424],[211,4.6859],[212,4.8103],[213,4.9128],[214,4.9984],[215,5.0701],[217,5.1824],[218,5.2232],[219,5.2499],[221,5.2499],[222,5.2209],[223,5.1726],[224,5.105],[225,5.0151],[226,4.9038],[228,4.6415],[229,4.4968],[230,4.3405],[231,4.1738],[232,3.9977],[233,3.8043],[235,3.3886],[236,3.1995],[237,3.029],[238,2.8695],[239,2.7255],[240,2.6011],[241,2.4988],[242,2.4158],[243,2.349],[244,2.2956],[245,2.2566],[246,2.233],[247,2.233],[248,2.2183],[254,2.2183],[255,2.2434],[256,2.2785],[257,2.3323],[258,2.4071],[259,2.5062],[260,2.634],[261,2.7953],[262,2.9944],[263,3.2729],[264,3.6162],[265,3.9448],[267,4.5528],[268,4.8313],[269,5.0577],[270,5.242],[271,5.41],[272,5.5587],[273,5.6853],[274,5.7868],[275,5.8603],[278,6.0171],[282,6.169],[284,6.2262],[287,6.271],[288,6.271],[289,6.2855],[291,6.2855],[293,6.2362],[294,6.1915],[295,6.13],[296,6.0544],[297,5.9657],[298,5.8601],[299,5.7336],[300,5.5822],[301,5.405],[302,5.2046],[303,4.9824],[304,4.74],[305,4.4786],[306,4.1875],[307,3.862],[308,3.5124],[309,3.1492],[310,2.7229],[311,2.2663],[312,1.9128],[313,1.6793],[314,1.5007],[315,1.3864],[316,1.3462],[317,1.3953],[318,1.5201],[319,1.6863],[320,1.8597],[321,2.0415],[322,2.2466],[323,2.4611],[324,2.6646]]
       }
       lowerarm_r_fk_ctrl.Scale.X {type=MovieSceneScriptingFloatChannel, default=1, const [1,325]}
       lowerarm_r_fk_ctrl.Scale.Y {type=MovieSceneScriptingFloatChannel, default=1, const [1,325]}
       lowerarm_r_fk_ctrl.Scale.Z {type=MovieSceneScriptingFloatChannel, default=1, const [1,325]}
       hand_r_fk_ctrl.Location.X {type=MovieSceneScriptingFloatChannel, default=0, const [-0.0001,325]}
       hand_r_fk_ctrl.Location.Y {type=MovieSceneScriptingFloatChannel, default=0, const [0,325]}
       hand_r_fk_ctrl.Location.Z {type=MovieSceneScriptingFloatChannel, default=0, const [0,325]}
       hand_r_fk_ctrl.Rotation.X {type=MovieSceneScriptingFloatChannel, default=0, const [0.321,325]}
       hand_r_fk_ctrl.Rotation.Y {type=MovieSceneScriptingFloatChannel, default=0, const [4.7589,325]}
       hand_r_fk_ctrl.Rotation.Z {type=MovieSceneScriptingFloatChannel, default=0, const [-11.7032,325]}
       hand_r_fk_ctrl.Scale.X {type=MovieSceneScriptingFloatChannel, default=1, const [1,325]}
       hand_r_fk_ctrl.Scale.Y {type=MovieSceneScriptingFloatChannel, default=1, const [1,325]}
       hand_r_fk_ctrl.Scale.Z {type=MovieSceneScriptingFloatChannel, default=1, const [1,325]}
       thumb_01_r_ctrl.Location.X {type=MovieSceneScriptingFloatChannel, default=0, const [0.0685,325]}
       thumb_01_r_ctrl.Location.Y {type=MovieSceneScriptingFloatChannel, default=0, const [-0.3984,325]}
       thumb_01_r_ctrl.Location.Z {type=MovieSceneScriptingFloatChannel, default=0, const [-0.4179,325]}
       thumb_01_r_ctrl.Rotation.X {type=MovieSceneScriptingFloatChannel, default=0, const [9.0456,325]}
       thumb_01_r_ctrl.Rotation.Y {type=MovieSceneScriptingFloatChannel, default=0, const [-15.3775,325]}
       thumb_01_r_ctrl.Rotation.Z {type=MovieSceneScriptingFloatChannel, default=0, const [-6.5546,325]}
       thumb_01_r_ctrl.Scale.X {type=MovieSceneScriptingFloatChannel, default=1, const [1,325]}
       thumb_01_r_ctrl.Scale.Y {type=MovieSceneScriptingFloatChannel, default=1, const [1,325]}
       thumb_01_r_ctrl.Scale.Z {type=MovieSceneScriptingFloatChannel, default=1, const [1,325]}
       thumb_02_r_ctrl.Location.X {type=MovieSceneScriptingFloatChannel, default=0, const [0.0001,325]}
       thumb_02_r_ctrl.Location.Y {type=MovieSceneScriptingFloatChannel, default=0, const [0,325]}
       thumb_02_r_ctrl.Location.Z {type=MovieSceneScriptingFloatChannel, default=0, const [0,325]}
       thumb_02_r_ctrl.Rotation.X {type=MovieSceneScriptingFloatChannel, default=0, const [-28.0109,325]}
       thumb_02_r_ctrl.Rotation.Y {type=MovieSceneScriptingFloatChannel, default=0, const [-14.6405,325]}
       thumb_02_r_ctrl.Rotation.Z {type=MovieSceneScriptingFloatChannel, default=0, const [-20.6578,325]}
       thumb_02_r_ctrl.Scale.X {type=MovieSceneScriptingFloatChannel, default=1, const [1,325]}
       thumb_02_r_ctrl.Scale.Y {type=MovieSceneScriptingFloatChannel, default=1, const [1,325]}
       thumb_02_r_ctrl.Scale.Z {type=MovieSceneScriptingFloatChannel, default=1, const [1,325]}
       thumb_03_r_ctrl.Location.X {type=MovieSceneScriptingFloatChannel, default=0, const [0,325]}
       thumb_03_r_ctrl.Location.Y {type=MovieSceneScriptingFloatChannel, default=0, const [0,325]}
       thumb_03_r_ctrl.Location.Z {type=MovieSceneScriptingFloatChannel, default=0, const [0,325]}
       thumb_03_r_ctrl.Rotation.X {type=MovieSceneScriptingFloatChannel, default=0, const [-4.6289,325]}
       thumb_03_r_ctrl.Rotation.Y {type=MovieSceneScriptingFloatChannel, default=0, const [-5.8424,325]}
       thumb_03_r_ctrl.Rotation.Z {type=MovieSceneScriptingFloatChannel, default=0, const [-69.3636,325]}
       thumb_03_r_ctrl.Scale.X {type=MovieSceneScriptingFloatChannel, default=1, const [1,325]}
       thumb_03_r_ctrl.Scale.Y {type=MovieSceneScriptingFloatChannel, default=1, const [1,325]}
       thumb_03_r_ctrl.Scale.Z {type=MovieSceneScriptingFloatChannel, default=1, const [1,325]}
       index_metacarpal_r_ctrl.Location.X {type=MovieSceneScriptingFloatChannel, default=0, const [0.0001,325]}
       index_metacarpal_r_ctrl.Location.Y {type=MovieSceneScriptingFloatChannel, default=0, const [0.0002,325]}
       index_metacarpal_r_ctrl.Location.Z {type=MovieSceneScriptingFloatChannel, default=0, const [0,325]}
       index_metacarpal_r_ctrl.Rotation.X {type=MovieSceneScriptingFloatChannel, default=0, const [-2.9693,325]}
       index_metacarpal_r_ctrl.Rotation.Y {type=MovieSceneScriptingFloatChannel, default=0, const [3.5534,325]}
       index_metacarpal_r_ctrl.Rotation.Z {type=MovieSceneScriptingFloatChannel, default=0, const [-13.3204,325]}
       index_metacarpal_r_ctrl.Scale.X {type=MovieSceneScriptingFloatChannel, default=1, const [1,325]}
       index_metacarpal_r_ctrl.Scale.Y {type=MovieSceneScriptingFloatChannel, default=1, const [1,325]}
       index_metacarpal_r_ctrl.Scale.Z {type=MovieSceneScriptingFloatChannel, default=1, const [1,325]}
       index_01_r_ctrl.Location.X {type=MovieSceneScriptingFloatChannel, default=0, const [-0.4363,325]}
       index_01_r_ctrl.Location.Y {type=MovieSceneScriptingFloatChannel, default=0, const [-1.0252,325]}
       index_01_r_ctrl.Location.Z {type=MovieSceneScriptingFloatChannel, default=0, const [-0.034,325]}
       index_01_r_ctrl.Rotation.X {type=MovieSceneScriptingFloatChannel, default=0, const [8.4656,325]}
       index_01_r_ctrl.Rotation.Y {type=MovieSceneScriptingFloatChannel, default=0, const [-1.3715,325]}
       index_01_r_ctrl.Rotation.Z {type=MovieSceneScriptingFloatChannel, default=0, const [-66.4322,325]}
       index_01_r_ctrl.Scale.X {type=MovieSceneScriptingFloatChannel, default=1, const [1,325]}
       index_01_r_ctrl.Scale.Y {type=MovieSceneScriptingFloatChannel, default=1, const [1,325]}
       index_01_r_ctrl.Scale.Z {type=MovieSceneScriptingFloatChannel, default=1, const [1,325]}
       index_02_r_ctrl.Location.X {type=MovieSceneScriptingFloatChannel, default=0, const [-0.0001,325]}
       index_02_r_ctrl.Location.Y {type=MovieSceneScriptingFloatChannel, default=0, const [-0.0001,325]}
       index_02_r_ctrl.Location.Z {type=MovieSceneScriptingFloatChannel, default=0, const [-0.0001,325]}
       index_02_r_ctrl.Rotation.X {type=MovieSceneScriptingFloatChannel, default=0, const [2.7848,325]}
       index_02_r_ctrl.Rotation.Y {type=MovieSceneScriptingFloatChannel, default=0, const [-2.6518,325]}
       index_02_r_ctrl.Rotation.Z {type=MovieSceneScriptingFloatChannel, default=0, const [-67.2511,325]}
       index_02_r_ctrl.Scale.X {type=MovieSceneScriptingFloatChannel, default=1, const [1,325]}
       index_02_r_ctrl.Scale.Y {type=MovieSceneScriptingFloatChannel, default=1, const [1,325]}
       index_02_r_ctrl.Scale.Z {type=MovieSceneScriptingFloatChannel, default=1, const [1,325]}
       index_03_r_ctrl.Location.X {type=MovieSceneScriptingFloatChannel, default=0, const [-0.0001,325]}
       index_03_r_ctrl.Location.Y {type=MovieSceneScriptingFloatChannel, default=0, const [0.0001,325]}
       index_03_r_ctrl.Location.Z {type=MovieSceneScriptingFloatChannel, default=0, const [-0.0001,325]}
       index_03_r_ctrl.Rotation.X {type=MovieSceneScriptingFloatChannel, default=0, const [0,325]}
       index_03_r_ctrl.Rotation.Y {type=MovieSceneScriptingFloatChannel, default=0, const [0,325]}
       index_03_r_ctrl.Rotation.Z {type=MovieSceneScriptingFloatChannel, default=0, const [-90.9605,325]}
       index_03_r_ctrl.Scale.X {type=MovieSceneScriptingFloatChannel, default=1, const [1,325]}
       index_03_r_ctrl.Scale.Y {type=MovieSceneScriptingFloatChannel, default=1, const [1,325]}
       index_03_r_ctrl.Scale.Z {type=MovieSceneScriptingFloatChannel, default=1, const [1,325]}
       middle_metacarpal_r_ctrl.Location.X {type=MovieSceneScriptingFloatChannel, default=0, const [0,325]}
       middle_metacarpal_r_ctrl.Location.Y {type=MovieSceneScriptingFloatChannel, default=0, const [0.0001,325]}
       middle_metacarpal_r_ctrl.Location.Z {type=MovieSceneScriptingFloatChannel, default=0, const [0,325]}
       middle_metacarpal_r_ctrl.Rotation.X {type=MovieSceneScriptingFloatChannel, default=0, const [0.2131,325]}
       middle_metacarpal_r_ctrl.Rotation.Y {type=MovieSceneScriptingFloatChannel, default=0, const [-1.1482,325]}
       middle_metacarpal_r_ctrl.Rotation.Z {type=MovieSceneScriptingFloatChannel, default=0, const [-10.5148,325]}
       middle_metacarpal_r_ctrl.Scale.X {type=MovieSceneScriptingFloatChannel, default=1, const [1,325]}
       middle_metacarpal_r_ctrl.Scale.Y {type=MovieSceneScriptingFloatChannel, default=1, const [1,325]}
       middle_metacarpal_r_ctrl.Scale.Z {type=MovieSceneScriptingFloatChannel, default=1, const [1,325]}
       middle_01_r_ctrl.Location.X {type=MovieSceneScriptingFloatChannel, default=0, const [-0.553,325]}
       middle_01_r_ctrl.Location.Y {type=MovieSceneScriptingFloatChannel, default=0, const [-0.9634,325]}
       middle_01_r_ctrl.Location.Z {type=MovieSceneScriptingFloatChannel, default=0, const [-0.0914,325]}
       middle_01_r_ctrl.Rotation.X {type=MovieSceneScriptingFloatChannel, default=0, const [0,325]}
       middle_01_r_ctrl.Rotation.Y {type=MovieSceneScriptingFloatChannel, default=0, const [0,325]}
       middle_01_r_ctrl.Rotation.Z {type=MovieSceneScriptingFloatChannel, default=0, const [-80.8858,325]}
       middle_01_r_ctrl.Scale.X {type=MovieSceneScriptingFloatChannel, default=1, const [1,325]}
       middle_01_r_ctrl.Scale.Y {type=MovieSceneScriptingFloatChannel, default=1, const [1,325]}
       middle_01_r_ctrl.Scale.Z {type=MovieSceneScriptingFloatChannel, default=1, const [1,325]}
       middle_02_r_ctrl.Location.X {type=MovieSceneScriptingFloatChannel, default=0, const [0.0008,325]}
       middle_02_r_ctrl.Location.Y {type=MovieSceneScriptingFloatChannel, default=0, const [0.0009,325]}
       middle_02_r_ctrl.Location.Z {type=MovieSceneScriptingFloatChannel, default=0, const [0.0476,325]}
       middle_02_r_ctrl.Rotation.X {type=MovieSceneScriptingFloatChannel, default=0, const [-3.6824,325]}
       middle_02_r_ctrl.Rotation.Y {type=MovieSceneScriptingFloatChannel, default=0, const [0.451,325]}
       middle_02_r_ctrl.Rotation.Z {type=MovieSceneScriptingFloatChannel, default=0, const [-62.569,325]}
       middle_02_r_ctrl.Scale.X {type=MovieSceneScriptingFloatChannel, default=1, const [1,325]}
       middle_02_r_ctrl.Scale.Y {type=MovieSceneScriptingFloatChannel, default=1, const [1,325]}
       middle_02_r_ctrl.Scale.Z {type=MovieSceneScriptingFloatChannel, default=1, const [1,325]}
       middle_03_r_ctrl.Location.X {type=MovieSceneScriptingFloatChannel, default=0, const [0.0005,325]}
       middle_03_r_ctrl.Location.Y {type=MovieSceneScriptingFloatChannel, default=0, const [0,325]}
       middle_03_r_ctrl.Location.Z {type=MovieSceneScriptingFloatChannel, default=0, const [0.0061,325]}
       middle_03_r_ctrl.Rotation.X {type=MovieSceneScriptingFloatChannel, default=0, const [-0.0051,325]}
       middle_03_r_ctrl.Rotation.Y {type=MovieSceneScriptingFloatChannel, default=0, const [-0.0098,325]}
       middle_03_r_ctrl.Rotation.Z {type=MovieSceneScriptingFloatChannel, default=0, const [-76.6019,325]}
       middle_03_r_ctrl.Scale.X {type=MovieSceneScriptingFloatChannel, default=1, const [1,325]}
       middle_03_r_ctrl.Scale.Y {type=MovieSceneScriptingFloatChannel, default=1, const [1,325]}
       middle_03_r_ctrl.Scale.Z {type=MovieSceneScriptingFloatChannel, default=1, const [1,325]}
       ring_metacarpal_r_ctrl.Location.X {type=MovieSceneScriptingFloatChannel, default=0, const [0.0001,325]}
       ring_metacarpal_r_ctrl.Location.Y {type=MovieSceneScriptingFloatChannel, default=0, const [0.0002,325]}
       ring_metacarpal_r_ctrl.Location.Z {type=MovieSceneScriptingFloatChannel, default=0, const [0,325]}
       ring_metacarpal_r_ctrl.Rotation.X {type=MovieSceneScriptingFloatChannel, default=0, const [5.3943,325]}
       ring_metacarpal_r_ctrl.Rotation.Y {type=MovieSceneScriptingFloatChannel, default=0, const [-2.9431,325]}
       ring_metacarpal_r_ctrl.Rotation.Z {type=MovieSceneScriptingFloatChannel, default=0, const [-10.4667,325]}
       ring_metacarpal_r_ctrl.Scale.X {type=MovieSceneScriptingFloatChannel, default=1, const [1,325]}
       ring_metacarpal_r_ctrl.Scale.Y {type=MovieSceneScriptingFloatChannel, default=1, const [1,325]}
       ring_metacarpal_r_ctrl.Scale.Z {type=MovieSceneScriptingFloatChannel, default=1, const [1,325]}
       ring_01_r_ctrl.Location.X {type=MovieSceneScriptingFloatChannel, default=0, const [-0.371,325]}
       ring_01_r_ctrl.Location.Y {type=MovieSceneScriptingFloatChannel, default=0, const [-1.0509,325]}
       ring_01_r_ctrl.Location.Z {type=MovieSceneScriptingFloatChannel, default=0, const [-0.0243,325]}
       ring_01_r_ctrl.Rotation.X {type=MovieSceneScriptingFloatChannel, default=0, const [-0.8239,325]}
       ring_01_r_ctrl.Rotation.Y {type=MovieSceneScriptingFloatChannel, default=0, const [-2.6856,325]}
       ring_01_r_ctrl.Rotation.Z {type=MovieSceneScriptingFloatChannel, default=0, const [-84.1083,325]}
       ring_01_r_ctrl.Scale.X {type=MovieSceneScriptingFloatChannel, default=1, const [1,325]}
       ring_01_r_ctrl.Scale.Y {type=MovieSceneScriptingFloatChannel, default=1, const [1,325]}
       ring_01_r_ctrl.Scale.Z {type=MovieSceneScriptingFloatChannel, default=1, const [1,325]}
       ring_02_r_ctrl.Location.X {type=MovieSceneScriptingFloatChannel, default=0, const [0.0003,325]}
       ring_02_r_ctrl.Location.Y {type=MovieSceneScriptingFloatChannel, default=0, const [0.0001,325]}
       ring_02_r_ctrl.Location.Z {type=MovieSceneScriptingFloatChannel, default=0, const [0,325]}
       ring_02_r_ctrl.Rotation.X {type=MovieSceneScriptingFloatChannel, default=0, const [-12.8636,325]}
       ring_02_r_ctrl.Rotation.Y {type=MovieSceneScriptingFloatChannel, default=0, const [-0.6911,325]}
       ring_02_r_ctrl.Rotation.Z {type=MovieSceneScriptingFloatChannel, default=0, const [-59.6606,325]}
       ring_02_r_ctrl.Scale.X {type=MovieSceneScriptingFloatChannel, default=1, const [1,325]}
       ring_02_r_ctrl.Scale.Y {type=MovieSceneScriptingFloatChannel, default=1, const [1,325]}
       ring_02_r_ctrl.Scale.Z {type=MovieSceneScriptingFloatChannel, default=1, const [1,325]}
       ring_03_r_ctrl.Location.X {type=MovieSceneScriptingFloatChannel, default=0, const [0,325]}
       ring_03_r_ctrl.Location.Y {type=MovieSceneScriptingFloatChannel, default=0, const [-0.0001,325]}
       ring_03_r_ctrl.Location.Z {type=MovieSceneScriptingFloatChannel, default=0, const [0,325]}
       ring_03_r_ctrl.Rotation.X {type=MovieSceneScriptingFloatChannel, default=0, const [0,325]}
       ring_03_r_ctrl.Rotation.Y {type=MovieSceneScriptingFloatChannel, default=0, const [0,325]}
       ring_03_r_ctrl.Rotation.Z {type=MovieSceneScriptingFloatChannel, default=0, const [-88.4683,325]}
       ring_03_r_ctrl.Scale.X {type=MovieSceneScriptingFloatChannel, default=1, const [1,325]}
       ring_03_r_ctrl.Scale.Y {type=MovieSceneScriptingFloatChannel, default=1, const [1,325]}
       ring_03_r_ctrl.Scale.Z {type=MovieSceneScriptingFloatChannel, default=1, const [1,325]}
       pinky_metacarpal_r_ctrl.Location.X {type=MovieSceneScriptingFloatChannel, default=0, const [-0.0001,325]}
       pinky_metacarpal_r_ctrl.Location.Y {type=MovieSceneScriptingFloatChannel, default=0, const [0,325]}
       pinky_metacarpal_r_ctrl.Location.Z {type=MovieSceneScriptingFloatChannel, default=0, const [0,325]}
       pinky_metacarpal_r_ctrl.Rotation.X {type=MovieSceneScriptingFloatChannel, default=0, const [9.8795,325]}
       pinky_metacarpal_r_ctrl.Rotation.Y {type=MovieSceneScriptingFloatChannel, default=0, const [-6.9881,325]}
       pinky_metacarpal_r_ctrl.Rotation.Z {type=MovieSceneScriptingFloatChannel, default=0, const [-8.607,325]}
       pinky_metacarpal_r_ctrl.Scale.X {type=MovieSceneScriptingFloatChannel, default=1, const [1,325]}
       pinky_metacarpal_r_ctrl.Scale.Y {type=MovieSceneScriptingFloatChannel, default=1, const [1,325]}
       pinky_metacarpal_r_ctrl.Scale.Z {type=MovieSceneScriptingFloatChannel, default=1, const [1,325]}
       pinky_01_r_ctrl.Location.X {type=MovieSceneScriptingFloatChannel, default=0, const [-0.3834,325]}
       pinky_01_r_ctrl.Location.Y {type=MovieSceneScriptingFloatChannel, default=0, const [-1.045,325]}
       pinky_01_r_ctrl.Location.Z {type=MovieSceneScriptingFloatChannel, default=0, const [-0.0533,325]}
       pinky_01_r_ctrl.Rotation.X {type=MovieSceneScriptingFloatChannel, default=0, const [-7.5971,325]}
       pinky_01_r_ctrl.Rotation.Y {type=MovieSceneScriptingFloatChannel, default=0, const [-7.928,325]}
       pinky_01_r_ctrl.Rotation.Z {type=MovieSceneScriptingFloatChannel, default=0, const [-83.5712,325]}
       pinky_01_r_ctrl.Scale.X {type=MovieSceneScriptingFloatChannel, default=1, const [1,325]}
       pinky_01_r_ctrl.Scale.Y {type=MovieSceneScriptingFloatChannel, default=1, const [1,325]}
       pinky_01_r_ctrl.Scale.Z {type=MovieSceneScriptingFloatChannel, default=1, const [1,325]}
       pinky_02_r_ctrl.Location.X {type=MovieSceneScriptingFloatChannel, default=0, const [0.3528,325]}
       pinky_02_r_ctrl.Location.Y {type=MovieSceneScriptingFloatChannel, default=0, const [0.0018,325]}
       pinky_02_r_ctrl.Location.Z {type=MovieSceneScriptingFloatChannel, default=0, const [-0.0756,325]}
       pinky_02_r_ctrl.Rotation.X {type=MovieSceneScriptingFloatChannel, default=0, const [1.9939,325]}
       pinky_02_r_ctrl.Rotation.Y {type=MovieSceneScriptingFloatChannel, default=0, const [10.0966,325]}
       pinky_02_r_ctrl.Rotation.Z {type=MovieSceneScriptingFloatChannel, default=0, const [-57.9295,325]}
       pinky_02_r_ctrl.Scale.X {type=MovieSceneScriptingFloatChannel, default=1, const [1,325]}
       pinky_02_r_ctrl.Scale.Y {type=MovieSceneScriptingFloatChannel, default=1, const [1,325]}
       pinky_02_r_ctrl.Scale.Z {type=MovieSceneScriptingFloatChannel, default=1, const [1,325]}
       pinky_03_r_ctrl.Location.X {type=MovieSceneScriptingFloatChannel, default=0, const [0.0002,325]}
       pinky_03_r_ctrl.Location.Y {type=MovieSceneScriptingFloatChannel, default=0, const [-0.0002,325]}
       pinky_03_r_ctrl.Location.Z {type=MovieSceneScriptingFloatChannel, default=0, const [0,325]}
       pinky_03_r_ctrl.Rotation.X {type=MovieSceneScriptingFloatChannel, default=0, const [3.1496,325]}
       pinky_03_r_ctrl.Rotation.Y {type=MovieSceneScriptingFloatChannel, default=0, const [-4.3172,325]}
       pinky_03_r_ctrl.Rotation.Z {type=MovieSceneScriptingFloatChannel, default=0, const [-111.6788,325]}
       pinky_03_r_ctrl.Scale.X {type=MovieSceneScriptingFloatChannel, default=1, const [1,325]}
       pinky_03_r_ctrl.Scale.Y {type=MovieSceneScriptingFloatChannel, default=1, const [1,325]}
       pinky_03_r_ctrl.Scale.Z {type=MovieSceneScriptingFloatChannel, default=1, const [1,325]}
       hand_r_ik_ctrl.Location.X {
        type=MovieSceneScriptingFloatChannel
        default=0
        thinned_from=325
        keys [[0,0.482],[1,0.4752],[2,0.4529],[3,0.4218],[4,0.3798],[5,0.3267],[6,0.2623],[7,0.1867],[8,0.0997],[9,-0.0005],[10,-0.113],[11,-0.2401],[12,-0.3825],[13,-0.5336],[14,-0.6881],[15,-0.8393],[16,-0.981],[17,-1.1268],[18,-1.2519],[19,-1.3727],[20,-1.4864],[21,-1.5903],[23,-1.7732],[24,-1.8365],[25,-1.8886],[26,-1.9343],[27,-1.9727],[28,-2.0056],[29,-2.0324],[30,-2.0457],[31,-2.0432],[32,-1.9908],[33,-1.8937],[35,-1.5725],[36,-1.4687],[37,-1.3601],[38,-1.2762],[39,-1.2005],[40,-1.1381],[42,-1.0208],[43,-0.9532],[44,-0.869],[45,-0.7685],[46,-0.6554],[47,-0.5199],[48,-0.387],[49,-0.2529],[50,-0.1098],[51,0.0426],[52,0.1957],[53,0.3484],[54,0.4974],[55,0.6388],[56,0.7688],[57,0.8886],[58,1.0005],[59,1.1025],[60,1.188],[61,1.257],[62,1.3157],[63,1.3649],[64,1.4033],[65,1.4309],[66,1.448],[67,1.4559],[68,1.456],[69,1.4492],[70,1.4366],[71,1.419],[72,1.3973],[77,1.2664],[80,1.1897],[82,1.1338],[84,1.0651],[85,1.0378],[86,1.0081],[87,0.9575],[88,0.8699],[89,0.6887],[90,0.2997],[91,0.0047],[92,-0.2381],[93,-0.415],[94,-0.564],[95,-0.6993],[97,-0.9648],[98,-1.0942],[99,-1.2181],[100,-1.3326],[101,-1.4591],[102,-1.5588],[103,-1.6537],[104,-1.7415],[105,-1.8202],[106,-1.8876],[107,-1.963],[108,-2.0058],[109,-2.0365],[110,-2.0593],[111,-2.0745],[113,-2.0912],[116,-2.0839],[118,-2.0632],[120,-2.0304],[121,-2.007],[122,-1.9793],[123,-1.9326],[124,-1.8984],[125,-1.8601],[127,-1.7726],[128,-1.718],[129,-1.6501],[130,-1.572],[131,-1.4701],[132,-1.3739],[133,-1.2666],[134,-1.1481],[136,-0.9012],[137,-0.7745],[139,-0.5177],[140,-0.3938],[141,-0.2729],[142,-0.1546],[143,-0.0387],[144,0.0733],[146,0.293],[147,0.3995],[148,0.5002],[149,0.5961],[150,0.6853],[151,0.7629],[152,0.8247],[153,0.8742],[154,0.9168],[155,0.9527],[156,0.9823],[157,1.0059],[158,1.0228],[159,1.036],[160,1.0435],[161,1.0475],[162,1.0469],[164,1.0352],[166,1.0123],[167,0.9968],[168,0.9784],[169,0.9533],[170,0.9127],[171,0.8612],[172,0.8002],[173,0.7315],[174,0.6569],[175,0.5783],[177,0.4191],[178,0.3424],[179,0.2669],[180,0.1901],[181,0.1119],[182,0.0333],[183,-0.0478],[184,-0.1445],[185,-0.2387],[186,-0.3892],[187,-0.6079],[188,-0.8638],[189,-1.1291],[191,-1.7418],[192,-2.032],[193,-2.3867],[194,-2.6951],[195,-2.8644],[196,-2.9941],[197,-3.0975],[198,-3.1693],[199,-3.2227],[200,-3.2668],[201,-3.3002],[203,-3.346],[206,-3.3848],[209,-3.4],[210,-3.4002],[211,-3.3934],[212,-3.3747],[213,-3.3435],[214,-3.303],[217,-3.1501],[218,-3.0878],[226,-2.4847],[227,-2.4136],[229,-2.2872],[231,-2.1736],[233,-2.0739],[234,-2.0303],[235,-1.9918],[236,-1.9592],[237,-1.9326],[238,-1.912],[239,-1.8977],[240,-1.8898],[242,-1.891],[243,-1.8994],[245,-1.9254],[247,-1.9655],[248,-1.9776],[249,-1.9637],[250,-1.9099],[251,-1.7705],[252,-1.4988],[253,-1.0969],[254,-0.7285],[255,-0.4726],[256,-0.3276],[257,-0.2814],[258,-0.2717],[259,-0.289],[260,-0.3236],[261,-0.3669],[262,-0.4128],[263,-0.46],[266,-0.6048],[268,-0.7064],[269,-0.7692],[270,-0.8198],[271,-0.8822],[272,-0.9546],[273,-1.0345],[274,-1.1195],[275,-1.2127],[276,-1.2796],[277,-1.3315],[278,-1.3703],[279,-1.3924],[281,-1.4137],[283,-1.4261],[285,-1.4294],[287,-1.4234],[289,-1.4073],[290,-1.3955],[291,-1.3726],[293,-1.3383],[294,-1.3174],[295,-1.2905],[296,-1.2542],[297,-1.2088],[298,-1.1561],[299,-1.0873],[300,-1.0231],[301,-0.9529],[304,-0.7184],[307,-0.4741],[309,-0.3165],[311,-0.1533],[312,-0.0768],[313,-0.005],[314,0.0659],[315,0.1347],[316,0.2003],[317,0.2613],[318,0.3171],[319,0.3663],[320,0.4086],[321,0.4428],[322,0.4669],[323,0.4794],[324,0.4798]]
       }
       hand_r_ik_ctrl.Location.Y {
        type=MovieSceneScriptingFloatChannel
        default=0
        thinned_from=325
        keys [[0,9.9827],[5,10.1043],[14,10.0685],[16,10.0314],[23,9.8443],[29,9.8161],[31,9.8762],[32,9.9889],[33,10.1565],[35,10.5589],[37,10.8388],[38,10.9183],[39,10.9646],[44,11.0176],[46,11.0787],[53,11.3366],[56,11.374],[58,11.3537],[60,11.2788],[65,10.8742],[74,10.0416],[77,9.8544],[83,9.6694],[84,9.6501],[87,9.6645],[88,9.6135],[89,9.4913],[91,9.0027],[92,8.7979],[93,8.6431],[94,8.5214],[95,8.4333],[96,8.3705],[97,8.3256],[98,8.2989],[99,8.29],[101,8.3115],[107,8.5364],[113,8.8655],[122,9.2827],[130,9.726],[138,10.3772],[140,10.4958],[142,10.5675],[143,10.5748],[149,10.3547],[151,10.24],[158,10.1021],[159,10.1072],[168,10.0421],[169,10.0563],[175,9.9792],[178,9.9881],[180,10.0358],[184,10.2215],[186,10.3755],[190,10.9328],[191,11.0491],[192,11.1264],[199,11.3372],[216,11.3916],[224,11.312],[230,11.0915],[241,10.4655],[244,10.3315],[249,10.1345],[250,10.0736],[251,9.9816],[253,9.6759],[254,9.5827],[256,9.6359],[259,9.6745],[261,9.7443],[262,9.7993],[263,9.8846],[268,10.4977],[272,10.9236],[273,10.9954],[274,11.0358],[291,11.286],[296,11.2126],[299,11.0801],[303,10.7506],[311,9.8711],[312,9.7875],[314,9.695],[316,9.6687],[318,9.7172],[324,9.9823]]
       }
       hand_r_ik_ctrl.Location.Z {
        type=MovieSceneScriptingFloatChannel
        default=0
        thinned_from=325
        keys [[0,2.1557],[1,2.2325],[2,2.3235],[3,2.4293],[4,2.5471],[5,2.673],[7,2.9419],[13,3.7858],[14,3.913],[15,4.026],[16,4.1192],[17,4.1513],[18,4.2163],[20,4.3173],[22,4.3975],[23,4.3901],[24,4.4147],[25,4.4291],[26,4.4311],[28,4.4156],[29,4.4185],[30,4.4338],[31,4.4711],[32,4.556],[33,4.6836],[34,4.8561],[35,5.0526],[37,5.4135],[38,5.5081],[39,5.5639],[40,5.5567],[41,5.5283],[45,5.3569],[46,5.3034],[47,5.2847],[48,5.2308],[49,5.1601],[50,5.0675],[57,4.3118],[59,4.0649],[60,3.9177],[61,3.7415],[67,2.5806],[68,2.3922],[69,2.2129],[70,2.0446],[71,1.8864],[72,1.7374],[73,1.5999],[74,1.4778],[75,1.3665],[76,1.2774],[77,1.2079],[78,1.1522],[79,1.1123],[80,1.0867],[81,1.0648],[82,1.0604],[83,1.0513],[84,1.063],[85,1.1149],[86,1.2281],[87,1.3489],[88,1.4357],[89,1.4257],[90,1.3292],[91,1.0888],[92,0.9106],[93,0.8464],[94,0.8285],[95,0.8526],[96,0.9067],[97,0.9816],[98,1.0712],[99,1.1701],[100,1.2744],[101,1.3461],[103,1.6041],[104,1.7285],[106,1.9698],[107,2.0433],[108,2.1503],[109,2.2511],[110,2.3435],[111,2.4232],[112,2.4958],[113,2.5547],[114,2.628],[115,2.6929],[117,2.8088],[122,3.0648],[123,3.1536],[128,3.3887],[129,3.4274],[130,3.4573],[131,3.5217],[133,3.5705],[135,3.6021],[140,3.7211],[141,3.7308],[142,3.7204],[143,3.6789],[144,3.6037],[147,3.3243],[148,3.238],[149,3.1292],[151,2.8529],[152,2.7535],[153,2.6897],[154,2.6354],[155,2.5897],[156,2.5525],[157,2.5235],[159,2.4824],[161,2.462],[163,2.4587],[165,2.4759],[168,2.5257],[170,2.5787],[171,2.6184],[172,2.6648],[173,2.7231],[174,2.7984],[175,2.8862],[177,3.0862],[178,3.1931],[179,3.3198],[180,3.4721],[182,3.8086],[183,3.9701],[184,4.0612],[185,4.1752],[186,4.3145],[187,4.3793],[188,4.3916],[189,4.313],[190,4.1399],[191,3.8999],[192,3.605],[193,3.1302],[194,2.7051],[195,2.5065],[196,2.3578],[197,2.2616],[198,2.205],[199,2.1728],[200,2.1559],[201,2.1449],[206,2.135],[208,2.1191],[210,2.094],[211,2.0762],[212,2.0454],[213,1.9944],[214,1.9266],[215,1.8442],[216,1.7499],[217,1.6458],[218,1.5262],[219,1.392],[220,1.2487],[221,1.1088],[222,0.9653],[224,0.6826],[225,0.5341],[226,0.3801],[228,0.0675],[229,-0.0892],[230,-0.2486],[231,-0.4098],[232,-0.572],[233,-0.7373],[234,-0.9044],[235,-1.0671],[236,-1.2196],[237,-1.3609],[238,-1.4935],[239,-1.6152],[240,-1.7235],[241,-1.8213],[242,-1.9049],[243,-1.9818],[244,-2.0523],[245,-2.1132],[246,-2.1695],[247,-2.2075],[248,-2.2367],[249,-2.2183],[250,-2.1347],[251,-1.9511],[252,-1.5893],[253,-1.0898],[254,-0.5388],[255,0.0217],[256,0.4178],[257,0.6589],[258,0.8668],[259,1.053],[260,1.2311],[261,1.4125],[262,1.609],[263,1.8566],[264,2.1633],[267,3.1525],[268,3.4595],[269,3.7114],[271,4.3294],[272,4.6219],[273,4.8766],[274,5.0737],[275,5.1797],[276,5.3062],[277,5.4182],[278,5.5105],[279,5.5816],[281,5.6852],[283,5.764],[286,5.8392],[289,5.875],[290,5.8788],[291,5.9158],[293,5.8779],[295,5.8005],[296,5.7423],[297,5.6621],[298,5.5574],[299,5.4685],[300,5.3228],[301,5.1448],[302,4.9373],[303,4.7124],[305,4.2271],[308,3.4577],[309,3.2075],[311,2.6853],[312,2.472],[313,2.3067],[314,2.1664],[315,2.0544],[316,1.9736],[317,1.93],[318,1.9196],[319,1.9299],[320,1.9507],[321,1.9825],[322,2.0301],[323,2.087],[324,2.1553]]
       }
       hand_r_ik_ctrl.Rotation.X {
        type=MovieSceneScriptingFloatChannel
        default=0
        thinned_from=325
        keys [[0,26.3712],[16,25.7124],[17,25.631],[22,25.5645],[36,25.2295],[38,25.2513],[46,25.4709],[59,26.2119],[74,27.4533],[84,27.6956],[90,27.4332],[100,26.6847],[101,26.5669],[106,26.3065],[113,25.879],[122,25.585],[130,25.3677],[141,25.343],[149,25.9128],[152,26.1859],[173,26.439],[183,26.1344],[185,25.9857],[186,26.011],[216,25.3503],[250,26.4387],[259,26.3735],[262,26.2273],[269,25.5173],[290,25.1021],[291,25.1511],[298,25.2055],[309,26.098],[312,26.4229],[316,26.6017],[324,26.3711]]
       }
       hand_r_ik_ctrl.Rotation.Y {
        type=MovieSceneScriptingFloatChannel
        default=0
        thinned_from=325
        keys [[0,3.2496],[2,3.6076],[10,5.1626],[13,5.708],[14,5.8785],[15,6.0348],[16,6.1763],[17,6.2142],[19,6.4811],[22,6.8234],[23,6.825],[25,6.9639],[29,7.1968],[30,7.2688],[32,7.4427],[34,7.5755],[35,7.6163],[36,7.6317],[37,7.5899],[38,7.478],[39,7.4022],[42,6.8474],[44,6.5137],[46,6.2096],[47,6.1612],[48,6.0459],[49,5.912],[51,5.594],[52,5.4211],[53,5.2354],[54,5.0351],[55,4.819],[56,4.5853],[57,4.3297],[58,4.0507],[59,3.7532],[60,3.4261],[61,3.0548],[63,2.2576],[64,1.8732],[65,1.504],[66,1.1426],[67,0.792],[68,0.457],[69,0.146],[70,-0.141],[71,-0.4083],[72,-0.6606],[73,-0.8948],[74,-1.1079],[75,-1.2991],[76,-1.4577],[77,-1.5848],[78,-1.6867],[79,-1.774],[80,-1.8314],[81,-1.8993],[82,-1.9388],[83,-1.9882],[84,-2.0034],[85,-1.9592],[86,-1.8238],[87,-1.6527],[88,-1.4725],[89,-1.3189],[90,-1.1503],[91,-0.9455],[92,-0.7442],[93,-0.541],[94,-0.332],[95,-0.1204],[96,0.0844],[97,0.2899],[98,0.5014],[99,0.7192],[100,0.9406],[101,1.0861],[102,1.3616],[103,1.6327],[104,1.8839],[106,2.3543],[107,2.4712],[108,2.6698],[110,3.0482],[112,3.4031],[115,3.8861],[116,4.0221],[117,4.1744],[121,4.6901],[122,4.813],[123,5.018],[128,5.5149],[130,5.6932],[131,5.88],[136,6.4826],[137,6.614],[138,6.719],[139,6.805],[140,6.8582],[141,6.8675],[142,6.8245],[143,6.7004],[144,6.4895],[145,6.2259],[148,5.3691],[149,5.0337],[150,4.6267],[151,4.2404],[152,3.9671],[153,3.7947],[154,3.6469],[155,3.5202],[156,3.4137],[157,3.3257],[158,3.2513],[159,3.1868],[160,3.1356],[161,3.0918],[163,3.0233],[164,3.0154],[165,2.9988],[168,2.9991],[169,3.0172],[170,3.0138],[172,3.0365],[173,3.0633],[174,3.1184],[175,3.1932],[177,3.3738],[178,3.4693],[179,3.5896],[180,3.7398],[183,4.2668],[184,4.3362],[185,4.5095],[186,4.7347],[188,5.0238],[189,5.1557],[190,5.2758],[191,5.3813],[193,5.5565],[197,5.8285],[199,5.9387],[206,6.3693],[210,6.5613],[211,6.5962],[212,6.6136],[213,6.6111],[215,6.5628],[217,6.4741],[219,6.3462],[221,6.1605],[222,6.0433],[223,5.9101],[224,5.7612],[225,5.5942],[226,5.4101],[228,5.0128],[229,4.8045],[231,4.3598],[232,4.1261],[233,3.8785],[235,3.363],[236,3.1237],[237,2.9026],[238,2.6927],[239,2.4985],[240,2.3258],[241,2.1688],[242,2.0364],[243,1.9161],[244,1.8068],[246,1.6217],[247,1.5734],[248,1.4997],[249,1.4578],[250,1.4112],[251,1.3802],[252,1.3542],[253,1.3368],[254,1.3324],[255,1.3544],[256,1.3886],[257,1.4448],[258,1.529],[259,1.6462],[260,1.805],[261,2.0123],[262,2.2756],[263,2.6478],[264,3.1239],[267,4.6092],[268,5.0649],[269,5.4005],[271,6.2689],[272,6.6605],[273,6.9896],[274,7.2308],[275,7.2945],[277,7.5163],[279,7.6938],[281,7.8262],[283,7.9285],[286,8.0199],[289,8.0552],[290,8.0537],[291,8.1303],[293,8.0388],[294,7.9663],[295,7.8714],[296,7.7537],[297,7.6113],[298,7.4402],[299,7.3356],[300,7.1276],[301,6.8891],[302,6.6159],[303,6.3179],[304,5.9954],[305,5.654],[306,5.2896],[307,4.8983],[309,4.087],[310,3.6403],[311,3.1837],[312,2.8285],[313,2.5827],[314,2.3913],[315,2.2611],[316,2.1993],[317,2.2187],[318,2.308],[319,2.4352],[320,2.573],[321,2.7212],[323,3.0613],[324,3.2488]]
       }
       hand_r_ik_ctrl.Rotation.Z {
        type=MovieSceneScriptingFloatChannel
        default=0
        thinned_from=325
        keys [[0,8.7836],[1,8.8607],[11,9.8871],[15,10.3523],[16,10.4539],[17,10.4992],[22,10.9548],[23,10.9775],[33,11.528],[35,11.6004],[36,11.6171],[37,11.6013],[38,11.5457],[39,11.5151],[46,10.8537],[47,10.8166],[49,10.6543],[52,10.3283],[54,10.0619],[56,9.7506],[58,9.3956],[60,8.9936],[64,8.067],[67,7.4421],[69,7.0732],[71,6.7587],[74,6.3453],[76,6.1507],[77,6.0699],[81,5.8995],[83,5.8433],[84,5.8355],[85,5.8767],[86,5.934],[88,6.1292],[89,6.255],[90,6.3962],[100,8.0134],[101,8.1403],[103,8.5101],[106,8.9984],[107,9.0982],[111,9.6103],[115,10.0304],[122,10.5946],[123,10.7073],[130,11.0489],[131,11.1401],[137,11.4173],[139,11.434],[140,11.4121],[141,11.367],[142,11.2954],[143,11.1797],[144,11.0133],[148,10.191],[149,9.9653],[151,9.4714],[152,9.2998],[154,9.0905],[156,8.9365],[158,8.8122],[160,8.7405],[164,8.6549],[168,8.6386],[169,8.6681],[172,8.6909],[174,8.7571],[178,8.998],[180,9.1686],[183,9.4974],[184,9.5741],[185,9.722],[190,10.2163],[196,10.6727],[207,11.2752],[211,11.4164],[214,11.4413],[218,11.3764],[221,11.2501],[224,11.0297],[229,10.4909],[233,9.9675],[236,9.5465],[239,9.2032],[241,9.0089],[244,8.8097],[248,8.6603],[250,8.6087],[252,8.5838],[253,8.5975],[254,8.5865],[256,8.6198],[258,8.6996],[259,8.7648],[260,8.8522],[261,8.9649],[262,9.1075],[263,9.3081],[267,10.3434],[268,10.5846],[269,10.76],[271,11.2093],[272,11.4111],[273,11.5807],[274,11.7067],[275,11.7386],[280,11.9961],[286,12.1359],[290,12.1487],[291,12.1897],[294,12.0815],[296,11.9432],[298,11.7275],[299,11.6375],[301,11.322],[303,10.9393],[305,10.5044],[307,10.0087],[309,9.48],[311,8.907],[312,8.6809],[313,8.521],[314,8.3931],[315,8.3014],[316,8.2501],[317,8.2469],[318,8.2852],[321,8.4937],[324,8.7847]]
       }
       hand_r_ik_ctrl.Scale.X {type=MovieSceneScriptingFloatChannel, default=1, const [1,325]}
       hand_r_ik_ctrl.Scale.Y {type=MovieSceneScriptingFloatChannel, default=1, const [1,325]}
       hand_r_ik_ctrl.Scale.Z {type=MovieSceneScriptingFloatChannel, default=1, const [1,325]}
       Softness_2 {type=MovieSceneScriptingFloatChannel, default=0, const [0,325]}
       hand_r_ikorient_ctrl.Rotation.Roll {type=MovieSceneScriptingFloatChannel, default=0, const [0,325]}
       hand_r_ikorient_ctrl.Rotation.Pitch {type=MovieSceneScriptingFloatChannel, default=0, const [0,325]}
       hand_r_ikorient_ctrl.Rotation.Yaw {type=MovieSceneScriptingFloatChannel, default=0, const [0,325]}
       arm_r_pv_ik_ctrl.Location.X {
        type=MovieSceneScriptingFloatChannel
        default=0
        thinned_from=325
        keys [[0,6.0659],[16,6.1141],[17,6.1011],[22,6.2425],[23,6.2302],[29,6.3054],[30,6.3051],[31,6.2883],[32,6.2105],[33,6.0853],[35,5.7778],[37,5.5182],[38,5.4229],[39,5.3928],[40,5.3421],[46,5.1323],[47,5.1255],[58,4.8971],[60,4.8751],[64,4.8951],[73,5.1016],[74,5.131],[77,5.1842],[84,5.2184],[86,5.2116],[87,5.2287],[88,5.2883],[89,5.4046],[90,5.5995],[91,5.8416],[92,6.0595],[93,6.2314],[94,6.3753],[95,6.4854],[96,6.5655],[97,6.6258],[98,6.6686],[99,6.6973],[100,6.7122],[101,6.6882],[106,6.7223],[107,6.6833],[113,6.6039],[122,6.5508],[123,6.5688],[130,6.4321],[131,6.4437],[143,6.2576],[149,6.1864],[157,6.0981],[158,6.0999],[168,6.0596],[177,5.994],[179,5.957],[182,5.8545],[183,5.8117],[184,5.7377],[185,5.6862],[186,5.6166],[187,5.4629],[188,5.2967],[189,5.1537],[190,4.9948],[191,4.8712],[192,4.7947],[193,4.7605],[198,4.6814],[201,4.6733],[207,4.735],[213,4.85],[241,5.4975],[247,5.5901],[249,5.6439],[250,5.6891],[251,5.7794],[252,5.9337],[253,6.1262],[254,6.2297],[256,6.1611],[262,6.0449],[264,5.9821],[268,5.8],[269,5.7244],[272,5.5923],[273,5.5584],[274,5.5368],[275,5.4894],[279,5.4066],[290,5.3238],[291,5.3462],[296,5.3407],[298,5.3643],[299,5.4155],[308,5.8313],[311,5.9341],[316,6.0245],[324,6.0662]]
       }
       arm_r_pv_ik_ctrl.Location.Y {
        type=MovieSceneScriptingFloatChannel
        default=0
        thinned_from=325
        keys [[0,7.3083],[3,7.344],[9,7.4946],[12,7.5226],[14,7.5085],[16,7.4571],[17,7.3753],[22,7.1399],[23,7.0561],[28,6.8119],[29,6.7713],[30,6.7459],[31,6.7428],[32,6.8089],[33,6.936],[34,7.1363],[35,7.3663],[36,7.5661],[37,7.7892],[38,7.9462],[39,8.0796],[40,8.1585],[46,8.4842],[47,8.5762],[49,8.6937],[59,9.175],[60,9.1998],[63,9.1771],[65,9.1105],[74,8.5424],[77,8.4092],[80,8.3431],[84,8.3251],[85,8.3518],[87,8.4395],[88,8.4209],[89,8.2716],[90,7.9432],[91,7.5147],[92,7.1664],[93,6.9486],[94,6.7821],[95,6.6581],[96,6.564],[98,6.4284],[100,6.336],[101,6.2628],[103,6.2292],[105,6.2222],[106,6.2282],[107,6.1969],[112,6.2287],[113,6.2186],[120,6.2761],[122,6.3118],[123,6.3708],[130,6.5689],[131,6.6373],[135,6.7869],[138,6.9444],[150,7.8119],[152,7.9125],[160,8.0896],[174,8.2571],[178,8.3741],[179,8.4238],[183,8.7073],[184,8.7081],[186,8.7466],[187,8.6684],[188,8.5275],[189,8.3023],[190,7.9758],[191,7.5918],[192,7.1683],[193,6.557],[194,6.0123],[195,5.7325],[196,5.5154],[197,5.3556],[198,5.2467],[199,5.1681],[200,5.102],[202,5.003],[212,4.6214],[214,4.5273],[216,4.4099],[218,4.2691],[221,4.0375],[223,3.907],[235,3.2457],[239,3.0541],[241,2.9676],[244,2.8525],[247,2.7505],[248,2.7377],[249,2.7696],[250,2.8797],[251,3.1141],[252,3.5663],[253,4.2008],[254,4.8589],[255,5.4669],[256,5.8746],[257,6.0933],[258,6.2574],[259,6.3803],[260,6.4759],[262,6.6328],[263,6.7224],[264,6.8356],[268,7.3687],[269,7.4583],[272,7.9053],[273,8.0271],[274,8.115],[275,8.1344],[278,8.2852],[284,8.4255],[290,8.4902],[291,8.5345],[296,8.5093],[298,8.4558],[299,8.4505],[300,8.4042],[302,8.2612],[307,7.8268],[310,7.625],[314,7.4568],[319,7.3416],[324,7.3073]]
       }
       arm_r_pv_ik_ctrl.Location.Z {
        type=MovieSceneScriptingFloatChannel
        default=0
        thinned_from=325
        keys [[0,4.17],[1,4.1782],[2,4.1981],[3,4.2299],[4,4.2734],[5,4.3302],[6,4.4002],[7,4.4841],[8,4.5822],[9,4.6965],[10,4.8265],[11,4.9714],[16,5.7981],[17,5.9985],[20,6.4315],[22,6.6822],[23,6.8343],[25,7.0234],[27,7.173],[29,7.286],[30,7.3185],[31,7.3271],[32,7.2865],[33,7.2022],[35,6.9416],[37,6.8144],[38,6.767],[39,6.6885],[41,6.605],[42,6.5513],[43,6.4789],[44,6.3785],[45,6.2513],[46,6.1038],[47,5.8945],[49,5.5225],[50,5.3239],[53,4.6815],[54,4.4765],[55,4.2838],[56,4.109],[57,3.9487],[58,3.8003],[59,3.6684],[60,3.5652],[61,3.4918],[62,3.4339],[63,3.3874],[64,3.3574],[65,3.3451],[66,3.3486],[67,3.3664],[68,3.3955],[69,3.4332],[71,3.5301],[73,3.6417],[74,3.7072],[77,3.8584],[78,3.8921],[79,3.9385],[81,4.0142],[82,4.0437],[84,4.1334],[86,4.1687],[87,4.2087],[88,4.305],[89,4.5102],[90,4.9221],[91,5.1851],[92,5.4189],[93,5.6222],[97,6.3609],[99,6.6991],[100,6.8438],[101,7.0131],[103,7.2126],[105,7.3697],[106,7.4292],[107,7.5278],[109,7.5958],[111,7.6185],[114,7.6066],[116,7.5807],[120,7.478],[122,7.3956],[123,7.3044],[126,7.1387],[128,6.9984],[130,6.8053],[131,6.6424],[132,6.5012],[133,6.3371],[139,5.1921],[141,4.849],[142,4.687],[143,4.5354],[148,3.8658],[149,3.7393],[150,3.6212],[151,3.5171],[152,3.4325],[153,3.3632],[154,3.3024],[155,3.2504],[156,3.2064],[157,3.1699],[158,3.1498],[159,3.1206],[161,3.0946],[162,3.0924],[163,3.0989],[164,3.0987],[166,3.1167],[169,3.1651],[170,3.2139],[171,3.2711],[172,3.3392],[177,3.7433],[178,3.8181],[179,3.8824],[180,3.9373],[183,4.0688],[184,4.1425],[185,4.1662],[186,4.2475],[187,4.3589],[188,4.4821],[190,4.6903],[191,4.7865],[193,4.9493],[194,5.0219],[195,5.0729],[202,5.3429],[210,5.5786],[211,5.5996],[212,5.6085],[213,5.6027],[215,5.5577],[217,5.4871],[218,5.436],[222,5.1804],[226,4.9676],[228,4.8806],[231,4.7841],[234,4.7218],[236,4.7018],[238,4.7028],[240,4.7277],[241,4.7591],[242,4.7763],[244,4.8517],[245,4.8808],[248,5.0209],[250,5.0923],[251,5.0973],[252,5.0807],[253,5.0059],[254,4.9554],[255,4.9556],[256,4.9816],[260,5.1952],[261,5.2399],[262,5.2713],[263,5.2838],[264,5.277],[268,5.1709],[269,5.1756],[270,5.1241],[271,5.0848],[272,5.0651],[273,5.0717],[274,5.1109],[275,5.2123],[276,5.2661],[277,5.3065],[278,5.3319],[284,5.3259],[290,5.2838],[291,5.2383],[298,5.2133],[299,5.1639],[301,5.1252],[309,4.8973],[311,4.8071],[318,4.3756],[320,4.2704],[321,4.2289],[322,4.1975],[324,4.1719]]
       }
       arm_r_pv_ik_ctrl.Rotation.X {
        type=MovieSceneScriptingFloatChannel
        default=0
        thinned_from=325
        keys [[0,13.8363],[10,13.636],[14,13.6429],[16,13.7206],[17,13.8665],[22,14.2471],[23,14.4241],[29,15.0833],[32,15.2692],[35,15.3387],[38,15.3257],[39,15.2421],[42,15.1963],[44,15.0806],[46,14.9038],[47,14.7139],[57,13.4556],[59,13.257],[60,13.1995],[62,13.1867],[64,13.2412],[66,13.3849],[74,14.3034],[76,14.4807],[79,14.6486],[81,14.6955],[84,14.7149],[85,14.6606],[87,14.4544],[88,14.4067],[89,14.4663],[98,15.4333],[100,15.5851],[101,15.7147],[106,15.6806],[107,15.7585],[121,15.7004],[122,15.6757],[123,15.5694],[130,15.3297],[131,15.198],[134,14.9173],[146,13.3143],[150,12.7303],[151,12.6123],[154,12.3908],[159,12.1707],[169,11.9499],[175,11.8368],[178,11.7044],[179,11.6252],[180,11.513],[183,11.1029],[184,11.0603],[186,10.7933],[188,10.6317],[190,10.5241],[192,10.488],[195,10.5287],[198,10.6548],[204,11.0513],[238,14.3232],[246,14.9175],[250,15.1426],[254,15.2246],[258,15.183],[260,15.0897],[262,14.9068],[263,14.7565],[264,14.5527],[268,13.5494],[269,13.3719],[271,12.7558],[272,12.4695],[273,12.2315],[274,12.0666],[275,12.0478],[278,11.8429],[281,11.7141],[283,11.6564],[290,11.6025],[291,11.5361],[294,11.6266],[296,11.7335],[298,11.9047],[299,11.9287],[301,12.1455],[308,13.1302],[310,13.3643],[313,13.5888],[317,13.7561],[323,13.8463],[324,13.8367]]
       }
       arm_r_pv_ik_ctrl.Rotation.Y {
        type=MovieSceneScriptingFloatChannel
        default=0
        thinned_from=325
        keys [[0,-7.4264],[1,-7.4062],[11,-7.3927],[17,-7.5037],[22,-7.5144],[30,-7.5407],[44,-7.6363],[52,-7.7041],[73,-7.4398],[74,-7.4108],[75,-7.4105],[77,-7.373],[78,-7.3813],[84,-7.3549],[85,-7.3767],[88,-7.3905],[100,-7.6223],[106,-7.7631],[113,-7.9253],[122,-8.0046],[130,-8.0583],[137,-8.0526],[148,-7.7797],[157,-7.7422],[158,-7.7272],[159,-7.7407],[160,-7.7313],[168,-7.735],[169,-7.7556],[178,-7.8668],[183,-8.0105],[185,-8.1429],[186,-8.1119],[196,-8.4262],[199,-8.4889],[209,-8.544],[223,-8.4055],[241,-8.0891],[242,-8.0908],[244,-8.0559],[245,-8.0616],[250,-8.01],[252,-8.0062],[253,-8.0237],[261,-8.0492],[268,-8.232],[272,-8.3634],[284,-8.4907],[290,-8.4895],[296,-8.4276],[298,-8.3711],[310,-7.6887],[314,-7.5495],[320,-7.4448],[324,-7.4277]]
       }
       arm_r_pv_ik_ctrl.Rotation.Z {
        type=MovieSceneScriptingFloatChannel
        default=0
        thinned_from=325
        keys [[0,27.241],[16,27.2816],[17,27.1736],[22,27.1404],[23,27.0241],[38,26.6959],[39,26.7801],[46,26.8051],[47,26.913],[60,27.4666],[84,26.9234],[88,27.0358],[100,26.6053],[101,26.5021],[106,26.618],[107,26.5389],[122,26.5616],[123,26.6546],[130,26.6415],[131,26.7355],[157,27.8114],[179,28.032],[183,28.2212],[184,28.1714],[186,28.3294],[256,26.7268],[262,26.8377],[268,27.3242],[269,27.3313],[273,27.8288],[274,27.9089],[275,27.8624],[290,28.0222],[291,28.1016],[298,27.8744],[299,27.9077],[323,27.2401],[324,27.2412]]
       }
       arm_r_pv_ik_ctrl.Scale.X {type=MovieSceneScriptingFloatChannel, default=1, const [1,325]}
       arm_r_pv_ik_ctrl.Scale.Y {type=MovieSceneScriptingFloatChannel, default=1, const [1,325]}
       arm_r_pv_ik_ctrl.Scale.Z {type=MovieSceneScriptingFloatChannel, default=1, const [1,325]}
       body_ctrl.Location.X {
        type=MovieSceneScriptingFloatChannel
        default=0
        thinned_from=325
        keys [[0,1.5753],[16,1.588],[17,1.5941],[22,1.5898],[23,1.5961],[29,1.5996],[30,1.6056],[31,1.6234],[32,1.6861],[33,1.7798],[34,1.8855],[35,1.9667],[36,2.0307],[37,2.0891],[38,2.1261],[40,2.1506],[42,2.1621],[46,2.1621],[55,2.1252],[84,2.1035],[85,2.097],[86,2.0854],[87,2.0616],[88,2.0175],[89,1.9479],[90,1.8495],[91,1.7447],[93,1.5205],[94,1.4216],[95,1.3451],[96,1.2913],[97,1.2522],[98,1.2259],[99,1.2105],[100,1.2038],[101,1.209],[106,1.2115],[107,1.2191],[122,1.2233],[123,1.2176],[130,1.2176],[131,1.2101],[139,1.1864],[170,1.1708],[177,1.1828],[179,1.1914],[181,1.2053],[182,1.2151],[183,1.2276],[184,1.2469],[185,1.2627],[186,1.3026],[187,1.3603],[188,1.4293],[189,1.5345],[190,1.6859],[191,1.8223],[192,1.9304],[193,2.0219],[194,2.0738],[195,2.0968],[197,2.1295],[199,2.1482],[202,2.1585],[247,2.0955],[248,2.0882],[249,2.0738],[250,2.0469],[251,1.9853],[252,1.8614],[253,1.7176],[254,1.6233],[255,1.5699],[256,1.5335],[257,1.5113],[258,1.4992],[259,1.4946],[268,1.5207],[269,1.5294],[274,1.5382],[275,1.547],[290,1.5795],[291,1.576],[298,1.5907],[299,1.5858],[324,1.5753]]
       }
       body_ctrl.Location.Y {
        type=MovieSceneScriptingFloatChannel
        default=0
        thinned_from=325
        keys [[0,-1.7915],[5,-1.7812],[10,-1.7517],[16,-1.7026],[17,-1.6834],[22,-1.6711],[23,-1.6526],[26,-1.6386],[38,-1.6387],[39,-1.6517],[45,-1.6532],[46,-1.6557],[47,-1.674],[53,-1.728],[55,-1.7401],[85,-1.7849],[89,-1.7759],[93,-1.7579],[100,-1.7033],[101,-1.6838],[106,-1.6713],[107,-1.6521],[110,-1.6387],[122,-1.6403],[123,-1.6541],[129,-1.6562],[130,-1.6588],[131,-1.6778],[137,-1.733],[140,-1.749],[169,-1.7915],[174,-1.7753],[183,-1.7035],[184,-1.6838],[185,-1.6822],[186,-1.6914],[188,-1.7021],[191,-1.7062],[208,-1.7101],[212,-1.7232],[221,-1.8088],[247,-1.8599],[250,-1.857],[251,-1.8516],[252,-1.8406],[254,-1.8042],[268,-1.7028],[269,-1.6836],[274,-1.6712],[275,-1.6527],[278,-1.6386],[290,-1.6406],[291,-1.654],[297,-1.6557],[298,-1.658],[299,-1.6756],[306,-1.7366],[318,-1.7848],[324,-1.7915]]
       }
       body_ctrl.Location.Z {
        type=MovieSceneScriptingFloatChannel
        default=0
        thinned_from=325
        keys [[0,1.6458],[2,1.649],[3,1.6544],[4,1.6637],[5,1.6778],[6,1.6978],[7,1.7247],[8,1.7594],[9,1.8049],[10,1.8615],[11,1.9259],[14,2.135],[15,2.1994],[16,2.256],[17,2.3025],[19,2.3726],[21,2.428],[23,2.4692],[25,2.494],[27,2.5024],[34,2.4931],[38,2.4675],[41,2.431],[43,2.397],[44,2.3728],[45,2.3397],[46,2.2992],[48,2.2006],[52,1.9838],[53,1.9344],[54,1.8899],[55,1.8521],[56,1.8226],[58,1.777],[60,1.7398],[62,1.7101],[65,1.678],[68,1.6585],[72,1.6477],[77,1.6492],[80,1.6619],[82,1.6798],[84,1.7078],[85,1.7269],[86,1.7503],[88,1.808],[90,1.8772],[98,2.1946],[101,2.2989],[106,2.454],[108,2.5032],[109,2.5203],[110,2.5313],[111,2.5351],[114,2.5244],[117,2.495],[122,2.4155],[128,2.291],[130,2.2354],[137,1.9986],[139,1.9407],[144,1.8329],[149,1.7442],[153,1.6881],[157,1.6472],[160,1.6286],[168,1.6168],[169,1.6192],[170,1.6264],[171,1.6377],[172,1.6529],[174,1.693],[176,1.743],[185,2.0027],[187,2.0696],[191,2.2166],[193,2.2793],[194,2.3042],[196,2.338],[199,2.3716],[204,2.3991],[211,2.4035],[213,2.3946],[215,2.3717],[218,2.3168],[223,2.1938],[225,2.1361],[235,1.7801],[237,1.7179],[239,1.6663],[240,1.6454],[242,1.6133],[244,1.5917],[246,1.5786],[250,1.5695],[254,1.573],[257,1.5864],[259,1.6029],[261,1.6261],[263,1.6568],[265,1.6956],[267,1.743],[269,1.8009],[270,1.8412],[271,1.8978],[272,1.9664],[275,2.2003],[276,2.2724],[277,2.3343],[278,2.3814],[279,2.4093],[282,2.4509],[286,2.4826],[296,2.4941],[298,2.486],[301,2.4568],[304,2.4083],[306,2.3642],[308,2.3104],[309,2.2771],[310,2.2369],[311,2.1909],[313,2.0869],[317,1.8656],[318,1.8148],[319,1.7684],[320,1.7276],[321,1.6937],[322,1.6679],[323,1.6516],[324,1.6458]]
       }
       body_ctrl.Rotation.X {
        type=MovieSceneScriptingFloatChannel
        default=0
        thinned_from=325
        keys [[0,-1.3872],[1,-1.3832],[2,-1.3347],[3,-1.2714],[4,-1.1869],[5,-1.0807],[6,-0.9554],[7,-0.8119],[8,-0.6519],[9,-0.4755],[10,-0.2849],[11,-0.0834],[12,0.1231],[13,0.3292],[14,0.5299],[15,0.7232],[16,0.8998],[17,1.1307],[18,1.2529],[19,1.3833],[21,1.6487],[22,1.7738],[23,1.9821],[24,2.1012],[25,2.1951],[26,2.255],[27,2.2763],[30,2.271],[31,2.2506],[32,2.236],[34,2.1932],[35,2.163],[36,2.1256],[37,2.0802],[38,2.0301],[39,1.9056],[40,1.86],[41,1.8095],[42,1.7534],[43,1.6912],[44,1.6109],[45,1.5041],[46,1.3754],[47,1.1372],[48,0.9444],[49,0.745],[50,0.5438],[51,0.3455],[52,0.1564],[53,-0.017],[54,-0.17],[55,-0.2969],[56,-0.3901],[57,-0.4642],[58,-0.536],[59,-0.6034],[60,-0.6682],[61,-0.7307],[62,-0.7885],[63,-0.8442],[64,-0.898],[66,-0.9945],[67,-1.0391],[69,-1.1195],[70,-1.1555],[72,-1.2204],[73,-1.2492],[75,-1.2981],[76,-1.3195],[78,-1.3541],[79,-1.3575],[80,-1.3788],[81,-1.3814],[82,-1.3943],[84,-1.3962],[85,-1.3923],[86,-1.3517],[87,-1.2894],[88,-1.204],[89,-1.0951],[90,-0.9619],[91,-0.8163],[92,-0.6555],[93,-0.4795],[94,-0.2877],[95,-0.0876],[96,0.1199],[97,0.3256],[98,0.5259],[99,0.7162],[100,0.8936],[101,1.1266],[102,1.2493],[103,1.3802],[105,1.6466],[106,1.7723],[107,1.9851],[108,2.1022],[109,2.1948],[110,2.2549],[111,2.2763],[114,2.271],[115,2.2506],[116,2.2478],[117,2.2166],[119,2.1645],[121,2.0903],[122,2.0458],[123,1.9206],[124,1.8765],[125,1.8267],[126,1.771],[127,1.7091],[128,1.6289],[129,1.522],[130,1.3934],[131,1.1529],[132,0.9636],[134,0.5623],[135,0.3628],[136,0.1744],[137,0.0002],[138,-0.153],[139,-0.2794],[140,-0.3738],[141,-0.4474],[142,-0.5193],[143,-0.5873],[144,-0.6521],[145,-0.7137],[146,-0.7725],[147,-0.8283],[148,-0.8809],[149,-0.9311],[150,-0.9788],[151,-1.023],[152,-1.0645],[153,-1.1036],[154,-1.1403],[156,-1.205],[158,-1.2599],[160,-1.3036],[161,-1.3228],[162,-1.3391],[163,-1.3427],[164,-1.3644],[165,-1.3667],[166,-1.38],[168,-1.3806],[169,-1.377],[170,-1.3286],[171,-1.2585],[172,-1.1624],[173,-1.0435],[174,-0.9033],[175,-0.7446],[176,-0.5706],[177,-0.3768],[178,-0.1651],[179,0.0552],[180,0.2796],[181,0.4964],[182,0.705],[183,0.8924],[184,1.129],[185,1.2486],[186,1.3522],[187,1.5049],[188,1.6541],[189,1.7961],[190,1.9269],[191,2.043],[192,2.1404],[193,2.2163],[194,2.2662],[195,2.2843],[199,2.2756],[200,2.2512],[201,2.233],[202,2.2102],[203,2.1824],[204,2.1493],[205,2.1104],[206,2.0654],[207,2.0138],[208,1.9555],[209,1.8898],[210,1.8165],[211,1.7352],[212,1.628],[213,1.4826],[214,1.3069],[215,1.1087],[216,0.8956],[218,0.4564],[219,0.2459],[220,0.0518],[221,-0.1181],[222,-0.2561],[223,-0.3542],[224,-0.4282],[225,-0.4989],[226,-0.5662],[227,-0.6304],[228,-0.6916],[229,-0.7496],[230,-0.8047],[231,-0.8567],[232,-0.906],[233,-0.9523],[234,-0.996],[235,-1.0369],[236,-1.0751],[237,-1.1107],[238,-1.1437],[239,-1.1742],[240,-1.2024],[242,-1.2514],[244,-1.2909],[245,-1.3072],[246,-1.3112],[247,-1.3335],[248,-1.3362],[249,-1.3516],[250,-1.3532],[251,-1.3638],[252,-1.3686],[253,-1.3611],[254,-1.3289],[255,-1.2705],[256,-1.1875],[257,-1.0816],[258,-0.9559],[259,-0.8109],[260,-0.6495],[261,-0.4721],[262,-0.2811],[263,-0.0801],[264,0.1252],[265,0.3297],[266,0.5287],[267,0.7177],[268,0.896],[269,1.1266],[270,1.2491],[271,1.3798],[273,1.6462],[274,1.7719],[275,1.9808],[276,2.1004],[277,2.1947],[278,2.2549],[279,2.2763],[282,2.271],[283,2.2534],[284,2.2464],[287,2.1645],[288,2.1292],[290,2.0434],[291,1.9206],[292,1.8764],[293,1.8267],[294,1.771],[295,1.709],[296,1.6307],[297,1.5286],[298,1.4067],[299,1.1808],[300,1.003],[302,0.6162],[303,0.4217],[304,0.234],[305,0.0564],[306,-0.1083],[307,-0.2528],[308,-0.3738],[309,-0.4791],[310,-0.5816],[311,-0.6797],[312,-0.7736],[313,-0.8617],[314,-0.9453],[315,-1.0232],[316,-1.0946],[317,-1.1593],[318,-1.2176],[319,-1.2671],[320,-1.3092],[321,-1.3427],[322,-1.3667],[323,-1.371],[324,-1.3872]]
       }
       body_ctrl.Rotation.Y {
        type=MovieSceneScriptingFloatChannel
        default=0
        thinned_from=325
        keys [[0,0],[30,0],[31,-0.0259],[32,-0.0769],[33,-0.1553],[34,-0.2683],[35,-0.3563],[36,-0.3793],[37,-0.3885],[38,-0.3895],[39,-0.3874],[46,-0.3679],[53,-0.3448],[55,-0.3399],[65,-0.3281],[74,-0.3239],[79,-0.324],[86,-0.3256],[87,-0.3148],[88,-0.2944],[89,-0.2456],[90,-0.0783],[91,-0.0372],[92,-0.016],[93,-0.0054],[94,-0.0021],[95,-0.0024],[96,-0.0042],[97,-0.0056],[98,-0.0063],[99,-0.0065],[110,-0.0047],[128,-0.0056],[140,-0.0079],[174,-0.0087],[178,-0.008],[182,-0.0068],[183,-0.01],[184,-0.0065],[185,-0.0069],[186,-0.1361],[187,-0.3203],[188,-0.5363],[189,-0.6832],[190,-0.8126],[191,-0.9168],[192,-0.9848],[193,-1.0368],[194,-1.0759],[195,-1.1011],[196,-1.1121],[197,-1.1373],[199,-1.1518],[200,-1.165],[202,-1.1727],[211,-1.1948],[215,-1.2197],[220,-1.2586],[222,-1.2693],[246,-1.2967],[247,-1.2964],[248,-1.2873],[249,-1.2631],[250,-1.2197],[251,-1.1101],[252,-0.9296],[253,-0.6036],[254,-0.3146],[255,-0.1882],[256,-0.1035],[257,-0.0559],[258,-0.0271],[259,-0.012],[260,-0.005],[261,-0.0017],[262,-0.0003],[263,0],[324,0]]
       }
       body_ctrl.Rotation.Z {
        type=MovieSceneScriptingFloatChannel
        default=0
        thinned_from=325
        keys [[0,-19.4399],[16,-19.4397],[17,-19.3676],[22,-19.5076],[23,-19.4394],[29,-19.4609],[30,-19.5058],[31,-19.5969],[32,-19.8948],[33,-20.3779],[34,-21.0631],[35,-21.649],[38,-21.9199],[39,-22.0432],[46,-21.9815],[47,-22.0347],[85,-22.0355],[86,-22.0126],[87,-21.9345],[88,-21.7835],[89,-21.5419],[90,-21.0699],[91,-20.4404],[92,-19.9842],[93,-19.7583],[94,-19.603],[95,-19.5058],[97,-19.4408],[100,-19.4418],[101,-19.3683],[106,-19.5082],[107,-19.4412],[122,-19.4388],[123,-19.5101],[130,-19.3819],[131,-19.4426],[183,-19.4358],[184,-19.3338],[186,-19.2932],[187,-19.1523],[188,-18.8838],[189,-18.4746],[192,-17.0908],[193,-16.7863],[195,-16.586],[198,-16.412],[247,-16.3608],[248,-16.4085],[249,-16.5332],[250,-16.7813],[251,-17.2317],[252,-18.126],[253,-18.8536],[254,-19.0674],[255,-19.2225],[256,-19.3268],[258,-19.4248],[268,-19.4402],[269,-19.3676],[274,-19.5076],[275,-19.4395],[290,-19.4391],[291,-19.5097],[298,-19.3847],[299,-19.4384],[324,-19.4399]]
       }
       body_ctrl.Scale.X {type=MovieSceneScriptingFloatChannel, default=1, const [1,325]}
       body_ctrl.Scale.Y {type=MovieSceneScriptingFloatChannel, default=1, const [1,325]}
       body_ctrl.Scale.Z {type=MovieSceneScriptingFloatChannel, default=1, const [1,325]}
       hips_ctrl.Location.X {type=MovieSceneScriptingFloatChannel, default=0, thinned_from=325, keys [[0,0],[324,0]]}
       hips_ctrl.Location.Y {type=MovieSceneScriptingFloatChannel, default=0, thinned_from=325, keys [[0,0],[324,0]]}
       hips_ctrl.Location.Z {type=MovieSceneScriptingFloatChannel, default=0, thinned_from=325, keys [[0,0],[324,0]]}
       hips_ctrl.Rotation.X {
        type=MovieSceneScriptingFloatChannel
        default=0
        thinned_from=325
        keys [[0,-1.2664],[1,-1.2664],[2,-1.3044],[3,-1.3505],[4,-1.4135],[5,-1.4926],[6,-1.5869],[7,-1.6956],[8,-1.8178],[9,-1.9548],[10,-2.1055],[11,-2.2656],[14,-2.7597],[15,-2.9146],[16,-3.0574],[18,-3.3058],[21,-3.7016],[22,-3.8267],[23,-3.9402],[24,-4.037],[25,-4.1124],[26,-4.1612],[27,-4.1785],[30,-4.1785],[34,-4.1211],[36,-4.0703],[39,-3.9586],[41,-3.8625],[43,-3.7442],[44,-3.664],[45,-3.5571],[46,-3.4285],[47,-3.283],[48,-3.1256],[51,-2.6309],[52,-2.4749],[53,-2.3315],[54,-2.2056],[55,-2.1022],[56,-2.0261],[58,-1.9102],[60,-1.8054],[62,-1.7113],[64,-1.6273],[66,-1.553],[68,-1.4879],[70,-1.4316],[72,-1.3836],[74,-1.3436],[76,-1.3113],[78,-1.2864],[79,-1.2864],[80,-1.2684],[81,-1.2684],[82,-1.2568],[85,-1.2568],[86,-1.2868],[87,-1.3324],[88,-1.3958],[89,-1.4775],[90,-1.5796],[91,-1.6907],[92,-1.8142],[93,-1.9519],[94,-2.1027],[95,-2.2627],[98,-2.7557],[99,-2.9104],[100,-3.0531],[102,-3.3022],[105,-3.6995],[106,-3.8252],[107,-3.9391],[108,-4.0364],[109,-4.1121],[110,-4.1611],[111,-4.1786],[114,-4.1786],[115,-4.1623],[116,-4.1623],[120,-4.0756],[122,-4.0124],[124,-3.9294],[126,-3.8239],[127,-3.762],[128,-3.6818],[129,-3.5749],[130,-3.4463],[131,-3.3007],[132,-3.1432],[135,-2.6482],[136,-2.492],[137,-2.3485],[138,-2.2226],[139,-2.119],[140,-2.0428],[142,-1.9268],[144,-1.8219],[146,-1.7276],[148,-1.6435],[150,-1.5691],[152,-1.5039],[154,-1.4475],[156,-1.3993],[158,-1.3591],[160,-1.3267],[162,-1.3017],[163,-1.3017],[164,-1.2836],[165,-1.2836],[166,-1.272],[169,-1.272],[170,-1.3078],[171,-1.3595],[172,-1.4301],[173,-1.5184],[174,-1.6232],[175,-1.7431],[176,-1.877],[177,-2.0278],[178,-2.1949],[181,-2.7307],[182,-2.8992],[183,-3.0518],[184,-3.1819],[188,-3.6704],[189,-3.7861],[190,-3.893],[191,-3.9879],[192,-4.0677],[193,-4.1302],[194,-4.1716],[195,-4.1867],[199,-4.1858],[203,-4.1172],[205,-4.0642],[207,-3.9915],[209,-3.8964],[211,-3.7758],[212,-3.6903],[213,-3.5725],[214,-3.4289],[215,-3.2662],[218,-2.7287],[219,-2.555],[220,-2.395],[221,-2.2552],[222,-2.1423],[223,-2.0627],[225,-1.9472],[227,-1.8431],[229,-1.7498],[231,-1.6668],[233,-1.5935],[235,-1.5295],[237,-1.4742],[239,-1.427],[241,-1.3875],[243,-1.3555],[245,-1.3304],[246,-1.3304],[247,-1.3118],[248,-1.3118],[249,-1.2987],[250,-1.2987],[251,-1.2889],[252,-1.2845],[253,-1.2882],[254,-1.3099],[255,-1.3514],[256,-1.4126],[257,-1.4913],[258,-1.586],[259,-1.6957],[260,-1.8198],[261,-1.9583],[262,-2.1093],[266,-2.7585],[267,-2.9119],[268,-3.0536],[270,-3.302],[273,-3.6991],[274,-3.8248],[275,-3.9388],[276,-4.0362],[277,-4.112],[278,-4.1611],[279,-4.1785],[282,-4.1785],[283,-4.1622],[284,-4.1622],[288,-4.0755],[290,-4.0124],[292,-3.9293],[294,-3.8239],[295,-3.7619],[296,-3.6836],[297,-3.5815],[298,-3.4596],[299,-3.3221],[300,-3.173],[304,-2.5403],[305,-2.3932],[306,-2.2586],[307,-2.1405],[308,-2.0428],[310,-1.8777],[311,-1.8007],[312,-1.7278],[313,-1.6593],[314,-1.5954],[315,-1.5364],[316,-1.4825],[317,-1.4339],[318,-1.391],[319,-1.354],[320,-1.3232],[321,-1.2987],[322,-1.281],[323,-1.281],[324,-1.2664]]
       }
       hips_ctrl.Rotation.Y {
        type=MovieSceneScriptingFloatChannel
        default=0
        thinned_from=325
        keys [[0,0],[30,0],[31,0.0259],[32,0.0767],[33,0.155],[34,0.2678],[35,0.3556],[36,0.3785],[37,0.3877],[38,0.3888],[39,0.3867],[47,0.3644],[53,0.3448],[55,0.34],[65,0.3283],[74,0.3242],[79,0.3243],[86,0.3259],[87,0.315],[88,0.2946],[89,0.2458],[90,0.0783],[91,0.0372],[92,0.016],[93,0.0054],[94,0.0021],[95,0.0024],[96,0.0042],[97,0.0056],[98,0.0063],[99,0.0065],[110,0.0047],[128,0.0056],[140,0.0079],[174,0.0087],[178,0.008],[183,0.0065],[184,0.0065],[185,0.0069],[186,0.0079],[187,0.0095],[188,0.0119],[189,0.0176],[190,0.0281],[191,0.0418],[192,0.0575],[193,0.0859],[194,0.1128],[195,0.1252],[196,0.1252],[197,0.1412],[199,0.1412],[200,0.1502],[205,0.1592],[207,0.1638],[209,0.1694],[211,0.1764],[212,0.1809],[213,0.1868],[214,0.1938],[215,0.2015],[218,0.2259],[219,0.2335],[220,0.2404],[221,0.2464],[222,0.251],[223,0.2542],[227,0.2623],[233,0.2704],[245,0.2789],[246,0.2789],[247,0.2798],[248,0.2798],[249,0.2784],[250,0.2784],[251,0.2632],[252,0.2462],[253,0.2088],[254,0.1517],[255,0.0939],[256,0.0552],[257,0.0355],[258,0.0211],[259,0.0112],[260,0.005],[261,0.0017],[262,0.0003],[263,0],[324,0]]
       }
       hips_ctrl.Rotation.Z {
        type=MovieSceneScriptingFloatChannel
        default=0
        thinned_from=325
        keys [[0,0],[30,0],[31,-0.0019],[32,-0.0056],[33,-0.0112],[34,-0.0193],[35,-0.0255],[36,-0.0269],[37,-0.0274],[38,-0.0272],[43,-0.0246],[45,-0.023],[52,-0.015],[54,-0.0132],[56,-0.012],[68,-0.0085],[78,-0.0073],[85,-0.0071],[87,-0.0073],[88,-0.0072],[89,-0.0063],[90,-0.0022],[91,-0.0011],[92,-0.0005],[93,-0.0002],[186,-0.0005],[188,-0.0008],[189,-0.0012],[190,-0.0019],[192,-0.0041],[194,-0.0082],[195,-0.0092],[196,-0.0092],[197,-0.0103],[199,-0.0103],[200,-0.0109],[213,-0.0117],[217,-0.0111],[223,-0.0092],[245,-0.0065],[250,-0.0063],[252,-0.0055],[253,-0.0047],[255,-0.0022],[256,-0.0014],[259,-0.0003],[324,0]]
       }
       hips_ctrl.Scale.X {type=MovieSceneScriptingFloatChannel, default=1, const [1,325]}
       hips_ctrl.Scale.Y {type=MovieSceneScriptingFloatChannel, default=1, const [1,325]}
       hips_ctrl.Scale.Z {type=MovieSceneScriptingFloatChannel, default=1, const [1,325]}
       hips_tan_ctrl.Location.X {
        type=MovieSceneScriptingFloatChannel
        default=0
        thinned_from=325
        keys [[0,0],[31,-0.0001],[33,-0.0008],[35,-0.002],[37,-0.0037],[38,-0.0043],[41,-0.0049],[88,-0.004],[89,-0.0037],[91,-0.0017],[92,-0.0009],[94,-0.0001],[186,0.0005],[188,0.0007],[189,0.0011],[190,0.0019],[192,0.004],[194,0.0081],[195,0.0091],[196,0.0098],[198,0.0105],[248,0.0113],[250,0.011],[251,0.0105],[252,0.0098],[253,0.0082],[255,0.0037],[256,0.0021],[257,0.0014],[259,0.0004],[324,0]]
       }
       hips_tan_ctrl.Location.Y {
        type=MovieSceneScriptingFloatChannel
        default=0.008
        thinned_from=325
        keys [[0,0.1374],[1,0.1372],[2,0.1367],[3,0.1359],[5,0.1334],[7,0.1302],[10,0.1241],[15,0.1131],[17,0.1094],[21,0.1028],[23,0.1],[24,0.0989],[25,0.098],[26,0.0975],[28,0.0973],[32,0.0979],[35,0.0992],[37,0.1006],[39,0.1024],[41,0.1047],[43,0.1077],[46,0.1131],[50,0.1208],[52,0.1242],[54,0.1269],[56,0.1287],[66,0.1341],[79,0.1372],[84,0.1374],[85,0.1372],[86,0.1367],[87,0.1358],[89,0.1333],[91,0.1299],[94,0.1238],[98,0.1148],[100,0.1108],[105,0.1026],[107,0.0999],[108,0.0988],[109,0.098],[110,0.0975],[112,0.0973],[116,0.0979],[119,0.0992],[121,0.1005],[123,0.1023],[125,0.1047],[127,0.1076],[130,0.113],[134,0.1207],[136,0.1241],[138,0.1268],[140,0.1287],[150,0.134],[163,0.1372],[168,0.1374],[169,0.1371],[170,0.1365],[171,0.1355],[173,0.1328],[175,0.1291],[178,0.1224],[181,0.1153],[183,0.111],[187,0.1045],[189,0.1017],[191,0.0994],[192,0.0985],[193,0.0979],[194,0.0975],[198,0.0976],[201,0.0983],[203,0.0992],[205,0.1006],[207,0.1024],[209,0.1047],[211,0.1077],[213,0.1115],[218,0.122],[220,0.1254],[221,0.1268],[222,0.128],[231,0.1333],[244,0.137],[252,0.1375],[253,0.1372],[254,0.1366],[256,0.1346],[258,0.1317],[260,0.128],[267,0.1127],[269,0.1091],[273,0.1026],[275,0.0999],[276,0.0988],[277,0.098],[278,0.0975],[280,0.0973],[284,0.0979],[287,0.0992],[289,0.1005],[291,0.1023],[293,0.1047],[295,0.1076],[303,0.1219],[305,0.125],[307,0.1276],[313,0.133],[318,0.136],[322,0.1372],[324,0.1373]]
       }
       hips_tan_ctrl.Location.Z {
        type=MovieSceneScriptingFloatChannel
        default=-0.0073
        thinned_from=325
        keys [[0,0.042],[1,0.0428],[2,0.0454],[3,0.0494],[4,0.0548],[5,0.0615],[6,0.0692],[7,0.078],[8,0.0876],[9,0.0979],[10,0.1087],[11,0.1201],[12,0.1319],[17,0.1928],[18,0.2044],[19,0.2152],[20,0.2253],[21,0.2344],[22,0.2424],[23,0.2493],[24,0.2548],[25,0.2589],[26,0.2615],[27,0.2624],[30,0.2617],[32,0.2602],[34,0.2576],[36,0.2536],[38,0.2477],[39,0.2439],[40,0.2396],[41,0.2348],[42,0.2293],[43,0.2231],[44,0.2156],[45,0.2064],[46,0.1958],[47,0.1841],[48,0.1715],[51,0.1321],[52,0.1193],[53,0.1073],[54,0.0961],[55,0.0862],[56,0.0778],[58,0.0628],[59,0.0556],[60,0.0487],[61,0.042],[62,0.0356],[63,0.0296],[64,0.024],[65,0.0189],[66,0.0142],[67,0.0101],[68,0.0065],[69,0.0036],[70,0.0012],[71,-0.0004],[72,-0.0013],[73,-0.0012],[74,-0.0007],[75,0.0006],[76,0.0026],[77,0.0053],[78,0.0087],[79,0.0128],[80,0.0175],[81,0.0229],[82,0.029],[83,0.0357],[84,0.0431],[85,0.0512],[86,0.0598],[87,0.0689],[88,0.0783],[90,0.098],[97,0.1714],[99,0.1917],[101,0.211],[102,0.22],[103,0.2283],[104,0.2358],[105,0.2425],[106,0.2483],[107,0.2532],[108,0.2571],[109,0.26],[110,0.2618],[111,0.2624],[114,0.2617],[116,0.2602],[118,0.2575],[120,0.2533],[122,0.2473],[123,0.2435],[124,0.2392],[125,0.2344],[126,0.2289],[127,0.2227],[128,0.2152],[129,0.206],[130,0.1954],[131,0.1837],[132,0.1711],[135,0.1317],[136,0.1189],[137,0.1069],[138,0.0957],[139,0.0858],[140,0.0773],[142,0.0624],[143,0.0552],[144,0.0482],[145,0.0415],[146,0.0352],[147,0.0292],[148,0.0236],[149,0.0184],[150,0.0138],[151,0.0096],[152,0.0061],[153,0.0031],[154,0.0008],[155,-0.0008],[156,-0.0017],[157,-0.0017],[158,-0.0012],[159,0.0002],[160,0.0023],[161,0.005],[162,0.0085],[163,0.0126],[164,0.0174],[165,0.0229],[166,0.029],[167,0.0358],[168,0.0433],[169,0.0513],[170,0.0598],[171,0.0688],[172,0.0782],[173,0.0879],[174,0.0978],[183,0.1906],[185,0.2101],[186,0.2191],[187,0.2274],[188,0.235],[189,0.2418],[190,0.2478],[191,0.2529],[192,0.2569],[193,0.2599],[194,0.2618],[195,0.2624],[198,0.2618],[200,0.2604],[202,0.2579],[204,0.2538],[206,0.2478],[207,0.2441],[208,0.2397],[209,0.2347],[210,0.2291],[211,0.2227],[212,0.2148],[213,0.2047],[214,0.1929],[215,0.1798],[216,0.1658],[218,0.1369],[219,0.1227],[220,0.1093],[221,0.0971],[222,0.0863],[223,0.0773],[224,0.0695],[225,0.0619],[226,0.0545],[227,0.0475],[228,0.0408],[229,0.0344],[230,0.0284],[231,0.0229],[232,0.0178],[233,0.0132],[234,0.0091],[235,0.0056],[236,0.0028],[237,0.0005],[238,-0.001],[239,-0.0019],[240,-0.0019],[241,-0.0014],[242,-0.0003],[243,0.0014],[244,0.0037],[245,0.0066],[246,0.0101],[247,0.0141],[248,0.0188],[249,0.0241],[250,0.0301],[251,0.0367],[252,0.0439],[253,0.0518],[254,0.0602],[255,0.0691],[256,0.0785],[257,0.0883],[258,0.0982],[266,0.1816],[268,0.2015],[269,0.211],[270,0.22],[271,0.2282],[272,0.2357],[273,0.2424],[274,0.2483],[275,0.2532],[276,0.2571],[277,0.26],[278,0.2618],[279,0.2624],[282,0.2617],[284,0.2602],[286,0.2575],[288,0.2533],[290,0.2473],[291,0.2435],[292,0.2392],[293,0.2344],[294,0.2289],[295,0.2227],[296,0.2151],[297,0.2057],[298,0.1947],[299,0.1825],[300,0.1694],[303,0.129],[304,0.1162],[305,0.1043],[306,0.0936],[307,0.0845],[308,0.0773],[309,0.0715],[310,0.0664],[311,0.0619],[312,0.058],[313,0.0546],[314,0.0517],[315,0.0493],[316,0.0473],[317,0.0458],[318,0.0445],[319,0.0436],[320,0.0429],[322,0.0421],[324,0.042]]
       }
       hips_tan_ctrl.Rotation.X {type=MovieSceneScriptingFloatChannel, default=0, const [0,325]}
       hips_tan_ctrl.Rotation.Y {type=MovieSceneScriptingFloatChannel, default=0, const [0,325]}
       hips_tan_ctrl.Rotation.Z {type=MovieSceneScriptingFloatChannel, default=0, const [0,325]}
       hips_tan_ctrl.Scale.X {type=MovieSceneScriptingFloatChannel, default=1, const [1,325]}
       hips_tan_ctrl.Scale.Y {type=MovieSceneScriptingFloatChannel, default=1, const [1,325]}
       hips_tan_ctrl.Scale.Z {type=MovieSceneScriptingFloatChannel, default=1, const [1,325]}
       spine_01_ctrl.Location.X {
        type=MovieSceneScriptingFloatChannel
        default=0
        thinned_from=325
        keys [[0,0],[29,0],[31,0.0004],[32,0.0012],[33,0.0025],[34,0.0044],[35,0.0058],[36,0.0062],[83,0.0076],[86,0.0073],[87,0.007],[88,0.0064],[89,0.0053],[90,0.0017],[91,0.0008],[92,0.0003],[189,0.0003],[192,0.0009],[194,0.0018],[213,0.0034],[236,0.0067],[249,0.0066],[251,0.0061],[252,0.0057],[253,0.0048],[255,0.0021],[256,0.0012],[258,0.0004],[324,0]]
       }
       spine_01_ctrl.Location.Y {
        type=MovieSceneScriptingFloatChannel
        default=0
        thinned_from=325
        keys [[0,0.1326],[2,0.1321],[9,0.1271],[13,0.1258],[29,0.1248],[35,0.1259],[38,0.1272],[41,0.1294],[44,0.1324],[46,0.1337],[48,0.1341],[51,0.1335],[56,0.1314],[69,0.1293],[73,0.1293],[85,0.1324],[87,0.1321],[90,0.1305],[94,0.1279],[104,0.1252],[114,0.1248],[119,0.1259],[122,0.1274],[125,0.1297],[128,0.1327],[130,0.134],[132,0.1345],[135,0.1338],[140,0.1318],[153,0.1297],[157,0.1297],[169,0.1328],[170,0.1326],[172,0.1316],[177,0.1273],[179,0.1264],[192,0.1248],[200,0.1252],[204,0.1264],[207,0.1282],[210,0.1308],[212,0.133],[214,0.1343],[216,0.1347],[219,0.1339],[223,0.1322],[236,0.1301],[240,0.1301],[252,0.133],[253,0.1331],[255,0.1325],[262,0.128],[271,0.1253],[282,0.1248],[287,0.1259],[290,0.1274],[293,0.1297],[296,0.1327],[298,0.1339],[300,0.1343],[303,0.1335],[307,0.1319],[324,0.1326]]
       }
       spine_01_ctrl.Location.Z {
        type=MovieSceneScriptingFloatChannel
        default=0
        thinned_from=325
        keys [[0,0.0359],[1,0.0366],[2,0.0386],[3,0.0418],[4,0.0462],[5,0.0515],[6,0.0577],[7,0.0647],[8,0.0723],[9,0.0806],[10,0.0892],[11,0.0983],[12,0.1077],[17,0.1564],[18,0.1657],[19,0.1744],[20,0.1824],[21,0.1897],[22,0.1962],[23,0.2017],[24,0.2061],[25,0.2094],[26,0.2115],[27,0.2122],[30,0.2117],[32,0.2104],[34,0.2084],[36,0.2051],[37,0.2029],[38,0.2003],[39,0.1973],[40,0.1938],[41,0.1898],[42,0.1854],[43,0.1804],[44,0.1744],[45,0.167],[46,0.1584],[47,0.149],[48,0.1389],[51,0.1075],[52,0.0973],[53,0.0877],[54,0.0788],[55,0.0709],[56,0.0641],[58,0.0522],[60,0.0409],[61,0.0356],[62,0.0305],[63,0.0257],[64,0.0212],[65,0.0171],[66,0.0134],[67,0.0101],[68,0.0073],[69,0.0049],[70,0.0031],[71,0.0018],[72,0.0011],[73,0.0011],[74,0.0016],[75,0.0026],[76,0.0043],[77,0.0065],[78,0.0092],[79,0.0125],[80,0.0163],[81,0.0206],[82,0.0255],[83,0.0309],[84,0.0369],[85,0.0434],[86,0.0503],[87,0.0575],[88,0.0651],[90,0.0808],[97,0.1394],[99,0.1556],[101,0.1711],[102,0.1783],[103,0.1849],[104,0.1909],[105,0.1963],[106,0.2009],[107,0.2049],[108,0.208],[109,0.2103],[110,0.2117],[111,0.2122],[114,0.2117],[116,0.2104],[118,0.2082],[120,0.2048],[122,0.2],[123,0.1969],[124,0.1934],[125,0.1895],[126,0.185],[127,0.18],[128,0.174],[129,0.1666],[130,0.158],[131,0.1486],[132,0.1386],[135,0.1071],[136,0.0969],[137,0.0873],[138,0.0784],[139,0.0705],[140,0.0637],[142,0.0518],[144,0.0405],[145,0.0352],[146,0.0301],[147,0.0253],[148,0.0208],[149,0.0167],[150,0.013],[151,0.0097],[152,0.0068],[153,0.0045],[154,0.0026],[155,0.0014],[156,0.0007],[157,0.0007],[158,0.0011],[159,0.0022],[160,0.0039],[161,0.0062],[162,0.0089],[163,0.0123],[164,0.0162],[165,0.0206],[166,0.0255],[167,0.031],[168,0.0369],[169,0.0434],[170,0.0502],[171,0.0574],[172,0.065],[173,0.0727],[175,0.0889],[183,0.1546],[185,0.1703],[186,0.1775],[187,0.1841],[188,0.1902],[189,0.1957],[190,0.2005],[191,0.2046],[192,0.2078],[193,0.2102],[194,0.2117],[195,0.2122],[198,0.2117],[200,0.2106],[202,0.2085],[204,0.2052],[205,0.203],[206,0.2004],[207,0.1973],[208,0.1938],[209,0.1898],[210,0.1852],[211,0.18],[212,0.1736],[213,0.1655],[214,0.156],[215,0.1455],[216,0.1343],[218,0.1112],[219,0.0999],[220,0.0892],[221,0.0794],[222,0.0708],[223,0.0637],[225,0.0514],[226,0.0455],[227,0.0399],[228,0.0345],[229,0.0294],[230,0.0246],[231,0.0202],[232,0.0161],[233,0.0125],[234,0.0092],[235,0.0065],[236,0.0042],[237,0.0024],[238,0.0012],[239,0.0005],[240,0.0005],[241,0.0009],[242,0.0018],[243,0.0032],[244,0.0051],[245,0.0074],[246,0.0102],[247,0.0135],[248,0.0172],[249,0.0215],[250,0.0263],[251,0.0316],[252,0.0374],[253,0.0437],[254,0.0505],[255,0.0577],[256,0.0652],[257,0.073],[259,0.0892],[266,0.1475],[268,0.1634],[269,0.171],[270,0.1782],[271,0.1848],[272,0.1909],[273,0.1962],[274,0.2009],[275,0.2048],[276,0.208],[277,0.2103],[278,0.2117],[279,0.2122],[282,0.2117],[284,0.2104],[286,0.2082],[288,0.2048],[290,0.2],[291,0.1969],[292,0.1934],[293,0.1895],[294,0.185],[295,0.18],[296,0.1739],[297,0.1663],[298,0.1574],[299,0.1476],[300,0.1372],[303,0.1049],[304,0.0947],[305,0.0852],[306,0.0767],[307,0.0695],[308,0.0637],[309,0.0591],[310,0.0551],[311,0.0515],[312,0.0484],[313,0.0457],[314,0.0435],[315,0.0416],[316,0.04],[317,0.0388],[318,0.0378],[319,0.0371],[321,0.0362],[323,0.0359],[324,0.0359]]
       }
       spine_01_ctrl.Rotation.X {
        type=MovieSceneScriptingFloatChannel
        default=0
        thinned_from=325
        keys [[0,2.2454],[2,2.2382],[10,2.1502],[15,2.1193],[27,2.0912],[33,2.1006],[37,2.122],[40,2.1505],[45,2.219],[47,2.2339],[49,2.2383],[67,2.1735],[73,2.174],[83,2.2345],[85,2.2493],[87,2.2486],[90,2.2269],[98,2.1468],[109,2.0937],[116,2.0973],[121,2.1231],[125,2.1648],[128,2.2094],[130,2.2302],[132,2.2397],[135,2.2365],[151,2.1767],[157,2.1772],[167,2.2382],[169,2.2524],[170,2.2527],[172,2.2403],[178,2.1622],[182,2.1296],[193,2.0933],[201,2.102],[205,2.124],[208,2.1535],[213,2.2247],[215,2.2395],[217,2.2427],[233,2.1803],[240,2.1797],[253,2.2554],[255,2.2525],[258,2.2284],[266,2.1473],[277,2.0937],[284,2.0973],[289,2.1231],[293,2.1648],[296,2.2091],[298,2.2278],[300,2.2344],[308,2.2112],[324,2.245]]
       }
       spine_01_ctrl.Rotation.Y {
        type=MovieSceneScriptingFloatChannel
        default=0
        thinned_from=325
        keys [[0,0],[28,0],[29,-0.0002],[30,-0.0012],[31,-0.0039],[32,-0.0117],[33,-0.0237],[34,-0.0413],[35,-0.0567],[36,-0.0651],[37,-0.0723],[38,-0.0766],[39,-0.0787],[40,-0.0804],[41,-0.0816],[44,-0.084],[46,-0.086],[54,-0.0964],[61,-0.1027],[65,-0.1056],[68,-0.107],[71,-0.1078],[73,-0.1078],[76,-0.1069],[79,-0.1052],[82,-0.1026],[85,-0.0992],[86,-0.0976],[87,-0.0937],[88,-0.0875],[89,-0.0738],[90,-0.0321],[91,-0.0177],[92,-0.0087],[93,-0.0037],[94,-0.0011],[95,0.0001],[98,0.0009],[177,0.0005],[185,0.0015],[186,0.0018],[187,0.0024],[188,0.0032],[189,0.005],[190,0.0085],[191,0.0133],[192,0.019],[193,0.0294],[194,0.0394],[195,0.0441],[196,0.0474],[197,0.0496],[198,0.0508],[199,0.0514],[200,0.0517],[201,0.0517],[202,0.0515],[204,0.0506],[206,0.049],[208,0.0469],[210,0.0441],[211,0.0425],[212,0.0406],[213,0.0383],[214,0.0358],[216,0.0303],[219,0.0215],[220,0.0189],[221,0.0164],[222,0.0144],[223,0.0128],[228,0.0068],[232,0.0029],[235,0.0007],[238,-0.0009],[239,-0.0012],[240,-0.0012],[241,-0.0017],[244,-0.0018],[247,-0.0014],[253,0.0011],[254,0.0013],[262,0],[324,0]]
       }
       spine_01_ctrl.Rotation.Z {
        type=MovieSceneScriptingFloatChannel
        default=0
        thinned_from=325
        keys [[0,0],[30,0.0001],[31,0.0002],[32,0.0007],[33,0.0014],[34,-0.0044],[35,-0.0519],[36,-0.1792],[37,-0.3282],[38,-0.4297],[39,-0.4843],[40,-0.5292],[41,-0.5579],[42,-0.5736],[43,-0.5842],[44,-0.5908],[45,-0.5943],[87,-0.5984],[88,-0.5924],[89,-0.5502],[90,-0.4365],[91,-0.282],[92,-0.1531],[93,-0.073],[94,-0.0184],[95,0.015],[96,0.0354],[97,0.0492],[98,0.0577],[99,0.0622],[100,0.0639],[101,0.0642],[182,0.0642],[183,0.0646],[184,0.0671],[185,0.0738],[186,0.0868],[187,0.1083],[188,0.1404],[189,0.2154],[190,0.3542],[191,0.5421],[192,0.7645],[193,1.1635],[194,1.5478],[195,1.7269],[196,1.8575],[197,1.9453],[198,1.9961],[199,2.0258],[200,2.047],[201,2.0613],[202,2.0699],[247,2.0763],[248,2.0742],[249,2.0593],[250,2.019],[251,1.9405],[252,1.8111],[253,1.5326],[255,0.687],[256,0.4044],[257,0.2611],[258,0.1562],[259,0.084],[260,0.0382],[261,0.013],[262,0.0022],[263,0],[324,0]]
       }
       spine_01_ctrl.Scale.X {type=MovieSceneScriptingFloatChannel, default=1, const [1,325]}
       spine_01_ctrl.Scale.Y {type=MovieSceneScriptingFloatChannel, default=1, const [1,325]}
       spine_01_ctrl.Scale.Z {type=MovieSceneScriptingFloatChannel, default=1, const [1,325]}
       spine_02_ctrl.Location.X {
        type=MovieSceneScriptingFloatChannel
        default=0
        thinned_from=325
        keys [[0,0],[30,-0.0001],[31,-0.0004],[32,-0.0011],[33,-0.0022],[34,-0.0037],[35,-0.0048],[36,-0.0047],[41,-0.0037],[86,-0.0031],[88,-0.0027],[89,-0.0021],[90,-0.0001],[91,0.0001],[187,-0.0004],[189,-0.0008],[191,-0.0021],[192,-0.003],[194,-0.006],[196,-0.0072],[198,-0.0078],[249,-0.0083],[251,-0.0078],[252,-0.0073],[253,-0.0062],[255,-0.0028],[256,-0.0016],[258,-0.0006],[261,-0.0001],[324,0]]
       }
       spine_02_ctrl.Location.Y {
        type=MovieSceneScriptingFloatChannel
        default=0
        thinned_from=325
        keys [[0,0.1227],[1,0.1225],[2,0.122],[3,0.121],[4,0.1198],[5,0.1183],[6,0.1165],[8,0.1121],[10,0.1069],[14,0.0952],[16,0.0899],[20,0.0811],[22,0.0771],[23,0.0754],[24,0.074],[25,0.0729],[26,0.0722],[27,0.072],[31,0.0724],[34,0.0734],[36,0.0746],[38,0.0763],[40,0.0785],[42,0.0811],[43,0.0827],[44,0.0845],[45,0.0866],[47,0.0914],[51,0.1017],[53,0.1063],[54,0.1082],[55,0.1098],[59,0.114],[65,0.1188],[70,0.1216],[76,0.1232],[82,0.1232],[84,0.1228],[85,0.1224],[86,0.1216],[87,0.1206],[88,0.1192],[89,0.1175],[91,0.1135],[93,0.1086],[98,0.0945],[99,0.0918],[100,0.0893],[105,0.0788],[106,0.077],[107,0.0753],[108,0.0739],[109,0.0729],[110,0.0722],[111,0.072],[115,0.0724],[118,0.0734],[120,0.0746],[122,0.0762],[124,0.0783],[126,0.081],[127,0.0825],[128,0.0844],[129,0.0864],[131,0.0913],[135,0.1016],[137,0.1061],[138,0.1081],[139,0.1097],[143,0.1139],[149,0.1187],[154,0.1215],[160,0.1231],[166,0.1231],[168,0.1227],[169,0.1222],[170,0.1214],[171,0.1202],[172,0.1187],[173,0.1169],[174,0.1149],[176,0.1101],[177,0.1074],[181,0.0952],[182,0.0923],[183,0.0897],[184,0.0874],[187,0.0813],[189,0.0777],[190,0.0761],[191,0.0747],[192,0.0736],[193,0.0727],[194,0.0722],[195,0.072],[199,0.0723],[202,0.0733],[204,0.0745],[206,0.0761],[208,0.0782],[210,0.0809],[211,0.0825],[212,0.0845],[213,0.0867],[215,0.092],[218,0.1006],[219,0.1032],[220,0.1057],[221,0.1078],[222,0.1096],[223,0.1109],[229,0.1164],[235,0.1204],[241,0.1226],[247,0.1232],[252,0.1226],[253,0.1222],[254,0.1214],[255,0.1204],[256,0.119],[257,0.1174],[258,0.1155],[260,0.111],[262,0.1058],[266,0.0944],[267,0.0918],[268,0.0893],[273,0.0788],[274,0.077],[275,0.0753],[276,0.0739],[277,0.0729],[278,0.0722],[279,0.072],[283,0.0724],[286,0.0734],[288,0.0746],[290,0.0762],[292,0.0783],[294,0.081],[295,0.0825],[296,0.0843],[298,0.0886],[304,0.1033],[306,0.1076],[307,0.1094],[310,0.1135],[314,0.1179],[317,0.1203],[320,0.1219],[323,0.1227],[324,0.1227]]
       }
       spine_02_ctrl.Location.Z {
        type=MovieSceneScriptingFloatChannel
        default=0
        thinned_from=325
        keys [[0,0.022],[1,0.0226],[2,0.0243],[3,0.0269],[4,0.0305],[5,0.0349],[6,0.0401],[7,0.0459],[8,0.0522],[9,0.0589],[10,0.066],[11,0.0734],[13,0.089],[17,0.1209],[18,0.1285],[19,0.1356],[20,0.1422],[21,0.1481],[22,0.1534],[23,0.1578],[24,0.1615],[25,0.1642],[26,0.1658],[27,0.1664],[30,0.166],[32,0.1649],[34,0.1632],[36,0.1605],[37,0.1587],[38,0.1565],[39,0.154],[40,0.1511],[41,0.1478],[42,0.1441],[43,0.14],[44,0.135],[45,0.1289],[46,0.1219],[47,0.1141],[48,0.1059],[51,0.0803],[52,0.072],[53,0.0642],[54,0.057],[55,0.0505],[56,0.045],[58,0.0353],[60,0.026],[61,0.0217],[62,0.0175],[63,0.0136],[64,0.01],[65,0.0066],[66,0.0036],[67,0.0009],[68,-0.0014],[69,-0.0033],[70,-0.0048],[71,-0.0059],[72,-0.0065],[73,-0.0065],[74,-0.0061],[75,-0.0052],[76,-0.0039],[77,-0.0021],[78,0.0002],[79,0.0028],[80,0.006],[81,0.0095],[82,0.0135],[83,0.018],[84,0.0229],[85,0.0282],[86,0.0339],[87,0.0399],[89,0.0526],[90,0.0591],[97,0.1071],[99,0.1203],[101,0.133],[102,0.1388],[103,0.1442],[104,0.1491],[105,0.1535],[106,0.1573],[107,0.1605],[108,0.163],[109,0.1649],[110,0.166],[111,0.1664],[114,0.166],[116,0.1649],[118,0.1631],[120,0.1603],[121,0.1584],[122,0.1562],[123,0.1537],[124,0.1508],[125,0.1475],[126,0.1438],[127,0.1396],[128,0.1346],[129,0.1285],[130,0.1215],[131,0.1138],[132,0.1056],[135,0.0799],[136,0.0717],[137,0.0638],[138,0.0566],[139,0.0502],[140,0.0447],[142,0.0349],[144,0.0257],[145,0.0213],[146,0.0172],[147,0.0133],[148,0.0096],[149,0.0062],[150,0.0032],[151,0.0005],[152,-0.0018],[153,-0.0037],[154,-0.0052],[155,-0.0063],[156,-0.0068],[157,-0.0069],[158,-0.0065],[159,-0.0056],[160,-0.0042],[161,-0.0023],[162,-0.0001],[163,0.0027],[164,0.0058],[165,0.0094],[166,0.0135],[167,0.018],[168,0.0228],[169,0.0281],[170,0.0338],[171,0.0397],[172,0.046],[174,0.059],[182,0.1129],[184,0.126],[185,0.1323],[186,0.1381],[187,0.1436],[188,0.1486],[189,0.153],[190,0.1569],[191,0.1602],[192,0.1629],[193,0.1648],[194,0.166],[195,0.1664],[198,0.166],[200,0.1651],[202,0.1634],[204,0.1606],[205,0.1588],[206,0.1566],[207,0.154],[208,0.1511],[209,0.1477],[210,0.1439],[211,0.1396],[212,0.1343],[213,0.1276],[214,0.1198],[215,0.1112],[216,0.1021],[218,0.0833],[219,0.0741],[220,0.0654],[221,0.0574],[222,0.0504],[223,0.0446],[225,0.0345],[227,0.0251],[228,0.0208],[229,0.0166],[230,0.0127],[231,0.0091],[232,0.0058],[233,0.0028],[234,0.0001],[235,-0.0021],[236,-0.004],[237,-0.0055],[238,-0.0065],[239,-0.007],[240,-0.0071],[241,-0.0067],[242,-0.0059],[243,-0.0048],[244,-0.0033],[245,-0.0014],[246,0.0009],[247,0.0036],[248,0.0067],[249,0.0102],[250,0.0141],[251,0.0184],[252,0.0232],[253,0.0284],[254,0.034],[255,0.04],[256,0.0462],[258,0.0592],[265,0.107],[267,0.1203],[269,0.1329],[270,0.1388],[271,0.1442],[272,0.1491],[273,0.1535],[274,0.1572],[275,0.1604],[276,0.163],[277,0.1649],[278,0.166],[279,0.1664],[282,0.166],[284,0.1649],[286,0.1631],[288,0.1603],[289,0.1584],[290,0.1562],[291,0.1537],[292,0.1508],[293,0.1475],[294,0.1438],[295,0.1396],[296,0.1346],[297,0.1283],[298,0.121],[299,0.113],[300,0.1045],[303,0.0781],[304,0.0698],[305,0.0621],[306,0.0552],[307,0.0493],[308,0.0447],[309,0.0409],[310,0.0376],[311,0.0347],[312,0.0322],[313,0.03],[314,0.0282],[315,0.0266],[316,0.0254],[317,0.0244],[319,0.023],[321,0.0223],[324,0.022]]
       }
       spine_02_ctrl.Rotation.X {
        type=MovieSceneScriptingFloatChannel
        default=0
        thinned_from=325
        keys [[0,2.4488],[1,2.4455],[2,2.4357],[3,2.4199],[4,2.3983],[5,2.3713],[6,2.3393],[8,2.2616],[10,2.1659],[14,1.9518],[15,1.902],[16,1.8565],[21,1.6661],[22,1.6315],[23,1.6005],[24,1.5743],[25,1.5541],[26,1.5412],[27,1.5366],[30,1.5405],[33,1.5546],[35,1.5715],[37,1.5966],[39,1.629],[41,1.6698],[43,1.7203],[44,1.7512],[45,1.7874],[47,1.8715],[51,2.0554],[53,2.1372],[54,2.1716],[55,2.2001],[60,2.286],[68,2.3814],[77,2.4388],[84,2.4509],[85,2.4474],[86,2.4375],[87,2.4216],[88,2.3999],[89,2.3725],[91,2.3023],[93,2.2155],[98,1.9532],[99,1.9036],[100,1.8583],[105,1.6671],[106,1.6323],[107,1.6011],[108,1.5747],[109,1.5543],[110,1.5412],[111,1.5366],[114,1.5405],[117,1.5546],[119,1.5712],[121,1.5948],[123,1.6262],[125,1.6666],[127,1.7169],[128,1.7478],[129,1.784],[131,1.8682],[135,2.0523],[137,2.1343],[138,2.1687],[139,2.1973],[144,2.2833],[152,2.379],[161,2.4366],[168,2.4487],[169,2.4447],[170,2.4334],[171,2.4154],[172,2.3911],[173,2.361],[174,2.3256],[176,2.2407],[177,2.1908],[181,1.9614],[182,1.9072],[183,1.8585],[184,1.8177],[188,1.6744],[190,1.613],[191,1.5874],[192,1.5661],[193,1.5498],[194,1.5393],[195,1.5356],[199,1.5418],[202,1.5589],[204,1.5783],[206,1.6054],[208,1.6415],[210,1.6877],[211,1.715],[212,1.7477],[213,1.787],[214,1.8316],[219,2.0798],[220,2.1237],[221,2.162],[222,2.1932],[223,2.2155],[230,2.3198],[238,2.3979],[248,2.4415],[252,2.4453],[253,2.4426],[254,2.4337],[255,2.4184],[256,2.397],[257,2.3699],[258,2.3378],[260,2.2594],[262,2.1636],[266,1.9523],[267,1.9031],[268,1.8581],[273,1.6673],[274,1.6324],[275,1.6012],[276,1.5747],[277,1.5543],[278,1.5412],[279,1.5366],[282,1.5405],[285,1.5546],[287,1.5712],[289,1.5948],[291,1.6262],[293,1.6666],[295,1.7169],[296,1.7472],[297,1.782],[299,1.8614],[304,2.0798],[306,2.1574],[307,2.1905],[311,2.2906],[315,2.369],[318,2.412],[321,2.4393],[323,2.4477],[324,2.4477]]
       }
       spine_02_ctrl.Rotation.Y {
        type=MovieSceneScriptingFloatChannel
        default=0
        thinned_from=325
        keys [[0,0],[28,0],[29,0.0003],[30,0.0025],[31,0.0085],[32,0.0252],[33,0.0508],[34,0.0877],[35,0.1156],[36,0.1206],[37,0.1209],[40,0.1168],[41,0.1158],[43,0.1149],[83,0.1167],[85,0.1164],[86,0.1156],[87,0.1113],[88,0.1037],[89,0.0861],[90,0.0283],[91,0.0134],[92,0.0057],[93,0.0018],[94,0.0007],[95,0.0008],[97,0.0022],[99,0.0027],[183,0.0029],[185,0.0033],[186,0.0039],[187,0.0048],[188,0.0063],[189,0.0097],[190,0.0159],[191,0.0244],[192,0.0344],[193,0.0524],[194,0.0698],[195,0.0779],[196,0.0838],[197,0.0878],[198,0.09],[199,0.0914],[200,0.0923],[201,0.0929],[202,0.0932],[204,0.0932],[208,0.0921],[218,0.0874],[248,0.0812],[249,0.0805],[250,0.0788],[251,0.0757],[252,0.0706],[253,0.0598],[255,0.0273],[256,0.0162],[257,0.0106],[258,0.0065],[259,0.0035],[260,0.0016],[261,0.0006],[262,0.0001],[324,0]]
       }
       spine_02_ctrl.Rotation.Z {
        type=MovieSceneScriptingFloatChannel
        default=0
        thinned_from=325
        keys [[0,0],[29,0],[30,-0.0003],[31,-0.001],[32,-0.0029],[33,-0.0059],[34,-0.0157],[35,-0.059],[36,-0.1656],[37,-0.2896],[38,-0.3738],[39,-0.419],[40,-0.456],[41,-0.4796],[42,-0.4924],[43,-0.5011],[44,-0.5063],[45,-0.509],[87,-0.5052],[88,-0.4998],[89,-0.4636],[90,-0.3647],[91,-0.2353],[92,-0.1276],[93,-0.0608],[94,-0.0153],[95,0.0124],[96,0.0293],[97,0.0407],[98,0.0477],[99,0.0514],[100,0.0528],[101,0.0531],[182,0.0531],[183,0.0534],[184,0.0554],[185,0.0609],[186,0.0717],[187,0.0895],[188,0.116],[189,0.1779],[190,0.2927],[191,0.448],[192,0.6317],[193,0.9614],[194,1.279],[195,1.427],[196,1.5349],[197,1.6074],[198,1.6494],[199,1.6739],[200,1.6915],[201,1.7032],[202,1.7103],[247,1.7177],[248,1.7159],[249,1.7036],[250,1.6703],[251,1.6053],[252,1.4983],[253,1.2679],[255,0.5683],[256,0.3345],[257,0.2159],[258,0.1292],[259,0.0694],[260,0.0316],[261,0.0107],[262,0.0019],[263,0],[324,0]]
       }
       spine_02_ctrl.Scale.X {type=MovieSceneScriptingFloatChannel, default=1, const [1,325]}
       spine_02_ctrl.Scale.Y {type=MovieSceneScriptingFloatChannel, default=1, const [1,325]}
       spine_02_ctrl.Scale.Z {type=MovieSceneScriptingFloatChannel, default=1, const [1,325]}
       spine_03_ctrl.Location.X {
        type=MovieSceneScriptingFloatChannel
        default=0
        thinned_from=325
        keys [[0,0],[29,-0.0001],[30,-0.0006],[31,-0.0019],[32,-0.0057],[33,-0.0116],[34,-0.02],[35,-0.0266],[36,-0.0285],[37,-0.0293],[38,-0.0296],[85,-0.0295],[86,-0.0293],[87,-0.0283],[88,-0.0265],[89,-0.0223],[90,-0.0083],[91,-0.0043],[92,-0.002],[93,-0.0007],[94,-0.0002],[186,-0.0004],[188,-0.0007],[189,-0.0011],[190,-0.0017],[192,-0.0037],[194,-0.0075],[195,-0.0083],[196,-0.009],[198,-0.0096],[248,-0.0116],[250,-0.0112],[251,-0.0108],[252,-0.01],[253,-0.0085],[255,-0.0038],[256,-0.0023],[257,-0.0015],[259,-0.0005],[262,0],[324,0]]
       }
       spine_03_ctrl.Location.Y {
        type=MovieSceneScriptingFloatChannel
        default=0
        thinned_from=325
        keys [[0,0.2854],[1,0.2849],[2,0.2834],[3,0.281],[4,0.2777],[5,0.2735],[6,0.2684],[7,0.2623],[8,0.2554],[9,0.2475],[10,0.2386],[14,0.199],[15,0.1895],[16,0.1807],[17,0.1729],[20,0.1505],[21,0.1433],[22,0.1365],[23,0.1305],[24,0.1254],[25,0.1215],[26,0.1189],[27,0.118],[28,0.118],[30,0.1187],[32,0.12],[34,0.1222],[35,0.1237],[36,0.1256],[38,0.1303],[39,0.133],[40,0.1361],[41,0.1395],[42,0.1433],[43,0.1474],[44,0.1525],[45,0.1589],[46,0.1664],[47,0.1746],[51,0.211],[52,0.2197],[53,0.2278],[54,0.2349],[55,0.2408],[56,0.2453],[60,0.2589],[64,0.27],[68,0.2786],[72,0.2843],[76,0.2874],[80,0.2878],[84,0.2859],[85,0.2848],[86,0.2828],[87,0.2799],[88,0.2761],[89,0.2714],[90,0.2657],[91,0.2594],[92,0.2523],[93,0.2444],[94,0.2355],[98,0.1967],[99,0.1875],[100,0.179],[101,0.1715],[105,0.1427],[106,0.1361],[107,0.1302],[108,0.1252],[109,0.1214],[110,0.1189],[111,0.118],[112,0.118],[114,0.1187],[116,0.12],[118,0.1222],[119,0.1236],[120,0.1254],[121,0.1274],[122,0.1297],[123,0.1324],[124,0.1354],[125,0.1387],[126,0.1425],[127,0.1467],[128,0.1518],[129,0.1581],[130,0.1656],[131,0.1739],[135,0.2103],[136,0.219],[137,0.227],[138,0.2341],[139,0.2401],[140,0.2446],[144,0.2582],[148,0.2693],[152,0.2779],[156,0.2836],[160,0.2867],[164,0.2871],[168,0.2852],[169,0.284],[170,0.2818],[171,0.2787],[172,0.2747],[173,0.2697],[174,0.2639],[175,0.2573],[176,0.2498],[177,0.2412],[178,0.2314],[181,0.1995],[182,0.1893],[183,0.1801],[184,0.1723],[187,0.1513],[188,0.1447],[189,0.1385],[190,0.1328],[191,0.1279],[192,0.1238],[193,0.1206],[194,0.1186],[195,0.1178],[197,0.118],[199,0.1189],[201,0.1205],[202,0.1216],[203,0.123],[204,0.1247],[205,0.1267],[206,0.129],[207,0.1317],[208,0.1347],[209,0.1382],[210,0.142],[211,0.1463],[212,0.1517],[213,0.1587],[214,0.167],[215,0.1762],[218,0.2062],[219,0.2159],[220,0.2249],[221,0.2328],[222,0.2392],[223,0.2439],[227,0.2576],[231,0.2686],[235,0.2769],[239,0.2826],[243,0.2857],[248,0.2862],[252,0.2845],[253,0.2835],[254,0.2817],[255,0.279],[256,0.2753],[257,0.2707],[258,0.2653],[259,0.259],[260,0.2519],[261,0.2439],[262,0.235],[266,0.1965],[267,0.1874],[268,0.179],[269,0.1715],[273,0.1427],[274,0.1361],[275,0.1302],[276,0.1252],[277,0.1214],[278,0.1189],[279,0.118],[280,0.118],[282,0.1187],[284,0.12],[286,0.1222],[287,0.1236],[288,0.1254],[289,0.1274],[290,0.1297],[291,0.1324],[292,0.1354],[293,0.1387],[294,0.1425],[295,0.1467],[296,0.1517],[297,0.1579],[298,0.165],[299,0.1729],[301,0.1902],[304,0.2168],[305,0.225],[306,0.2325],[307,0.2391],[308,0.2446],[310,0.2537],[312,0.2618],[314,0.2688],[316,0.2746],[318,0.2793],[320,0.2826],[322,0.2847],[324,0.2854]]
       }
       spine_03_ctrl.Location.Z {
        type=MovieSceneScriptingFloatChannel
        default=0
        thinned_from=325
        keys [[0,0.0392],[1,0.04],[2,0.0422],[3,0.0458],[4,0.0507],[5,0.0566],[6,0.0635],[7,0.0713],[8,0.0798],[9,0.0887],[10,0.0983],[11,0.1082],[12,0.1184],[17,0.1715],[18,0.1817],[19,0.1912],[20,0.1999],[21,0.2078],[22,0.2147],[23,0.2207],[24,0.2254],[25,0.229],[26,0.2312],[27,0.2319],[30,0.2313],[32,0.23],[34,0.2277],[36,0.2241],[37,0.2217],[38,0.2189],[39,0.2156],[40,0.2117],[41,0.2074],[42,0.2025],[43,0.197],[44,0.1904],[45,0.1823],[46,0.1729],[47,0.1627],[48,0.1517],[51,0.1174],[52,0.1063],[53,0.0958],[54,0.0862],[55,0.0775],[56,0.0701],[58,0.0569],[59,0.0506],[60,0.0444],[61,0.0385],[62,0.0329],[63,0.0276],[64,0.0227],[65,0.0182],[66,0.0141],[67,0.0104],[68,0.0073],[69,0.0047],[70,0.0026],[71,0.0012],[72,0.0005],[73,0.0004],[74,0.001],[75,0.0022],[76,0.004],[77,0.0064],[78,0.0095],[79,0.0131],[80,0.0173],[81,0.0222],[82,0.0276],[83,0.0336],[84,0.0402],[85,0.0474],[86,0.0551],[87,0.0632],[88,0.0717],[90,0.0891],[96,0.1444],[99,0.1711],[101,0.1879],[102,0.1957],[103,0.2028],[104,0.2093],[105,0.215],[106,0.22],[107,0.2242],[108,0.2275],[109,0.2299],[110,0.2314],[111,0.2319],[114,0.2313],[116,0.23],[118,0.2276],[120,0.2239],[122,0.2185],[123,0.2152],[124,0.2114],[125,0.207],[126,0.2021],[127,0.1966],[128,0.19],[129,0.1819],[130,0.1726],[131,0.1623],[132,0.1513],[135,0.117],[136,0.1059],[137,0.0955],[138,0.0858],[139,0.0771],[140,0.0697],[142,0.0565],[143,0.0502],[144,0.044],[145,0.0381],[146,0.0325],[147,0.0272],[148,0.0223],[149,0.0178],[150,0.0136],[151,0.01],[152,0.0068],[153,0.0042],[154,0.0022],[155,0.0008],[156,0],[157,0],[158,0.0006],[159,0.0018],[160,0.0037],[161,0.0062],[162,0.0093],[163,0.013],[164,0.0173],[165,0.0222],[166,0.0276],[167,0.0337],[168,0.0403],[169,0.0475],[170,0.0551],[171,0.0632],[172,0.0715],[174,0.089],[181,0.1522],[184,0.1784],[185,0.1868],[186,0.1946],[187,0.2018],[188,0.2084],[189,0.2143],[190,0.2195],[191,0.2238],[192,0.2273],[193,0.2298],[194,0.2314],[195,0.2319],[198,0.2314],[200,0.2302],[202,0.2279],[204,0.2243],[205,0.2219],[206,0.219],[207,0.2156],[208,0.2118],[209,0.2073],[210,0.2023],[211,0.1966],[212,0.1896],[213,0.1807],[214,0.1703],[215,0.1589],[216,0.1467],[218,0.1215],[219,0.1092],[220,0.0975],[221,0.0869],[222,0.0774],[223,0.0696],[225,0.056],[226,0.0495],[227,0.0433],[228,0.0374],[229,0.0318],[230,0.0265],[231,0.0216],[232,0.0171],[233,0.0131],[234,0.0095],[235,0.0064],[236,0.0039],[237,0.0019],[238,0.0006],[239,-0.0002],[240,-0.0002],[241,0.0003],[242,0.0013],[243,0.0029],[244,0.0049],[245,0.0075],[246,0.0106],[247,0.0142],[248,0.0184],[249,0.0232],[250,0.0285],[251,0.0344],[252,0.0408],[253,0.0479],[254,0.0554],[255,0.0635],[256,0.0718],[258,0.0893],[265,0.1533],[267,0.171],[269,0.1878],[270,0.1956],[271,0.2028],[272,0.2093],[273,0.215],[274,0.22],[275,0.2242],[276,0.2275],[277,0.2299],[278,0.2314],[279,0.2319],[282,0.2313],[284,0.23],[286,0.2276],[288,0.2239],[290,0.2185],[291,0.2152],[292,0.2114],[293,0.207],[294,0.2021],[295,0.1966],[296,0.1899],[297,0.1816],[298,0.1719],[299,0.1612],[300,0.1498],[303,0.1145],[304,0.1034],[305,0.0931],[306,0.0839],[307,0.0759],[308,0.0697],[309,0.0646],[310,0.0601],[311,0.0562],[312,0.0528],[313,0.0499],[314,0.0474],[315,0.0454],[316,0.0437],[317,0.0423],[318,0.0413],[319,0.0405],[321,0.0395],[324,0.0392]]
       }
       spine_03_ctrl.Rotation.X {
        type=MovieSceneScriptingFloatChannel
        default=0
        thinned_from=325
        keys [[0,4.2433],[1,4.2367],[2,4.2168],[3,4.1836],[4,4.1372],[5,4.0776],[6,4.0048],[7,3.9189],[8,3.8199],[9,3.7055],[10,3.5762],[14,2.9987],[15,2.8612],[16,2.7352],[17,2.6247],[21,2.2011],[22,2.1037],[23,2.0165],[24,1.9428],[25,1.886],[26,1.8494],[27,1.8365],[28,1.8365],[30,1.8456],[32,1.8639],[34,1.8941],[35,1.9147],[36,1.94],[38,2.003],[39,2.0402],[40,2.0819],[41,2.1284],[42,2.1799],[43,2.2369],[44,2.3075],[45,2.3972],[46,2.5026],[47,2.6204],[49,2.8791],[51,3.1447],[52,3.2707],[53,3.3873],[54,3.4904],[55,3.5762],[56,3.6406],[60,3.8328],[64,3.9897],[68,4.1108],[72,4.1958],[77,4.2518],[82,4.258],[84,4.248],[85,4.2352],[86,4.2101],[87,4.1726],[88,4.1224],[89,4.059],[90,3.9812],[91,3.8929],[92,3.7926],[93,3.6779],[94,3.5496],[98,2.9824],[99,2.8479],[100,2.7245],[101,2.6162],[105,2.1985],[106,2.102],[107,2.0154],[108,1.9422],[109,1.8857],[110,1.8493],[111,1.8365],[112,1.8365],[114,1.8456],[116,1.864],[118,1.8942],[120,1.9383],[121,1.9661],[122,1.9981],[123,2.0346],[124,2.0758],[125,2.1219],[126,2.1733],[127,2.2302],[128,2.3006],[129,2.3903],[130,2.4958],[131,2.6137],[133,2.8724],[135,3.1381],[136,3.2642],[137,3.3808],[138,3.484],[139,3.5698],[140,3.6342],[144,3.8266],[148,3.9836],[152,4.1049],[156,4.19],[161,4.2461],[166,4.2521],[168,4.2421],[169,4.2284],[170,4.2009],[171,4.1597],[172,4.105],[173,4.037],[174,3.9557],[175,3.8615],[176,3.7546],[177,3.63],[178,3.4872],[181,3.0166],[182,2.8677],[183,2.7337],[184,2.6211],[187,2.3195],[188,2.2236],[189,2.1336],[190,2.0517],[191,1.9799],[192,1.9203],[193,1.8751],[194,1.8464],[195,1.8364],[197,1.8398],[199,1.8517],[201,1.8743],[203,1.9099],[204,1.9332],[205,1.9607],[206,1.9926],[207,2.0292],[208,2.0708],[209,2.1177],[210,2.1703],[211,2.2289],[212,2.3037],[213,2.4021],[214,2.5193],[215,2.6508],[219,3.2231],[220,3.353],[221,3.4674],[222,3.561],[223,3.6283],[227,3.8199],[231,3.9752],[235,4.0943],[239,4.1776],[244,4.2341],[250,4.2427],[252,4.2342],[253,4.2233],[254,4.2004],[255,4.1642],[256,4.1146],[257,4.0518],[258,3.9761],[259,3.8875],[260,3.7859],[261,3.67],[262,3.5414],[266,2.979],[267,2.846],[268,2.7238],[269,2.6162],[273,2.199],[274,2.1024],[275,2.0158],[276,1.9425],[277,1.8859],[278,1.8494],[279,1.8365],[280,1.8365],[282,1.8456],[284,1.8639],[286,1.8942],[288,1.9383],[289,1.9661],[290,1.9981],[291,2.0346],[292,2.0758],[293,2.1219],[294,2.1733],[295,2.2302],[296,2.2993],[297,2.3856],[298,2.4861],[299,2.5981],[300,2.7187],[304,3.2291],[305,3.3484],[306,3.4578],[307,3.5542],[308,3.6342],[311,3.8308],[313,3.9432],[315,4.0392],[317,4.118],[319,4.1784],[321,4.2196],[323,4.2406],[324,4.2433]]
       }
       spine_03_ctrl.Rotation.Y {
        type=MovieSceneScriptingFloatChannel
        default=0
        thinned_from=325
        keys [[0,0],[28,0],[29,0.0011],[30,0.009],[31,0.0304],[32,0.0902],[33,0.1822],[34,0.3148],[35,0.4182],[36,0.4454],[37,0.4568],[38,0.4593],[41,0.456],[43,0.4558],[45,0.4581],[55,0.4816],[69,0.4976],[78,0.4984],[85,0.4925],[86,0.4886],[87,0.4716],[88,0.4415],[89,0.3711],[90,0.1437],[91,0.0743],[92,0.0342],[93,0.0131],[94,0.0041],[95,0.0019],[97,0.004],[98,0.0047],[99,0.0052],[131,0.0055],[170,0.0041],[183,0.0054],[184,0.0057],[185,0.0063],[186,0.0075],[187,0.0095],[188,0.0125],[189,0.0194],[190,0.0322],[191,0.0498],[192,0.0708],[193,0.1084],[194,0.1448],[195,0.1618],[196,0.1741],[197,0.1823],[198,0.187],[199,0.1897],[200,0.1916],[201,0.1927],[202,0.1933],[204,0.1931],[207,0.1913],[210,0.1883],[211,0.187],[212,0.1852],[213,0.1828],[214,0.18],[219,0.1624],[220,0.1591],[221,0.1562],[222,0.1538],[223,0.1521],[231,0.142],[237,0.1361],[243,0.1319],[248,0.1296],[249,0.1284],[250,0.1257],[251,0.1207],[252,0.1125],[253,0.0956],[254,0.0701],[255,0.0441],[256,0.0266],[257,0.0176],[258,0.0108],[259,0.006],[260,0.0028],[261,0.001],[262,0.0002],[324,0]]
       }
       spine_03_ctrl.Rotation.Z {
        type=MovieSceneScriptingFloatChannel
        default=0
        thinned_from=325
        keys [[0,0],[28,0],[29,-0.0002],[30,-0.0013],[31,-0.0043],[32,-0.0126],[33,-0.0253],[34,-0.0513],[35,-0.1197],[36,-0.2662],[37,-0.4347],[38,-0.5486],[39,-0.6095],[40,-0.6593],[41,-0.691],[42,-0.7081],[43,-0.7196],[44,-0.7266],[45,-0.7299],[86,-0.7182],[87,-0.7167],[88,-0.7072],[89,-0.653],[90,-0.5015],[91,-0.3222],[92,-0.1743],[93,-0.0827],[94,-0.0209],[95,0.0164],[96,0.0391],[97,0.0544],[98,0.0638],[99,0.0688],[100,0.0707],[101,0.071],[182,0.071],[183,0.0714],[184,0.0742],[185,0.0816],[186,0.096],[187,0.1198],[188,0.1554],[189,0.2383],[190,0.392],[191,0.6001],[192,0.8463],[193,1.2881],[194,1.7137],[195,1.912],[196,2.0565],[197,2.1537],[198,2.21],[199,2.2428],[200,2.2662],[201,2.2819],[202,2.2914],[247,2.2923],[248,2.2899],[249,2.2736],[250,2.2291],[251,2.1424],[252,1.9995],[253,1.6921],[255,0.7584],[256,0.4465],[257,0.2882],[258,0.1725],[259,0.0927],[260,0.0422],[261,0.0143],[262,0.0025],[263,0],[324,0]]
       }
       spine_03_ctrl.Scale.X {type=MovieSceneScriptingFloatChannel, default=1, const [1,325]}
       spine_03_ctrl.Scale.Y {type=MovieSceneScriptingFloatChannel, default=1, const [1,325]}
       spine_03_ctrl.Scale.Z {type=MovieSceneScriptingFloatChannel, default=1, const [1,325]}
       spine_04_ctrl.Location.X {
        type=MovieSceneScriptingFloatChannel
        default=0
        thinned_from=325
        keys [[0,0],[28,0],[29,-0.0002],[30,-0.0018],[31,-0.0062],[32,-0.0182],[33,-0.0368],[34,-0.0637],[35,-0.0851],[36,-0.0918],[37,-0.0956],[38,-0.097],[39,-0.0973],[85,-0.0975],[86,-0.0971],[87,-0.0942],[88,-0.0888],[89,-0.0754],[90,-0.0311],[91,-0.0166],[92,-0.0079],[93,-0.0032],[94,-0.001],[95,-0.0002],[185,-0.0005],[188,-0.0011],[189,-0.0017],[190,-0.0029],[191,-0.0044],[192,-0.0064],[194,-0.0131],[195,-0.0146],[196,-0.0158],[197,-0.0165],[199,-0.0172],[211,-0.0172],[213,-0.0168],[221,-0.0127],[240,-0.009],[250,-0.0085],[251,-0.0082],[252,-0.0077],[253,-0.0066],[255,-0.0032],[256,-0.002],[258,-0.0008],[260,-0.0002],[324,0]]
       }
       spine_04_ctrl.Location.Y {
        type=MovieSceneScriptingFloatChannel
        default=0
        thinned_from=325
        keys [[0,0.5737],[1,0.5729],[2,0.5702],[3,0.5656],[4,0.559],[5,0.5503],[6,0.5395],[7,0.5263],[8,0.5108],[9,0.4924],[10,0.4709],[11,0.4473],[14,0.3713],[15,0.3469],[16,0.3243],[17,0.3043],[18,0.2852],[20,0.2454],[21,0.2259],[22,0.2076],[23,0.1911],[24,0.1772],[25,0.1664],[26,0.1594],[27,0.1569],[28,0.1571],[29,0.1576],[30,0.1586],[31,0.1599],[32,0.1618],[33,0.1641],[34,0.1671],[35,0.1707],[36,0.1751],[37,0.1803],[38,0.186],[39,0.1924],[40,0.1996],[41,0.2075],[42,0.2162],[43,0.2258],[44,0.2379],[45,0.2538],[46,0.2728],[47,0.294],[48,0.3169],[51,0.3882],[52,0.4106],[53,0.4311],[54,0.4491],[55,0.464],[56,0.475],[59,0.4997],[62,0.5206],[65,0.538],[68,0.5518],[72,0.5652],[76,0.5731],[81,0.5763],[84,0.575],[85,0.5734],[86,0.5701],[87,0.5649],[88,0.5578],[89,0.5484],[90,0.5364],[91,0.5227],[92,0.5067],[93,0.4879],[94,0.4664],[95,0.4428],[98,0.3676],[99,0.3436],[100,0.3214],[101,0.3018],[102,0.2831],[104,0.2439],[105,0.2248],[106,0.2068],[107,0.1906],[108,0.1768],[109,0.1662],[110,0.1594],[111,0.1569],[112,0.1571],[113,0.1576],[114,0.1586],[115,0.1599],[116,0.1618],[117,0.1641],[118,0.1671],[119,0.1706],[120,0.1747],[121,0.1795],[122,0.185],[123,0.1913],[124,0.1983],[125,0.2061],[126,0.2148],[127,0.2244],[128,0.2365],[129,0.2524],[130,0.2713],[131,0.2926],[132,0.3155],[135,0.3868],[136,0.4091],[137,0.4297],[138,0.4477],[139,0.4625],[140,0.4736],[143,0.4983],[146,0.5192],[149,0.5365],[152,0.5504],[156,0.5638],[160,0.5717],[165,0.5748],[168,0.5736],[169,0.5719],[170,0.5682],[171,0.5626],[172,0.5548],[173,0.5449],[174,0.5327],[175,0.5181],[176,0.5012],[177,0.4807],[178,0.4566],[179,0.43],[181,0.3743],[182,0.3477],[183,0.3236],[184,0.3032],[187,0.2478],[188,0.2299],[189,0.2131],[190,0.1977],[191,0.1841],[192,0.1729],[193,0.1643],[194,0.1589],[195,0.157],[196,0.1571],[197,0.1576],[198,0.1585],[199,0.1597],[200,0.1615],[201,0.1637],[202,0.1665],[203,0.1699],[204,0.1739],[205,0.1787],[206,0.1842],[207,0.1904],[208,0.1975],[209,0.2055],[210,0.2144],[211,0.2243],[212,0.2372],[213,0.2547],[214,0.2758],[215,0.2995],[216,0.3249],[218,0.3771],[219,0.402],[220,0.4249],[221,0.4449],[222,0.4611],[223,0.4726],[226,0.4973],[229,0.5181],[232,0.5352],[235,0.5488],[239,0.5619],[243,0.5698],[248,0.5734],[252,0.5724],[253,0.5711],[254,0.5681],[255,0.5631],[256,0.5561],[257,0.5468],[258,0.5354],[259,0.5216],[260,0.5054],[261,0.4863],[262,0.4647],[263,0.4413],[266,0.367],[267,0.3433],[268,0.3213],[269,0.3018],[270,0.2831],[272,0.244],[273,0.2249],[274,0.2069],[275,0.1906],[276,0.1769],[277,0.1662],[278,0.1594],[279,0.1569],[280,0.1571],[281,0.1576],[282,0.1586],[283,0.1599],[284,0.1618],[285,0.1641],[286,0.1671],[287,0.1706],[288,0.1747],[289,0.1795],[290,0.185],[291,0.1913],[292,0.1983],[293,0.2061],[294,0.2148],[295,0.2244],[296,0.2363],[297,0.2516],[298,0.2697],[299,0.29],[300,0.3119],[303,0.381],[304,0.4033],[305,0.4243],[306,0.4433],[307,0.46],[308,0.4736],[310,0.4961],[312,0.516],[314,0.5331],[316,0.5474],[318,0.5587],[320,0.567],[322,0.572],[324,0.5737]]
       }
       spine_04_ctrl.Location.Z {
        type=MovieSceneScriptingFloatChannel
        default=0
        thinned_from=325
        keys [[0,0.0672],[1,0.0681],[2,0.0709],[3,0.0753],[4,0.0813],[5,0.0885],[6,0.097],[7,0.1064],[8,0.1168],[9,0.1279],[10,0.1394],[11,0.1514],[17,0.2265],[18,0.2385],[19,0.2496],[20,0.2596],[21,0.2687],[22,0.2766],[23,0.2832],[24,0.2885],[25,0.2924],[26,0.2949],[27,0.2957],[30,0.295],[32,0.2935],[34,0.2907],[36,0.2863],[38,0.2802],[39,0.2764],[40,0.272],[41,0.2669],[42,0.2612],[43,0.2548],[44,0.2471],[45,0.2376],[46,0.2267],[47,0.2147],[48,0.2019],[51,0.1612],[52,0.1479],[53,0.1353],[54,0.1235],[55,0.113],[56,0.1039],[58,0.0879],[59,0.0801],[60,0.0726],[61,0.0654],[62,0.0586],[63,0.0521],[64,0.046],[65,0.0405],[66,0.0354],[67,0.0309],[68,0.0271],[69,0.0239],[70,0.0214],[71,0.0197],[72,0.0187],[73,0.0187],[74,0.0194],[75,0.0209],[76,0.0232],[77,0.0262],[78,0.0299],[79,0.0344],[80,0.0396],[81,0.0456],[82,0.0523],[83,0.0598],[84,0.0679],[85,0.0768],[86,0.0863],[87,0.0963],[88,0.1068],[89,0.1176],[94,0.1738],[97,0.2064],[99,0.227],[101,0.2467],[102,0.2558],[103,0.264],[104,0.2713],[105,0.2776],[106,0.2831],[107,0.2876],[108,0.2911],[109,0.2936],[110,0.2952],[113,0.2954],[116,0.2935],[118,0.2907],[120,0.2865],[122,0.2803],[123,0.2765],[124,0.272],[125,0.267],[126,0.2612],[127,0.2548],[128,0.2471],[129,0.2376],[130,0.2268],[131,0.2148],[132,0.2019],[135,0.1612],[136,0.148],[137,0.1352],[138,0.1235],[139,0.113],[140,0.104],[142,0.0879],[143,0.0802],[144,0.0727],[145,0.0655],[146,0.0586],[147,0.0521],[148,0.0461],[149,0.0405],[150,0.0354],[151,0.031],[152,0.0271],[153,0.0239],[154,0.0214],[155,0.0197],[156,0.0188],[157,0.0187],[158,0.0194],[159,0.021],[160,0.0233],[161,0.0264],[162,0.0302],[163,0.0348],[164,0.0401],[165,0.0461],[166,0.0529],[167,0.0604],[168,0.0686],[169,0.0774],[170,0.0869],[171,0.0967],[172,0.107],[173,0.1177],[178,0.1727],[182,0.2147],[184,0.235],[185,0.2448],[186,0.2539],[187,0.2621],[188,0.2696],[189,0.2763],[190,0.282],[191,0.2868],[192,0.2906],[193,0.2934],[194,0.2951],[195,0.2957],[199,0.2945],[201,0.2926],[203,0.2893],[205,0.2842],[206,0.2809],[207,0.277],[208,0.2725],[209,0.2673],[210,0.2614],[211,0.2548],[212,0.2465],[213,0.2362],[214,0.2241],[215,0.2107],[216,0.1964],[219,0.1519],[220,0.1378],[221,0.1248],[222,0.1134],[223,0.1038],[224,0.0955],[225,0.0873],[226,0.0794],[227,0.0718],[228,0.0645],[229,0.0576],[230,0.0512],[231,0.0452],[232,0.0397],[233,0.0347],[234,0.0303],[235,0.0265],[236,0.0234],[237,0.021],[238,0.0194],[239,0.0185],[240,0.0184],[241,0.0191],[242,0.0204],[243,0.0223],[244,0.0248],[245,0.028],[246,0.0318],[247,0.0364],[248,0.0415],[249,0.0474],[250,0.054],[251,0.0612],[252,0.0692],[253,0.0779],[254,0.0873],[255,0.0971],[256,0.1074],[257,0.1181],[263,0.1847],[266,0.2167],[268,0.2369],[269,0.2466],[270,0.2557],[271,0.2639],[272,0.2712],[273,0.2776],[274,0.283],[275,0.2876],[276,0.2911],[277,0.2936],[278,0.2952],[281,0.2954],[284,0.2935],[286,0.2907],[288,0.2865],[290,0.2803],[291,0.2765],[292,0.272],[293,0.267],[294,0.2612],[295,0.2548],[296,0.247],[297,0.2372],[298,0.2259],[299,0.2133],[300,0.1999],[303,0.158],[304,0.1447],[305,0.1322],[306,0.1211],[307,0.1115],[308,0.104],[309,0.0979],[310,0.0925],[311,0.0878],[312,0.0837],[313,0.0802],[314,0.0772],[315,0.0747],[316,0.0726],[317,0.071],[318,0.0697],[319,0.0688],[320,0.0681],[322,0.0673],[324,0.0672]]
       }
       spine_04_ctrl.Rotation.X {
        type=MovieSceneScriptingFloatChannel
        default=0
        thinned_from=325
        keys [[0,8.6154],[2,8.6046],[4,8.5535],[5,8.5042],[6,8.4343],[7,8.3403],[8,8.2188],[9,8.0586],[10,7.8575],[11,7.6258],[15,6.5953],[16,6.3626],[21,5.3762],[22,5.1925],[23,5.0272],[24,4.8869],[25,4.7785],[26,4.7085],[27,4.6837],[30,4.7009],[33,4.7556],[35,4.8145],[37,4.8939],[39,4.995],[41,5.119],[43,5.2669],[44,5.3699],[45,5.5134],[46,5.6902],[47,5.8927],[51,6.8146],[52,7.0364],[53,7.2401],[54,7.4185],[55,7.5643],[56,7.6704],[60,7.9689],[65,8.2484],[71,8.4665],[79,8.5982],[85,8.6106],[87,8.5818],[88,8.5484],[89,8.4977],[90,8.4264],[91,8.3299],[92,8.2047],[93,8.0414],[94,7.8392],[95,7.6079],[99,6.5871],[100,6.3569],[105,5.3748],[106,5.1915],[107,5.0266],[108,4.8866],[109,4.7783],[110,4.7084],[111,4.6836],[114,4.7009],[117,4.7558],[119,4.8156],[121,4.8955],[123,4.9969],[125,5.121],[127,5.269],[128,5.3719],[129,5.5154],[130,5.6922],[131,5.8947],[135,6.8165],[136,7.0383],[137,7.242],[138,7.4204],[139,7.5663],[140,7.6723],[144,7.9708],[149,8.2502],[155,8.4683],[163,8.6003],[169,8.6124],[171,8.5812],[172,8.5455],[173,8.4916],[174,8.4158],[175,8.3142],[176,8.1834],[177,8.0039],[178,7.7702],[179,7.4988],[182,6.6248],[183,6.369],[184,6.1586],[188,5.4203],[189,5.2504],[190,5.0948],[191,4.9577],[192,4.8434],[193,4.7553],[194,4.6983],[195,4.678],[198,4.6921],[201,4.7438],[203,4.8019],[205,4.8809],[207,4.9827],[209,5.1089],[211,5.2614],[212,5.3721],[213,5.5316],[214,5.7299],[215,5.9574],[219,6.9638],[220,7.191],[221,7.3891],[222,7.5486],[223,7.6597],[227,7.9576],[232,8.2348],[238,8.4493],[247,8.5864],[254,8.5967],[256,8.5449],[257,8.4919],[258,8.4167],[259,8.3156],[260,8.1849],[261,8.0174],[262,7.8142],[263,7.5844],[267,6.5812],[268,6.3548],[273,5.3763],[274,5.1929],[275,5.0276],[276,4.8873],[277,4.7787],[278,4.7085],[279,4.6837],[282,4.7009],[285,4.7558],[287,4.8156],[289,4.8955],[291,4.9969],[293,5.121],[295,5.269],[296,5.3697],[297,5.5073],[298,5.6755],[299,5.8681],[301,6.3014],[304,6.9784],[305,7.1865],[306,7.3757],[307,7.5396],[308,7.6723],[311,7.9869],[314,8.24],[317,8.4297],[320,8.5542],[323,8.6115],[324,8.6154]]
       }
       spine_04_ctrl.Rotation.Y {
        type=MovieSceneScriptingFloatChannel
        default=0
        thinned_from=325
        keys [[0,0],[28,0],[29,0.0035],[30,0.0282],[31,0.0953],[32,0.2825],[33,0.5709],[34,0.988],[35,1.3242],[36,1.4418],[37,1.5154],[38,1.548],[39,1.5575],[40,1.5617],[85,1.5795],[86,1.5734],[87,1.5288],[88,1.444],[89,1.2304],[90,0.5275],[91,0.2873],[92,0.1385],[93,0.0567],[94,0.0167],[95,0.0021],[96,0],[111,0.0018],[129,0.0021],[169,-0.0009],[185,0.0008],[186,0.0011],[187,0.0016],[188,0.0024],[189,0.0042],[190,0.0078],[191,0.013],[192,0.0197],[194,0.0435],[195,0.0491],[196,0.0529],[197,0.0554],[198,0.0569],[199,0.0578],[201,0.059],[205,0.0609],[207,0.0624],[208,0.0634],[209,0.0647],[210,0.0662],[211,0.0679],[212,0.0684],[213,0.0666],[214,0.0628],[215,0.0575],[216,0.051],[217,0.0438],[219,0.0284],[220,0.0211],[221,0.0146],[222,0.0092],[223,0.0054],[225,-0.0002],[227,-0.0054],[229,-0.01],[231,-0.0143],[233,-0.018],[235,-0.0212],[237,-0.024],[239,-0.0263],[241,-0.028],[243,-0.0292],[245,-0.03],[247,-0.0302],[248,-0.0302],[249,-0.0299],[250,-0.0291],[251,-0.0278],[252,-0.0257],[253,-0.0214],[255,-0.0088],[256,-0.0048],[257,-0.0028],[258,-0.0015],[259,-0.0007],[260,-0.0002],[324,0]]
       }
       spine_04_ctrl.Rotation.Z {
        type=MovieSceneScriptingFloatChannel
        default=0
        thinned_from=325
        keys [[0,0],[28,0],[29,-0.0002],[30,-0.0017],[31,-0.0058],[32,-0.0172],[33,-0.0347],[34,-0.0695],[35,-0.1569],[36,-0.3412],[37,-0.5528],[38,-0.6959],[39,-0.7726],[40,-0.8355],[41,-0.8757],[42,-0.8978],[43,-0.9129],[44,-0.9223],[45,-0.9272],[86,-0.9209],[87,-0.9187],[88,-0.9059],[89,-0.8354],[90,-0.6372],[91,-0.4087],[92,-0.2208],[93,-0.1047],[94,-0.0265],[95,0.0207],[96,0.0492],[97,0.0683],[98,0.0801],[99,0.0863],[100,0.0887],[101,0.0891],[182,0.0892],[183,0.0896],[184,0.093],[185,0.1023],[186,0.1204],[187,0.1502],[188,0.1946],[189,0.2985],[190,0.4908],[191,0.7511],[192,1.059],[193,1.6116],[194,2.1438],[195,2.3918],[196,2.5726],[197,2.6943],[198,2.7647],[199,2.8058],[200,2.8352],[201,2.855],[202,2.867],[247,2.8889],[248,2.8859],[249,2.8652],[250,2.8091],[251,2.6999],[252,2.5197],[253,2.1323],[255,0.9557],[256,0.5625],[257,0.3631],[258,0.2173],[259,0.1167],[260,0.0531],[261,0.018],[262,0.0031],[263,0],[324,0]]
       }
       spine_04_ctrl.Scale.X {type=MovieSceneScriptingFloatChannel, default=1, const [1,325]}
       spine_04_ctrl.Scale.Y {type=MovieSceneScriptingFloatChannel, default=1, const [1,325]}
       spine_04_ctrl.Scale.Z {type=MovieSceneScriptingFloatChannel, default=1, const [1,325]}
       head_ctrl.Location.X {type=MovieSceneScriptingFloatChannel, default=0, thinned_from=325, keys [[0,0],[324,0]]}
       head_ctrl.Location.Y {type=MovieSceneScriptingFloatChannel, default=0, thinned_from=325, keys [[0,0],[324,0]]}
       head_ctrl.Location.Z {type=MovieSceneScriptingFloatChannel, default=0, thinned_from=325, keys [[0,0],[324,0]]}
       head_ctrl.Rotation.X {
        type=MovieSceneScriptingFloatChannel
        default=0
        thinned_from=325
        keys [[0,9.8055],[1,9.7918],[3,9.6995],[4,9.6203],[6,9.4038],[8,9.1179],[10,8.7667],[16,7.5686],[18,7.1976],[22,6.3558],[23,6.1544],[24,5.9698],[25,5.8086],[26,5.6769],[27,5.5812],[28,5.5153],[29,5.5174],[30,5.6476],[31,5.9719],[32,7.0188],[33,8.344],[34,8.956],[35,9.1243],[36,9.0672],[37,8.96],[39,8.6633],[40,8.5425],[41,8.4722],[42,8.4729],[43,8.5248],[44,8.6066],[45,8.7262],[46,8.8777],[47,9.055],[53,10.3173],[54,10.5031],[55,10.6655],[62,11.4629],[68,11.9771],[73,12.2929],[78,12.5085],[80,12.6397],[82,12.8328],[84,13.1116],[86,13.4618],[88,13.8762],[89,13.9304],[90,13.5742],[91,13.0732],[92,12.5352],[93,12.0227],[94,11.5751],[95,11.1877],[96,10.8579],[97,10.5832],[98,10.3562],[100,9.9523],[102,9.5759],[107,8.5179],[108,8.331],[109,8.168],[110,8.045],[111,7.9402],[115,7.6688],[118,7.5238],[120,7.4696],[123,7.4719],[125,7.5405],[126,7.5978],[127,7.6719],[128,7.7758],[129,7.9177],[130,8.0914],[131,8.2906],[136,9.4478],[138,9.8685],[139,10.0448],[144,10.6841],[147,11.0048],[150,11.2745],[154,11.5759],[161,12.0021],[166,12.1842],[168,12.2216],[169,12.2216],[170,12.1841],[171,12.121],[172,12.0314],[174,11.7819],[176,11.4537],[178,11.0518],[180,10.6084],[181,10.4163],[182,10.2825],[183,10.2079],[184,10.1993],[185,10.2428],[186,10.5019],[187,10.9172],[188,11.1147],[189,11.0178],[190,10.8663],[191,10.6576],[192,10.3343],[195,9.2324],[196,8.9029],[197,8.6594],[198,8.5035],[199,8.3931],[200,8.318],[202,8.2314],[205,8.1578],[207,8.1461],[209,8.182],[211,8.2746],[212,8.3553],[213,8.47],[214,8.6126],[216,8.9574],[220,9.7153],[221,9.8834],[222,10.0307],[229,10.7436],[234,11.1368],[246,11.7984],[247,12.085],[250,13.7843],[251,14.0702],[252,13.732],[253,13.1823],[255,12.154],[258,10.435],[259,9.8962],[260,9.42],[261,9.0344],[262,8.7697],[268,7.5458],[270,7.1658],[274,6.3143],[275,6.1122],[276,5.9276],[277,5.767],[278,5.6369],[279,5.5439],[282,5.3416],[284,5.2304],[286,5.1466],[288,5.0976],[290,5.0907],[291,5.1052],[292,5.1331],[293,5.1751],[294,5.2322],[295,5.3054],[296,5.4049],[297,5.5372],[298,5.6974],[299,5.8805],[300,6.0816],[305,7.1838],[306,7.3889],[307,7.5775],[312,8.3442],[315,8.7504],[319,9.2181],[322,9.6563],[323,9.7587],[324,9.8041]]
       }
       head_ctrl.Rotation.Y {
        type=MovieSceneScriptingFloatChannel
        default=0
        thinned_from=325
        keys [[0,-5.1409],[3,-5.1184],[6,-5.0565],[10,-4.925],[16,-4.6913],[23,-4.4749],[25,-4.433],[26,-4.4232],[28,-4.4274],[29,-4.3848],[30,-4.2407],[31,-3.9386],[32,-2.9477],[33,-1.3321],[34,1.006],[35,3.0054],[36,3.3195],[37,3.565],[38,3.7691],[39,3.9333],[40,4.0458],[41,4.1116],[42,4.1347],[44,4.155],[46,4.2079],[53,4.4579],[55,4.5101],[61,4.5905],[62,4.6099],[63,4.6117],[64,4.629],[65,4.6308],[66,4.6461],[67,4.648],[69,4.6736],[70,4.67],[72,4.6904],[73,4.6875],[76,4.7099],[84,4.7209],[85,4.5606],[86,4.0534],[87,3.1582],[88,1.6954],[89,-0.1622],[90,-2.0007],[91,-3.4762],[92,-4.3853],[93,-4.6647],[94,-4.8218],[95,-4.8901],[96,-4.9004],[108,-4.7146],[110,-4.6942],[122,-4.7161],[129,-4.7797],[138,-4.9373],[144,-4.9898],[147,-5.0102],[148,-5.006],[150,-5.0226],[151,-5.0194],[153,-5.0329],[169,-5.0567],[175,-4.9998],[182,-4.8654],[183,-4.8358],[184,-4.7858],[185,-4.8001],[186,-5.0077],[187,-5.3608],[188,-5.9793],[189,-7.8098],[190,-10.2696],[191,-11.3584],[192,-10.9642],[193,-10.4665],[194,-10.057],[195,-9.766],[196,-9.5694],[197,-9.4303],[198,-9.3262],[199,-9.2555],[200,-9.2164],[202,-9.2175],[206,-9.3093],[210,-9.4838],[212,-9.6165],[213,-9.7191],[215,-9.9834],[219,-10.5934],[221,-10.8509],[222,-10.9485],[232,-11.4106],[244,-11.6853],[249,-11.6667],[250,-11.4228],[251,-10.4509],[252,-8.1068],[253,-6.2922],[254,-5.9068],[255,-5.5739],[256,-5.3371],[257,-5.1745],[258,-5.0774],[259,-5.0238],[268,-4.6919],[275,-4.4751],[277,-4.4327],[279,-4.4164],[289,-4.4456],[295,-4.5274],[297,-4.5721],[306,-4.8902],[315,-5.0698],[317,-5.1065],[324,-5.1406]]
       }
       head_ctrl.Rotation.Z {
        type=MovieSceneScriptingFloatChannel
        default=0
        thinned_from=325
        keys [[0,11.2508],[25,11.5385],[26,11.4974],[27,11.3754],[28,11.1366],[29,10.7535],[30,10.1992],[31,9.4464],[32,7.1332],[33,3.5671],[34,-0.6536],[35,-4.121],[36,-6.4166],[37,-8.3881],[38,-9.6383],[39,-10.3185],[40,-10.855],[41,-11.2635],[42,-11.5593],[43,-11.7523],[44,-11.8524],[45,-11.8751],[84,-11.6594],[85,-11.517],[86,-10.9224],[87,-9.6243],[88,-6.6363],[89,-2.8788],[90,0.2211],[91,1.7257],[92,3.0967],[93,3.996],[94,4.4105],[95,4.7571],[96,5.04],[97,5.2642],[98,5.4242],[99,5.5219],[100,5.5735],[110,5.7003],[122,5.685],[130,5.6187],[139,5.4819],[144,5.4444],[147,5.4253],[148,5.4297],[150,5.4139],[169,5.3851],[176,5.4564],[181,5.5481],[182,5.6783],[183,6.0601],[184,6.7326],[185,7.7426],[186,9.139],[187,11.9049],[188,16.8395],[189,24.0322],[190,32.1466],[191,38.8416],[192,42.1948],[193,43.1981],[197,45.1084],[200,46.2477],[202,46.8128],[204,47.1635],[244,46.837],[245,46.7537],[246,46.4434],[247,45.8161],[248,44.8068],[249,43.3557],[250,41.4178],[251,35.7084],[252,28.3994],[253,21.9351],[254,17.1092],[255,15.2053],[256,13.716],[257,12.612],[258,11.8644],[259,11.4436],[260,11.3196],[279,11.5528],[324,11.2509]]
       }
       head_ctrl.Scale.X {type=MovieSceneScriptingFloatChannel, default=1, const [1,325]}
       head_ctrl.Scale.Y {type=MovieSceneScriptingFloatChannel, default=1, const [1,325]}
       head_ctrl.Scale.Z {type=MovieSceneScriptingFloatChannel, default=1, const [1,325]}
       clavicle_l_ctrl.Location.X {type=MovieSceneScriptingFloatChannel, default=0, thinned_from=325, keys [[0,0],[324,0]]}
       clavicle_l_ctrl.Location.Y {type=MovieSceneScriptingFloatChannel, default=0, thinned_from=325, keys [[0,0],[324,0]]}
       clavicle_l_ctrl.Location.Z {type=MovieSceneScriptingFloatChannel, default=0, thinned_from=325, keys [[0,0],[324,0]]}
       clavicle_l_ctrl.Rotation.X {
        type=MovieSceneScriptingFloatChannel
        default=0
        thinned_from=325
        keys [[0,-0.0085],[1,-0.01],[2,-0.014],[3,-0.0203],[4,-0.0283],[5,-0.0378],[6,-0.0483],[7,-0.0594],[9,-0.0819],[10,-0.0924],[11,-0.1041],[12,-0.118],[13,-0.1329],[14,-0.1474],[15,-0.1604],[16,-0.1706],[17,-0.1788],[18,-0.1862],[19,-0.1929],[20,-0.1987],[21,-0.2034],[22,-0.207],[23,-0.2092],[24,-0.21],[26,-0.209],[28,-0.2063],[30,-0.2021],[32,-0.1964],[34,-0.1894],[36,-0.1814],[38,-0.1724],[41,-0.1576],[42,-0.1524],[43,-0.1457],[44,-0.1365],[45,-0.1256],[46,-0.1136],[47,-0.1013],[48,-0.0894],[49,-0.0786],[50,-0.0691],[51,-0.0622],[54,-0.0471],[55,-0.0423],[56,-0.0377],[57,-0.0334],[58,-0.0296],[59,-0.0263],[60,-0.0237],[61,-0.0216],[62,-0.0198],[63,-0.0183],[64,-0.017],[66,-0.0151],[69,-0.0134],[75,-0.0103],[76,-0.0103],[77,-0.0095],[79,-0.0095],[80,-0.0089],[84,-0.0089],[85,-0.0101],[86,-0.0142],[87,-0.0204],[88,-0.0285],[89,-0.038],[90,-0.0484],[91,-0.0595],[93,-0.0819],[94,-0.0925],[95,-0.1041],[96,-0.118],[97,-0.1329],[98,-0.1474],[99,-0.1604],[100,-0.1706],[101,-0.1788],[102,-0.1862],[103,-0.1929],[104,-0.1987],[105,-0.2034],[106,-0.207],[107,-0.2092],[108,-0.2099],[109,-0.2098],[111,-0.2081],[113,-0.205],[115,-0.2005],[117,-0.1948],[119,-0.1879],[121,-0.1802],[123,-0.1716],[126,-0.1574],[127,-0.1524],[128,-0.1445],[129,-0.132],[131,-0.1022],[132,-0.0892],[133,-0.0784],[134,-0.0691],[135,-0.0622],[137,-0.0515],[138,-0.0465],[139,-0.0417],[140,-0.0372],[141,-0.0331],[142,-0.0295],[143,-0.0263],[144,-0.0237],[145,-0.0215],[146,-0.0197],[147,-0.0182],[148,-0.017],[150,-0.0151],[152,-0.0138],[159,-0.0102],[160,-0.0102],[161,-0.0094],[163,-0.0094],[164,-0.0088],[168,-0.0088],[169,-0.01],[170,-0.0141],[171,-0.0203],[172,-0.0284],[173,-0.0379],[174,-0.0483],[177,-0.0819],[178,-0.0924],[179,-0.1041],[180,-0.118],[182,-0.1474],[183,-0.1604],[184,-0.1706],[185,-0.1788],[186,-0.1862],[187,-0.1929],[188,-0.1987],[189,-0.2034],[190,-0.207],[191,-0.2092],[192,-0.21],[193,-0.2097],[195,-0.2079],[197,-0.2044],[199,-0.1994],[201,-0.193],[203,-0.1855],[205,-0.177],[208,-0.1628],[210,-0.1525],[211,-0.1457],[212,-0.1366],[213,-0.1256],[214,-0.1136],[215,-0.1013],[216,-0.0894],[217,-0.0786],[218,-0.0691],[219,-0.0622],[222,-0.0471],[223,-0.0422],[224,-0.0376],[225,-0.0333],[226,-0.0295],[227,-0.0262],[228,-0.0234],[229,-0.0213],[230,-0.0194],[231,-0.018],[232,-0.0167],[234,-0.015],[241,-0.0114],[242,-0.0114],[243,-0.0104],[244,-0.0104],[245,-0.0097],[253,-0.0097],[254,-0.0142],[255,-0.0204],[256,-0.0285],[257,-0.038],[258,-0.0485],[259,-0.0595],[261,-0.082],[262,-0.0925],[263,-0.1042],[264,-0.118],[265,-0.1329],[266,-0.1474],[267,-0.1604],[268,-0.1706],[269,-0.1788],[270,-0.1862],[271,-0.1929],[272,-0.1987],[273,-0.2034],[274,-0.207],[275,-0.2092],[276,-0.21],[277,-0.2097],[279,-0.2078],[281,-0.2042],[283,-0.1991],[285,-0.1927],[287,-0.1851],[289,-0.1766],[292,-0.1625],[294,-0.1524],[295,-0.1455],[296,-0.1356],[297,-0.1237],[298,-0.1111],[299,-0.0987],[300,-0.0876],[301,-0.0778],[302,-0.0691],[303,-0.0622],[306,-0.0455],[307,-0.0403],[308,-0.0353],[309,-0.0307],[310,-0.0266],[311,-0.0231],[312,-0.0202],[313,-0.0178],[314,-0.0158],[315,-0.0141],[316,-0.0126],[317,-0.0114],[318,-0.0105],[320,-0.0093],[322,-0.0087],[324,-0.0086]]
       }
       clavicle_l_ctrl.Rotation.Y {
        type=MovieSceneScriptingFloatChannel
        default=0
        thinned_from=325
        keys [[0,-6.7741],[1,-6.7628],[2,-6.7302],[3,-6.679],[4,-6.6114],[6,-6.437],[10,-5.9989],[11,-5.8601],[12,-5.6841],[14,-5.2867],[15,-5.0981],[16,-4.9383],[19,-4.5187],[20,-4.3899],[21,-4.2735],[22,-4.1734],[23,-4.0939],[24,-4.039],[26,-3.9653],[29,-3.8846],[32,-3.8338],[37,-3.7966],[42,-3.7913],[43,-3.7994],[44,-3.8222],[45,-3.8575],[46,-3.9031],[48,-4.0161],[49,-4.0791],[50,-4.1963],[51,-4.3679],[52,-4.5495],[56,-5.3489],[57,-5.5375],[58,-5.7098],[59,-5.8599],[60,-5.9819],[62,-6.174],[64,-6.3297],[66,-6.4534],[69,-6.5886],[73,-6.6984],[75,-6.7316],[76,-6.7316],[77,-6.7529],[79,-6.7529],[80,-6.7682],[84,-6.7682],[85,-6.7571],[86,-6.7155],[87,-6.651],[88,-6.5673],[90,-6.3578],[94,-5.8782],[95,-5.7473],[96,-5.5912],[99,-5.0835],[103,-4.525],[104,-4.3931],[105,-4.2729],[106,-4.1697],[107,-4.0886],[108,-4.035],[111,-3.9377],[114,-3.8709],[118,-3.8192],[127,-3.7935],[128,-3.8003],[129,-3.821],[130,-3.8562],[131,-3.9068],[132,-3.9734],[133,-4.078],[134,-4.2258],[135,-4.3922],[136,-4.5686],[140,-5.3549],[141,-5.5416],[142,-5.7123],[143,-5.8611],[144,-5.9819],[146,-6.1723],[148,-6.3271],[150,-6.4506],[153,-6.5866],[157,-6.6984],[159,-6.7325],[160,-6.7325],[161,-6.7543],[163,-6.7543],[164,-6.7701],[168,-6.7701],[169,-6.759],[170,-6.7174],[171,-6.6528],[172,-6.569],[174,-6.3592],[178,-5.8791],[179,-5.7481],[180,-5.5919],[183,-5.0837],[187,-4.525],[188,-4.3932],[189,-4.2728],[190,-4.1691],[191,-4.087],[192,-4.0318],[194,-3.9602],[197,-3.8819],[200,-3.8326],[205,-3.7965],[210,-3.7913],[211,-3.8003],[212,-3.8256],[213,-3.8643],[214,-3.9138],[217,-4.1004],[218,-4.2177],[219,-4.3875],[220,-4.5664],[224,-5.3501],[225,-5.5358],[226,-5.7067],[227,-5.8572],[228,-5.9819],[230,-6.1828],[232,-6.3471],[234,-6.4778],[236,-6.578],[239,-6.6774],[241,-6.7152],[242,-6.7152],[243,-6.7407],[244,-6.7407],[245,-6.7571],[253,-6.7571],[254,-6.7153],[255,-6.6508],[256,-6.5671],[258,-6.3576],[262,-5.8782],[263,-5.7473],[264,-5.5914],[267,-5.0838],[271,-4.5233],[272,-4.3906],[273,-4.2695],[274,-4.1653],[275,-4.0831],[276,-4.0282],[278,-3.9577],[281,-3.8805],[284,-3.8319],[289,-3.7964],[294,-3.7913],[295,-3.7969],[296,-3.8138],[297,-3.8419],[298,-3.8812],[299,-3.9318],[300,-3.9937],[301,-4.0918],[302,-4.2348],[303,-4.3985],[304,-4.5725],[308,-5.3456],[309,-5.531],[310,-5.7025],[311,-5.8546],[312,-5.9819],[315,-6.2966],[317,-6.4724],[319,-6.6136],[321,-6.714],[323,-6.7672],[324,-6.7672]]
       }
       clavicle_l_ctrl.Rotation.Z {
        type=MovieSceneScriptingFloatChannel
        default=0
        thinned_from=325
        keys [[0,-0.1435],[1,-0.1595],[2,-0.2044],[3,-0.2734],[4,-0.3615],[5,-0.464],[6,-0.576],[8,-0.8092],[9,-0.9207],[10,-1.0224],[11,-1.1262],[12,-1.2422],[13,-1.3622],[14,-1.4782],[15,-1.5821],[16,-1.6659],[17,-1.7354],[18,-1.8008],[19,-1.8604],[20,-1.9128],[21,-1.9562],[22,-1.9892],[23,-2.0102],[24,-2.0175],[26,-2.0129],[28,-1.9992],[30,-1.9767],[32,-1.9457],[34,-1.9064],[36,-1.8591],[38,-1.8041],[40,-1.7415],[42,-1.6716],[43,-1.6235],[44,-1.5556],[45,-1.4733],[46,-1.3815],[48,-1.1901],[49,-1.1006],[50,-1.002],[51,-0.9081],[52,-0.8293],[53,-0.7456],[55,-0.5754],[56,-0.4948],[57,-0.4212],[58,-0.3574],[59,-0.3065],[60,-0.2713],[61,-0.2473],[62,-0.2273],[63,-0.2111],[64,-0.1981],[65,-0.188],[66,-0.1803],[67,-0.1745],[68,-0.1703],[69,-0.1672],[72,-0.1604],[74,-0.1546],[75,-0.1521],[76,-0.1521],[77,-0.1486],[79,-0.1486],[80,-0.146],[84,-0.146],[85,-0.1615],[86,-0.2062],[87,-0.275],[88,-0.3629],[89,-0.4651],[90,-0.5769],[92,-0.8095],[93,-0.9208],[94,-1.0224],[95,-1.1262],[96,-1.2421],[97,-1.3621],[98,-1.4781],[99,-1.5821],[100,-1.6659],[101,-1.7354],[102,-1.8008],[103,-1.8604],[104,-1.9128],[105,-1.9562],[106,-1.9892],[107,-2.0102],[108,-2.0175],[110,-2.0124],[112,-1.9975],[114,-1.9739],[117,-1.9242],[120,-1.86],[124,-1.7573],[127,-1.6716],[128,-1.6294],[129,-1.5659],[130,-1.4874],[132,-1.312],[133,-1.2042],[134,-1.0792],[135,-0.9719],[136,-0.8819],[138,-0.692],[139,-0.5985],[140,-0.5101],[141,-0.4299],[142,-0.3612],[143,-0.3071],[144,-0.2707],[145,-0.2465],[146,-0.2264],[147,-0.21],[148,-0.1969],[149,-0.1867],[150,-0.1789],[151,-0.1731],[152,-0.1688],[153,-0.1657],[156,-0.1589],[158,-0.1529],[159,-0.1505],[160,-0.1505],[161,-0.1469],[163,-0.1469],[164,-0.1443],[168,-0.1443],[169,-0.1598],[170,-0.2047],[171,-0.2736],[172,-0.3617],[173,-0.4641],[174,-0.5761],[176,-0.8092],[177,-0.9207],[178,-1.0224],[179,-1.1262],[180,-1.2422],[181,-1.3622],[182,-1.4782],[183,-1.5821],[184,-1.6659],[185,-1.7354],[186,-1.8008],[187,-1.8604],[188,-1.9128],[189,-1.9562],[190,-1.9892],[191,-2.0102],[192,-2.0175],[194,-2.0129],[196,-1.9992],[198,-1.9767],[200,-1.9457],[202,-1.9064],[204,-1.8591],[206,-1.804],[208,-1.7415],[210,-1.6716],[211,-1.6235],[212,-1.5557],[213,-1.4734],[214,-1.3817],[216,-1.1904],[217,-1.1009],[218,-1.0023],[219,-0.9083],[220,-0.8292],[221,-0.7453],[223,-0.5747],[224,-0.4939],[225,-0.4199],[226,-0.3558],[227,-0.3044],[228,-0.2687],[229,-0.2442],[230,-0.2244],[231,-0.2086],[232,-0.1963],[233,-0.187],[234,-0.1802],[235,-0.1753],[236,-0.1718],[239,-0.1643],[241,-0.1575],[242,-0.1575],[243,-0.1521],[244,-0.1521],[245,-0.1487],[253,-0.1487],[254,-0.2064],[255,-0.2751],[256,-0.363],[257,-0.4652],[258,-0.5769],[260,-0.8096],[261,-0.9209],[262,-1.0224],[263,-1.1262],[264,-1.2421],[265,-1.3621],[266,-1.4781],[267,-1.5821],[268,-1.6659],[269,-1.7354],[270,-1.8008],[271,-1.8604],[272,-1.9128],[273,-1.9562],[274,-1.9892],[275,-2.0102],[276,-2.0175],[278,-2.0122],[280,-1.9968],[282,-1.9721],[284,-1.9388],[287,-1.8746],[290,-1.7956],[294,-1.6717],[295,-1.6269],[296,-1.5626],[297,-1.4844],[298,-1.3978],[300,-1.2212],[301,-1.1227],[302,-1.0124],[303,-0.9168],[304,-0.8348],[306,-0.6632],[307,-0.5786],[308,-0.4981],[309,-0.4243],[310,-0.3596],[311,-0.3065],[312,-0.2676],[313,-0.2391],[314,-0.2153],[315,-0.1958],[316,-0.1802],[317,-0.1681],[318,-0.159],[319,-0.1524],[320,-0.148],[321,-0.1454],[322,-0.144],[323,-0.1435],[324,-0.1435]]
       }
       clavicle_l_ctrl.Scale.X {type=MovieSceneScriptingFloatChannel, default=1, const [1,325]}
       clavicle_l_ctrl.Scale.Y {type=MovieSceneScriptingFloatChannel, default=1, const [1,325]}
       clavicle_l_ctrl.Scale.Z {type=MovieSceneScriptingFloatChannel, default=1, const [1,325]}
       clavicle_r_ctrl.Location.X {type=MovieSceneScriptingFloatChannel, default=0, thinned_from=325, keys [[0,0],[324,0]]}
       clavicle_r_ctrl.Location.Y {type=MovieSceneScriptingFloatChannel, default=0, thinned_from=325, keys [[0,0],[324,0]]}
       clavicle_r_ctrl.Location.Z {type=MovieSceneScriptingFloatChannel, default=0, thinned_from=325, keys [[0,-0.0001],[324,-0.0001]]}
       clavicle_r_ctrl.Rotation.X {
        type=MovieSceneScriptingFloatChannel
        default=0
        thinned_from=325
        keys [[0,-0.0085],[1,-0.0102],[2,-0.0151],[3,-0.0225],[4,-0.032],[5,-0.0431],[6,-0.0553],[8,-0.0806],[9,-0.0928],[10,-0.1041],[11,-0.1158],[12,-0.129],[13,-0.1429],[14,-0.1562],[15,-0.168],[16,-0.1774],[17,-0.1848],[18,-0.1915],[19,-0.1974],[20,-0.2025],[21,-0.2065],[22,-0.2096],[23,-0.2115],[24,-0.2121],[26,-0.2117],[28,-0.2105],[30,-0.2083],[32,-0.205],[34,-0.2007],[36,-0.1951],[38,-0.1882],[40,-0.1799],[41,-0.1751],[42,-0.17],[43,-0.1626],[44,-0.1517],[45,-0.1382],[46,-0.1233],[47,-0.1078],[48,-0.0928],[49,-0.0793],[50,-0.067],[51,-0.0582],[53,-0.0469],[54,-0.0415],[55,-0.0362],[56,-0.0313],[57,-0.0269],[58,-0.0231],[59,-0.02],[60,-0.0177],[61,-0.016],[62,-0.0145],[64,-0.0124],[66,-0.0112],[68,-0.0105],[73,-0.0096],[74,-0.0096],[75,-0.0092],[77,-0.0092],[78,-0.0088],[85,-0.0088],[86,-0.0153],[87,-0.0227],[88,-0.0323],[89,-0.0434],[90,-0.0556],[92,-0.0811],[93,-0.0934],[94,-0.1046],[95,-0.1164],[96,-0.1297],[97,-0.1436],[98,-0.157],[99,-0.1688],[100,-0.1781],[101,-0.1855],[102,-0.1921],[103,-0.1979],[104,-0.2028],[105,-0.2067],[106,-0.2097],[107,-0.2115],[108,-0.2121],[111,-0.2115],[112,-0.211],[113,-0.211],[114,-0.2093],[116,-0.2068],[118,-0.2032],[120,-0.1984],[122,-0.1922],[123,-0.1886],[124,-0.1846],[125,-0.1802],[126,-0.1753],[127,-0.17],[128,-0.1594],[129,-0.1412],[131,-0.0973],[132,-0.0793],[133,-0.0665],[134,-0.0575],[135,-0.0512],[136,-0.0462],[137,-0.0414],[138,-0.0369],[139,-0.0327],[140,-0.0289],[141,-0.0254],[142,-0.0223],[143,-0.0197],[144,-0.0176],[145,-0.0159],[146,-0.0144],[147,-0.0133],[148,-0.0123],[150,-0.011],[152,-0.0103],[157,-0.0094],[158,-0.0094],[159,-0.009],[160,-0.009],[161,-0.0088],[169,-0.0088],[170,-0.0152],[171,-0.0227],[172,-0.0323],[173,-0.0435],[174,-0.0557],[176,-0.0813],[177,-0.0936],[178,-0.1049],[179,-0.1167],[180,-0.1301],[181,-0.144],[182,-0.1574],[183,-0.1693],[184,-0.1786],[185,-0.1859],[186,-0.1925],[187,-0.1982],[188,-0.203],[189,-0.2069],[190,-0.2098],[191,-0.2115],[192,-0.2121],[195,-0.2112],[197,-0.2095],[199,-0.2068],[201,-0.203],[203,-0.198],[205,-0.1918],[207,-0.1842],[209,-0.1751],[210,-0.17],[211,-0.1626],[212,-0.1517],[213,-0.1382],[214,-0.1233],[215,-0.1078],[216,-0.0928],[217,-0.0793],[218,-0.067],[219,-0.0582],[222,-0.0414],[223,-0.0362],[224,-0.0313],[225,-0.0268],[226,-0.023],[227,-0.0198],[228,-0.0175],[229,-0.0158],[230,-0.0143],[231,-0.0132],[232,-0.0123],[234,-0.0112],[240,-0.0098],[241,-0.0098],[242,-0.0094],[244,-0.0094],[245,-0.0089],[253,-0.0089],[254,-0.0154],[255,-0.0229],[256,-0.0326],[257,-0.0438],[258,-0.0561],[260,-0.0819],[261,-0.0943],[262,-0.1056],[263,-0.1175],[264,-0.1309],[265,-0.1449],[266,-0.1583],[267,-0.1702],[268,-0.1795],[269,-0.1868],[270,-0.1932],[271,-0.1987],[272,-0.2034],[273,-0.2071],[274,-0.2099],[275,-0.2115],[276,-0.2121],[279,-0.2113],[281,-0.2097],[283,-0.2072],[285,-0.2035],[287,-0.1986],[288,-0.1957],[290,-0.1887],[291,-0.1847],[292,-0.1802],[293,-0.1753],[294,-0.17],[295,-0.1613],[296,-0.1474],[297,-0.1304],[298,-0.1121],[299,-0.0944],[300,-0.0793],[301,-0.0673],[302,-0.058],[303,-0.0514],[304,-0.0464],[305,-0.0416],[306,-0.0371],[307,-0.0329],[308,-0.029],[309,-0.0255],[310,-0.0224],[311,-0.0197],[312,-0.0174],[313,-0.0155],[314,-0.0139],[315,-0.0126],[316,-0.0115],[317,-0.0106],[319,-0.0094],[322,-0.0086],[324,-0.0085]]
       }
       clavicle_r_ctrl.Rotation.Y {
        type=MovieSceneScriptingFloatChannel
        default=0
        thinned_from=325
        keys [[0,-10.4771],[1,-10.4604],[2,-10.4132],[3,-10.34],[5,-10.1334],[10,-9.4727],[11,-9.3268],[15,-8.5966],[16,-8.4481],[19,-8.0632],[21,-7.8449],[22,-7.7591],[23,-7.694],[27,-7.5866],[42,-7.4942],[43,-7.5038],[45,-7.5711],[49,-7.8149],[50,-7.9311],[52,-8.2782],[57,-9.2525],[58,-9.4228],[59,-9.5715],[60,-9.6929],[63,-9.9725],[66,-10.1811],[70,-10.3591],[73,-10.4255],[74,-10.4255],[75,-10.4476],[77,-10.4476],[85,-10.4665],[87,-10.3377],[89,-10.1315],[94,-9.4718],[95,-9.3261],[99,-8.5967],[100,-8.4481],[103,-8.0606],[105,-7.8402],[106,-7.7541],[107,-7.6897],[108,-7.651],[112,-7.5728],[113,-7.5728],[114,-7.5463],[127,-7.497],[128,-7.5056],[129,-7.5312],[130,-7.5731],[131,-7.6309],[132,-7.704],[133,-7.8165],[136,-8.3156],[141,-9.2628],[142,-9.4288],[143,-9.574],[145,-9.7927],[148,-10.0454],[151,-10.2319],[155,-10.3859],[157,-10.4255],[158,-10.4255],[159,-10.4486],[169,-10.4635],[170,-10.4128],[171,-10.3397],[173,-10.1332],[178,-9.4728],[179,-9.3269],[183,-8.5969],[184,-8.4481],[187,-8.0607],[189,-7.8404],[190,-7.754],[191,-7.6888],[192,-7.649],[200,-7.5208],[210,-7.4942],[211,-7.5047],[212,-7.5337],[214,-7.6336],[217,-7.8362],[218,-7.9525],[220,-8.2951],[225,-9.2508],[226,-9.4196],[227,-9.5688],[229,-9.7987],[232,-10.0673],[235,-10.2627],[238,-10.3849],[240,-10.4255],[244,-10.446],[253,-10.4645],[255,-10.3374],[257,-10.1312],[262,-9.4717],[263,-9.3261],[267,-8.5969],[268,-8.4481],[271,-8.0596],[273,-7.8383],[274,-7.7515],[275,-7.6862],[276,-7.6466],[284,-7.5204],[294,-7.4943],[296,-7.5221],[298,-7.6014],[300,-7.7259],[301,-7.8308],[303,-8.1466],[305,-8.5017],[309,-9.2533],[311,-9.5684],[314,-9.9051],[317,-10.1767],[320,-10.3724],[322,-10.4498],[324,-10.4701]]
       }
       clavicle_r_ctrl.Rotation.Z {
        type=MovieSceneScriptingFloatChannel
        default=0
        thinned_from=325
        keys [[0,-0.0793],[1,-0.0895],[2,-0.1183],[3,-0.1634],[4,-0.2222],[5,-0.2923],[6,-0.3712],[7,-0.4565],[8,-0.5456],[10,-0.7258],[11,-0.8275],[12,-0.9508],[13,-1.0872],[14,-1.2282],[15,-1.3655],[17,-1.6166],[18,-1.756],[19,-1.8995],[20,-2.0378],[21,-2.1616],[22,-2.2615],[23,-2.3283],[24,-2.3526],[26,-2.3462],[28,-2.3273],[30,-2.2968],[32,-2.2554],[34,-2.2038],[36,-2.1427],[39,-2.0349],[42,-1.9099],[43,-1.852],[44,-1.771],[45,-1.6735],[46,-1.5658],[48,-1.3458],[49,-1.2464],[50,-1.1573],[51,-1.0765],[52,-0.9979],[55,-0.748],[56,-0.667],[57,-0.5906],[58,-0.5209],[59,-0.4601],[60,-0.4101],[61,-0.3684],[62,-0.3306],[63,-0.2967],[64,-0.2663],[65,-0.2391],[66,-0.2151],[67,-0.1938],[68,-0.1752],[69,-0.1589],[70,-0.1447],[71,-0.1324],[72,-0.1217],[73,-0.1124],[74,-0.1124],[75,-0.0981],[77,-0.0981],[78,-0.0859],[85,-0.0859],[86,-0.1297],[87,-0.1855],[88,-0.2573],[89,-0.3416],[90,-0.4347],[91,-0.5329],[92,-0.6328],[93,-0.7306],[94,-0.8228],[95,-0.9183],[96,-1.0257],[97,-1.1409],[99,-1.3776],[100,-1.4908],[101,-1.6115],[102,-1.7484],[104,-2.031],[105,-2.1568],[106,-2.2589],[107,-2.3275],[108,-2.3526],[110,-2.3464],[112,-2.3283],[113,-2.3283],[114,-2.2993],[116,-2.2604],[119,-2.1853],[122,-2.0931],[126,-1.949],[127,-1.9099],[128,-1.8477],[129,-1.7504],[130,-1.6337],[131,-1.5129],[132,-1.4037],[133,-1.3058],[134,-1.2112],[136,-1.0379],[139,-0.7677],[140,-0.6806],[141,-0.5988],[142,-0.5245],[143,-0.4598],[144,-0.4071],[145,-0.3635],[146,-0.3246],[147,-0.29],[148,-0.2594],[149,-0.2324],[150,-0.2089],[151,-0.1884],[152,-0.1706],[153,-0.1552],[154,-0.1419],[155,-0.1303],[156,-0.1201],[157,-0.1111],[158,-0.1111],[159,-0.0968],[160,-0.0968],[161,-0.0877],[169,-0.0877],[170,-0.1285],[171,-0.1844],[172,-0.2562],[173,-0.3406],[174,-0.4337],[175,-0.532],[176,-0.6319],[177,-0.7298],[178,-0.822],[179,-0.9176],[180,-1.0251],[181,-1.1404],[183,-1.3772],[184,-1.4904],[185,-1.6112],[186,-1.7481],[188,-2.0308],[189,-2.1567],[190,-2.2589],[191,-2.3275],[192,-2.3526],[194,-2.3462],[196,-2.3273],[198,-2.2968],[200,-2.2554],[202,-2.2037],[204,-2.1426],[207,-2.0349],[210,-1.9099],[211,-1.8521],[212,-1.7712],[213,-1.6737],[214,-1.5661],[216,-1.3463],[217,-1.247],[218,-1.1578],[220,-0.9973],[223,-0.746],[224,-0.6644],[225,-0.5873],[226,-0.5167],[227,-0.4547],[228,-0.4032],[229,-0.3602],[230,-0.3219],[231,-0.288],[232,-0.2581],[233,-0.2319],[234,-0.209],[235,-0.189],[236,-0.1717],[237,-0.1566],[238,-0.1433],[239,-0.1316],[240,-0.1211],[241,-0.1211],[242,-0.1041],[244,-0.1041],[245,-0.0888],[253,-0.0888],[254,-0.1297],[255,-0.1855],[256,-0.2574],[257,-0.3416],[258,-0.4347],[259,-0.533],[260,-0.6328],[261,-0.7306],[262,-0.8228],[263,-0.9183],[264,-1.0257],[265,-1.1409],[267,-1.3777],[268,-1.4908],[269,-1.6115],[270,-1.7484],[272,-2.031],[273,-2.1568],[274,-2.2589],[275,-2.3275],[276,-2.3526],[278,-2.3458],[280,-2.326],[282,-2.2943],[284,-2.2517],[286,-2.1991],[289,-2.1039],[292,-1.992],[294,-1.9099],[295,-1.8509],[296,-1.7647],[297,-1.6607],[299,-1.4372],[300,-1.3364],[301,-1.2457],[302,-1.1592],[304,-0.9979],[307,-0.746],[308,-0.6645],[309,-0.5873],[310,-0.5162],[311,-0.4532],[312,-0.4],[313,-0.3538],[314,-0.3102],[315,-0.2697],[316,-0.2323],[317,-0.1985],[318,-0.1683],[319,-0.1422],[320,-0.1202],[321,-0.1027],[322,-0.0899],[323,-0.082],[324,-0.082]]
       }
       clavicle_r_ctrl.Scale.X {type=MovieSceneScriptingFloatChannel, default=1, const [1,325]}
       clavicle_r_ctrl.Scale.Y {type=MovieSceneScriptingFloatChannel, default=1, const [1,325]}
       clavicle_r_ctrl.Scale.Z {type=MovieSceneScriptingFloatChannel, default=1, const [1,325]}
       neck_01_ctrl.Location.X {type=MovieSceneScriptingFloatChannel, default=0, const [0,325]}
       neck_01_ctrl.Location.Y {type=MovieSceneScriptingFloatChannel, default=0, const [0,325]}
       neck_01_ctrl.Location.Z {type=MovieSceneScriptingFloatChannel, default=0, const [0,325]}
       neck_01_ctrl.Rotation.X {type=MovieSceneScriptingFloatChannel, default=0, const [0,325]}
       neck_01_ctrl.Rotation.Y {type=MovieSceneScriptingFloatChannel, default=0, const [0,325]}
       neck_01_ctrl.Rotation.Z {type=MovieSceneScriptingFloatChannel, default=0, const [0,325]}
       neck_01_ctrl.Scale.X {type=MovieSceneScriptingFloatChannel, default=1, const [1,325]}
       neck_01_ctrl.Scale.Y {type=MovieSceneScriptingFloatChannel, default=1, const [1,325]}
       neck_01_ctrl.Scale.Z {type=MovieSceneScriptingFloatChannel, default=1, const [1,325]}
       neck_02_ctrl.Location.X {type=MovieSceneScriptingFloatChannel, default=0, thinned_from=325, keys [[0,0],[324,0]]}
       neck_02_ctrl.Location.Y {
        type=MovieSceneScriptingFloatChannel
        default=0
        thinned_from=325
        keys [[0,0.1981],[1,0.1983],[3,0.1999],[6,0.2044],[11,0.2137],[13,0.2166],[15,0.2185],[16,0.2188],[17,0.2187],[18,0.2181],[19,0.2169],[20,0.215],[21,0.2125],[22,0.2093],[24,0.2014],[25,0.1988],[26,0.1971],[28,0.1954],[30,0.1949],[33,0.1955],[37,0.1986],[48,0.2128],[49,0.2137],[50,0.2138],[51,0.2133],[52,0.2122],[54,0.2086],[57,0.2008],[61,0.1897],[63,0.1853],[64,0.1837],[65,0.1825],[66,0.1819],[74,0.185],[77,0.1872],[80,0.1907],[82,0.194],[90,0.2124],[93,0.2181],[95,0.2208],[97,0.2224],[99,0.2225],[101,0.2211],[102,0.2197],[103,0.2178],[104,0.2154],[105,0.2124],[106,0.2089],[107,0.2047],[108,0.201],[109,0.1985],[110,0.197],[112,0.1955],[114,0.1951],[116,0.1955],[119,0.1976],[132,0.2153],[133,0.2162],[134,0.2162],[135,0.2154],[136,0.214],[137,0.212],[139,0.2069],[144,0.1913],[146,0.1861],[147,0.1842],[148,0.1828],[149,0.182],[150,0.182],[158,0.1854],[161,0.1876],[164,0.1911],[166,0.1943],[175,0.2142],[178,0.2192],[180,0.2214],[182,0.2223],[184,0.2217],[185,0.2208],[186,0.2194],[187,0.2175],[188,0.2151],[189,0.2121],[190,0.2086],[191,0.2044],[192,0.2007],[193,0.1984],[194,0.197],[196,0.1955],[198,0.1951],[200,0.1955],[203,0.1976],[216,0.2153],[217,0.2162],[218,0.2162],[219,0.2155],[220,0.2141],[221,0.2121],[223,0.2069],[228,0.1913],[230,0.1861],[231,0.1842],[232,0.1828],[233,0.182],[234,0.182],[242,0.1854],[245,0.1876],[248,0.1911],[250,0.1944],[258,0.2123],[261,0.2176],[263,0.2201],[265,0.2215],[267,0.2216],[269,0.22],[270,0.2186],[271,0.2166],[272,0.2142],[273,0.2112],[274,0.2077],[275,0.2035],[276,0.2],[277,0.1979],[279,0.1961],[281,0.1951],[284,0.195],[287,0.1965],[300,0.2108],[301,0.2115],[302,0.2114],[303,0.2105],[304,0.209],[306,0.2047],[311,0.1902],[313,0.1852],[314,0.1833],[315,0.182],[316,0.1812],[317,0.1812],[318,0.182],[320,0.1853],[321,0.1876],[322,0.1904],[323,0.1938],[324,0.1981]]
       }
       neck_02_ctrl.Location.Z {
        type=MovieSceneScriptingFloatChannel
        default=0
        thinned_from=325
        keys [[0,-0.6721],[1,-0.6742],[2,-0.6802],[3,-0.69],[4,-0.7032],[5,-0.7194],[6,-0.7384],[7,-0.76],[8,-0.7837],[10,-0.8366],[18,-1.0767],[20,-1.1319],[22,-1.1806],[23,-1.2017],[24,-1.2181],[25,-1.2283],[26,-1.2336],[27,-1.235],[29,-1.231],[31,-1.2201],[33,-1.203],[35,-1.1805],[38,-1.1382],[42,-1.0698],[54,-0.8357],[57,-0.7826],[59,-0.7514],[61,-0.7247],[62,-0.7134],[63,-0.7034],[64,-0.695],[65,-0.6884],[66,-0.6835],[72,-0.6733],[84,-0.6724],[85,-0.675],[86,-0.6815],[87,-0.6916],[88,-0.7051],[89,-0.7216],[90,-0.7408],[91,-0.7625],[92,-0.7863],[94,-0.8393],[102,-1.079],[104,-1.1342],[106,-1.1827],[107,-1.2037],[108,-1.2198],[109,-1.2294],[110,-1.2339],[112,-1.2339],[114,-1.2259],[116,-1.2111],[118,-1.1903],[121,-1.1495],[124,-1.0995],[138,-0.8227],[140,-0.7866],[142,-0.7542],[144,-0.7263],[145,-0.7144],[146,-0.7041],[147,-0.6954],[148,-0.6885],[149,-0.6835],[155,-0.6735],[168,-0.6722],[169,-0.6748],[170,-0.6813],[171,-0.6915],[172,-0.705],[173,-0.7215],[174,-0.7408],[175,-0.7626],[176,-0.7865],[178,-0.8397],[186,-1.0801],[188,-1.1354],[190,-1.184],[191,-1.205],[192,-1.2209],[193,-1.23],[194,-1.2341],[196,-1.2339],[198,-1.2259],[200,-1.2111],[202,-1.1903],[205,-1.1494],[208,-1.0994],[222,-0.8225],[224,-0.7864],[226,-0.7539],[228,-0.7261],[229,-0.7143],[230,-0.7039],[231,-0.6953],[232,-0.6884],[233,-0.6834],[239,-0.6736],[252,-0.6725],[253,-0.6751],[254,-0.6817],[255,-0.692],[256,-0.7057],[257,-0.7224],[258,-0.7419],[259,-0.7639],[260,-0.7881],[262,-0.8418],[270,-1.0841],[272,-1.1395],[274,-1.1881],[275,-1.2091],[276,-1.2242],[277,-1.232],[278,-1.2347],[280,-1.234],[282,-1.2267],[284,-1.2133],[286,-1.1942],[289,-1.1568],[292,-1.1106],[298,-1.0013],[307,-0.8259],[310,-0.7735],[312,-0.7427],[314,-0.7164],[315,-0.7051],[316,-0.6953],[317,-0.6871],[318,-0.6805],[319,-0.6757],[320,-0.6728],[322,-0.6708],[324,-0.6721]]
       }
       neck_02_ctrl.Rotation.X {
        type=MovieSceneScriptingFloatChannel
        default=0
        thinned_from=325
        keys [[0,2.3241],[61,2.2733],[108,2.3113],[144,2.2719],[192,2.3113],[227,2.2718],[276,2.3113],[310,2.2708],[323,2.3234],[324,2.3241]]
       }
       neck_02_ctrl.Rotation.Y {type=MovieSceneScriptingFloatChannel, default=0, thinned_from=325, keys [[0,0],[324,0]]}
       neck_02_ctrl.Rotation.Z {type=MovieSceneScriptingFloatChannel, default=0, thinned_from=325, keys [[0,0],[324,0]]}
       neck_02_ctrl.Scale.X {type=MovieSceneScriptingFloatChannel, default=1, const [1,325]}
       neck_02_ctrl.Scale.Y {type=MovieSceneScriptingFloatChannel, default=1, const [1,325]}
       neck_02_ctrl.Scale.Z {type=MovieSceneScriptingFloatChannel, default=1, const [1,325]}
       head_ik_ctrl.Location.X {
        type=MovieSceneScriptingFloatChannel
        default=0
        thinned_from=325
        keys [[0,0],[28,0],[29,0.0001],[30,0.0009],[31,0.003],[32,0.0089],[33,0.0178],[34,0.0307],[35,0.0415],[36,0.0466],[37,0.0507],[38,0.0527],[39,0.0534],[40,0.0537],[41,0.0536],[44,0.0523],[58,0.0425],[62,0.0404],[65,0.0394],[82,0.0396],[86,0.0403],[87,0.0398],[88,0.0385],[89,0.034],[90,0.0181],[91,0.0107],[92,0.0055],[93,0.0024],[94,0.0007],[95,-0.0002],[97,-0.0008],[185,-0.0012],[187,-0.0017],[188,-0.0022],[189,-0.0033],[190,-0.0053],[191,-0.0079],[192,-0.0109],[193,-0.0164],[194,-0.0217],[195,-0.0241],[196,-0.0259],[197,-0.027],[198,-0.0278],[212,-0.033],[216,-0.0333],[232,-0.0316],[248,-0.0332],[249,-0.0332],[250,-0.0328],[251,-0.0318],[252,-0.03],[253,-0.0256],[255,-0.0115],[256,-0.0068],[257,-0.0043],[258,-0.0026],[259,-0.0014],[260,-0.0006],[261,-0.0002],[324,0]]
       }
       head_ik_ctrl.Location.Y {
        type=MovieSceneScriptingFloatChannel
        default=0
        thinned_from=325
        keys [[0,0.1727],[1,0.1725],[3,0.171],[7,0.1667],[8,0.166],[9,0.166],[10,0.1668],[11,0.1683],[15,0.1764],[16,0.1776],[17,0.1779],[21,0.1777],[22,0.177],[24,0.1741],[26,0.1739],[29,0.1725],[31,0.1728],[33,0.1741],[38,0.1786],[43,0.1853],[44,0.186],[45,0.1857],[46,0.1846],[48,0.1813],[49,0.1791],[53,0.1677],[54,0.1655],[55,0.1639],[56,0.1631],[64,0.1592],[66,0.1588],[74,0.1602],[77,0.1619],[79,0.1636],[81,0.166],[83,0.1693],[85,0.173],[86,0.174],[87,0.1743],[90,0.1741],[92,0.1733],[93,0.1732],[94,0.1737],[96,0.1759],[98,0.1787],[99,0.1797],[100,0.1802],[101,0.1796],[103,0.1781],[105,0.1771],[106,0.1762],[107,0.1745],[108,0.1734],[110,0.1738],[111,0.1735],[113,0.1726],[115,0.1732],[117,0.1749],[120,0.179],[127,0.191],[128,0.192],[129,0.1919],[130,0.1911],[132,0.1882],[133,0.1861],[137,0.1743],[138,0.1717],[139,0.1698],[140,0.1686],[147,0.1624],[149,0.1614],[156,0.1617],[160,0.1632],[163,0.1656],[165,0.168],[167,0.1711],[169,0.1747],[170,0.1757],[171,0.1761],[173,0.1757],[176,0.1745],[177,0.1749],[178,0.1763],[179,0.1784],[181,0.1834],[182,0.1856],[183,0.1868],[184,0.1867],[188,0.1829],[189,0.1815],[190,0.1795],[191,0.1766],[192,0.1744],[193,0.1733],[196,0.1711],[197,0.1707],[198,0.1707],[200,0.1719],[202,0.1741],[211,0.1885],[212,0.1891],[213,0.1886],[214,0.187],[216,0.1825],[217,0.1797],[220,0.1699],[221,0.1672],[222,0.1652],[223,0.1643],[231,0.1587],[233,0.158],[241,0.159],[244,0.1604],[247,0.1629],[249,0.1654],[251,0.1689],[253,0.1733],[254,0.1749],[255,0.1756],[256,0.1756],[260,0.1729],[261,0.1727],[262,0.173],[264,0.1748],[266,0.1771],[267,0.1779],[268,0.1782],[271,0.1758],[273,0.1748],[274,0.1739],[275,0.1723],[276,0.1716],[278,0.1735],[279,0.1735],[281,0.1725],[283,0.1726],[285,0.1737],[295,0.1848],[296,0.1853],[297,0.1849],[298,0.1839],[300,0.1807],[301,0.1785],[302,0.1756],[305,0.1657],[306,0.1626],[307,0.1601],[308,0.1583],[310,0.1555],[312,0.1537],[313,0.1533],[314,0.1532],[315,0.1535],[316,0.1543],[317,0.1556],[320,0.1613],[321,0.1635],[322,0.166],[323,0.169],[324,0.1727]]
       }
       head_ik_ctrl.Location.Z {
        type=MovieSceneScriptingFloatChannel
        default=0
        thinned_from=325
        keys [[0,-0.8146],[1,-0.8166],[2,-0.8225],[3,-0.8318],[4,-0.8444],[5,-0.86],[6,-0.8781],[7,-0.8986],[9,-0.9452],[12,-1.025],[18,-1.1949],[20,-1.2461],[22,-1.2905],[23,-1.3095],[24,-1.3242],[25,-1.3331],[26,-1.3376],[28,-1.3373],[30,-1.3296],[32,-1.3157],[35,-1.2844],[39,-1.2285],[57,-0.913],[59,-0.8823],[61,-0.8559],[63,-0.835],[64,-0.8269],[65,-0.8206],[66,-0.8161],[75,-0.8083],[84,-0.8143],[85,-0.8174],[86,-0.8242],[87,-0.8344],[88,-0.8477],[89,-0.8638],[90,-0.8826],[91,-0.9033],[93,-0.95],[96,-1.0293],[102,-1.1976],[104,-1.2483],[106,-1.2925],[107,-1.3115],[108,-1.3258],[109,-1.3341],[110,-1.3379],[112,-1.3372],[114,-1.3292],[116,-1.3147],[119,-1.2826],[122,-1.2404],[140,-0.9179],[142,-0.8858],[144,-0.8581],[146,-0.8361],[147,-0.8276],[148,-0.821],[149,-0.8164],[159,-0.8087],[168,-0.8148],[169,-0.818],[170,-0.8247],[171,-0.8349],[172,-0.848],[173,-0.864],[174,-0.8825],[175,-0.9031],[177,-0.9498],[180,-1.0288],[186,-1.1975],[188,-1.2486],[190,-1.2932],[191,-1.3123],[192,-1.3266],[193,-1.3346],[194,-1.338],[196,-1.3372],[198,-1.3292],[200,-1.3147],[203,-1.2825],[206,-1.2402],[224,-0.918],[226,-0.8859],[228,-0.8582],[230,-0.8363],[231,-0.8278],[232,-0.8211],[233,-0.8165],[243,-0.8088],[252,-0.815],[253,-0.8182],[254,-0.8251],[255,-0.8354],[256,-0.8488],[257,-0.865],[258,-0.8837],[259,-0.9046],[261,-0.9518],[264,-1.0322],[270,-1.2022],[272,-1.2533],[274,-1.2976],[275,-1.3165],[276,-1.33],[277,-1.3366],[279,-1.3386],[281,-1.3344],[283,-1.3241],[286,-1.2983],[289,-1.2623],[293,-1.2024],[309,-0.9199],[311,-0.8881],[313,-0.8604],[315,-0.8379],[316,-0.8289],[317,-0.8216],[318,-0.8161],[319,-0.8125],[321,-0.8101],[324,-0.8146]]
       }
       head_ik_ctrl.Rotation.X {
        type=MovieSceneScriptingFloatChannel
        default=0
        thinned_from=325
        keys [[0,9.8055],[1,9.7918],[3,9.6995],[4,9.6203],[6,9.4038],[8,9.1179],[10,8.7667],[16,7.5686],[18,7.1976],[22,6.3558],[23,6.1544],[24,5.9698],[25,5.8086],[26,5.6769],[27,5.5812],[28,5.5153],[29,5.5174],[30,5.6476],[31,5.9719],[32,7.0188],[33,8.344],[34,8.956],[35,9.1243],[36,9.0672],[37,8.96],[39,8.6633],[40,8.5425],[41,8.4722],[42,8.4729],[43,8.5248],[44,8.6066],[45,8.7262],[46,8.8777],[47,9.055],[53,10.3173],[54,10.5031],[55,10.6655],[62,11.4629],[68,11.9771],[73,12.2929],[78,12.5085],[80,12.6397],[82,12.8328],[84,13.1116],[86,13.4618],[88,13.8762],[89,13.9304],[90,13.5742],[91,13.0732],[92,12.5352],[93,12.0227],[94,11.5751],[95,11.1877],[96,10.8579],[97,10.5832],[98,10.3562],[100,9.9523],[102,9.5759],[107,8.5179],[108,8.331],[109,8.168],[110,8.045],[111,7.9402],[115,7.6688],[118,7.5238],[120,7.4696],[122,7.4587],[124,7.4988],[125,7.5405],[126,7.5978],[127,7.6719],[128,7.7758],[129,7.9177],[130,8.0914],[131,8.2906],[136,9.4478],[138,9.8685],[139,10.0448],[144,10.6841],[147,11.0048],[150,11.2745],[154,11.5759],[161,12.0021],[166,12.1842],[168,12.2216],[169,12.2185],[170,12.1841],[171,12.121],[172,12.0314],[174,11.7819],[176,11.4537],[178,11.0518],[180,10.6084],[181,10.4163],[182,10.2825],[183,10.2079],[184,10.1993],[185,10.2428],[186,10.5019],[187,10.9172],[188,11.1147],[189,11.0178],[190,10.8663],[191,10.6576],[192,10.3343],[195,9.2324],[196,8.9029],[197,8.6594],[198,8.5035],[199,8.3931],[200,8.318],[202,8.2314],[205,8.1578],[207,8.1461],[209,8.182],[211,8.2746],[212,8.3553],[213,8.47],[214,8.6126],[216,8.9574],[220,9.7153],[221,9.8834],[222,10.0307],[229,10.7436],[234,11.1368],[246,11.7984],[247,12.085],[250,13.7843],[251,14.0702],[252,13.732],[253,13.1823],[255,12.154],[258,10.435],[259,9.8962],[260,9.42],[261,9.0344],[262,8.7697],[268,7.5458],[270,7.1658],[274,6.3143],[275,6.1122],[276,5.9276],[277,5.767],[278,5.6369],[279,5.5439],[282,5.3416],[284,5.2304],[286,5.1466],[288,5.0976],[289,5.0884],[290,5.0907],[291,5.1052],[292,5.1331],[293,5.1751],[294,5.2322],[295,5.3054],[296,5.4049],[297,5.5372],[298,5.6974],[299,5.8805],[300,6.0816],[305,7.1838],[306,7.3889],[307,7.5775],[312,8.3442],[315,8.7504],[319,9.2181],[322,9.6563],[323,9.7587],[324,9.8041]]
       }
       head_ik_ctrl.Rotation.Y {
        type=MovieSceneScriptingFloatChannel
        default=0
        thinned_from=325
        keys [[0,-5.1409],[3,-5.1184],[6,-5.0565],[10,-4.925],[16,-4.6913],[23,-4.4749],[25,-4.433],[26,-4.4232],[28,-4.4274],[29,-4.3848],[30,-4.2407],[31,-3.9386],[32,-2.9477],[33,-1.3321],[34,1.006],[35,3.0054],[36,3.3195],[37,3.565],[38,3.7691],[39,3.9333],[40,4.0458],[41,4.1116],[42,4.1347],[44,4.155],[46,4.2079],[53,4.4579],[55,4.5101],[61,4.5905],[62,4.6099],[63,4.6117],[64,4.629],[65,4.6308],[66,4.6461],[67,4.648],[69,4.6736],[70,4.67],[72,4.6904],[73,4.6875],[76,4.7099],[84,4.7209],[85,4.5606],[86,4.0534],[87,3.1582],[88,1.6954],[89,-0.1622],[90,-2.0007],[91,-3.4762],[92,-4.3853],[93,-4.6647],[94,-4.8218],[95,-4.8901],[96,-4.9004],[108,-4.7146],[110,-4.6942],[127,-4.757],[139,-4.9493],[144,-4.9898],[147,-5.0102],[148,-5.006],[150,-5.0226],[151,-5.0194],[153,-5.0329],[170,-5.0521],[176,-4.9839],[182,-4.8654],[183,-4.8358],[184,-4.7858],[185,-4.8001],[186,-5.0077],[187,-5.3608],[188,-5.9793],[189,-7.8098],[190,-10.2696],[191,-11.3584],[192,-10.9642],[193,-10.4665],[194,-10.057],[195,-9.766],[196,-9.5694],[197,-9.4303],[198,-9.3262],[199,-9.2555],[200,-9.2164],[202,-9.2175],[206,-9.3093],[210,-9.4838],[212,-9.6165],[213,-9.7191],[215,-9.9834],[219,-10.5934],[221,-10.8509],[222,-10.9485],[232,-11.4106],[244,-11.6853],[249,-11.6667],[250,-11.4228],[251,-10.4509],[252,-8.1068],[253,-6.2922],[254,-5.9068],[255,-5.5739],[256,-5.3371],[257,-5.1745],[258,-5.0774],[259,-5.0238],[268,-4.6919],[275,-4.4751],[277,-4.4327],[279,-4.4164],[289,-4.4535],[295,-4.5274],[297,-4.5721],[306,-4.8902],[315,-5.0698],[317,-5.1065],[324,-5.1406]]
       }
       head_ik_ctrl.Rotation.Z {
        type=MovieSceneScriptingFloatChannel
        default=0
        thinned_from=325
        keys [[0,11.2508],[25,11.5385],[26,11.4974],[27,11.3754],[28,11.1366],[29,10.7535],[30,10.1992],[31,9.4464],[32,7.1332],[33,3.5671],[34,-0.6536],[35,-4.121],[36,-6.4166],[37,-8.3881],[38,-9.6383],[39,-10.3185],[40,-10.855],[41,-11.2635],[42,-11.5593],[43,-11.7523],[44,-11.8524],[45,-11.8751],[84,-11.6594],[85,-11.517],[86,-10.9224],[87,-9.6243],[88,-6.6363],[89,-2.8788],[90,0.2211],[91,1.7257],[92,3.0967],[93,3.996],[94,4.4105],[95,4.7571],[96,5.04],[97,5.2642],[98,5.4242],[99,5.5219],[100,5.5735],[110,5.7003],[128,5.6428],[139,5.4819],[144,5.4444],[147,5.4253],[148,5.4297],[150,5.4139],[170,5.3902],[177,5.4731],[181,5.5481],[182,5.6783],[183,6.0601],[184,6.7326],[185,7.7426],[186,9.139],[187,11.9049],[188,16.8395],[189,24.0322],[190,32.1466],[191,38.8416],[192,42.1948],[193,43.1981],[197,45.1084],[200,46.2477],[202,46.8128],[204,47.1635],[244,46.837],[245,46.7537],[246,46.4434],[247,45.8161],[248,44.8068],[249,43.3557],[250,41.4178],[251,35.7084],[252,28.3994],[253,21.9351],[254,17.1092],[255,15.2053],[256,13.716],[257,12.612],[258,11.8644],[259,11.4436],[260,11.3196],[279,11.5528],[324,11.2509]]
       }
       head_ik_ctrl.Scale.X {type=MovieSceneScriptingFloatChannel, default=1, const [1,325]}
       head_ik_ctrl.Scale.Y {type=MovieSceneScriptingFloatChannel, default=1, const [1,325]}
       head_ik_ctrl.Scale.Z {type=MovieSceneScriptingFloatChannel, default=1, const [1,325]}
       chest_ctrl.Location.X {
        type=MovieSceneScriptingFloatChannel
        default=0
        thinned_from=325
        keys [[0,0],[29,-0.0001],[30,-0.0006],[31,-0.0021],[32,-0.0063],[33,-0.0126],[34,-0.0218],[35,-0.0289],[36,-0.0308],[37,-0.0315],[38,-0.0316],[44,-0.0292],[54,-0.0226],[65,-0.019],[73,-0.018],[81,-0.0189],[86,-0.0202],[87,-0.0199],[88,-0.0189],[89,-0.0161],[90,-0.0054],[91,-0.0026],[92,-0.0012],[93,-0.0004],[94,-0.0002],[186,-0.0005],[188,-0.0008],[189,-0.0013],[190,-0.002],[192,-0.0042],[194,-0.0082],[195,-0.0091],[196,-0.0098],[198,-0.0105],[212,-0.0125],[219,-0.0143],[239,-0.0135],[249,-0.0145],[250,-0.0144],[251,-0.014],[252,-0.0133],[253,-0.0115],[255,-0.0053],[256,-0.0032],[257,-0.0021],[258,-0.0013],[259,-0.0007],[261,-0.0001],[324,0]]
       }
       chest_ctrl.Location.Y {
        type=MovieSceneScriptingFloatChannel
        default=0
        thinned_from=325
        keys [[0,1.4408],[1,1.4392],[3,1.4267],[5,1.404],[7,1.373],[10,1.315],[14,1.2298],[16,1.1925],[22,1.1037],[24,1.0808],[25,1.0726],[26,1.0673],[28,1.0655],[32,1.0716],[35,1.0836],[38,1.1047],[40,1.1238],[42,1.1478],[44,1.1773],[50,1.2813],[52,1.3115],[54,1.3354],[56,1.3508],[72,1.41],[84,1.4417],[85,1.4422],[86,1.4392],[87,1.433],[89,1.412],[91,1.3814],[93,1.3439],[99,1.2156],[101,1.181],[106,1.1049],[108,1.0812],[109,1.0728],[110,1.0674],[112,1.0655],[116,1.0716],[119,1.0834],[122,1.1036],[124,1.1225],[126,1.1465],[128,1.1759],[134,1.28],[136,1.3103],[138,1.3343],[140,1.3498],[156,1.4094],[168,1.4412],[169,1.4413],[170,1.4375],[171,1.43],[172,1.4193],[174,1.3897],[176,1.3517],[182,1.2146],[183,1.1951],[188,1.1211],[190,1.0964],[192,1.0774],[193,1.0707],[194,1.0664],[198,1.0664],[201,1.0731],[204,1.087],[206,1.1013],[208,1.1204],[210,1.145],[212,1.1757],[218,1.2885],[220,1.3196],[222,1.3418],[239,1.4069],[252,1.4402],[253,1.4408],[254,1.4381],[255,1.432],[257,1.4113],[259,1.3811],[261,1.3435],[266,1.2356],[268,1.1971],[274,1.1049],[276,1.0812],[277,1.0728],[278,1.0674],[280,1.0655],[284,1.0716],[287,1.0834],[290,1.1036],[292,1.1225],[294,1.1465],[304,1.3027],[307,1.3398],[314,1.4004],[319,1.4297],[323,1.4404],[324,1.4405]]
       }
       chest_ctrl.Location.Z {
        type=MovieSceneScriptingFloatChannel
        default=0
        thinned_from=325
        keys [[0,0.1284],[1,0.1305],[2,0.1366],[3,0.1463],[4,0.1594],[5,0.1754],[6,0.1941],[7,0.2152],[8,0.2382],[9,0.2628],[10,0.2888],[11,0.3159],[13,0.3726],[17,0.4885],[18,0.516],[19,0.5418],[20,0.5655],[21,0.5871],[22,0.6061],[23,0.6223],[24,0.6354],[25,0.6451],[26,0.6512],[27,0.6533],[30,0.6517],[33,0.6453],[35,0.6376],[37,0.6257],[38,0.6181],[39,0.6091],[40,0.5988],[41,0.5871],[42,0.5739],[43,0.5592],[44,0.5413],[45,0.5193],[46,0.494],[47,0.4661],[48,0.4363],[51,0.3428],[52,0.3126],[53,0.2839],[54,0.2576],[55,0.2341],[56,0.2141],[58,0.1786],[59,0.1616],[60,0.1451],[61,0.1292],[62,0.1142],[63,0.1],[64,0.0867],[65,0.0745],[66,0.0634],[67,0.0536],[68,0.0451],[69,0.0381],[70,0.0326],[71,0.0287],[72,0.0266],[73,0.0264],[74,0.0278],[75,0.0309],[76,0.0356],[77,0.042],[78,0.0499],[79,0.0595],[80,0.0707],[81,0.0834],[82,0.0978],[83,0.1137],[84,0.1312],[85,0.1503],[86,0.1707],[87,0.1924],[88,0.215],[90,0.2621],[97,0.4375],[99,0.4859],[101,0.5318],[102,0.5532],[103,0.5727],[104,0.5905],[105,0.6063],[106,0.6201],[107,0.6317],[108,0.6409],[109,0.6477],[110,0.6519],[111,0.6533],[114,0.6517],[116,0.648],[118,0.6415],[120,0.6315],[122,0.6171],[123,0.6082],[124,0.5979],[125,0.5862],[126,0.573],[127,0.5582],[128,0.5403],[129,0.5183],[130,0.493],[131,0.4651],[132,0.4352],[135,0.3417],[136,0.3115],[137,0.283],[138,0.2566],[139,0.233],[140,0.213],[142,0.1776],[143,0.1605],[144,0.144],[145,0.1281],[146,0.1131],[147,0.0988],[148,0.0856],[149,0.0733],[150,0.0623],[151,0.0524],[152,0.044],[153,0.0369],[154,0.0314],[155,0.0276],[156,0.0254],[157,0.0252],[158,0.0267],[159,0.0298],[160,0.0347],[161,0.0412],[162,0.0494],[163,0.0591],[164,0.0705],[165,0.0834],[166,0.0979],[167,0.1139],[168,0.1314],[169,0.1504],[170,0.1708],[171,0.1923],[172,0.2147],[173,0.238],[174,0.2619],[182,0.4592],[184,0.5068],[185,0.5296],[186,0.5509],[187,0.5706],[188,0.5886],[189,0.6048],[190,0.6189],[191,0.6308],[192,0.6404],[193,0.6474],[194,0.6518],[195,0.6533],[198,0.6519],[200,0.6486],[202,0.6424],[204,0.6327],[206,0.6184],[207,0.6094],[208,0.5989],[209,0.587],[210,0.5734],[211,0.5581],[212,0.5391],[213,0.515],[214,0.4869],[215,0.4558],[216,0.4226],[218,0.354],[219,0.3205],[220,0.2887],[221,0.2597],[222,0.234],[223,0.2128],[224,0.1943],[225,0.1763],[226,0.1589],[227,0.1422],[228,0.1262],[229,0.1111],[230,0.0969],[231,0.0838],[232,0.0717],[233,0.0608],[234,0.0512],[235,0.0429],[236,0.0361],[237,0.0308],[238,0.027],[239,0.025],[240,0.0248],[241,0.026],[242,0.0287],[243,0.0327],[244,0.038],[245,0.0448],[246,0.053],[247,0.0626],[248,0.0737],[249,0.0862],[250,0.1002],[251,0.1158],[252,0.1328],[253,0.1515],[254,0.1716],[255,0.193],[256,0.2155],[257,0.2388],[258,0.2627],[265,0.4373],[267,0.4857],[269,0.5317],[270,0.553],[271,0.5726],[272,0.5904],[273,0.6062],[274,0.62],[275,0.6316],[276,0.6409],[277,0.6477],[278,0.6519],[279,0.6533],[282,0.6517],[284,0.648],[286,0.6415],[288,0.6315],[290,0.6171],[291,0.6082],[292,0.5979],[293,0.5862],[294,0.573],[295,0.5582],[296,0.5401],[297,0.5175],[298,0.4912],[299,0.4623],[300,0.4313],[303,0.3354],[304,0.3051],[305,0.277],[306,0.2517],[307,0.2301],[308,0.213],[309,0.1992],[310,0.1869],[311,0.1762],[312,0.1668],[313,0.1588],[314,0.1519],[315,0.1462],[316,0.1414],[317,0.1376],[318,0.1346],[319,0.1323],[320,0.1306],[321,0.1295],[322,0.1288],[323,0.1285],[324,0.1284]]
       }
       chest_ctrl.Rotation.X {
        type=MovieSceneScriptingFloatChannel
        default=0
        thinned_from=325
        keys [[0,8.9375],[1,8.9256],[2,8.8907],[3,8.8337],[4,8.7557],[5,8.6575],[6,8.5403],[7,8.405],[8,8.2526],[9,8.0812],[11,7.6913],[14,7.0741],[15,6.8825],[16,6.7076],[21,5.9704],[22,5.8355],[23,5.7144],[24,5.6121],[25,5.5331],[26,5.4822],[27,5.4642],[30,5.4794],[33,5.5326],[35,5.5959],[37,5.6883],[39,5.8087],[41,5.961],[43,6.1495],[44,6.2657],[45,6.4035],[47,6.7258],[51,7.4341],[53,7.7488],[54,7.8807],[55,7.9896],[56,8.0709],[63,8.4636],[71,8.7571],[81,8.9283],[84,8.9422],[85,8.9321],[86,8.8987],[87,8.8429],[88,8.7656],[89,8.6672],[91,8.4113],[92,8.2583],[93,8.0865],[95,7.6975],[98,7.0824],[99,6.8913],[100,6.7164],[105,5.9748],[106,5.8387],[107,5.7167],[108,5.6134],[109,5.5337],[110,5.4824],[111,5.4642],[114,5.4793],[117,5.5327],[119,5.5953],[121,5.6839],[123,5.802],[125,5.9533],[127,6.1416],[128,6.2578],[129,6.3956],[131,6.7181],[135,7.4269],[137,7.7419],[138,7.8738],[139,7.9828],[140,8.0641],[147,8.4576],[155,8.7518],[165,8.9235],[168,8.9375],[169,8.9256],[170,8.887],[171,8.8235],[172,8.7364],[173,8.6273],[174,8.4977],[175,8.349],[176,8.1827],[177,7.9946],[181,7.114],[182,6.9045],[183,6.7165],[184,6.5586],[188,6.0033],[190,5.7637],[191,5.6638],[192,5.5811],[193,5.5186],[194,5.4792],[195,5.4656],[199,5.4912],[202,5.5561],[204,5.6291],[206,5.731],[208,5.8662],[210,6.0389],[211,6.1408],[212,6.2635],[213,6.4132],[214,6.5837],[219,7.5375],[220,7.7059],[221,7.8525],[222,7.9712],[223,8.0558],[230,8.4438],[238,8.7322],[249,8.9116],[252,8.9267],[253,8.9191],[254,8.8887],[255,8.8347],[256,8.758],[257,8.6601],[258,8.5423],[259,8.4057],[260,8.251],[262,7.8888],[266,7.0786],[267,6.8891],[268,6.7156],[273,5.9753],[274,5.8392],[275,5.717],[276,5.6137],[277,5.5339],[278,5.4824],[279,5.4642],[282,5.4793],[285,5.5327],[287,5.5953],[289,5.6839],[291,5.802],[293,5.9533],[295,6.1416],[296,6.2557],[297,6.3876],[299,6.6919],[304,7.5325],[306,7.8298],[307,7.9566],[310,8.2501],[314,8.5677],[317,8.7494],[320,8.8738],[323,8.9334],[324,8.936]]
       }
       chest_ctrl.Rotation.Y {
        type=MovieSceneScriptingFloatChannel
        default=0
        thinned_from=325
        keys [[0,0],[28,0],[29,0.0013],[30,0.0104],[31,0.0352],[32,0.1043],[33,0.2108],[34,0.3644],[35,0.486],[36,0.5231],[37,0.5432],[38,0.5509],[39,0.5527],[44,0.5559],[55,0.5733],[84,0.5868],[85,0.5866],[86,0.5835],[87,0.5658],[88,0.5329],[89,0.4526],[90,0.1935],[91,0.1043],[92,0.0496],[93,0.0199],[94,0.0058],[95,0.001],[96,0.0008],[99,0.0022],[109,0.004],[126,0.0032],[140,0.0002],[168,-0.0014],[172,-0.0008],[184,0.0028],[185,0.0033],[186,0.0041],[187,0.0054],[188,0.0074],[189,0.0119],[190,0.0204],[191,0.0324],[192,0.047],[193,0.073],[194,0.0984],[195,0.1103],[196,0.1186],[197,0.124],[198,0.127],[199,0.1284],[200,0.129],[201,0.1289],[202,0.1281],[203,0.1268],[204,0.1249],[205,0.1225],[206,0.1196],[207,0.1163],[208,0.1124],[209,0.108],[210,0.1029],[211,0.0972],[212,0.0906],[213,0.0831],[214,0.0749],[215,0.0661],[217,0.0481],[218,0.0393],[219,0.0309],[220,0.0233],[221,0.0165],[222,0.011],[223,0.0069],[226,-0.0025],[229,-0.011],[231,-0.0162],[234,-0.0233],[236,-0.0275],[239,-0.0329],[240,-0.0341],[241,-0.036],[243,-0.0387],[245,-0.0409],[247,-0.0426],[248,-0.0433],[249,-0.0435],[250,-0.0431],[251,-0.0417],[252,-0.0391],[253,-0.0323],[254,-0.0216],[255,-0.0116],[256,-0.0055],[257,-0.0025],[258,-0.0008],[259,0],[260,0.0002],[324,0]]
       }
       chest_ctrl.Rotation.Z {
        type=MovieSceneScriptingFloatChannel
        default=0
        thinned_from=325
        keys [[0,0],[28,0],[29,-0.0001],[30,-0.0008],[31,-0.0028],[32,-0.0081],[33,-0.0163],[34,-0.048],[35,-0.1991],[36,-0.5772],[37,-1.0172],[38,-1.3161],[39,-1.4766],[40,-1.6081],[41,-1.6918],[42,-1.737],[43,-1.7674],[44,-1.7856],[45,-1.7946],[87,-1.7748],[88,-1.7569],[89,-1.6311],[90,-1.2883],[91,-0.832],[92,-0.4515],[93,-0.2151],[94,-0.0542],[95,0.0439],[96,0.1039],[97,0.1445],[98,0.1695],[99,0.1826],[100,0.1877],[101,0.1886],[182,0.1886],[183,0.1897],[184,0.1969],[185,0.2166],[186,0.2549],[187,0.3181],[188,0.4124],[189,0.6326],[190,1.0405],[191,1.5928],[192,2.246],[193,3.4185],[194,4.548],[195,5.0742],[196,5.4578],[197,5.7158],[198,5.8652],[199,5.9522],[200,6.0146],[201,6.0563],[202,6.0816],[247,6.0911],[248,6.0849],[249,6.0413],[250,5.923],[251,5.6927],[252,5.313],[253,4.4962],[255,2.0154],[256,1.1865],[257,0.7659],[258,0.4584],[259,0.2464],[260,0.1121],[261,0.0381],[262,0.0066],[263,0],[324,0]]
       }
       chest_ctrl.Scale.X {type=MovieSceneScriptingFloatChannel, default=1, const [1,325]}
       chest_ctrl.Scale.Y {type=MovieSceneScriptingFloatChannel, default=1, const [1,325]}
       chest_ctrl.Scale.Z {type=MovieSceneScriptingFloatChannel, default=1, const [1,325]}
       chest_tan_ctrl.Location.X {
        type=MovieSceneScriptingFloatChannel
        default=0
        thinned_from=325
        keys [[0,0],[29,-0.0001],[30,-0.0005],[31,-0.0017],[32,-0.0049],[33,-0.0099],[34,-0.0172],[35,-0.0232],[36,-0.0258],[37,-0.0276],[38,-0.0286],[40,-0.0292],[85,-0.0303],[86,-0.0301],[87,-0.0293],[88,-0.0278],[89,-0.0238],[90,-0.0111],[91,-0.0062],[92,-0.003],[93,-0.0013],[94,-0.0004],[95,0],[188,0.0005],[189,0.0007],[191,0.0018],[192,0.0025],[194,0.005],[196,0.006],[248,0.0101],[250,0.0099],[251,0.0095],[252,0.0088],[253,0.0075],[255,0.0033],[256,0.0019],[258,0.0007],[260,0.0002],[324,0]]
       }
       chest_tan_ctrl.Location.Y {
        type=MovieSceneScriptingFloatChannel
        default=-0.0303
        thinned_from=325
        keys [[0,0.1887],[1,0.1885],[2,0.1879],[3,0.1868],[4,0.1852],[5,0.183],[6,0.1804],[7,0.1772],[8,0.1735],[9,0.1689],[10,0.1637],[12,0.152],[14,0.1399],[15,0.1342],[16,0.1291],[19,0.1163],[21,0.1073],[22,0.1031],[23,0.0993],[24,0.096],[25,0.0934],[26,0.0918],[27,0.0912],[30,0.0916],[32,0.0924],[34,0.0937],[36,0.0957],[38,0.0984],[40,0.1017],[42,0.1057],[43,0.108],[44,0.1109],[45,0.1145],[46,0.1188],[47,0.1236],[51,0.1445],[52,0.1494],[53,0.1539],[54,0.1578],[55,0.1609],[56,0.1631],[60,0.169],[67,0.1764],[84,0.1892],[85,0.1897],[86,0.1896],[87,0.1889],[88,0.1877],[89,0.1857],[90,0.1831],[91,0.1799],[92,0.1761],[93,0.1716],[94,0.1663],[96,0.1544],[98,0.1421],[99,0.1363],[100,0.131],[103,0.1175],[105,0.1081],[106,0.1037],[107,0.0996],[108,0.0962],[109,0.0935],[110,0.0918],[111,0.0912],[114,0.0916],[116,0.0924],[118,0.0937],[120,0.0956],[122,0.0981],[124,0.1013],[126,0.1053],[127,0.1076],[128,0.1105],[129,0.1141],[130,0.1184],[131,0.1232],[135,0.1441],[136,0.149],[137,0.1535],[138,0.1573],[139,0.1604],[140,0.1626],[144,0.1685],[151,0.176],[168,0.1889],[169,0.1893],[170,0.1891],[171,0.1882],[172,0.1867],[173,0.1846],[174,0.1817],[175,0.1783],[176,0.1741],[177,0.1691],[178,0.163],[181,0.1425],[182,0.136],[183,0.1302],[184,0.1255],[188,0.1086],[189,0.1046],[190,0.101],[191,0.0977],[192,0.095],[193,0.093],[194,0.0918],[195,0.0913],[198,0.0918],[200,0.0925],[202,0.0938],[204,0.0956],[206,0.0981],[208,0.1014],[209,0.1033],[210,0.1054],[211,0.1078],[212,0.1108],[213,0.1148],[214,0.1196],[215,0.1249],[218,0.1421],[219,0.1476],[220,0.1526],[221,0.1569],[222,0.1603],[223,0.1626],[227,0.1684],[235,0.1765],[252,0.1887],[253,0.1892],[254,0.1892],[255,0.1885],[256,0.1872],[257,0.1854],[258,0.1828],[259,0.1796],[260,0.1758],[261,0.1712],[262,0.1659],[266,0.1419],[267,0.1362],[268,0.131],[270,0.1221],[273,0.1081],[274,0.1037],[275,0.0996],[276,0.0962],[277,0.0935],[278,0.0918],[279,0.0912],[282,0.0916],[284,0.0924],[286,0.0937],[288,0.0956],[290,0.0981],[292,0.1013],[294,0.1053],[295,0.1076],[296,0.1104],[297,0.1138],[298,0.1178],[299,0.1222],[304,0.1469],[305,0.1515],[306,0.1557],[307,0.1595],[308,0.1626],[312,0.173],[315,0.1794],[318,0.1844],[320,0.1868],[322,0.1882],[324,0.1888]]
       }
       chest_tan_ctrl.Location.Z {
        type=MovieSceneScriptingFloatChannel
        default=0.265
        thinned_from=325
        keys [[0,0.2675],[1,0.2665],[2,0.2635],[3,0.2586],[4,0.2522],[5,0.2442],[6,0.2348],[7,0.2243],[8,0.2126],[9,0.2002],[10,0.1869],[11,0.1729],[12,0.1584],[16,0.0986],[17,0.0841],[18,0.0702],[19,0.057],[20,0.0447],[21,0.0335],[22,0.0236],[23,0.015],[24,0.0081],[25,0.0029],[26,-0.0004],[27,-0.0015],[28,-0.0015],[30,-0.0006],[31,0.0002],[32,0.0013],[33,0.0027],[34,0.0045],[35,0.0067],[36,0.0095],[37,0.0129],[38,0.0168],[39,0.0214],[40,0.0267],[41,0.0327],[42,0.0394],[43,0.0469],[44,0.056],[45,0.0672],[46,0.0801],[47,0.0944],[48,0.1096],[51,0.1572],[52,0.1725],[53,0.187],[54,0.2002],[55,0.212],[56,0.2219],[59,0.2478],[61,0.2636],[63,0.278],[65,0.2905],[66,0.296],[67,0.3008],[68,0.305],[69,0.3085],[70,0.3112],[71,0.3132],[72,0.3144],[73,0.3146],[74,0.3141],[75,0.3127],[76,0.3106],[77,0.3077],[78,0.3041],[79,0.2996],[80,0.2945],[81,0.2885],[82,0.2818],[83,0.2744],[84,0.2662],[85,0.2572],[86,0.2475],[87,0.2371],[88,0.2261],[90,0.2032],[92,0.1783],[97,0.1132],[98,0.1003],[99,0.0878],[100,0.0756],[101,0.0639],[102,0.0529],[103,0.0426],[104,0.0331],[105,0.0246],[106,0.0171],[107,0.0107],[108,0.0055],[109,0.0017],[110,-0.0007],[111,-0.0015],[112,-0.0015],[114,-0.0006],[115,0.0002],[116,0.0013],[117,0.0028],[118,0.0047],[119,0.007],[120,0.0099],[121,0.0133],[122,0.0172],[123,0.0218],[124,0.0271],[125,0.033],[126,0.0398],[127,0.0473],[128,0.0564],[129,0.0676],[130,0.0805],[131,0.0948],[132,0.11],[135,0.1576],[136,0.173],[137,0.1873],[138,0.2006],[139,0.2124],[140,0.2224],[143,0.2482],[145,0.2641],[147,0.2785],[149,0.291],[150,0.2964],[151,0.3013],[152,0.3055],[153,0.309],[154,0.3117],[155,0.3137],[156,0.3149],[157,0.3151],[158,0.3146],[159,0.3132],[160,0.311],[161,0.3081],[162,0.3043],[163,0.2998],[164,0.2945],[165,0.2885],[166,0.2818],[167,0.2743],[168,0.2661],[169,0.2571],[170,0.2474],[171,0.2371],[172,0.2261],[174,0.203],[176,0.1782],[182,0.1004],[183,0.0878],[184,0.0755],[185,0.0639],[186,0.0528],[187,0.0425],[188,0.0331],[189,0.0245],[190,0.017],[191,0.0106],[192,0.0055],[193,0.0017],[194,-0.0007],[195,-0.0015],[197,-0.0012],[198,-0.0007],[199,0],[200,0.001],[201,0.0024],[202,0.0042],[203,0.0065],[204,0.0092],[205,0.0126],[206,0.0166],[207,0.0212],[208,0.0265],[209,0.0326],[210,0.0395],[211,0.0473],[212,0.057],[213,0.0692],[214,0.0836],[215,0.0995],[216,0.1164],[218,0.1514],[219,0.1684],[220,0.1845],[221,0.199],[222,0.2119],[223,0.2224],[225,0.2404],[227,0.2571],[229,0.2724],[231,0.2858],[233,0.297],[234,0.3018],[235,0.3059],[236,0.3093],[237,0.312],[238,0.3139],[239,0.315],[240,0.3152],[241,0.3148],[242,0.3136],[243,0.3119],[244,0.3094],[245,0.3063],[246,0.3026],[247,0.2981],[248,0.293],[249,0.2871],[250,0.2806],[251,0.2733],[252,0.2653],[253,0.2566],[254,0.247],[255,0.2368],[256,0.2259],[258,0.2028],[260,0.178],[265,0.1132],[266,0.1004],[267,0.0878],[268,0.0757],[269,0.064],[270,0.0529],[271,0.0426],[272,0.0332],[273,0.0246],[274,0.0171],[275,0.0107],[276,0.0055],[277,0.0017],[278,-0.0007],[279,-0.0015],[280,-0.0015],[282,-0.0006],[283,0.0002],[284,0.0013],[285,0.0028],[286,0.0047],[287,0.007],[288,0.0099],[289,0.0133],[290,0.0172],[291,0.0218],[292,0.0271],[293,0.033],[294,0.0398],[295,0.0473],[296,0.0565],[297,0.0679],[298,0.0812],[299,0.0959],[300,0.1116],[303,0.1602],[304,0.1756],[305,0.1897],[306,0.2026],[307,0.2136],[308,0.2224],[309,0.2295],[310,0.2358],[311,0.2414],[312,0.2463],[313,0.2506],[314,0.2542],[316,0.2599],[318,0.2638],[320,0.2661],[323,0.2674],[324,0.2675]]
       }
       chest_tan_ctrl.Rotation.X {type=MovieSceneScriptingFloatChannel, default=0, thinned_from=325, keys [[0,4.6942],[1,0],[324,0]]}
       chest_tan_ctrl.Rotation.Y {
        type=MovieSceneScriptingFloatChannel
        default=0
        thinned_from=325
        keys [[0,0],[28,0],[29,0.0002],[93,0],[94,0.006],[95,0.0058],[96,0.0026],[97,0],[189,0],[190,0.0032],[191,0],[324,0]]
       }
       chest_tan_ctrl.Rotation.Z {
        type=MovieSceneScriptingFloatChannel
        default=0
        thinned_from=325
        keys [[0,0],[30,0],[31,0.0009],[32,0.0035],[33,0.0128],[34,0.0152],[35,0],[36,0],[37,0.0098],[38,0.2656],[39,0.4651],[41,0.6235],[42,0.6793],[43,0.7059],[44,0.7253],[45,0.7384],[46,0.7463],[47,0.7501],[87,0.7522],[88,0.7603],[89,0.8147],[90,0.8884],[91,0.7911],[92,0.5606],[93,0.3216],[94,0.1824],[95,0.0816],[96,0.0206],[97,0],[189,0],[190,0.0144],[191,0],[324,0]]
       }
       chest_tan_ctrl.Scale.X {type=MovieSceneScriptingFloatChannel, default=1, const [1,325]}
       chest_tan_ctrl.Scale.Y {type=MovieSceneScriptingFloatChannel, default=1, const [1,325]}
       chest_tan_ctrl.Scale.Z {type=MovieSceneScriptingFloatChannel, default=1, const [1,325]}
      ]
     }]
   }
  ]
 }
 Head {
  id=#2
  parent_id=#1
  possessed_class=SkeletalMeshComponentBudgeted
  tracks [{
    type=MovieSceneControlRigParameterTrack
    display_name=Fortnite_Face_BoneTweaker_CtrlRig
    sections [{type=MovieSceneControlRigParameterSection}]
   }]
 }
]
display_rate {denominator=1, numerator=30}
tick_resolution {denominator=1, numerator=24000}
tracks []
